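# Physics asset for face accessories on the average-build male player character base, judging by its name and path.
# It holds a single body on the head bone, shaped as one large, offset and rotated capsule.
# It defines no constraints, so there is no jointed ragdoll chain, only this one simple volume.
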PhysicsAsset /Game/Characters/Player/Male/Male_Avg_Base/Base_FaceAcc_PhysAsset
bodies {head [{shapeType=Capsule, length=43.25, radius=40, center=<-8.2455 -2.5216 0>, rotation=<rot 0 -90.0002 90>}]}
constraints []
