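# Physics asset for the head mesh of the average-build male base player character.
# It defines a single body on the head bone, shaped as one capsule offset and rotated from the bone origin.
# It has no constraints, so it holds no joints between bodies.
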PhysicsAsset /Game/Characters/Player/Male/Male_Avg_Base/Base_Head_PhysAsset
bodies {head [{shapeType=Capsule, length=8.25, radius=10.505, center=<3.3075 2.0918 0>, rotation=<rot 0 -80 90>}]}
constraints []
