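# Physics asset for the average male Fortnite player with a large head, giving the skeleton simplified collision shapes.
# Capsules cover the pelvis, two spine bones, arms, hands, legs and feet, while a single sphere covers the head.
# Only two constraints exist, right hand to forearm and right foot to calf, both with free swing and twist, so the bodies do not form a fully jointed ragdoll.
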PhysicsAsset /Game/Characters/Player/Male/Male_Avg_Base/Fortnite_M_AvgLargeHead_Player_Physics
bodies {
 Spine_03 [{shapeType=Capsule, length=3.5702, radius=14.4018, center=<0.8298 2.1243 -0.0107>, rotation=<rot -0.0003 88.5965 179.7999>}]
 calf_l [{shapeType=Capsule, length=38.6878, radius=7.1396, center=<-20.9094 -1.1698 -0.4407>, rotation=<rot -90 0 0>}]
 calf_r [{shapeType=Capsule, length=38.6878, radius=7.1396, center=<20.9094 1.1698 0.4407>, rotation=<rot -90 0 0>}]
 foot_l [{shapeType=Capsule, length=16.398, radius=6.8277, shapeName=hand_l_capsule, center=<-4.7972 6.887 0>, rotation=<rot 0 17.6807 90>}]
 foot_r [{shapeType=Capsule, length=16.398, radius=6.8277, shapeName=hand_l_capsule, center=<4.7972 -6.887 0>, rotation=<rot 0 17.6807 -90>}]
 hand_l [{shapeType=Capsule, length=6.0543, radius=6.1794, shapeName=hand_l_capsule, center=<7.5363 -1.7418 0>, rotation=<rot 90 0 0>}]
 hand_r [{shapeType=Capsule, length=6.0543, radius=6.1794, shapeName=hand_l_capsule, center=<-7.5363 1.7418 0>, rotation=<rot 90 0 180>}]
 head [{shapeType=Sphere, radius=14.5475, center=<5.9832 1.5155 0>}]
 lowerarm_l [{shapeType=Capsule, length=16.9921, radius=6.3, center=<9.7344 0 0>, rotation=<rot -90 -0.0148 0.0149>}]
 lowerarm_r [{shapeType=Capsule, length=16.9921, radius=6.3, center=<-9.7344 0 0>, rotation=<rot -90 -0.0148 0.0149>}]
 pelvis [{shapeType=Capsule, length=4.9303, radius=14.1069, center=<2.8556 -0.6243 0>, rotation=<rot 0 90 0>}]
 spine_05 [{shapeType=Capsule, length=7.7178, radius=15.0931, center=<0.2068 1.521 0.0001>}]
 thigh_l [{shapeType=Capsule, length=32.1783, radius=7.9305, center=<-19.5847 0 -0.8714>, rotation=<rot 90 0.0079 0.0094>}]
 thigh_r [{shapeType=Capsule, length=32.1783, radius=7.9305, center=<19.5847 0 0.8714>, rotation=<rot 90 0.0079 0.0094>}]
 upperarm_l [{shapeType=Capsule, length=24.3038, radius=6.3, center=<10.1374 0 -0.2114>, rotation=<rot -88.7876 -179.9763 179.9773>}]
 upperarm_r [{shapeType=Capsule, length=24.3038, radius=6.3, center=<-10.1374 0 0.2114>, rotation=<rot -88.7876 -179.9763 179.9773>}]
}
constraints [
 {
  bone1Name=hand_r
  bone2Name=lowerarm_r
  swing1LimitDegrees=45
  swing1Motion=Free
  swing2LimitDegrees=45
  swing2Motion=Free
  twistLimitDegrees=45
  twistMotion=Free
 }
 {
  bone1Name=foot_r
  bone2Name=calf_r
  swing1LimitDegrees=45
  swing1Motion=Free
  swing2LimitDegrees=45
  swing2Motion=Free
  twistLimitDegrees=45
  twistMotion=Free
 }
]
