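# Animation Blueprint for the main player skeleton that drives additive facial animation during gameplay.
# A default face idle blends by FaceStateMachineAlpha toward either a Locomotion or a Weapons state machine, chosen by bIsUsingWeapon.
# Locomotion covers running, tactical sprint, sliding, jumping, skydiving and DBNO; Weapons covers AR targeting and firing, with reload using the idle face.
# State aliases share transitions, blendspaces follow speed or smoothed yaw delta, and the event graph is empty.
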AnimBlueprint /Game/Animation/Game/MainPlayer/Facial/FaceGameplay/Gameplay_FaceSystem_StateMachine_AnimBP
target_skeleton=&Fortnite_M_Avg_Player_Skeleton
graphs [
 AnimGraph {nodes [
   AnimGraphNode_StateMachine_1 = Weapons() {is=title AnimGraphNode_StateMachine}
   AnimGraphNode_StateMachine_2 = Locomotion() {is=title AnimGraphNode_StateMachine}
   n1 = Get bIsUsingWeapon() -> (bIsUsingWeapon:bool)
   Output Pose(Result=Two Way Blend(
     A="Sequence Player (additive) 'DefaultFaceIdle_Passthrough'"() {sequence=&/DefaultFaceIdle_Passthrough}
     B=Two Way Blend!(A=@AnimGraphNode_StateMachine_2, B=@AnimGraphNode_StateMachine_1, bAlphaBoolEnabled:bool=@n1) -> (Pose) {blendNode {alphaInputType=Bool, alphaBoolBlend {blendInTime=0.1, blendOutTime=0.5}}}
     Alpha=Get FaceStateMachineAlpha() -> (FaceStateMachineAlpha:float)
    ) {blendNode {alphaBoolBlend {blendOption=Linear}}}) {layerGroup=None}
  ]}
 Weapons {
  entry=Idle
  path=AnimGraph.AnimGraphNode_StateMachine_1.Weapons
  ids=id
  nodes [
   BlendToIdle() {is=title AnimStateAliasNode, also {aliases [Targeted_To_NonTargeted]}}
   TargetingIntro() {is=title AnimStateAliasNode, also {aliases [Idle]}}
   TargetingFireIntro() {is=title AnimStateAliasNode, also {aliases [NonTargeted_To_Targeted]}}
   NonTargetingIntro() {is=title AnimStateAliasNode, also {aliases [Idle]}}
   NonTargetingFireIntro() {is=title AnimStateAliasNode, also {aliases [Targeted_Fire_Loop, Targeted_To_NonTargeted]}}
   TargetingLoopIntro() {is=title AnimStateAliasNode, also {aliases [NonTargeted_Loop]}}
   Reloading() {is=title AnimStateAliasNode, also {aliases [Nontargeted_Fire_Loop, NonTargeted_Loop, Targeted_Fire_Loop, Targeted_Loop]}}
   AnimStateNode_0 = ReloadIdle() {
    is=title AnimStateNode
    stateEntered {guid=#1}
    stateFullyBlended {guid=#2}
    stateLeft {guid=#3}
   }
   AnimStateNode_1 = NonTargeted_To_Targeted() {is=title AnimStateNode}
   AnimStateNode_2 = Targeted_Loop() {is=title AnimStateNode}
   AnimStateNode_3 = Targeted_To_NonTargeted() {is=title AnimStateNode}
   AnimStateNode_4 = Targeted_Fire_Loop() {is=title AnimStateNode}
   AnimStateNode_5 = Nontargeted_Fire_Loop() {is=title AnimStateNode}
   AnimStateNode_6 = NonTargeted_Loop() {
    is=title AnimStateNode
    stateEntered {guid=#4}
    stateFullyBlended {guid=#5}
    stateLeft {guid=#6}
   }
   AnimStateNode_7 = Idle_To_Targeted() {is=title AnimStateNode}
   AnimStateNode_8 = Idle() {
    stateEntered {guid=#1}
    stateFullyBlended {guid=#2}
    stateLeft {guid=#3}
   }
   AnimStateNode_9 = NonTargetingFireIntro_State() {is=title AnimStateNode}
   AnimStateTransitionNode_0 = AnimStateTransitionNode()
   AnimStateTransitionNode_1 = AnimStateTransitionNode() {bAutomaticRuleBasedOnSequencePlayerInState=true, crossfadeDuration=0.1}
   AnimStateTransitionNode_10 = AnimStateTransitionNode()
   AnimStateTransitionNode_11 = AnimStateTransitionNode()
   AnimStateTransitionNode_12 = AnimStateTransitionNode() {bAutomaticRuleBasedOnSequencePlayerInState=true, crossfadeDuration=0.1}
   AnimStateTransitionNode_16 = AnimStateTransitionNode()
   AnimStateTransitionNode_17 = AnimStateTransitionNode()
   AnimStateTransitionNode_19 = AnimStateTransitionNode() {bAutomaticRuleBasedOnSequencePlayerInState=true}
   AnimStateTransitionNode_2 = AnimStateTransitionNode() {crossfadeDuration=0.1}
   AnimStateTransitionNode_22 = AnimStateTransitionNode()
   AnimStateTransitionNode_23 = AnimStateTransitionNode()
   AnimStateTransitionNode_25 = AnimStateTransitionNode()
   AnimStateTransitionNode_26 = AnimStateTransitionNode() {bAutomaticRuleBasedOnSequencePlayerInState=true}
   AnimStateTransitionNode_4 = AnimStateTransitionNode()
   AnimStateTransitionNode_6 = AnimStateTransitionNode() {crossfadeDuration=0.15}
   AnimStateTransitionNode_7 = AnimStateTransitionNode()
   AnimStateTransitionNode_8 = AnimStateTransitionNode()
   AnimStateTransitionNode_9 = AnimStateTransitionNode() {crossfadeDuration=0.05}
  ]
  links [
   {
    dst=Idle
    rule=AnimGraph.AnimGraphNode_StateMachine_1.Weapons.AnimStateTransitionNode_0.Transition
    src=BlendToIdle
    via=AnimStateTransitionNode_0
   }
   {
    dst=Targeted_Fire_Loop
    rule=AnimGraph.AnimGraphNode_StateMachine_1.Weapons.AnimStateTransitionNode_10.Transition
    src=Idle_To_Targeted
    via=AnimStateTransitionNode_10
   }
   {
    dst=Targeted_Loop
    rule=AnimGraph.AnimGraphNode_StateMachine_1.Weapons.AnimStateTransitionNode_1.Transition
    src=Idle_To_Targeted
    via=AnimStateTransitionNode_1
   }
   {
    dst=NonTargetingIntro
    rule=AnimGraph.AnimGraphNode_StateMachine_1.Weapons.AnimStateTransitionNode_22.Transition
    src=NonTargeted_Loop
    via=AnimStateTransitionNode_22
   }
   {
    dst=Nontargeted_Fire_Loop
    rule=AnimGraph.AnimGraphNode_StateMachine_1.Weapons.AnimStateTransitionNode_26.Transition
    src=NonTargeted_Loop
    via=AnimStateTransitionNode_26
   }
   {
    dst=NonTargetingFireIntro_State
    rule=AnimGraph.AnimGraphNode_StateMachine_1.Weapons.AnimStateTransitionNode_8.Transition
    src=NonTargetingFireIntro
    via=AnimStateTransitionNode_8
   }
   {
    dst=Nontargeted_Fire_Loop
    rule=AnimGraph.AnimGraphNode_StateMachine_1.Weapons.AnimStateTransitionNode_12.Transition
    src=NonTargetingFireIntro_State
    via=AnimStateTransitionNode_12
   }
   {
    dst=Nontargeted_Fire_Loop
    rule=AnimGraph.AnimGraphNode_StateMachine_1.Weapons.AnimStateTransitionNode_23.Transition
    src=NonTargetingIntro
    via=AnimStateTransitionNode_23
   }
   {
    dst=NonTargeted_Loop
    rule=AnimGraph.AnimGraphNode_StateMachine_1.Weapons.AnimStateTransitionNode_19.Transition
    src=Nontargeted_Fire_Loop
    via=AnimStateTransitionNode_19
   }
   {
    dst=NonTargeted_To_Targeted
    rule=AnimGraph.AnimGraphNode_StateMachine_1.Weapons.AnimStateTransitionNode_17.Transition
    src=Nontargeted_Fire_Loop
    via=AnimStateTransitionNode_17
   }
   {
    dst=NonTargeted_Loop
    rule=AnimGraph.AnimGraphNode_StateMachine_1.Weapons.AnimStateTransitionNode_7.Transition
    src=ReloadIdle
    via=AnimStateTransitionNode_7
   }
   {
    dst=ReloadIdle
    rule=AnimGraph.AnimGraphNode_StateMachine_1.Weapons.AnimStateTransitionNode_6.Transition
    src=Reloading
    via=AnimStateTransitionNode_6
   }
   {
    dst=Targeted_Loop
    rule=AnimGraph.AnimGraphNode_StateMachine_1.Weapons.AnimStateTransitionNode_4.Transition
    src=Targeted_Fire_Loop
    via=AnimStateTransitionNode_4
   }
   {
    dst=Targeted_Fire_Loop
    rule=AnimGraph.AnimGraphNode_StateMachine_1.Weapons.AnimStateTransitionNode_11.Transition
    src=Targeted_Loop
    via=AnimStateTransitionNode_11
   }
   {
    dst=Targeted_To_NonTargeted
    rule=AnimGraph.AnimGraphNode_StateMachine_1.Weapons.AnimStateTransitionNode_2.Transition
    src=Targeted_Loop
    via=AnimStateTransitionNode_2
   }
   {
    dst=Targeted_Fire_Loop
    rule=AnimGraph.AnimGraphNode_StateMachine_1.Weapons.AnimStateTransitionNode_16.Transition
    src=TargetingFireIntro
    via=AnimStateTransitionNode_16
   }
   {
    dst=Idle_To_Targeted
    rule=AnimGraph.AnimGraphNode_StateMachine_1.Weapons.AnimStateTransitionNode_9.Transition
    src=TargetingIntro
    via=AnimStateTransitionNode_9
   }
   {
    dst=Targeted_Loop
    rule=AnimGraph.AnimGraphNode_StateMachine_1.Weapons.AnimStateTransitionNode_25.Transition
    src=TargetingLoopIntro
    via=AnimStateTransitionNode_25
   }
  ]
 }
 ReloadIdle {
  path=AnimGraph.AnimGraphNode_StateMachine_1.Weapons.AnimStateNode_0.ReloadIdle
  nodes [Output Animation Pose(Result="Sequence Player (additive) 'DefaultFaceIdle_Passthrough'"() {sequence=&DefaultFaceIdle_Passthrough})]
 }
 NonTargeted_To_Targeted {
  path=AnimGraph.AnimGraphNode_StateMachine_1.Weapons.AnimStateNode_1.NonTargeted_To_Targeted
  nodes [Output Animation Pose(Result="Sequence Player (additive) 'Facial_AR_NonTargetingToTargeting_CMM'"() {sequence=/Game/Animation/Game/MainPlayer/Facial/FaceGameplay/Weapons/MainPlayer/Facial_AR_NonTargetingToTargeting_CMM})]
 }
 Targeted_Loop {
  path=AnimGraph.AnimGraphNode_StateMachine_1.Weapons.AnimStateNode_2.Targeted_Loop
  nodes [
   "Sequence Player (additive) 'Facial_AR_TargetingLoop_CMM'"() {sequence=/Game/Animation/Game/MainPlayer/Facial/FaceGameplay/Weapons/MainPlayer/Facial_AR_TargetingLoop_CMM}
   Output Animation Pose(Result=Blendspace Player 'Facial_AR_Targeting_BS_CMM'!(X=Get Player_YawDeltaSmoothed() -> (Player_YawDeltaSmoothed:float)) -> (Pose) {blendSpace=/Game/Animation/Game/MainPlayer/Facial/FaceGameplay/Weapons/MainPlayer/Facial_AR_Targeting_BS_CMM})
  ]
 }
 Targeted_To_NonTargeted {
  path=AnimGraph.AnimGraphNode_StateMachine_1.Weapons.AnimStateNode_3.Targeted_To_NonTargeted
  nodes [Output Animation Pose(Result="Sequence Player (additive) 'Facial_AR_TargetingToNonTarget_CMM'"() {sequence=/Game/Animation/Game/MainPlayer/Facial/FaceGameplay/Weapons/MainPlayer/Facial_AR_TargetingToNonTarget_CMM})]
 }
 Targeted_Fire_Loop {
  path=AnimGraph.AnimGraphNode_StateMachine_1.Weapons.AnimStateNode_4.Targeted_Fire_Loop
  nodes [
   "Sequence Player (additive) 'Facial_AR_TargetingFireLoop_CMM'"() {sequence=/Game/Animation/Game/MainPlayer/Facial/FaceGameplay/Weapons/MainPlayer/Facial_AR_TargetingFireLoop_CMM}
   Output Animation Pose(Result=Blendspace Player 'Facial_AR_TargetingFire_BS_CMM'!(X=Get Player_YawDeltaSmoothed()) -> (Pose) {blendSpace=/Game/Animation/Game/MainPlayer/Facial/FaceGameplay/Weapons/MainPlayer/Facial_AR_TargetingFire_BS_CMM})
  ]
 }
 Nontargeted_Fire_Loop {
  path=AnimGraph.AnimGraphNode_StateMachine_1.Weapons.AnimStateNode_5.Nontargeted_Fire_Loop
  nodes [Output Animation Pose(Result="Sequence Player (additive) 'Facial_AR_NonTargetingFireLoop_CMM'"() {sequence=/Game/Animation/Game/MainPlayer/Facial/FaceGameplay/Weapons/MainPlayer/Facial_AR_NonTargetingFireLoop_CMM})]
 }
 NonTargeted_Loop {
  path=AnimGraph.AnimGraphNode_StateMachine_1.Weapons.AnimStateNode_6.NonTargeted_Loop
  nodes [Output Animation Pose(Result="Sequence Player (additive) 'Facial_AR_NonTargetingLoop_CMM'"() {sequence=/Game/Animation/Game/MainPlayer/Facial/FaceGameplay/Weapons/MainPlayer/Facial_AR_NonTargetingLoop_CMM})]
 }
 Idle_To_Targeted {
  path=AnimGraph.AnimGraphNode_StateMachine_1.Weapons.AnimStateNode_7.Idle_To_Targeted
  nodes [Output Animation Pose(Result="Sequence Player (additive) 'Facial_AR_NonTargetingToTargeting_CMM'"() {sequence=&Facial_AR_NonTargetingToTargeting_CMM})]
 }
 Idle {
  path=AnimGraph.AnimGraphNode_StateMachine_1.Weapons.AnimStateNode_8.Idle
  nodes [Output Animation Pose(Result="Sequence Player (additive) 'DefaultFaceIdle_Passthrough'"() {sequence=&DefaultFaceIdle_Passthrough})]
 }
 NonTargetingFireIntro_State {
  path=AnimGraph.AnimGraphNode_StateMachine_1.Weapons.AnimStateNode_9.NonTargetingFireIntro_State
  nodes [
   n1 = "Sequence Player (additive) 'Facial_AR_NonTargetingFireIntro_CMM'"() {bLoopAnimation=false, sequence=/Game/Animation/Game/MainPlayer/Facial/FaceGameplay/Weapons/MainPlayer/Facial_AR_NonTargetingFireIntro_CMM}
   Output Animation Pose(Result=@n1)
  ]
 }
 Transition {
  path=AnimGraph.AnimGraphNode_StateMachine_1.Weapons.AnimStateTransitionNode_0.Transition
  nodes [Result(bCanEnterTransition=Get bIsNotTargetingAndIsNotCurrentlyShooting() -> (bIsNotTargetingAndIsNotCurrentlyShooting:bool))]
 }
 Transition {path=AnimGraph.AnimGraphNode_StateMachine_1.Weapons.AnimStateTransitionNode_1.Transition, nodes [Result()]}
 Transition {
  path=AnimGraph.AnimGraphNode_StateMachine_1.Weapons.AnimStateTransitionNode_10.Transition
  nodes [Result(bCanEnterTransition=Get bRecentlyFired() -> (bRecentlyFired:bool))]
 }
 Transition {
  path=AnimGraph.AnimGraphNode_StateMachine_1.Weapons.AnimStateTransitionNode_11.Transition
  nodes [Result(bCanEnterTransition=Get bIsCurrentlyShooting() -> (bIsCurrentlyShooting:bool))]
 }
 Transition {path=AnimGraph.AnimGraphNode_StateMachine_1.Weapons.AnimStateTransitionNode_12.Transition, nodes [Result()]}
 Transition {
  path=AnimGraph.AnimGraphNode_StateMachine_1.Weapons.AnimStateTransitionNode_16.Transition
  nodes [Result(bCanEnterTransition=Get bIsTargeting())]
 }
 Transition {
  path=AnimGraph.AnimGraphNode_StateMachine_1.Weapons.AnimStateTransitionNode_17.Transition
  nodes [Result(bCanEnterTransition=Get bIsTargeting())]
 }
 Transition {path=AnimGraph.AnimGraphNode_StateMachine_1.Weapons.AnimStateTransitionNode_19.Transition, nodes [Result()]}
 Transition {
  path=AnimGraph.AnimGraphNode_StateMachine_1.Weapons.AnimStateTransitionNode_2.Transition
  nodes [
   Result(bCanEnterTransition=NOT Boolean(A=Get bIsTargeting()))
   Get bIsTargeting_0() -> (bIsTargeting_0:bool)
  ]
 }
 Transition {
  path=AnimGraph.AnimGraphNode_StateMachine_1.Weapons.AnimStateTransitionNode_22.Transition
  nodes [Result(bCanEnterTransition=NOT Boolean(A=Get bIsCurrentlyShooting()))]
 }
 Transition {
  path=AnimGraph.AnimGraphNode_StateMachine_1.Weapons.AnimStateTransitionNode_23.Transition
  nodes [Result(bCanEnterTransition=Get bIsCurrentlyShooting())]
 }
 Transition {
  path=AnimGraph.AnimGraphNode_StateMachine_1.Weapons.AnimStateTransitionNode_25.Transition
  nodes [Result(bCanEnterTransition=Get bIsTargeting())]
 }
 Transition {path=AnimGraph.AnimGraphNode_StateMachine_1.Weapons.AnimStateTransitionNode_26.Transition, nodes [Result()]}
 Transition {
  path=AnimGraph.AnimGraphNode_StateMachine_1.Weapons.AnimStateTransitionNode_4.Transition
  nodes [Result(bCanEnterTransition=NOT Boolean(A=Get bIsCurrentlyShooting()))]
 }
 Transition {
  path=AnimGraph.AnimGraphNode_StateMachine_1.Weapons.AnimStateTransitionNode_6.Transition
  nodes [Result(bCanEnterTransition=Get bIsReloading() -> (bIsReloading:bool))]
 }
 Transition {
  path=AnimGraph.AnimGraphNode_StateMachine_1.Weapons.AnimStateTransitionNode_7.Transition
  nodes [Result(bCanEnterTransition=NOT Boolean(A=Get bIsReloading()))]
 }
 Transition {
  path=AnimGraph.AnimGraphNode_StateMachine_1.Weapons.AnimStateTransitionNode_8.Transition
  nodes [Result(bCanEnterTransition=NOT Boolean(A=Get bIsTargeting()))]
 }
 Transition {
  path=AnimGraph.AnimGraphNode_StateMachine_1.Weapons.AnimStateTransitionNode_9.Transition
  nodes [Result(bCanEnterTransition=Get bIsTargeting())]
 }
 Locomotion {
  entry=Idle
  path=AnimGraph.AnimGraphNode_StateMachine_2.Locomotion
  ids=id
  nodes [
   IsWalkJogRunSprint() {is=title AnimStateAliasNode, also {aliases [Idle]}}
   IsSkydiving() {is=title AnimStateAliasNode, also {aliases [Idle,JumpIntro,JumpLoop,JumpOutro]}}
   ToIdle() {is=title AnimStateAliasNode, also {aliases [DiveIdle, JumpOutro, Sliding_Outro, TacSprintToIdle]}}
   IsSliding() {is=title AnimStateAliasNode, also {aliases [Jog/RunIntro, TacSprint, TacSprintIntro, TacSprintToIdle, TacSprintToRun, Walk/Jog/Run_BS]}}
   IsJumping() {
    is=title AnimStateAliasNode
    also {aliases [Idle, Jog/RunIntro, Slopes, TacSprint, TacSprintIntro, TacSprintToIdle, TacSprintToRun, Walk/Jog/Run_BS]}
   }
   IsFalling() {is=title AnimStateAliasNode}
   IsLanding() {is=title AnimStateAliasNode, also {aliases [ActiveDive,DiveIdle,DiveIntro,GlideLoop]}}
   AnimStateNode_0 = Idle() {
    stateEntered {guid=#7}
    stateFullyBlended {guid=#8}
    stateLeft {guid=#9}
   }
   AnimStateNode_1 = Walk/Jog/Run_BS() {is=title AnimStateNode}
   AnimStateNode_10 = DiveIntro() {is=title AnimStateNode}
   AnimStateNode_11 = TacSprintIntro() {
    is=title AnimStateNode
    stateEntered {guid=#10}
    stateFullyBlended {guid=#11}
    stateLeft {guid=#12}
   }
   AnimStateNode_12 = TacSprint() {is=title AnimStateNode}
   AnimStateNode_13 = TacSprintToRun() {is=title AnimStateNode}
   AnimStateNode_14 = TacSprintToIdle() {is=title AnimStateNode}
   AnimStateNode_15 = Sliding() {
    is=title AnimStateNode
    stateEntered {guid=#13}
    stateFullyBlended {guid=#14}
    stateLeft {guid=#15}
   }
   AnimStateNode_16 = JumpIntro() {is=title AnimStateNode}
   AnimStateNode_17 = JumpLoop() {is=title AnimStateNode}
   AnimStateNode_18 = JumpOutro() {is=title AnimStateNode}
   AnimStateNode_2 = Sliding_Outro() {
    is=title AnimStateNode
    stateEntered {guid=#13}
    stateFullyBlended {guid=#14}
    stateLeft {guid=#15}
   }
   AnimStateNode_3 = Slopes() {
    is=title AnimStateNode
    stateEntered {guid=#16}
    stateFullyBlended {guid=#17}
    stateLeft {guid=#18}
   }
   AnimStateNode_5 = DBNO() {is=title AnimStateNode}
   AnimStateNode_6 = Jog/RunIntro() {
    is=title AnimStateNode
    stateEntered {guid=#19}
    stateFullyBlended {guid=#20}
    stateLeft {guid=#21}
   }
   AnimStateNode_7 = GlideLoop() {is=title AnimStateNode}
   AnimStateNode_8 = ActiveDive() {is=title AnimStateNode}
   AnimStateNode_9 = DiveIdle() {is=title AnimStateNode}
   AnimStateTransitionNode_0 = AnimStateTransitionNode()
   AnimStateTransitionNode_1 = AnimStateTransitionNode() {bAutomaticRuleBasedOnSequencePlayerInState=true, crossfadeDuration=0.4}
   AnimStateTransitionNode_10 = AnimStateTransitionNode()
   AnimStateTransitionNode_12 = AnimStateTransitionNode() {bAutomaticRuleBasedOnSequencePlayerInState=true}
   AnimStateTransitionNode_13 = AnimStateTransitionNode()
   AnimStateTransitionNode_14 = AnimStateTransitionNode() {crossfadeDuration=0.175}
   AnimStateTransitionNode_15 = AnimStateTransitionNode() {bAutomaticRuleBasedOnSequencePlayerInState=true}
   AnimStateTransitionNode_16 = AnimStateTransitionNode()
   AnimStateTransitionNode_17 = AnimStateTransitionNode() {bAutomaticRuleBasedOnSequencePlayerInState=true}
   AnimStateTransitionNode_19 = AnimStateTransitionNode() {crossfadeDuration=0.15}
   AnimStateTransitionNode_2 = AnimStateTransitionNode() {
    transitionEnd {guid=#22}
    transitionInterrupt {guid=#23}
    transitionStart {guid=#24}
   }
   AnimStateTransitionNode_20 = AnimStateTransitionNode()
   AnimStateTransitionNode_21 = AnimStateTransitionNode()
   AnimStateTransitionNode_22 = AnimStateTransitionNode() {bAutomaticRuleBasedOnSequencePlayerInState=true, crossfadeDuration=0.35}
   AnimStateTransitionNode_23 = AnimStateTransitionNode() {crossfadeDuration=0.05}
   AnimStateTransitionNode_24 = AnimStateTransitionNode() {crossfadeDuration=0.05}
   AnimStateTransitionNode_25 = AnimStateTransitionNode() {crossfadeDuration=0.15}
   AnimStateTransitionNode_26 = AnimStateTransitionNode() {bAutomaticRuleBasedOnSequencePlayerInState=true, crossfadeDuration=0.05}
   AnimStateTransitionNode_27 = AnimStateTransitionNode() {crossfadeDuration=0.05}
   AnimStateTransitionNode_28 = AnimStateTransitionNode() {crossfadeDuration=0.1}
   AnimStateTransitionNode_29 = AnimStateTransitionNode() {crossfadeDuration=0.05}
   AnimStateTransitionNode_3 = AnimStateTransitionNode() {bAutomaticRuleBasedOnSequencePlayerInState=true}
   AnimStateTransitionNode_30 = AnimStateTransitionNode() {crossfadeDuration=0.15}
   AnimStateTransitionNode_31 = AnimStateTransitionNode()
   AnimStateTransitionNode_32 = AnimStateTransitionNode()
   AnimStateTransitionNode_4 = AnimStateTransitionNode() {crossfadeDuration=0.1}
   AnimStateTransitionNode_5 = AnimStateTransitionNode()
   AnimStateTransitionNode_6 = AnimStateTransitionNode() {bAutomaticRuleBasedOnSequencePlayerInState=true}
   AnimStateTransitionNode_7 = AnimStateTransitionNode() {bAutomaticRuleBasedOnSequencePlayerInState=true}
   AnimStateTransitionNode_8 = AnimStateTransitionNode() {bAutomaticRuleBasedOnSequencePlayerInState=true, crossfadeDuration=0.4}
   AnimStateTransitionNode_9 = AnimStateTransitionNode() {crossfadeDuration=0.25}
  ]
  links [
   {
    dst=DiveIdle
    rule=AnimGraph.AnimGraphNode_StateMachine_2.Locomotion.AnimStateTransitionNode_17.Transition
    src=ActiveDive
    via=AnimStateTransitionNode_17
   }
   {
    dst=Idle
    rule=AnimGraph.AnimGraphNode_StateMachine_2.Locomotion.AnimStateTransitionNode_6.Transition
    src=DBNO
    via=AnimStateTransitionNode_6
   }
   {
    dst=ActiveDive
    rule=AnimGraph.AnimGraphNode_StateMachine_2.Locomotion.AnimStateTransitionNode_16.Transition
    src=DiveIdle
    via=AnimStateTransitionNode_16
   }
   {
    dst=GlideLoop
    rule=AnimGraph.AnimGraphNode_StateMachine_2.Locomotion.AnimStateTransitionNode_13.Transition
    src=DiveIdle
    via=AnimStateTransitionNode_13
   }
   {
    dst=DiveIdle
    rule=AnimGraph.AnimGraphNode_StateMachine_2.Locomotion.AnimStateTransitionNode_3.Transition
    src=DiveIntro
    via=AnimStateTransitionNode_3
   }
   {
    dst=DiveIdle
    rule=AnimGraph.AnimGraphNode_StateMachine_2.Locomotion.AnimStateTransitionNode_15.Transition
    src=GlideLoop
    via=AnimStateTransitionNode_15
   }
   {
    dst=DBNO
    rule=AnimGraph.AnimGraphNode_StateMachine_2.Locomotion.AnimStateTransitionNode_5.Transition
    src=Idle
    via=AnimStateTransitionNode_5
   }
   {
    dst=Slopes
    rule=AnimGraph.AnimGraphNode_StateMachine_2.Locomotion.AnimStateTransitionNode_2.Transition
    src=Idle
    via=AnimStateTransitionNode_2
   }
   {
    dst=JumpLoop
    rule=AnimGraph.AnimGraphNode_StateMachine_2.Locomotion.AnimStateTransitionNode_29.Transition
    src=IsFalling
    via=AnimStateTransitionNode_29
   }
   {
    dst=JumpIntro
    rule=AnimGraph.AnimGraphNode_StateMachine_2.Locomotion.AnimStateTransitionNode_24.Transition
    src=IsJumping
    via=AnimStateTransitionNode_24
   }
   {
    dst=JumpOutro
    rule=AnimGraph.AnimGraphNode_StateMachine_2.Locomotion.AnimStateTransitionNode_31.Transition
    src=IsLanding
    via=AnimStateTransitionNode_31
   }
   {
    dst=DiveIntro
    rule=AnimGraph.AnimGraphNode_StateMachine_2.Locomotion.AnimStateTransitionNode_10.Transition
    src=IsSkydiving
    via=AnimStateTransitionNode_10
   }
   {
    dst=Sliding
    rule=AnimGraph.AnimGraphNode_StateMachine_2.Locomotion.AnimStateTransitionNode_4.Transition
    src=IsSliding
    via=AnimStateTransitionNode_4
   }
   {
    dst=Jog/RunIntro
    rule=AnimGraph.AnimGraphNode_StateMachine_2.Locomotion.AnimStateTransitionNode_25.Transition
    src=IsWalkJogRunSprint
    via=AnimStateTransitionNode_25
   }
   {
    dst=TacSprintIntro
    rule=AnimGraph.AnimGraphNode_StateMachine_2.Locomotion.AnimStateTransitionNode_19.Transition
    src=IsWalkJogRunSprint
    via=AnimStateTransitionNode_19
   }
   {
    dst=IsWalkJogRunSprint
    rule=AnimGraph.AnimGraphNode_StateMachine_2.Locomotion.AnimStateTransitionNode_28.Transition
    src=Jog/RunIntro
    via=AnimStateTransitionNode_28
   }
   {
    dst=TacSprintIntro
    rule=AnimGraph.AnimGraphNode_StateMachine_2.Locomotion.AnimStateTransitionNode_21.Transition
    src=Jog/RunIntro
    via=AnimStateTransitionNode_21
   }
   {
    dst=Walk/Jog/Run_BS
    rule=AnimGraph.AnimGraphNode_StateMachine_2.Locomotion.AnimStateTransitionNode_1.Transition
    src=Jog/RunIntro
    via=AnimStateTransitionNode_1
   }
   {
    dst=JumpLoop
    rule=AnimGraph.AnimGraphNode_StateMachine_2.Locomotion.AnimStateTransitionNode_26.Transition
    src=JumpIntro
    via=AnimStateTransitionNode_26
   }
   {
    dst=JumpOutro
    rule=AnimGraph.AnimGraphNode_StateMachine_2.Locomotion.AnimStateTransitionNode_27.Transition
    src=JumpLoop
    via=AnimStateTransitionNode_27
   }
   {
    dst=Sliding_Outro
    rule=AnimGraph.AnimGraphNode_StateMachine_2.Locomotion.AnimStateTransitionNode_23.Transition
    src=Sliding
    via=AnimStateTransitionNode_23
   }
   {
    dst=Idle
    rule=AnimGraph.AnimGraphNode_StateMachine_2.Locomotion.AnimStateTransitionNode_7.Transition
    src=Slopes
    via=AnimStateTransitionNode_7
   }
   {
    dst=TacSprintToIdle
    rule=AnimGraph.AnimGraphNode_StateMachine_2.Locomotion.AnimStateTransitionNode_14.Transition
    src=TacSprint
    via=AnimStateTransitionNode_14
   }
   {
    dst=TacSprintToRun
    rule=AnimGraph.AnimGraphNode_StateMachine_2.Locomotion.AnimStateTransitionNode_9.Transition
    src=TacSprint
    via=AnimStateTransitionNode_9
   }
   {
    dst=IsWalkJogRunSprint
    rule=AnimGraph.AnimGraphNode_StateMachine_2.Locomotion.AnimStateTransitionNode_20.Transition
    src=TacSprintIntro
    via=AnimStateTransitionNode_20
   }
   {
    dst=Jog/RunIntro
    rule=AnimGraph.AnimGraphNode_StateMachine_2.Locomotion.AnimStateTransitionNode_32.Transition
    src=TacSprintIntro
    via=AnimStateTransitionNode_32
   }
   {
    dst=TacSprint
    rule=AnimGraph.AnimGraphNode_StateMachine_2.Locomotion.AnimStateTransitionNode_8.Transition
    src=TacSprintIntro
    via=AnimStateTransitionNode_8
   }
   {
    dst=Walk/Jog/Run_BS
    rule=AnimGraph.AnimGraphNode_StateMachine_2.Locomotion.AnimStateTransitionNode_12.Transition
    src=TacSprintToRun
    via=AnimStateTransitionNode_12
   }
   {
    dst=Idle
    rule=AnimGraph.AnimGraphNode_StateMachine_2.Locomotion.AnimStateTransitionNode_22.Transition
    src=ToIdle
    via=AnimStateTransitionNode_22
   }
   {
    dst=IsWalkJogRunSprint
    rule=AnimGraph.AnimGraphNode_StateMachine_2.Locomotion.AnimStateTransitionNode_30.Transition
    src=Walk/Jog/Run_BS
    via=AnimStateTransitionNode_30
   }
   {
    dst=TacSprint
    rule=AnimGraph.AnimGraphNode_StateMachine_2.Locomotion.AnimStateTransitionNode_0.Transition
    src=Walk/Jog/Run_BS
    via=AnimStateTransitionNode_0
   }
  ]
 }
 Idle {
  path=AnimGraph.AnimGraphNode_StateMachine_2.Locomotion.AnimStateNode_0.Idle
  nodes [Output Animation Pose(Result="Sequence Player (additive) 'DefaultFaceIdle_Passthrough'"() {sequence=&DefaultFaceIdle_Passthrough})]
 }
 Walk/Jog/Run_BS {
  path=AnimGraph.AnimGraphNode_StateMachine_2.Locomotion.AnimStateNode_1.Walk/Jog/Run_BS
  nodes [
   Blend Poses by bool(
    BlendPose_0="Sequence Player (additive) 'Facial_TacSprint_CMM'"() {sequence=/Game/Animation/Game/MainPlayer/Facial/FaceGameplay/Locomotion/MainPlayer/Facial_TacSprint_CMM}
    BlendTime_1=0.25
    bActiveValue=Get bIsTacticalSprint() -> (bIsTacticalSprint:bool)
   ) {blendType=Linear}
   Output Animation Pose(Result=Blendspace Player 'CharacterLocomotion_Facial_BS_CMM'!(X=Get PawnSpeed2D()) -> (Pose) {blendSpace=/Game/Animation/Game/MainPlayer/Facial/FaceGameplay/Locomotion/MainPlayer/CharacterLocomotion_Facial_BS_CMM})
  ]
 }
 DiveIntro {
  path=AnimGraph.AnimGraphNode_StateMachine_2.Locomotion.AnimStateNode_10.DiveIntro
  nodes [Output Animation Pose(Result="Sequence Player (additive) 'Facial_SkyDive_Intro_CMM'"() {sequence=/Game/Animation/Game/MainPlayer/Facial/FaceGameplay/Locomotion/MainPlayer/Facial_SkyDive_Intro_CMM})]
 }
 TacSprintIntro {
  path=AnimGraph.AnimGraphNode_StateMachine_2.Locomotion.AnimStateNode_11.TacSprintIntro
  nodes [Output Animation Pose(Result="Sequence Player (additive) 'Facial_TacSprint_Intro_CMM'"() {bLoopAnimation=false, sequence=/Game/Animation/Game/MainPlayer/Facial/FaceGameplay/Locomotion/MainPlayer/Facial_TacSprint_Intro_CMM})]
 }
 TacSprint {
  path=AnimGraph.AnimGraphNode_StateMachine_2.Locomotion.AnimStateNode_12.TacSprint
  nodes [Output Animation Pose(Result="Sequence Player (additive) 'Facial_TacSprint_CMM'"() {sequence=&Facial_TacSprint_CMM})]
 }
 TacSprintToRun {
  path=AnimGraph.AnimGraphNode_StateMachine_2.Locomotion.AnimStateNode_13.TacSprintToRun
  nodes [Output Animation Pose(Result="Sequence Player (additive) 'Facial_TacSprintToRun_CMM'"() {bLoopAnimation=false, sequence=/Game/Animation/Game/MainPlayer/Facial/FaceGameplay/Locomotion/MainPlayer/Facial_TacSprintToRun_CMM})]
 }
 TacSprintToIdle {
  path=AnimGraph.AnimGraphNode_StateMachine_2.Locomotion.AnimStateNode_14.TacSprintToIdle
  nodes [Output Animation Pose(Result="Sequence Player (additive) 'Facial_TacSprintToIdle_CMM'"() {bLoopAnimation=false, sequence=/Game/Animation/Game/MainPlayer/Facial/FaceGameplay/Locomotion/MainPlayer/Facial_TacSprintToIdle_CMM})]
 }
 Sliding {
  path=AnimGraph.AnimGraphNode_StateMachine_2.Locomotion.AnimStateNode_15.Sliding
  nodes [Output Animation Pose(Result="Sequence Player (additive) 'Facial_Sloping_CMM'"() {sequence=/Game/Animation/Game/MainPlayer/Facial/FaceGameplay/Locomotion/MainPlayer/Facial_Sloping_CMM})]
 }
 JumpIntro {
  path=AnimGraph.AnimGraphNode_StateMachine_2.Locomotion.AnimStateNode_16.JumpIntro
  nodes [Output Animation Pose(Result="Sequence Player (additive) 'Facial_Jump_Intro_CMM'"() {bLoopAnimation=false, sequence=/Game/Animation/Game/MainPlayer/Facial/FaceGameplay/Locomotion/MainPlayer/Facial_Jump_Intro_CMM})]
 }
 JumpLoop {
  path=AnimGraph.AnimGraphNode_StateMachine_2.Locomotion.AnimStateNode_17.JumpLoop
  nodes [Output Animation Pose(Result="Sequence Player (additive) 'Facial_Jump_Loop_CMM'"() {sequence=/Game/Animation/Game/MainPlayer/Facial/FaceGameplay/Locomotion/MainPlayer/Facial_Jump_Loop_CMM})]
 }
 JumpOutro {
  path=AnimGraph.AnimGraphNode_StateMachine_2.Locomotion.AnimStateNode_18.JumpOutro
  nodes [Output Animation Pose(Result="Sequence Player (additive) 'Facial_Jump_Outro_CMM'"() {bLoopAnimation=false, sequence=/Game/Animation/Game/MainPlayer/Facial/FaceGameplay/Locomotion/MainPlayer/Facial_Jump_Outro_CMM})]
 }
 Sliding_Outro {
  path=AnimGraph.AnimGraphNode_StateMachine_2.Locomotion.AnimStateNode_2.Sliding_Outro
  nodes [Output Animation Pose(Result="Sequence Player (additive) 'Facial_SlopingOutro_CMM'"() {bLoopAnimation=false, sequence=/Game/Animation/Game/MainPlayer/Facial/FaceGameplay/Locomotion/MainPlayer/Facial_SlopingOutro_CMM})]
 }
 Slopes {
  path=AnimGraph.AnimGraphNode_StateMachine_2.Locomotion.AnimStateNode_3.Slopes
  nodes [Output Animation Pose(Result="Sequence Player (additive) 'Facial_Sloping_CMM'"() {sequence=&Facial_Sloping_CMM})]
 }
 DBNO {
  path=AnimGraph.AnimGraphNode_StateMachine_2.Locomotion.AnimStateNode_5.DBNO
  nodes [Output Animation Pose(Result="Sequence Player (additive) 'Facial_DBNO_CMM'"() {sequence=/Game/Animation/Game/MainPlayer/Facial/FaceGameplay/Locomotion/MainPlayer/Facial_DBNO_CMM})]
 }
 Jog/RunIntro {
  path=AnimGraph.AnimGraphNode_StateMachine_2.Locomotion.AnimStateNode_6.Jog/RunIntro
  nodes [Output Animation Pose(Result="Sequence Player (additive) 'Facial_Run_Intro_CMM'"() {bLoopAnimation=false, sequence=/Game/Animation/Game/MainPlayer/Facial/FaceGameplay/Locomotion/MainPlayer/Facial_Run_Intro_CMM})]
 }
 GlideLoop {
  path=AnimGraph.AnimGraphNode_StateMachine_2.Locomotion.AnimStateNode_7.GlideLoop
  nodes [
   "Sequence Player (additive) 'Facial_SkyGlide_CMM'"() {sequence=/Game/Animation/Game/MainPlayer/Facial/FaceGameplay/Locomotion/MainPlayer/Facial_SkyGlide_CMM}
   Output Animation Pose(Result=Blendspace Player 'SkyGlide_Facial_BS_CMM'!(X=Get Player_YawDeltaSmoothed()) -> (Pose) {blendSpace=/Game/Animation/Game/MainPlayer/Facial/FaceGameplay/Locomotion/MainPlayer/SkyGlide_Facial_BS_CMM})
  ]
 }
 ActiveDive {
  path=AnimGraph.AnimGraphNode_StateMachine_2.Locomotion.AnimStateNode_8.ActiveDive
  nodes [Output Animation Pose(Result="Sequence Player (additive) 'Facial_SkyActiveDive_CMM'"() {sequence=/Game/Animation/Game/MainPlayer/Facial/FaceGameplay/Locomotion/MainPlayer/Facial_SkyActiveDive_CMM})]
 }
 DiveIdle {
  path=AnimGraph.AnimGraphNode_StateMachine_2.Locomotion.AnimStateNode_9.DiveIdle
  nodes [Output Animation Pose(Result="Sequence Player (additive) 'Facial_SkyDiveIdle_CMM'"() {sequence=/Game/Animation/Game/MainPlayer/Facial/FaceGameplay/Locomotion/MainPlayer/Facial_SkyDiveIdle_CMM})]
 }
 Transition {
  path=AnimGraph.AnimGraphNode_StateMachine_2.Locomotion.AnimStateTransitionNode_0.Transition
  nodes [Result(bCanEnterTransition=Get bIsTacticalSprint())]
 }
 Transition {path=AnimGraph.AnimGraphNode_StateMachine_2.Locomotion.AnimStateTransitionNode_1.Transition, nodes [Result()]}
 Transition {
  path=AnimGraph.AnimGraphNode_StateMachine_2.Locomotion.AnimStateTransitionNode_10.Transition
  nodes [Result(bCanEnterTransition=Get!() -> (
     Skydiving_ParaGilderRootOffset:Vector=~
     Skydiving_ParaGliderRootRotationOverride:Rotator=~
     Skydiving_bUseParaGlideRootModifier:bool=~
     Skydiving_bIsSkydivingFromLaunchPad:bool=~
     Skydiving_bIsSkydivingFromBus:bool=~
     Skydiving_bIsInVortex:bool=~
     Skydiving_bIsDiving:bool=~
     Skydiving_bIsParachuteOpen:bool=~
     Skydiving_bIsSkydiving:bool=~
     Skydiving_bIsParachuteLeaning:bool=~
     Skydiving_bIsSkydiveLeaning:bool=~
     Skydiving_bIsLeaning:bool=~
     Skydiving_bIsSkydiveDiveMode:bool=~
     Skydiving_bPlaySkydiveDrift:bool=~
     Skydiving_bSkydiveDriftDelayActive:bool=~
     Skydiving_bSkydiveDriftAnimAllowed:bool=~
     Skydiving_bIsGliderRight:bool=~
     Skydiving_bIsGliderCenter:bool=~
     Skydiving_bIsGliderLeft:bool=~
     Skydiving_bIsGliderForward:bool=~
     Skydiving_bIsGliderBack:bool=~
     Skydiving_bHasDeployedAndIsStandingUpright:bool=~
     Skydiving_LocalAccelForward:float=~
     Skydiving_LocalAccelRight:float=~
     Skydiving_LocalVelocityRight:float=~
     Skydiving_SkydiveAimPitch:float=~
     Skydiving_SkydiveAimYaw:float=~
     Skydiving_DeployChuteAnimRate:float=~
     Skydiving_SkydiveDriftAnimRate:float=~
     Skydiving_SkydiveDriftAnimRateCurrent:float=~
     Skydiving_SkydiveFidgetAnimRate:float=~
     Skydiving_SkydiveFidgetAnimRateCurrent:float=~
     Skydiving_SkydiveAdditiveAlpha:float=~
     Skydiving_SkydiveDriftDelay:float=~
     Skydiving_SkydiveDriftAnim:int
     Skydiving_SkydiveDriftAnimMax:int
     Skydiving_LaunchpadAnim:int
     Skydiving_LocalAccelDir:ESkydivingDirection Enum
     Skydiving_DirectionLast:ESkydivingDirection Enum
     Skydiving_PoseAdditive:Anim Sequence Object Reference
    ).Skydiving_bIsSkydiving)]
 }
 Transition {path=AnimGraph.AnimGraphNode_StateMachine_2.Locomotion.AnimStateTransitionNode_12.Transition, nodes [Result()]}
 Transition {
  path=AnimGraph.AnimGraphNode_StateMachine_2.Locomotion.AnimStateTransitionNode_13.Transition
  nodes [Result(bCanEnterTransition=Get bIsSkydiveFloating())]
 }
 Transition {
  path=AnimGraph.AnimGraphNode_StateMachine_2.Locomotion.AnimStateTransitionNode_14.Transition
  nodes [Result(bCanEnterTransition=Get bSlowOrInVehcileOrPlayingEmote() -> (bSlowOrInVehcileOrPlayingEmote:bool))]
 }
 Transition {path=AnimGraph.AnimGraphNode_StateMachine_2.Locomotion.AnimStateTransitionNode_15.Transition, nodes [Result()]}
 Transition {
  path=AnimGraph.AnimGraphNode_StateMachine_2.Locomotion.AnimStateTransitionNode_16.Transition
  nodes [Result(bCanEnterTransition=Get bIsSkydiveDiveMode())]
 }
 Transition {path=AnimGraph.AnimGraphNode_StateMachine_2.Locomotion.AnimStateTransitionNode_17.Transition, nodes [Result()]}
 Transition {
  path=AnimGraph.AnimGraphNode_StateMachine_2.Locomotion.AnimStateTransitionNode_19.Transition
  nodes [Result(bCanEnterTransition=Get bIsTacticalSprint())]
 }
 Transition {
  path=AnimGraph.AnimGraphNode_StateMachine_2.Locomotion.AnimStateTransitionNode_2.Transition
  nodes [Result(bCanEnterTransition=Get bIsSlopeSliding() -> (bIsSlopeSliding:bool))]
 }
 Transition {
  path=AnimGraph.AnimGraphNode_StateMachine_2.Locomotion.AnimStateTransitionNode_20.Transition
  nodes [Result(bCanEnterTransition=NOT Boolean(A=Get bIsTacticalSprint()))]
 }
 Transition {
  path=AnimGraph.AnimGraphNode_StateMachine_2.Locomotion.AnimStateTransitionNode_21.Transition
  nodes [Result(bCanEnterTransition=Get bIsTacticalSprint())]
 }
 Transition {path=AnimGraph.AnimGraphNode_StateMachine_2.Locomotion.AnimStateTransitionNode_22.Transition, nodes [Result()]}
 Transition {
  path=AnimGraph.AnimGraphNode_StateMachine_2.Locomotion.AnimStateTransitionNode_23.Transition
  nodes [Result(bCanEnterTransition=Get bIsNeitherSlidingNorSlopeSliding() -> (bIsNeitherSlidingNorSlopeSliding:bool))]
 }
 Transition {
  path=AnimGraph.AnimGraphNode_StateMachine_2.Locomotion.AnimStateTransitionNode_24.Transition
  nodes [Result(bCanEnterTransition=Get bIsJumping())]
 }
 Transition {
  path=AnimGraph.AnimGraphNode_StateMachine_2.Locomotion.AnimStateTransitionNode_25.Transition
  nodes [Result(bCanEnterTransition=Get bFastAndNotInVehicleAndNotPlayingEmote() -> (bFastAndNotInVehicleAndNotPlayingEmote:bool))]
 }
 Transition {path=AnimGraph.AnimGraphNode_StateMachine_2.Locomotion.AnimStateTransitionNode_26.Transition, nodes [Result()]}
 Transition {
  path=AnimGraph.AnimGraphNode_StateMachine_2.Locomotion.AnimStateTransitionNode_27.Transition
  nodes [Result(bCanEnterTransition=Get bLandingPredicted() -> (bLandingPredicted:bool))]
 }
 Transition {
  path=AnimGraph.AnimGraphNode_StateMachine_2.Locomotion.AnimStateTransitionNode_28.Transition
  nodes [Result(bCanEnterTransition=Get bSlowOrInVehcileOrPlayingEmote())]
 }
 Transition {
  path=AnimGraph.AnimGraphNode_StateMachine_2.Locomotion.AnimStateTransitionNode_29.Transition
  nodes [Result(bCanEnterTransition=Get bIsFalling())]
 }
 Transition {path=AnimGraph.AnimGraphNode_StateMachine_2.Locomotion.AnimStateTransitionNode_3.Transition, nodes [Result()]}
 Transition {
  path=AnimGraph.AnimGraphNode_StateMachine_2.Locomotion.AnimStateTransitionNode_30.Transition
  nodes [Result(bCanEnterTransition=Get bSlowOrInVehcileOrPlayingEmote())]
 }
 Transition {
  path=AnimGraph.AnimGraphNode_StateMachine_2.Locomotion.AnimStateTransitionNode_31.Transition
  nodes [Result(bCanEnterTransition=Get bLandingPredicted())]
 }
 Transition {
  path=AnimGraph.AnimGraphNode_StateMachine_2.Locomotion.AnimStateTransitionNode_32.Transition
  nodes [Result(bCanEnterTransition=NOT Boolean(A=Get bIsTacticalSprint()))]
 }
 Transition {
  path=AnimGraph.AnimGraphNode_StateMachine_2.Locomotion.AnimStateTransitionNode_4.Transition
  nodes [Result(bCanEnterTransition=Get bIsSlidingOrIsSlopeSliding() -> (bIsSlidingOrIsSlopeSliding:bool))]
 }
 Transition {
  path=AnimGraph.AnimGraphNode_StateMachine_2.Locomotion.AnimStateTransitionNode_5.Transition
  nodes [Result(bCanEnterTransition=Get bIsDBNO())]
 }
 Transition {path=AnimGraph.AnimGraphNode_StateMachine_2.Locomotion.AnimStateTransitionNode_6.Transition, nodes [Result()]}
 Transition {path=AnimGraph.AnimGraphNode_StateMachine_2.Locomotion.AnimStateTransitionNode_7.Transition, nodes [Result()]}
 Transition {path=AnimGraph.AnimGraphNode_StateMachine_2.Locomotion.AnimStateTransitionNode_8.Transition, nodes [Result()]}
 Transition {
  path=AnimGraph.AnimGraphNode_StateMachine_2.Locomotion.AnimStateTransitionNode_9.Transition
  nodes [
   Result(bCanEnterTransition=NOT Boolean(A=Get bIsTacticalSprint()))
   AND Boolean(A="float > float"(A=Get PawnSpeed2D(), B=199), B=NOT Boolean(A=Get bIsInVehicle() -> (bIsInVehicle:bool)))
  ]
 }
 EventGraph {links [], nodes []}
]
variables []
