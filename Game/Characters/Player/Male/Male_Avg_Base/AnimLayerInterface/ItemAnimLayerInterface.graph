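# Animation Layer Interface for the male average player skeleton family, declaring the hooks that items use to override parts of the character's animation.
# Every layer sits in the Item Override group, covering idle, upper and full body, aim offsets, pre-IK, right-hand IK, jump, fall, swim, jetpack, zipline and final pose.
# Most layers take input poses; aim layers also receive alpha, yaw and pitch or a Fort adjusted-aim struct, while some supply only an output.
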AnimBlueprint /Game/Characters/Player/Male/Male_Avg_Base/AnimLayerInterface/ItemAnimLayerInterface
target_skeleton=null
graphs [
 ItemAimOffset {nodes [
   Input Pose!() -> (Pose, AimOffsetAlpha:double=~, Yaw:double=~, Pitch:double=~) {
    name=InPose_AimOffset
    inputs [
     AimOffsetAlpha {type {
       bIsConst=false
       bIsReference=false
       bIsUObjectWrapper=false
       bIsWeakPointer=false
       bSerializeAsSinglePrecisionFloat=true
       containerType=None
       pinCategory=real
       pinSubCategory=double
       pinSubCategoryObject=null
       pinSubCategoryMemberReference {memberGuid=#1, memberName=None, memberParent=null}
       pinValueType {
        bTerminalIsConst=false
        bTerminalIsUObjectWrapper=false
        bTerminalIsWeakPointer=false
        terminalCategory=None
        terminalSubCategory=None
        terminalSubCategoryObject=null
       }
      }}
     Yaw {type {
       bIsConst=false
       bIsReference=false
       bIsUObjectWrapper=false
       bIsWeakPointer=false
       bSerializeAsSinglePrecisionFloat=true
       containerType=None
       pinCategory=real
       pinSubCategory=double
       pinSubCategoryObject=null
       pinSubCategoryMemberReference {memberGuid=#1, memberName=None, memberParent=null}
       pinValueType {
        bTerminalIsConst=false
        bTerminalIsUObjectWrapper=false
        bTerminalIsWeakPointer=false
        terminalCategory=None
        terminalSubCategory=None
        terminalSubCategoryObject=null
       }
      }}
     Pitch {type {
       bIsConst=false
       bIsReference=false
       bIsUObjectWrapper=false
       bIsWeakPointer=false
       bSerializeAsSinglePrecisionFloat=true
       containerType=None
       pinCategory=real
       pinSubCategory=double
       pinSubCategoryObject=null
       pinSubCategoryMemberReference {memberGuid=#1, memberName=None, memberParent=null}
       pinValueType {
        bTerminalIsConst=false
        bTerminalIsUObjectWrapper=false
        bTerminalIsWeakPointer=false
        terminalCategory=None
        terminalSubCategory=None
        terminalSubCategoryObject=null
       }
      }}
    ]
   }
   Input Pose!() -> (Pose) {name=InPose_PreAimOffsets}
   Output Pose() {layerGroup=Item Override}
  ]}
 ItemFallAdditive {nodes [
   Input Pose() {name=InPoseItemFallAdditive}
   Output Pose() {layerGroup=Item Override}
  ]}
 ItemFallLandAdditive {nodes [
   Input Pose() {name=InPoseItemFallLandAdditive}
   Output Pose() {layerGroup=Item Override}
  ]}
 ItemFallLevel2Override {nodes [
   Input Pose() {name=InPose_FallLevel2}
   Output Pose() {layerGroup=Item Override}
  ]}
 ItemFinalPoseOverride {nodes [
   Input Pose() {name=InFinalPose}
   Output Pose() {layerGroup=Item Override}
  ]}
 ItemFullBodyAimOffsets {nodes [
   Input Pose!() -> (Pose, FullBody_AimOffsetAlpha:double=~, Aim_Yaw:double=~, Aim_Pitch:double=~) {
    name=InPoseFullBodyAimOffsets
    inputs [
     FullBody_AimOffsetAlpha {type {
       bIsConst=false
       bIsReference=false
       bIsUObjectWrapper=false
       bIsWeakPointer=false
       bSerializeAsSinglePrecisionFloat=false
       containerType=None
       pinCategory=real
       pinSubCategory=double
       pinSubCategoryObject=null
       pinSubCategoryMemberReference {memberGuid=#1, memberName=None, memberParent=null}
       pinValueType {
        bTerminalIsConst=false
        bTerminalIsUObjectWrapper=false
        bTerminalIsWeakPointer=false
        terminalCategory=None
        terminalSubCategory=None
        terminalSubCategoryObject=null
       }
      }}
     Aim_Yaw {type {
       bIsConst=false
       bIsReference=false
       bIsUObjectWrapper=false
       bIsWeakPointer=false
       bSerializeAsSinglePrecisionFloat=false
       containerType=None
       pinCategory=real
       pinSubCategory=double
       pinSubCategoryObject=null
       pinSubCategoryMemberReference {memberGuid=#1, memberName=None, memberParent=null}
       pinValueType {
        bTerminalIsConst=false
        bTerminalIsUObjectWrapper=false
        bTerminalIsWeakPointer=false
        terminalCategory=None
        terminalSubCategory=None
        terminalSubCategoryObject=null
       }
      }}
     Aim_Pitch {type {
       bIsConst=false
       bIsReference=false
       bIsUObjectWrapper=false
       bIsWeakPointer=false
       bSerializeAsSinglePrecisionFloat=false
       containerType=None
       pinCategory=real
       pinSubCategory=double
       pinSubCategoryObject=null
       pinSubCategoryMemberReference {memberGuid=#1, memberName=None, memberParent=null}
       pinValueType {
        bTerminalIsConst=false
        bTerminalIsUObjectWrapper=false
        bTerminalIsWeakPointer=false
        terminalCategory=None
        terminalSubCategory=None
        terminalSubCategoryObject=null
       }
      }}
    ]
   }
   Input Pose!() -> (Pose) {name=InPosePreFullBodyAimOffsets}
   Output Pose() {layerGroup=Item Override}
  ]}
 ItemFullBodyOverride {nodes [
   Input Pose() {name=InPoseFullBody}
   Output Pose() {layerGroup=Item Override}
  ]}
 ItemIdleOverride {nodes [
   Input Pose() {name=InPose_Idle}
   Output Pose() {layerGroup=Item Override}
  ]}
 ItemJetPackJumpAdditive {nodes [Output Pose() {layerGroup=Item Override}]}
 ItemJetPackStartAdditive {nodes [Output Pose() {layerGroup=Item Override}]}
 ItemJumpLoopAdditive {nodes [
   Input Pose() {name=InPoseItemJumpLoopAdditive}
   Output Pose() {layerGroup=Item Override}
  ]}
 ItemJumpOverride {nodes [
   Input Pose() {name=InPoseJump}
   Output Pose() {layerGroup=Item Override}
  ]}
 ItemJumpUpAdditive {nodes [
   Input Pose() {name=InPoseItemJumpUpAdditive}
   Output Pose() {layerGroup=Item Override}
  ]}
 ItemPreAimOffsets {nodes [
   Input Pose() {name=InPosePreAimOffsets}
   Output Pose() {layerGroup=Item Override}
  ]}
 ItemPreIK {nodes [
   Input Pose() {name=InPose_PreIK}
   Output Pose() {layerGroup=Item Override}
  ]}
 ItemRightHandIK {nodes [
   Input Pose() {name=InPosePreRightHandIK}
   Input Pose() {name=InPosePostRightHandIK}
   Output Pose() {layerGroup=Item Override}
  ]}
 ItemSkeletalControl {nodes [
   Input Pose()
   Output Pose() {layerGroup=Item Override}
  ]}
 ItemSwimJumpFallLoop {nodes [Output Pose() {layerGroup=Item Override}]}
 ItemSwimJumpSurfaceEnd {nodes [Output Pose() {layerGroup=Item Override}]}
 ItemSwimJumpSurfaceLoop {nodes [Output Pose() {layerGroup=Item Override}]}
 ItemUpperBody {nodes [
   Output Pose() {layerGroup=Item Override}
   AnimGraphNode_SubInput_0 = Input Pose!() -> (Pose, InputParam:Fort Anim Input Adjusted Aim Structure) {
    name=InPoseUpperBody
    inputs [InputParam {type {
       bIsConst=false
       bIsReference=false
       bIsUObjectWrapper=false
       bIsWeakPointer=false
       bSerializeAsSinglePrecisionFloat=false
       containerType=None
       pinCategory=struct
       pinSubCategory=None
       pinSubCategoryObject=/Script/FortniteGame.FortAnimInput_AdjustedAim
       pinSubCategoryMemberReference {memberGuid=#1, memberName=None, memberParent=null}
       pinValueType {
        bTerminalIsConst=false
        bTerminalIsUObjectWrapper=false
        bTerminalIsWeakPointer=false
        terminalCategory=None
        terminalSubCategory=None
        terminalSubCategoryObject=null
       }
      }}]
   }
  ]}
 ItemZipLineStartAdditive {nodes [Output Pose() {layerGroup=Item Override}]}
]
variables []
