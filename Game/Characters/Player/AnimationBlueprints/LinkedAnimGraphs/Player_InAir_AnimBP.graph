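# Player in-air linked anim graph that plays jump and fall animations, passing the input pose through for jetpack, goop or disabled in-air animation.
# FullBodyFall switches between Jump, which alternates duplicate players for re-entry, and two fall levels.
# Level 2 adds predicted landing anticipation, level 3 picks loops by held item, and item layers can override Jump and level 2.
# Inertialized blends resolve through Dead Blending before the LocomotionInAir slot.
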AnimBlueprint /Game/Characters/Player/AnimationBlueprints/LinkedAnimGraphs/Player_InAir_AnimBP
target_skeleton=&Fortnite_M_Avg_Player_Skeleton
graphs [
 AnimGraph {nodes [
   AnimGraphNode_StateMachine_0 = FullBodyFall() {is=title AnimGraphNode_StateMachine, bAllowConduitEntryStates=true, maxTransitionsPerFrame=1}
   n1 = Use cached pose 'SourcePose'()
   n2 = Get bDisableInAirAnimations() -> (bDisableInAirAnimations:bool)
   n3 = Blend Poses by bool(
    BlendPose_0=Use cached pose 'SourcePose'()
    BlendPose_1=Blend Poses by bool(
     BlendPose_0=Use cached pose 'SourcePose'()
     BlendPose_1=Blend Poses by bool(BlendPose_0=@n1, BlendPose_1=@AnimGraphNode_StateMachine_0, BlendTime_1=0.3, bActiveValue=@n2) {blendType=Linear}
     bActiveValue=Get bIsInGoop() -> (bIsInGoop:bool)
    ) {blendType=Linear}
    BlendTime_1=0.5
    bActiveValue=Get bIsUsingJetpack() -> (bIsUsingJetpack:bool)
   ) {blendType=Linear, transitionType=Inertialization}
   Output Pose(Result=Slot 'LocomotionInAir'(Source=Dead Blending!(Source:Pose Link Structure=@n3) -> (Pose:Pose Link Structure) {is=title AnimGraphNode_DeadBlending, defaultBlendMode=Linear}) {slotName=LocomotionInAir}) {layerGroup=None}
   Save cached pose 'SourcePose'(Pose=Input Pose() {name=SourcePose}) {cacheName=SourcePose}
  ]}
 FullBodyFall {
  entry=Conduit
  path=AnimGraph.AnimGraphNode_StateMachine_0.FullBodyFall
  ids=id
  nodes [
   AnimStateConduitNode_0 = Conduit()
   AnimStateNode_0 = Jump() {
    is=title AnimStateNode
    stateEntered {guid=#1}
    stateFullyBlended {guid=#2}
    stateLeft {guid=#3}
   }
   AnimStateNode_1 = Fall_Level_03() {
    is=title AnimStateNode
    stateEntered {guid=#4}
    stateFullyBlended {guid=#5}
    stateLeft {guid=#6}
   }
   AnimStateNode_2 = Fall_Level_02() {
    is=title AnimStateNode
    stateEntered {guid=#7}
    stateFullyBlended {guid=#8}
    stateLeft {guid=#9}
   }
   AnimStateTransitionNode_0 = AnimStateTransitionNode() {crossfadeDuration=0.5, logicType=TLT_Inertialization}
   AnimStateTransitionNode_1 = AnimStateTransitionNode() {logicType=TLT_Inertialization}
   AnimStateTransitionNode_10 = AnimStateTransitionNode() {crossfadeDuration=0.1, logicType=TLT_Inertialization, priorityOrder=10}
   AnimStateTransitionNode_2 = AnimStateTransitionNode() {crossfadeDuration=1, logicType=TLT_Inertialization}
   AnimStateTransitionNode_3 = AnimStateTransitionNode() {crossfadeDuration=0.5, logicType=TLT_Inertialization, priorityOrder=10}
   AnimStateTransitionNode_4 = AnimStateTransitionNode() {logicType=TLT_Inertialization, priorityOrder=100}
   AnimStateTransitionNode_5 = AnimStateTransitionNode() {crossfadeDuration=0}
   AnimStateTransitionNode_6 = AnimStateTransitionNode() {crossfadeDuration=0.5, logicType=TLT_Inertialization, priorityOrder=10}
   AnimStateTransitionNode_7 = AnimStateTransitionNode() {bAutomaticRuleBasedOnSequencePlayerInState=true, crossfadeDuration=0.5, logicType=TLT_Inertialization}
   AnimStateTransitionNode_8 = AnimStateTransitionNode() {logicType=TLT_Inertialization}
   AnimStateTransitionNode_9 = AnimStateTransitionNode() {crossfadeDuration=0.5, logicType=TLT_Inertialization}
  ]
  links [
   {
    dst=Fall_Level_02
    rule=AnimGraph.AnimGraphNode_StateMachine_0.FullBodyFall.AnimStateTransitionNode_2.Transition
    src=Conduit
    via=AnimStateTransitionNode_2
   }
   {
    dst=Jump
    rule=AnimGraph.AnimGraphNode_StateMachine_0.FullBodyFall.AnimStateTransitionNode_4.Transition
    src=Conduit
    via=AnimStateTransitionNode_4
   }
   {
    dst=Fall_Level_03
    rule=AnimGraph.AnimGraphNode_StateMachine_0.FullBodyFall.AnimStateTransitionNode_0.Transition
    src=Fall_Level_02
    via=AnimStateTransitionNode_0
   }
   {
    dst=Jump
    rule=AnimGraph.AnimGraphNode_StateMachine_0.FullBodyFall.AnimStateTransitionNode_10.Transition
    src=Fall_Level_02
    via=AnimStateTransitionNode_10
   }
   {
    dst=Jump
    rule=AnimGraph.AnimGraphNode_StateMachine_0.FullBodyFall.AnimStateTransitionNode_8.Transition
    src=Fall_Level_02
    via=AnimStateTransitionNode_8
   }
   {
    dst=Fall_Level_02
    rule=AnimGraph.AnimGraphNode_StateMachine_0.FullBodyFall.AnimStateTransitionNode_5.Transition
    src=Fall_Level_03
    via=AnimStateTransitionNode_5
   }
   {
    dst=Fall_Level_02
    rule=AnimGraph.AnimGraphNode_StateMachine_0.FullBodyFall.AnimStateTransitionNode_6.Transition
    src=Fall_Level_03
    via=AnimStateTransitionNode_6
   }
   {
    dst=Jump
    rule=AnimGraph.AnimGraphNode_StateMachine_0.FullBodyFall.AnimStateTransitionNode_1.Transition
    src=Fall_Level_03
    via=AnimStateTransitionNode_1
   }
   {
    dst=Fall_Level_02
    rule=AnimGraph.AnimGraphNode_StateMachine_0.FullBodyFall.AnimStateTransitionNode_3.Transition
    src=Jump
    via=AnimStateTransitionNode_3
   }
   {
    dst=Fall_Level_02
    rule=AnimGraph.AnimGraphNode_StateMachine_0.FullBodyFall.AnimStateTransitionNode_7.Transition
    src=Jump
    via=AnimStateTransitionNode_7
   }
   {
    dst=Fall_Level_03
    rule=AnimGraph.AnimGraphNode_StateMachine_0.FullBodyFall.AnimStateTransitionNode_9.Transition
    src=Jump
    via=AnimStateTransitionNode_9
   }
  ]
 }
 Conduit {path=AnimGraph.AnimGraphNode_StateMachine_0.FullBodyFall.AnimStateConduitNode_0.Conduit, nodes [Result(bCanEnterTransition=true)]}
 Jump {
  path=AnimGraph.AnimGraphNode_StateMachine_0.FullBodyFall.AnimStateNode_0.Jump
  nodes [
   n1 = Blendspace Player 'Player_Jump_Start_BS'!(X=Get MovementDirectionAzimuthSmoothed() -> (MovementDirectionAzimuthSmoothed:float)) -> (Pose) {
    bLoop=false
    blendSpace=/Game/Animation/Game/MainPlayer/Locomotion/Medium/Male/Jump/Layered/Player_Jump_Start_BS
    groupName=Jump2
    method=SyncGroup
   }
   n2 = Sequence Player 'NoWep_Jump_From_Idle'() {
    bLoopAnimation=false
    groupName=Jump2
    method=SyncGroup
    sequence=/Game/Animation/Game/MainPlayer/Locomotion/Medium/Male/Jump/Layered/NoWep_Jump_From_Idle
   }
   n3 = Blendspace Player 'Player_Jump_Start_BS'(X=Get MovementDirectionAzimuthSmoothed()) {
    bLoop=false
    blendSpace=&Player_Jump_Start_BS
    groupName=Jump
    method=SyncGroup
   }
   n4 = Sequence Player 'NoWep_Jump_From_Idle'() {
    bLoopAnimation=false
    groupName=Jump
    method=SyncGroup
    sequence=&NoWep_Jump_From_Idle
   }
   n5 = Blend Poses by bool(
    BlendPose_0=Blend Poses by bool(BlendPose_0=@n3, BlendPose_1=@n4, BlendTime_0=0.4, BlendTime_1=0.4, bActiveValue=Get bIsMoving2D()) {blendType=Linear, transitionType=Inertialization}
    BlendPose_1=Blend Poses by bool(BlendPose_0=@n1, BlendPose_1=@n2, BlendTime_0=0.4, BlendTime_1=0.4, bActiveValue=Get bIsMoving2D()) {blendType=Linear, transitionType=Inertialization}
    bActiveValue=Get bJumpReentryFlip() -> (bJumpReentryFlip:bool)
   ) {
    node {blendType=Linear, childUpateMode=ResetChildOnActivate, updateFunction {functionName=OnJumpStateUpdate}}
    updateFunction {bSelfContext=true, memberName=OnJumpStateUpdate}
   }
   n6 = Sequence Player 'NoWep_Jump_From_Idle'!(StartPosition:float=0.2) -> (Pose) {
    bLoopAnimation=false
    groupName=Jump2
    method=SyncGroup
    sequence=&NoWep_Jump_From_Idle
   }
   n7 = ItemAnimLayerInterface - ItemJumpOverride!(InPoseJump:Pose Link Structure=Blend Poses by bool(BlendPose_0=@n6, BlendPose_1=@n5, bActiveValue=Get bIsWaterJump() -> (bIsWaterJump:bool)) {blendType=Linear, childUpateMode=ResetChildOnActivate}) -> (Pose:Pose Link Structure) {
    is=title AnimGraphNode_LinkedAnimLayer
    interface=/Game/Characters/Player/Male/Male_Avg_Base/AnimLayerInterface/ItemAnimLayerInterface.ItemAnimLayerInterface_C
    layer=ItemJumpOverride
   }
   Output Animation Pose(Result=@n7)
  ]
 }
 Fall_Level_03 {
  path=AnimGraph.AnimGraphNode_StateMachine_0.FullBodyFall.AnimStateNode_1.Fall_Level_03
  nodes [
   AnimGraphNode_StateMachine_0 = FallLevel_3() {is=title AnimGraphNode_StateMachine, bAllowConduitEntryStates=true}
   Output Animation Pose(Result=@AnimGraphNode_StateMachine_0)
  ]
 }
 FallLevel_3 {
  entry=Conduit
  path=AnimGraph.AnimGraphNode_StateMachine_0.FullBodyFall.AnimStateNode_1.Fall_Level_03.AnimGraphNode_StateMachine_0.FallLevel_3
  ids=id
  nodes [
   AnimStateConduitNode_0 = Conduit()
   AnimStateNode_0 = Default() {is=title AnimStateNode}
   AnimStateNode_1 = Shoulder() {is=title AnimStateNode}
   AnimStateNode_2 = Melee_DualWield() {is=title AnimStateNode}
   AnimStateNode_3 = DualWield() {is=title AnimStateNode}
   AnimStateNode_4 = Melee() {is=title AnimStateNode}
   AnimStateNode_5 = Small_Object() {is=title AnimStateNode}
   AnimStateNode_6 = Large_Consumable() {is=title AnimStateNode}
   AnimStateTransitionNode_0 = AnimStateTransitionNode() {crossfadeDuration=0.4, logicType=TLT_Inertialization, priorityOrder=100}
   AnimStateTransitionNode_1 = AnimStateTransitionNode() {crossfadeDuration=0.4, logicType=TLT_Inertialization, priorityOrder=2}
   AnimStateTransitionNode_10 = AnimStateTransitionNode() {crossfadeDuration=0.4, logicType=TLT_Inertialization, priorityOrder=6}
   AnimStateTransitionNode_11 = AnimStateTransitionNode()
   AnimStateTransitionNode_12 = AnimStateTransitionNode() {crossfadeDuration=0.4, logicType=TLT_Inertialization, priorityOrder=5}
   AnimStateTransitionNode_13 = AnimStateTransitionNode()
   AnimStateTransitionNode_2 = AnimStateTransitionNode() {crossfadeDuration=0.4, logicType=TLT_Inertialization}
   AnimStateTransitionNode_3 = AnimStateTransitionNode() {crossfadeDuration=0.4, logicType=TLT_Inertialization, priorityOrder=3}
   AnimStateTransitionNode_4 = AnimStateTransitionNode()
   AnimStateTransitionNode_5 = AnimStateTransitionNode()
   AnimStateTransitionNode_6 = AnimStateTransitionNode()
   AnimStateTransitionNode_7 = AnimStateTransitionNode()
   AnimStateTransitionNode_8 = AnimStateTransitionNode() {crossfadeDuration=0.4, logicType=TLT_Inertialization, priorityOrder=4}
   AnimStateTransitionNode_9 = AnimStateTransitionNode()
  ]
  links [
   {
    dst=Default
    rule=AnimGraph.AnimGraphNode_StateMachine_0.FullBodyFall.AnimStateNode_1.Fall_Level_03.AnimGraphNode_StateMachine_0.FallLevel_3.AnimStateTransitionNode_0.Transition
    src=Conduit
    via=AnimStateTransitionNode_0
   }
   {
    dst=DualWield
    rule=AnimGraph.AnimGraphNode_StateMachine_0.FullBodyFall.AnimStateNode_1.Fall_Level_03.AnimGraphNode_StateMachine_0.FallLevel_3.AnimStateTransitionNode_10.Transition
    src=Conduit
    via=AnimStateTransitionNode_10
   }
   {
    dst=Large_Consumable
    rule=AnimGraph.AnimGraphNode_StateMachine_0.FullBodyFall.AnimStateNode_1.Fall_Level_03.AnimGraphNode_StateMachine_0.FallLevel_3.AnimStateTransitionNode_12.Transition
    src=Conduit
    via=AnimStateTransitionNode_12
   }
   {
    dst=Melee
    rule=AnimGraph.AnimGraphNode_StateMachine_0.FullBodyFall.AnimStateNode_1.Fall_Level_03.AnimGraphNode_StateMachine_0.FallLevel_3.AnimStateTransitionNode_3.Transition
    src=Conduit
    via=AnimStateTransitionNode_3
   }
   {
    dst=Melee_DualWield
    rule=AnimGraph.AnimGraphNode_StateMachine_0.FullBodyFall.AnimStateNode_1.Fall_Level_03.AnimGraphNode_StateMachine_0.FallLevel_3.AnimStateTransitionNode_1.Transition
    src=Conduit
    via=AnimStateTransitionNode_1
   }
   {
    dst=Shoulder
    rule=AnimGraph.AnimGraphNode_StateMachine_0.FullBodyFall.AnimStateNode_1.Fall_Level_03.AnimGraphNode_StateMachine_0.FallLevel_3.AnimStateTransitionNode_2.Transition
    src=Conduit
    via=AnimStateTransitionNode_2
   }
   {
    dst=Small_Object
    rule=AnimGraph.AnimGraphNode_StateMachine_0.FullBodyFall.AnimStateNode_1.Fall_Level_03.AnimGraphNode_StateMachine_0.FallLevel_3.AnimStateTransitionNode_8.Transition
    src=Conduit
    via=AnimStateTransitionNode_8
   }
   {
    dst=Conduit
    rule=AnimGraph.AnimGraphNode_StateMachine_0.FullBodyFall.AnimStateNode_1.Fall_Level_03.AnimGraphNode_StateMachine_0.FallLevel_3.AnimStateTransitionNode_7.Transition
    src=Default
    via=AnimStateTransitionNode_7
   }
   {
    dst=Conduit
    rule=AnimGraph.AnimGraphNode_StateMachine_0.FullBodyFall.AnimStateNode_1.Fall_Level_03.AnimGraphNode_StateMachine_0.FallLevel_3.AnimStateTransitionNode_11.Transition
    src=DualWield
    via=AnimStateTransitionNode_11
   }
   {
    dst=Conduit
    rule=AnimGraph.AnimGraphNode_StateMachine_0.FullBodyFall.AnimStateNode_1.Fall_Level_03.AnimGraphNode_StateMachine_0.FallLevel_3.AnimStateTransitionNode_13.Transition
    src=Large_Consumable
    via=AnimStateTransitionNode_13
   }
   {
    dst=Conduit
    rule=AnimGraph.AnimGraphNode_StateMachine_0.FullBodyFall.AnimStateNode_1.Fall_Level_03.AnimGraphNode_StateMachine_0.FallLevel_3.AnimStateTransitionNode_4.Transition
    src=Melee
    via=AnimStateTransitionNode_4
   }
   {
    dst=Conduit
    rule=AnimGraph.AnimGraphNode_StateMachine_0.FullBodyFall.AnimStateNode_1.Fall_Level_03.AnimGraphNode_StateMachine_0.FallLevel_3.AnimStateTransitionNode_5.Transition
    src=Melee_DualWield
    via=AnimStateTransitionNode_5
   }
   {
    dst=Conduit
    rule=AnimGraph.AnimGraphNode_StateMachine_0.FullBodyFall.AnimStateNode_1.Fall_Level_03.AnimGraphNode_StateMachine_0.FallLevel_3.AnimStateTransitionNode_6.Transition
    src=Shoulder
    via=AnimStateTransitionNode_6
   }
   {
    dst=Conduit
    rule=AnimGraph.AnimGraphNode_StateMachine_0.FullBodyFall.AnimStateNode_1.Fall_Level_03.AnimGraphNode_StateMachine_0.FallLevel_3.AnimStateTransitionNode_9.Transition
    src=Small_Object
    via=AnimStateTransitionNode_9
   }
  ]
 }
 Conduit {
  path=AnimGraph.AnimGraphNode_StateMachine_0.FullBodyFall.AnimStateNode_1.Fall_Level_03.AnimGraphNode_StateMachine_0.FallLevel_3.AnimStateConduitNode_0.Conduit
  nodes [Result(bCanEnterTransition=true)]
 }
 Default {
  path=AnimGraph.AnimGraphNode_StateMachine_0.FullBodyFall.AnimStateNode_1.Fall_Level_03.AnimGraphNode_StateMachine_0.FallLevel_3.AnimStateNode_0.Default
  nodes [Output Animation Pose(Result=Sequence Player 'AR_InAir_Fall_Loop_Lvl3'() {sequence=/Game/Animation/Game/MainPlayer/Locomotion/Medium/Male/InAir/AR_InAir_Fall_Loop_Lvl3})]
 }
 Shoulder {
  path=AnimGraph.AnimGraphNode_StateMachine_0.FullBodyFall.AnimStateNode_1.Fall_Level_03.AnimGraphNode_StateMachine_0.FallLevel_3.AnimStateNode_1.Shoulder
  nodes [Output Animation Pose(Result=Sequence Player 'HeavyWep_InAir_Fall_Loop_Lvl3'() {sequence=/Game/Animation/Game/MainPlayer/Locomotion/Medium/Male/InAir/HeavyWep_InAir_Fall_Loop_Lvl3})]
 }
 Melee_DualWield {
  path=AnimGraph.AnimGraphNode_StateMachine_0.FullBodyFall.AnimStateNode_1.Fall_Level_03.AnimGraphNode_StateMachine_0.FallLevel_3.AnimStateNode_2.Melee_DualWield
  nodes [Output Animation Pose(Result=Sequence Player 'NoWep_InAir_Fall_Loop_Lvl3'() {sequence=/Game/Animation/Game/MainPlayer/Locomotion/Medium/Male/InAir/NoWep_InAir_Fall_Loop_Lvl3})]
 }
 DualWield {
  path=AnimGraph.AnimGraphNode_StateMachine_0.FullBodyFall.AnimStateNode_1.Fall_Level_03.AnimGraphNode_StateMachine_0.FallLevel_3.AnimStateNode_3.DualWield
  nodes [Output Animation Pose(Result=Sequence Player 'NoWep_InAir_Fall_Loop_Lvl3'() {sequence=&NoWep_InAir_Fall_Loop_Lvl3})]
 }
 Melee {
  path=AnimGraph.AnimGraphNode_StateMachine_0.FullBodyFall.AnimStateNode_1.Fall_Level_03.AnimGraphNode_StateMachine_0.FallLevel_3.AnimStateNode_4.Melee
  nodes [Output Animation Pose(Result=Sequence Player 'Pickaxe_InAir_Fall_Loop_Lvl3'() {sequence=/Game/Animation/Game/MainPlayer/Locomotion/Medium/Male/InAir/Pickaxe_InAir_Fall_Loop_Lvl3})]
 }
 Small_Object {
  path=AnimGraph.AnimGraphNode_StateMachine_0.FullBodyFall.AnimStateNode_1.Fall_Level_03.AnimGraphNode_StateMachine_0.FallLevel_3.AnimStateNode_5.Small_Object
  nodes [Output Animation Pose(Result=Sequence Player 'NoWep_InAir_Fall_Loop_Lvl3'() {sequence=&NoWep_InAir_Fall_Loop_Lvl3})]
 }
 Large_Consumable {
  path=AnimGraph.AnimGraphNode_StateMachine_0.FullBodyFall.AnimStateNode_1.Fall_Level_03.AnimGraphNode_StateMachine_0.FallLevel_3.AnimStateNode_6.Large_Consumable
  nodes [Output Animation Pose(Result=Sequence Player 'ConsumableLarge_InAir_Fall_Loop_Lvl3'() {sequence=/Game/Animation/Game/MainPlayer/Locomotion/Medium/Male/InAir/ConsumableLarge_InAir_Fall_Loop_Lvl3})]
 }
 Transition {
  path=AnimGraph.AnimGraphNode_StateMachine_0.FullBodyFall.AnimStateNode_1.Fall_Level_03.AnimGraphNode_StateMachine_0.FallLevel_3.AnimStateTransitionNode_0.Transition
  nodes [Result(bCanEnterTransition=true)]
 }
 Transition {
  path=AnimGraph.AnimGraphNode_StateMachine_0.FullBodyFall.AnimStateNode_1.Fall_Level_03.AnimGraphNode_StateMachine_0.FallLevel_3.AnimStateTransitionNode_1.Transition
  nodes [Result(bCanEnterTransition=Get bIsMeleeDualWield() -> (bIsMeleeDualWield:bool))]
 }
 Transition {
  path=AnimGraph.AnimGraphNode_StateMachine_0.FullBodyFall.AnimStateNode_1.Fall_Level_03.AnimGraphNode_StateMachine_0.FallLevel_3.AnimStateTransitionNode_10.Transition
  nodes [Result(bCanEnterTransition=Get bIsDualWield() -> (bIsDualWield:bool))]
 }
 Transition {
  path=AnimGraph.AnimGraphNode_StateMachine_0.FullBodyFall.AnimStateNode_1.Fall_Level_03.AnimGraphNode_StateMachine_0.FallLevel_3.AnimStateTransitionNode_11.Transition
  nodes [Result(bCanEnterTransition=true)]
 }
 Transition {
  path=AnimGraph.AnimGraphNode_StateMachine_0.FullBodyFall.AnimStateNode_1.Fall_Level_03.AnimGraphNode_StateMachine_0.FallLevel_3.AnimStateTransitionNode_12.Transition
  nodes [Result(bCanEnterTransition=Get bIsConsumableLarge() -> (bIsConsumableLarge:bool))]
 }
 Transition {
  path=AnimGraph.AnimGraphNode_StateMachine_0.FullBodyFall.AnimStateNode_1.Fall_Level_03.AnimGraphNode_StateMachine_0.FallLevel_3.AnimStateTransitionNode_13.Transition
  nodes [Result(bCanEnterTransition=true)]
 }
 Transition {
  path=AnimGraph.AnimGraphNode_StateMachine_0.FullBodyFall.AnimStateNode_1.Fall_Level_03.AnimGraphNode_StateMachine_0.FallLevel_3.AnimStateTransitionNode_2.Transition
  nodes [Result(bCanEnterTransition=Get bIsShoulderLauncher() -> (bIsShoulderLauncher:bool))]
 }
 Transition {
  path=AnimGraph.AnimGraphNode_StateMachine_0.FullBodyFall.AnimStateNode_1.Fall_Level_03.AnimGraphNode_StateMachine_0.FallLevel_3.AnimStateTransitionNode_3.Transition
  nodes [Result(bCanEnterTransition=Get bIsMelee() -> (bIsMelee:bool))]
 }
 Transition {
  path=AnimGraph.AnimGraphNode_StateMachine_0.FullBodyFall.AnimStateNode_1.Fall_Level_03.AnimGraphNode_StateMachine_0.FallLevel_3.AnimStateTransitionNode_4.Transition
  nodes [Result(bCanEnterTransition=true)]
 }
 Transition {
  path=AnimGraph.AnimGraphNode_StateMachine_0.FullBodyFall.AnimStateNode_1.Fall_Level_03.AnimGraphNode_StateMachine_0.FallLevel_3.AnimStateTransitionNode_5.Transition
  nodes [Result(bCanEnterTransition=true)]
 }
 Transition {
  path=AnimGraph.AnimGraphNode_StateMachine_0.FullBodyFall.AnimStateNode_1.Fall_Level_03.AnimGraphNode_StateMachine_0.FallLevel_3.AnimStateTransitionNode_6.Transition
  nodes [Result(bCanEnterTransition=true)]
 }
 Transition {
  path=AnimGraph.AnimGraphNode_StateMachine_0.FullBodyFall.AnimStateNode_1.Fall_Level_03.AnimGraphNode_StateMachine_0.FallLevel_3.AnimStateTransitionNode_7.Transition
  nodes [Result(bCanEnterTransition=true)]
 }
 Transition {
  path=AnimGraph.AnimGraphNode_StateMachine_0.FullBodyFall.AnimStateNode_1.Fall_Level_03.AnimGraphNode_StateMachine_0.FallLevel_3.AnimStateTransitionNode_8.Transition
  nodes [
   n1 = Get WeaponCoreAnim!() -> (WeaponCoreAnim)
   Result(bCanEnterTransition=OR Boolean!(
     A=Get bIsConsumableSmall() -> (bIsConsumableSmall:bool)
     B="Equal (Enum)"(A:EFortWeaponCoreAnimation Enum=@n1, B:EFortWeaponCoreAnimation Enum=Pistol)
     C:bool="Equal (Enum)"(A=@n1, B=MeleeOneHand)
    ) -> (ReturnValue))
  ]
 }
 Transition {
  path=AnimGraph.AnimGraphNode_StateMachine_0.FullBodyFall.AnimStateNode_1.Fall_Level_03.AnimGraphNode_StateMachine_0.FallLevel_3.AnimStateTransitionNode_9.Transition
  nodes [Result(bCanEnterTransition=true)]
 }
 Fall_Level_02 {
  path=AnimGraph.AnimGraphNode_StateMachine_0.FullBodyFall.AnimStateNode_2.Fall_Level_02
  nodes [
   AnimGraphNode_StateMachine_1 = Lvl2FallStateMachine() {is=title AnimGraphNode_StateMachine, bAllowConduitEntryStates=true}
   n1 = ItemAnimLayerInterface - ItemFallLevel2Override!(InPose_FallLevel2:Pose Link Structure=@AnimGraphNode_StateMachine_1) -> (Pose:Pose Link Structure) {
    is=title AnimGraphNode_LinkedAnimLayer
    interface=&ItemAnimLayerInterface_C
    layer=ItemFallLevel2Override
   }
   Output Animation Pose(Result=@n1)
  ]
 }
 Lvl2FallStateMachine {
  entry=Conduit
  path=AnimGraph.AnimGraphNode_StateMachine_0.FullBodyFall.AnimStateNode_2.Fall_Level_02.AnimGraphNode_StateMachine_1.Lvl2FallStateMachine
  ids=id
  nodes [
   AnimStateConduitNode_0 = Conduit()
   AnimStateNode_0 = LandAnticipation() {
    is=title AnimStateNode
    stateEntered {guid=#10}
    stateFullyBlended {guid=#11}
    stateLeft {guid=#12}
   }
   AnimStateNode_1 = Lvl2Fall() {is=title AnimStateNode}
   AnimStateNode_2 = LandAnticipationLoop() {
    is=title AnimStateNode
    stateEntered {guid=#13}
    stateFullyBlended {guid=#14}
    stateLeft {guid=#15}
   }
   AnimStateTransitionNode_1 = AnimStateTransitionNode() {crossfadeDuration=0.3, logicType=TLT_Inertialization}
   AnimStateTransitionNode_2 = AnimStateTransitionNode() {priorityOrder=100}
   AnimStateTransitionNode_3 = AnimStateTransitionNode() {automaticRuleTriggerTime=0, bAutomaticRuleBasedOnSequencePlayerInState=true, crossfadeDuration=0.1, logicType=TLT_Inertialization}
   AnimStateTransitionNode_4 = AnimStateTransitionNode()
  ]
  links [
   {
    dst=LandAnticipation
    rule=AnimGraph.AnimGraphNode_StateMachine_0.FullBodyFall.AnimStateNode_2.Fall_Level_02.AnimGraphNode_StateMachine_1.Lvl2FallStateMachine.AnimStateTransitionNode_1.Transition
    src=Conduit
    via=AnimStateTransitionNode_1
   }
   {
    dst=Lvl2Fall
    rule=AnimGraph.AnimGraphNode_StateMachine_0.FullBodyFall.AnimStateNode_2.Fall_Level_02.AnimGraphNode_StateMachine_1.Lvl2FallStateMachine.AnimStateTransitionNode_2.Transition
    src=Conduit
    via=AnimStateTransitionNode_2
   }
   {
    dst=LandAnticipationLoop
    rule=AnimGraph.AnimGraphNode_StateMachine_0.FullBodyFall.AnimStateNode_2.Fall_Level_02.AnimGraphNode_StateMachine_1.Lvl2FallStateMachine.AnimStateTransitionNode_3.Transition
    src=LandAnticipation
    via=AnimStateTransitionNode_3
   }
   {
    dst=Conduit
    rule=AnimGraph.AnimGraphNode_StateMachine_0.FullBodyFall.AnimStateNode_2.Fall_Level_02.AnimGraphNode_StateMachine_1.Lvl2FallStateMachine.AnimStateTransitionNode_4.Transition
    src=Lvl2Fall
    via=AnimStateTransitionNode_4
   }
  ]
 }
 Conduit {
  path=AnimGraph.AnimGraphNode_StateMachine_0.FullBodyFall.AnimStateNode_2.Fall_Level_02.AnimGraphNode_StateMachine_1.Lvl2FallStateMachine.AnimStateConduitNode_0.Conduit
  nodes [Result(bCanEnterTransition=true)]
 }
 LandAnticipation {
  path=AnimGraph.AnimGraphNode_StateMachine_0.FullBodyFall.AnimStateNode_2.Fall_Level_02.AnimGraphNode_StateMachine_1.Lvl2FallStateMachine.AnimStateNode_0.LandAnticipation
  nodes [Output Animation Pose(Result=Sequence Player 'NoWep_Land_Anticipation'!(StartPosition:float=Get LandAnticipationStartPosition() -> (LandAnticipationStartPosition:float)) -> (Pose) {bLoopAnimation=false, sequence=/Game/Animation/Game/MainPlayer/Locomotion/Medium/Male/InAir/NoWep_Land_Anticipation})]
 }
 Lvl2Fall {
  path=AnimGraph.AnimGraphNode_StateMachine_0.FullBodyFall.AnimStateNode_2.Fall_Level_02.AnimGraphNode_StateMachine_1.Lvl2FallStateMachine.AnimStateNode_1.Lvl2Fall
  nodes [Output Animation Pose(Result=Sequence Player 'NoWep_InAir_Fall_Loop_Lvl2'() {sequence=/Game/Animation/Game/MainPlayer/Locomotion/Medium/Male/InAir/NoWep_InAir_Fall_Loop_Lvl2})]
 }
 LandAnticipationLoop {
  path=AnimGraph.AnimGraphNode_StateMachine_0.FullBodyFall.AnimStateNode_2.Fall_Level_02.AnimGraphNode_StateMachine_1.Lvl2FallStateMachine.AnimStateNode_2.LandAnticipationLoop
  nodes [Output Animation Pose(Result=Sequence Player 'NoWep_Land_Anticipation_Loop'() {bLoopAnimation=false, sequence=/Game/Animation/Game/MainPlayer/Locomotion/Medium/Male/InAir/NoWep_Land_Anticipation_Loop})]
 }
 Transition {
  path=AnimGraph.AnimGraphNode_StateMachine_0.FullBodyFall.AnimStateNode_2.Fall_Level_02.AnimGraphNode_StateMachine_1.Lvl2FallStateMachine.AnimStateTransitionNode_1.Transition
  nodes [Result(bCanEnterTransition=Get bLandingPredicted() -> (bLandingPredicted:bool))]
 }
 Transition {
  path=AnimGraph.AnimGraphNode_StateMachine_0.FullBodyFall.AnimStateNode_2.Fall_Level_02.AnimGraphNode_StateMachine_1.Lvl2FallStateMachine.AnimStateTransitionNode_2.Transition
  nodes [Result(bCanEnterTransition=true)]
 }
 Transition {
  path=AnimGraph.AnimGraphNode_StateMachine_0.FullBodyFall.AnimStateNode_2.Fall_Level_02.AnimGraphNode_StateMachine_1.Lvl2FallStateMachine.AnimStateTransitionNode_3.Transition
  nodes [Result()]
 }
 Transition {
  path=AnimGraph.AnimGraphNode_StateMachine_0.FullBodyFall.AnimStateNode_2.Fall_Level_02.AnimGraphNode_StateMachine_1.Lvl2FallStateMachine.AnimStateTransitionNode_4.Transition
  nodes [Result(bCanEnterTransition=true)]
 }
 Transition {
  path=AnimGraph.AnimGraphNode_StateMachine_0.FullBodyFall.AnimStateTransitionNode_0.Transition
  nodes [Result(bCanEnterTransition=Get bIsInFallLevel3() -> (bIsInFallLevel3:bool))]
 }
 Transition {
  path=AnimGraph.AnimGraphNode_StateMachine_0.FullBodyFall.AnimStateTransitionNode_1.Transition
  nodes [Result(bCanEnterTransition=Break Fort Anim Input Transition Properties!(FortAnimInput_TransitionProperties:"Fort Anim Input Transition Properties Structure (by ref)"=Get TransitionPropertiesInput() -> (TransitionPropertiesInput:Fort Anim Input Transition Properties Structure)) -> (bTransition_DoubleJump_Fast:bool=~, bTransition_DoubleJump:bool=~) {is=title K2Node_BreakStruct}.bTransition_DoubleJump)]
 }
 Transition {
  path=AnimGraph.AnimGraphNode_StateMachine_0.FullBodyFall.AnimStateTransitionNode_10.Transition
  nodes [Result(bCanEnterTransition=Get bTransitionFallToJump() -> (bTransitionFallToJump:bool))]
 }
 Transition {
  path=AnimGraph.AnimGraphNode_StateMachine_0.FullBodyFall.AnimStateTransitionNode_2.Transition
  nodes [Result(bCanEnterTransition=Get bIsFalling())]
 }
 Transition {
  path=AnimGraph.AnimGraphNode_StateMachine_0.FullBodyFall.AnimStateTransitionNode_3.Transition
  nodes [Result(bCanEnterTransition=Get bTransitionJumpToFallLevel2() -> (bTransitionJumpToFallLevel2:bool))]
 }
 Transition {path=AnimGraph.AnimGraphNode_StateMachine_0.FullBodyFall.AnimStateTransitionNode_4.Transition, nodes [Result(bCanEnterTransition=true)]}
 Transition {
  path=AnimGraph.AnimGraphNode_StateMachine_0.FullBodyFall.AnimStateTransitionNode_5.Transition
  nodes [Result(bCanEnterTransition=Get bTransitionToFallLevel2NoBlend() -> (bTransitionToFallLevel2NoBlend:bool))]
 }
 Transition {
  path=AnimGraph.AnimGraphNode_StateMachine_0.FullBodyFall.AnimStateTransitionNode_6.Transition
  nodes [Result(bCanEnterTransition=Get bTransitionToFallLevel2() -> (bTransitionToFallLevel2:bool))]
 }
 Transition {path=AnimGraph.AnimGraphNode_StateMachine_0.FullBodyFall.AnimStateTransitionNode_7.Transition, nodes [Result()]}
 Transition {
  path=AnimGraph.AnimGraphNode_StateMachine_0.FullBodyFall.AnimStateTransitionNode_8.Transition
  nodes [Result(bCanEnterTransition=Break Fort Anim Input Transition Properties(FortAnimInput_TransitionProperties=Get TransitionPropertiesInput()) -> (bTransition_DoubleJump_Fast=~, bTransition_DoubleJump=~).bTransition_DoubleJump)]
 }
 Transition {
  path=AnimGraph.AnimGraphNode_StateMachine_0.FullBodyFall.AnimStateTransitionNode_9.Transition
  nodes [Result(bCanEnterTransition=Get bIsInFallLevel3())]
 }
 EventGraph {links [], nodes []}
 ItemAimOffset {nodes [
   Input Pose!() -> (Pose, AimOffsetAlpha:double, Yaw:double, Pitch:double) {name=InPose_AimOffset}
   Input Pose!() -> (Pose) {name=InPose_PreAimOffsets}
   Output Pose() {layerGroup=Item Override}
  ]}
 ItemFallAdditive {nodes [Output Pose(Result=Input Pose() {name=InPoseItemFallAdditive}) {layerGroup=Item Override}]}
 ItemFallLandAdditive {nodes [
   Input Pose() {name=InPoseItemFallLandAdditive}
   Output Pose() {layerGroup=Item Override}
  ]}
 ItemFallLevel2Override {nodes [Output Pose(Result=Input Pose() {name=InPose_FallLevel2}) {layerGroup=Item Override}]}
 ItemFinalPoseOverride {nodes [
   Input Pose() {name=InFinalPose}
   Output Pose() {layerGroup=Item Override}
  ]}
 ItemFullBodyAimOffsets {nodes [
   Input Pose!() -> (Pose, FullBody_AimOffsetAlpha:double, Aim_Yaw:double, Aim_Pitch:double) {name=InPoseFullBodyAimOffsets}
   Input Pose!() -> (Pose) {name=InPosePreFullBodyAimOffsets}
   Output Pose() {layerGroup=Item Override}
  ]}
 ItemFullBodyOverride {nodes [
   Input Pose() {name=InPoseFullBody}
   Output Pose() {layerGroup=Item Override}
  ]}
 ItemIdleOverride {nodes [
   Input Pose() {name=InPose_Idle}
   Output Pose() {layerGroup=Item Override}
  ]}
 ItemJetPackJumpAdditive {nodes [Output Pose() {layerGroup=Item Override}]}
 ItemJetPackStartAdditive {nodes [Output Pose() {layerGroup=Item Override}]}
 ItemJumpLoopAdditive {nodes [Output Pose(Result=Input Pose() {name=InPoseItemJumpLoopAdditive}) {layerGroup=Item Override}]}
 ItemJumpOverride {nodes [Output Pose(Result=Input Pose() {name=InPoseJump}) {layerGroup=Item Override}]}
 ItemJumpUpAdditive {nodes [Output Pose(Result=Input Pose() {name=InPoseItemJumpUpAdditive}) {layerGroup=Item Override}]}
 ItemPreAimOffsets {nodes [
   Input Pose() {name=InPosePreAimOffsets}
   Output Pose() {layerGroup=Item Override}
  ]}
 ItemPreIK {nodes [
   Input Pose() {name=InPose_PreIK}
   Output Pose() {layerGroup=Item Override}
  ]}
 ItemRightHandIK {nodes [
   Input Pose() {name=InPosePreRightHandIK}
   Input Pose() {name=InPosePostRightHandIK}
   Output Pose() {layerGroup=Item Override}
  ]}
 ItemSkeletalControl {nodes [
   Input Pose()
   Output Pose() {layerGroup=Item Override}
  ]}
 ItemSwimJumpFallLoop {nodes [Output Pose() {layerGroup=Item Override}]}
 ItemSwimJumpSurfaceEnd {nodes [Output Pose() {layerGroup=Item Override}]}
 ItemSwimJumpSurfaceLoop {nodes [Output Pose() {layerGroup=Item Override}]}
 ItemUpperBody {nodes [
   Input Pose!() -> (Pose, InputParam:Fort Anim Input Adjusted Aim Structure) {name=InPoseUpperBody}
   Output Pose() {layerGroup=Item Override}
  ]}
 ItemZipLineStartAdditive {nodes [Output Pose() {layerGroup=Item Override}]}
]
variables []
