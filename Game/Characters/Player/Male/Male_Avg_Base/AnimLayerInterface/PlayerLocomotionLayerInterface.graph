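# Animation layer interface for player locomotion, declaring override points that Anim Blueprints implement, with no logic or target skeleton of its own.
# Nine layers share the LocomotionOverride group, covering base and full locomotion, in-air, lower, upper and full body, additives, ground-land additive and final pose.
# Most layers take one input pose to pass through or modify; InAir also takes a source pose, while Additives takes none.
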AnimBlueprint /Game/Characters/Player/Male/Male_Avg_Base/AnimLayerInterface/PlayerLocomotionLayerInterface
target_skeleton=null
graphs [
 LocomotionLayer_Additives {nodes [Output Pose() {layerGroup=LocomotionOverride}]}
 LocomotionLayer_BaseLocomotion {nodes [
   Input Pose() {name=InPose_BaseLocomotion}
   Output Pose() {layerGroup=LocomotionOverride}
  ]}
 LocomotionLayer_FinalPose {nodes [
   Input Pose() {name=InPoseFinalPose}
   Output Pose() {layerGroup=LocomotionOverride}
  ]}
 LocomotionLayer_FullBody {nodes [
   Input Pose() {name=InPose_FullBody}
   Output Pose() {layerGroup=LocomotionOverride}
  ]}
 LocomotionLayer_FullLocomotion {nodes [
   Input Pose() {name=InPose_FullLocomotion}
   Output Pose() {layerGroup=LocomotionOverride}
  ]}
 LocomotionLayer_GroundLandAdditive {nodes [
   Input Pose()
   Output Pose() {layerGroup=LocomotionOverride}
  ]}
 LocomotionLayer_InAir {nodes [
   Input Pose() {name=InPose_InAir}
   Input Pose() {name=InPose_InAirSourcePose}
   Output Pose() {layerGroup=LocomotionOverride}
  ]}
 LocomotionLayer_LowerBody {nodes [
   Input Pose() {name=InPose_LowerBody}
   Output Pose() {layerGroup=LocomotionOverride}
  ]}
 LocomotionLayer_UpperBody {nodes [
   Input Pose() {name=InPose_UpperBody}
   Output Pose() {layerGroup=LocomotionOverride}
  ]}
]
variables []
