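# Linked anim graph for Fortnite player landing additives on ground and water.
# A conduit-entry state machine picks ground land or the swim fall, surface loop and surface end chain, using anim-set clips.
# Ground land selects additives by weapon, stance and movement through overridable item/locomotion layers, blends in Jackal vehicle landing, then damps by alpha.
# Node functions reset landing exit on becoming relevant and allow it near animation end.
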AnimBlueprint /Game/Characters/Player/AnimationBlueprints/LinkedAnimGraphs/Player_Land_AnimBP
target_skeleton=&Fortnite_M_Avg_Player_Skeleton
graphs [
 AnimGraph {nodes [
   AnimGraphNode_StateMachine_0 = Landing() {is=title AnimGraphNode_StateMachine, bAllowConduitEntryStates=true}
   Output Pose(Result=@AnimGraphNode_StateMachine_0) {
    becomeRelevantFunction {bSelfContext=true, memberGuid=#1, memberName=OnLandingBecomesRelevant}
    node {layerGroup=None, becomeRelevantFunction {functionName=OnLandingBecomesRelevant}}
   }
  ]}
 Landing {
  entry=Conduit
  path=AnimGraph.AnimGraphNode_StateMachine_0.Landing
  ids=id
  nodes [
   AnimStateConduitNode_0 = Conduit()
   AnimStateNode_17 = SwimJumpFallLoop() {
    is=title AnimStateNode
    bAlwaysResetOnEntry=true
    stateEntered {guid=#2}
    stateFullyBlended {guid=#3}
    stateLeft {guid=#4}
   }
   AnimStateNode_18 = SwimJumpSurfaceLoop() {
    is=title AnimStateNode
    bAlwaysResetOnEntry=true
    stateEntered {guid=#5}
    stateFullyBlended {guid=#6}
    stateLeft {guid=#7}
   }
   AnimStateNode_19 = SwimJumpSurfaceEnd() {
    is=title AnimStateNode
    bAlwaysResetOnEntry=true
    stateEntered {guid=#8}
    stateFullyBlended {guid=#9}
    stateLeft {guid=#10}
   }
   AnimStateNode_9 = GroundLandState() {
    is=title AnimStateNode
    stateEntered {guid=#11, notifyName=Land}
    stateFullyBlended {guid=#12}
    stateLeft {guid=#13}
   }
   AnimStateTransitionNode_0 = AnimStateTransitionNode() {crossfadeDuration=0.1, logicType=TLT_Inertialization, priorityOrder=2}
   AnimStateTransitionNode_1 = AnimStateTransitionNode() {crossfadeDuration=0.1, logicType=TLT_Inertialization}
   AnimStateTransitionNode_14 = AnimStateTransitionNode() {priorityOrder=4}
   AnimStateTransitionNode_17 = AnimStateTransitionNode() {priorityOrder=3}
   AnimStateTransitionNode_52 = AnimStateTransitionNode() {crossfadeDuration=0.4, priorityOrder=2}
   AnimStateTransitionNode_72 = AnimStateTransitionNode() {bAutomaticRuleBasedOnSequencePlayerInState=true, priorityOrder=3}
  ]
  links [
   {
    dst=GroundLandState
    rule=AnimGraph.AnimGraphNode_StateMachine_0.Landing.AnimStateTransitionNode_0.Transition
    src=Conduit
    via=AnimStateTransitionNode_0
   }
   {
    dst=SwimJumpFallLoop
    rule=AnimGraph.AnimGraphNode_StateMachine_0.Landing.AnimStateTransitionNode_1.Transition
    src=Conduit
    via=AnimStateTransitionNode_1
   }
   {
    dst=SwimJumpFallLoop
    rule=AnimGraph.AnimGraphNode_StateMachine_0.Landing.AnimStateTransitionNode_17.AnimationTransitionGraph_1
    src=GroundLandState
    via=AnimStateTransitionNode_17
   }
   {
    dst=SwimJumpSurfaceLoop
    rule=AnimGraph.AnimGraphNode_StateMachine_0.Landing.AnimStateTransitionNode_14.AnimationTransitionGraph_1
    src=SwimJumpFallLoop
    via=AnimStateTransitionNode_14
   }
   {
    dst=SwimJumpSurfaceLoop
    rule=AnimGraph.AnimGraphNode_StateMachine_0.Landing.AnimStateTransitionNode_72.AnimationTransitionGraph_0
    src=SwimJumpFallLoop
    via=AnimStateTransitionNode_72
   }
   {
    dst=SwimJumpSurfaceEnd
    rule=AnimGraph.AnimGraphNode_StateMachine_0.Landing.AnimStateTransitionNode_52.AnimationTransitionGraph_0
    src=SwimJumpSurfaceLoop
    via=AnimStateTransitionNode_52
   }
  ]
 }
 Conduit {path=AnimGraph.AnimGraphNode_StateMachine_0.Landing.AnimStateConduitNode_0.Conduit, nodes [Result(bCanEnterTransition=true)]}
 SwimJumpFallLoop {
  path=AnimGraph.AnimGraphNode_StateMachine_0.Landing.AnimStateNode_17.SwimJumpFallLoop
  nodes [
   n1 = ItemAnimLayerInterface - ItemSwimJumpFallLoop!() -> (Pose:Pose Link Structure) {
    is=title AnimGraphNode_LinkedAnimLayer
    interface=/Game/Characters/Player/Male/Male_Avg_Base/AnimLayerInterface/ItemAnimLayerInterface.ItemAnimLayerInterface_C
    layer=ItemSwimJumpFallLoop
   }
   Output Animation Pose(Result=@n1)
  ]
 }
 SwimJumpSurfaceLoop {
  path=AnimGraph.AnimGraphNode_StateMachine_0.Landing.AnimStateNode_18.SwimJumpSurfaceLoop
  nodes [
   n1 = ItemAnimLayerInterface - ItemSwimJumpSurfaceLoop!() -> (Pose:Pose Link Structure) {
    is=title AnimGraphNode_LinkedAnimLayer
    interface=&ItemAnimLayerInterface_C
    layer=ItemSwimJumpSurfaceLoop
   }
   Output Animation Pose(Result=@n1)
  ]
 }
 SwimJumpSurfaceEnd {
  path=AnimGraph.AnimGraphNode_StateMachine_0.Landing.AnimStateNode_19.SwimJumpSurfaceEnd
  nodes [
   n1 = ItemAnimLayerInterface - ItemSwimJumpSurfaceEnd!() -> (Pose:Pose Link Structure) {
    is=title AnimGraphNode_LinkedAnimLayer
    interface=&ItemAnimLayerInterface_C
    layer=ItemSwimJumpSurfaceEnd
   }
   Output Animation Pose(Result=@n1) {
    node {updateFunction {functionName=On New Function}}
    updateFunction {bSelfContext=true, memberGuid=#14, memberName=OnLandingUpdate}
   }
  ]
 }
 GroundLandState {
  path=AnimGraph.AnimGraphNode_StateMachine_0.Landing.AnimStateNode_9.GroundLandState
  nodes [
   AnimGraphNode_Layer_0 = ItemAnimLayerInterface - ItemFallLandAdditive!(InPoseItemFallLandAdditive:Pose Link Structure=FallLandAdditive!() -> (Pose:Pose Link Structure) {is=title AnimGraphNode_LinkedAnimLayer, layer=FallLandAdditive}) -> (Pose:Pose Link Structure) {
    is=title AnimGraphNode_LinkedAnimLayer
    interface=&ItemAnimLayerInterface_C
    layer=ItemFallLandAdditive
   }
   n1 = "Sequence Player (additive) 'M_JumpLand_ARDS_Additive'"() {
    bLoopAnimation=false
    groupName=JumpLand
    method=SyncGroup
    sequence=/Game/Animation/Game/MainPlayer/Locomotion/Medium/Male/Jump/AssaultRifle/M_JumpLand_ARDS_Additive
   }
   n2 = "Sequence Player (additive) 'JumpLand_GC_CMM_Additive'"() {
    bLoopAnimation=false
    groupName=JumpLand
    method=SyncGroup
    sequence=/Game/Animation/Game/MainPlayer/Locomotion/Medium/Male/Jump/GoingCommando/JumpLand_GC_CMM_Additive
   }
   n3 = "Sequence Player (additive) 'JumpLand_NoWep_CMM_Additive'"() {groupName=JumpLand, method=SyncGroup, sequence=/Game/Animation/Game/MainPlayer/Locomotion/Medium/Male/Jump/NoWep/JumpLand_NoWep_CMM_Additive}
   n4 = "Blend Poses (EFortWeaponCoreAnimation)"!(
    BlendPose_0:Pose Link Structure=@n1
    BlendPose_1:Pose Link Structure=@n3
    BlendPose_2:Pose Link Structure=@n2
    BlendPose_3:Pose Link Structure="Sequence Player (additive) 'DualWield_Land_CMM'"() {sequence=/Game/Animation/Game/MainPlayer/Locomotion/Medium/Male/Jump/DualWield/DualWield_Land_CMM}
    BlendTime_0:float=0
    BlendTime_1:float=0.1
    BlendTime_2:float=0.1
    BlendTime_3:float=0.1
    ActiveEnumValue:Byte=Get WeaponCoreAnim!() -> (WeaponCoreAnim)
   ) -> (Pose:Pose Link Structure) {
    is=title AnimGraphNode_BlendListByEnum
    blendPose [{linkId=-1, sourceLinkId=-1}, {linkId=-1, sourceLinkId=-1}, {linkId=-1, sourceLinkId=-1}, {linkId=-1, sourceLinkId=-1}]
    blendTime [0.1,0.1,0.1,0.1]
   }
   n5 = Additive Identity Pose()
   n6 = "Sequence Player (additive) 'JackalVehicle_JumpCombat_Land'"() {
    bLoopAnimation=false
    groupName=JumpLand
    method=SyncGroup
    sequence=/Game/Animation/Game/MainPlayer/Vehicles/JackalVehicle/JackalVehicle_JumpCombat_Land
   }
   n7 = Layered blend per bone(
    BasePose=Blend Poses by bool(
     BlendPose_0=Two Way Blend(A=@n4, B=@n5, Alpha=0.6) {blendNode {alphaBoolBlend {blendOption=Linear}}}
     BlendPose_1=@AnimGraphNode_Layer_0
     BlendTime_0=.2
     BlendTime_1=.2
     bActiveValue=Get bIsCrouching()
    ) {blendType=Linear}
    BlendPoses_0=@n6
    BlendWeights_0=Break Fort Anim Input Jackal Vehicle!(FortAnimInput_JackalVehicle:"Fort Anim Input Jackal Vehicle Structure (by ref)"=Get JackalVehicleInput() -> (JackalVehicleInput:Fort Anim Input Jackal Vehicle Structure)) -> (
     bIsUsingJackalVehicle:bool=~
     bIsSteeringLeft:bool=~
     bIsSteeringRight:bool=~
     SteerAngleDeadZoneDegrees:float=~
     SteerAngle:float=~
     LeanYaw:float=~
     QuantizedSteerAngle:float=~
     SteerAlpha:float=~
     RunForwardAlpha:float=~
     SlopePitchDegreeAngle:float=~
     SlopeRollDegreeAngle:float=~
     DistanceFromGround:float=~
     PivotPlayRate:float=~
     bIsReversing:bool=~
     bIsBraking:bool=~
     bIsMoving:bool=~
     bIsMovingForward:bool=~
     bIsSprinting:bool=~
     bInAir:bool=~
     bIsFalling:bool=~
     bIsJumping:bool=~
     bIsRelaxed:bool=~
     bIsBoosting:bool=~
     bHasReachedJumpChargeStartThreshold:bool=~
     bHasReachedJumpChargeMidThreshold:bool=~
     bHasReachedJumpChargeFullThreshold:bool=~
     bAdjustRootForFemaleRider:bool=~
     bIsPlayingEmoteOnHoverboard:bool=~
     bShouldApplyAdditive:bool=~
     bPlayPivotOnGroundAndNotBoosting:bool=~
     bIsOnLowerHill:bool=~
     bLowerBodyIdleToLoopTransition:bool=~
     bInterruptHoverboardFullybody:bool=~
     bEnterFullBodyHoverboardState:bool=~
     bDefaultToJumpStartTransition:bool=~
     bLocomotionPoseToJumpTransition:bool=~
     bBoostingToBoostingJumpTransition:bool=~
     bJumpToLocomotionPoseTransition:bool=~
     bJumpApexToJumpFallTransition:bool=~
     bIdleToMovementStartTransition:bool=~
     bIdleToMovementLoopTransition:bool=~
     bMovementLoopToMovementStopTransition:bool=~
     bMovementLoopToPivotTransition:bool=~
     bMovementLoopToIdleTransition:bool=~
     bIdleAdditiveToCollisionNTransition:bool=~
     bSplitBodyToHoverboardBRTransition:bool=~
     bHoverboardBRToSplitBodyTransition:bool=~
     bHoverboardBRMovementToJumpChargeTransition:bool=~
     bIdlesToJackalVehicleTransition:bool=~
     bPlayAdditiveLeans:bool=~
     bPlayBalloonLeans:bool=~
     bPlayJumpTrickVertical:bool=~
     bPlayJumpTrick:bool=~
     bPlayMovingFast:bool=~
     bPlayHipAdjustmentAdditive:bool=~
     bPlayDriveSouth:bool=~
     bPlayHeadAimOffset:bool=~
     bPlaySlopeAimOffset:bool=~
     JumpCombatAdditiveWeight:float=~
     MeleeTwistIdle:float=~
     MeleeTwistLocomotionLoop:float=~
     VelocityImpact:Fort Anim Input Velocity Impact Structure
     ImpactDisplacement:Vector=~
     SteerYaw:float=~
     EmoteHoverboardPosition:Vector=~
     EmoteHoverboardRotation:Rotator=~
     bShouldAttachFeetToHoverboard:bool=~
     FootLeftLocationOffset:Vector=~
     FootLeftRotationOffset:Rotator=~
     FootRightLocationOffset:Vector=~
     FootRightRotationOffset:Rotator=~
     StoppedThreshold:float=~
     MovingForwardThreshold:float=~
     MovingFowardFastThreshold:float=~
     DefaultToJumpStartTransitionThreshold:float=~
     JumpTrickAngularVelocityThreshold:float=~
     JumpDistanceFromGroundThreshold:float=~
     VelocityStartThreshold:float=~
     MovingFastThreshold:float=~
     RelaxedSpeedThreshold:float=~
     JumpChargeStartThreshold:float=~
     JumpChargeMidThreshold:float=~
     JumpChargeFullThreshold:float=~
     RotatingAngularVelocityThreshold:float=~
     IdleToLoopTransitionSpeedThreshold:float=~
     IdleToMovementLoopTransitionThreshold:float=~
     MeleeTwistIdleMultiplier:float=~
     MeleeTwistLocomotionLoopMultiplier:float=~
     LeanYawForMaxPivotPlayRate:float=~
     SteerAlphaForMaxPivotPlayRate:float=~
     PrePivotCardinalDirection:EFortCardinalDirection Enum
    ) {is=title K2Node_BreakStruct}.JumpCombatAdditiveWeight
   ) {
    bMeshSpaceRotationBlend=true
    lODThreshold=2
    layerSetup [{branchFilters [
       {boneName=root}
       {blendDepth=-1, boneName=pelvis}
       {boneName=thigh_r}
       {boneName=thigh_l}
       {blendDepth=-1, boneName=spine_01}
       {blendDepth=-1, boneName=Ik_hand_root}
       {boneName=ik_foot_root}
      ]}]
   }
   n8 = PlayerLocomotionLayerInterface - LocomotionLayer_GroundLandAdditive!(InPose:Pose Link Structure=@n7) -> (Pose:Pose Link Structure) {
    is=title AnimGraphNode_LinkedAnimLayer
    interface=/Game/Characters/Player/Male/Male_Avg_Base/AnimLayerInterface/PlayerLocomotionLayerInterface.PlayerLocomotionLayerInterface_C
    layer=LocomotionLayer_GroundLandAdditive
   }
   Output Animation Pose(Result=Two Way Blend(A=@n8, B=Additive Identity Pose(), Alpha=Get GroundLandingAlphaDamper() -> (GroundLandingAlphaDamper:float)) {blendNode {alphaBoolBlend {blendOption=Linear}}}) {
    node {updateFunction {functionName=On Surface Swim Update}}
    updateFunction {bSelfContext=true, memberGuid=#14, memberName=OnLandingUpdate}
   }
  ]
 }
 Transition {path=AnimGraph.AnimGraphNode_StateMachine_0.Landing.AnimStateTransitionNode_0.Transition, nodes [Result(bCanEnterTransition=true)]}
 Transition {
  path=AnimGraph.AnimGraphNode_StateMachine_0.Landing.AnimStateTransitionNode_1.Transition
  nodes [Result(bCanEnterTransition=Get bSwimmingIsWaterLanding() -> (bSwimmingIsWaterLanding:bool))]
 }
 AnimationTransitionGraph_1 {
  path=AnimGraph.AnimGraphNode_StateMachine_0.Landing.AnimStateTransitionNode_14.AnimationTransitionGraph_1
  nodes [Result(bCanEnterTransition=Get bSwimmingIsWaterPaddlingToSurface() -> (bSwimmingIsWaterPaddlingToSurface:bool))]
 }
 AnimationTransitionGraph_1 {
  path=AnimGraph.AnimGraphNode_StateMachine_0.Landing.AnimStateTransitionNode_17.AnimationTransitionGraph_1
  nodes [Result(bCanEnterTransition=Get bIsSurfaceSwimming())]
 }
 AnimationTransitionGraph_0 {
  path=AnimGraph.AnimGraphNode_StateMachine_0.Landing.AnimStateTransitionNode_52.AnimationTransitionGraph_0
  nodes [Result(bCanEnterTransition=Get bTransition_SwimJumpSurfaceLoop_to_SwimJumpSurfaceEnd() -> (bTransition_SwimJumpSurfaceLoop_to_SwimJumpSurfaceEnd:bool))]
 }
 AnimationTransitionGraph_0 {path=AnimGraph.AnimGraphNode_StateMachine_0.Landing.AnimStateTransitionNode_72.AnimationTransitionGraph_0, nodes [Result()]}
 EventGraph {links [], nodes []}
 FallLandAdditive {nodes [
   n1 = "Sequence Player (additive) 'JumpLandMoving_NoWep_CMM_Additive'"() {
    bLoopAnimation=false
    groupName=JumpLand
    method=SyncGroup
    sequence=/Game/Animation/Game/MainPlayer/Locomotion/Medium/Male/Jump/NoWep/JumpLandMoving_NoWep_CMM_Additive
   }
   n2 = "Sequence Player (additive) 'JumpLand_NoWep_CMM_Additive'"() {
    bLoopAnimation=false
    groupName=JumpLand
    method=SyncGroup
    sequence=&JumpLand_NoWep_CMM_Additive
   }
   n3 = Blend Poses by bool(
    BlendPose_0=@n1
    BlendPose_1=@n2
    BlendTime_0=0.2
    BlendTime_1=0.2
    bActiveValue=Get bIsAboveMinimumLocomotionSpeed() -> (bIsAboveMinimumLocomotionSpeed:bool)
   ) {blendType=Linear}
   n4 = "Sequence Player (additive) 'M_JumpLand_ARDS_Additive'"() {
    bLoopAnimation=false
    groupName=JumpLand
    method=SyncGroup
    sequence=&M_JumpLand_ARDS_Additive
   }
   n5 = "Sequence Player (additive) 'M_JumpLand_ARDS_Additive'"() {
    bLoopAnimation=false
    groupName=JumpLand
    method=SyncGroup
    sequence=&M_JumpLand_ARDS_Additive
   }
   n6 = "Sequence Player (additive) 'JumpLandMoving_NoWep_CMM_Additive'"() {
    bLoopAnimation=false
    groupName=JumpLand
    method=SyncGroup
    sequence=&JumpLandMoving_NoWep_CMM_Additive
   }
   n7 = "Sequence Player (additive) 'JumpLand_NoWep_CMM_Additive'"() {
    bLoopAnimation=false
    groupName=JumpLand
    method=SyncGroup
    sequence=&JumpLand_NoWep_CMM_Additive
   }
   n8 = Get bIsAboveMinimumLocomotionSpeed()
   n9 = "Sequence Player (additive) 'DualWield_Land_CMM'"() {
    bLoopAnimation=false
    groupName=JumpLand
    method=SyncGroup
    sequence=&DualWield_Land_CMM
   }
   n10 = "Sequence Player (additive) 'DualWield_Land_MovingAdditive_CMM'"() {
    bLoopAnimation=false
    groupName=JumpLand
    method=SyncGroup
    sequence=/Game/Animation/Game/MainPlayer/Locomotion/Medium/Male/Jump/DualWield/DualWield_Land_MovingAdditive_CMM
   }
   n11 = "Sequence Player (additive) 'Balloons_JumpLand_Low_Additive_CMM'"() {
    bLoopAnimation=false
    groupName=JumpLand
    method=SyncGroup
    sequence=/Game/Animation/Game/MainPlayer/Combat/Gadgets/Medium/Male/Balloons/Balloons_JumpLand_Low_Additive_CMM
   }
   n12 = "Sequence Player (additive) 'Balloons_JumpLand_Low_Additive_InMotion_CMM'"() {
    bLoopAnimation=false
    groupName=JumpLand
    method=SyncGroup
    sequence=/Game/Animation/Game/MainPlayer/Combat/Gadgets/Medium/Male/Balloons/Balloons_JumpLand_Low_Additive_InMotion_CMM
   }
   n13 = "Sequence Player (additive) 'Balloons_JumpLand_Low_Additive_CMF'"() {
    bLoopAnimation=false
    groupName=JumpLand
    method=SyncGroup
    sequence=/Game/Animation/Game/MainPlayer/Combat/Gadgets/Medium/Female/Balloons/Balloons_JumpLand_Low_Additive_CMF
   }
   n14 = "Sequence Player (additive) 'Balloons_JumpLand_Low_Additive_InMotion_CMF'"() {
    bLoopAnimation=false
    groupName=JumpLand
    method=SyncGroup
    sequence=/Game/Animation/Game/MainPlayer/Combat/Gadgets/Medium/Female/Balloons/Balloons_JumpLand_Low_Additive_InMotion_CMF
   }
   n15 = "Blend Poses (EFortCustomGender)"!(
    BlendPose_0:Pose Link Structure=Blend Poses by bool(BlendPose_0=@n12, BlendPose_1=@n11, BlendTime_0=.25, BlendTime_1=.25, bActiveValue=Get bIsMoving2D()) {blendType=Linear}
    BlendPose_1:Pose Link Structure=Blend Poses by bool(BlendPose_0=@n14, BlendPose_1=@n13, BlendTime_0=.25, BlendTime_1=.25, bActiveValue=Get bIsMoving2D()) {blendType=Linear}
    BlendTime_0:float=0.1
    BlendTime_1:float=0.1
    ActiveEnumValue:Byte=Get Gender() -> (Gender:EFortCustomGender Enum=Invalid)
   ) -> (Pose:Pose Link Structure) {
    is=title AnimGraphNode_BlendListByEnum
    blendType=Linear
    blendPose [{linkId=-1, sourceLinkId=-1}, {linkId=-1, sourceLinkId=-1}]
    blendTime [0.1,0.1]
   }
   n16 = "Sequence Player (additive) 'Balloons_JumpLand_Base_Additive_CMM'"() {
    bLoopAnimation=false
    groupName=JumpLand
    method=SyncGroup
    sequence=/Game/Animation/Game/MainPlayer/Combat/Gadgets/Medium/Male/Balloons/Balloons_JumpLand_Base_Additive_CMM
   }
   n17 = "Sequence Player (additive) 'M_JumpLand_OneHanded_Additive'"() {
    bLoopAnimation=false
    groupName=JumpLand
    method=SyncGroup
    sequence=/Game/Animation/Game/MainPlayer/Locomotion/Medium/Male/Jump/AssaultRifle/M_JumpLand_OneHanded_Additive
   }
   n18 = "Sequence Player (additive) 'M_JumpLand_ARDS_Additive'"() {
    bLoopAnimation=false
    groupName=JumpLand
    method=SyncGroup
    sequence=&M_JumpLand_ARDS_Additive
   }
   n19 = "Sequence Player (additive) 'M_JumpLand_DualSMG_NonTargeted_Additive'"() {
    bLoopAnimation=false
    sequence=/Game/Animation/Game/MainPlayer/Locomotion/Medium/Male/Jump/AssaultRifle/M_JumpLand_DualSMG_NonTargeted_Additive
   }
   n20 = "Blend Poses (EFortWeaponCoreAnimation)"!(
    BlendPose_0=Blend Poses by bool(BlendPose_0=@n18, BlendPose_1=@n17, bActiveValue=Get bBothHandsOnWeapon() -> (bBothHandsOnWeapon:bool)) {blendType=Linear}
    BlendPose_1=Blend Poses by bool(BlendPose_0=@n15, BlendPose_1=@n16, BlendTime_0=0.2, BlendTime_1=0.2) {blendType=Linear}
    BlendPose_2=@n19
    BlendTime_0=0
    BlendTime_1=0.1
    BlendTime_2=0.1
    ActiveEnumValue=Get WeaponCoreAnim()
   ) -> (Pose) {blendPose [{linkId=-1, sourceLinkId=-1}, {linkId=-1, sourceLinkId=-1}, {linkId=-1, sourceLinkId=-1}], blendTime [0.1,0.1,0.1]}
   n21 = "Sequence Player (additive) 'JumpLandMoving_NoWep_CMM_Additive'"() {
    bLoopAnimation=false
    groupName=JumpLand
    method=SyncGroup
    sequence=&JumpLandMoving_NoWep_CMM_Additive
   }
   n22 = "Sequence Player (additive) 'JumpLand_NoWep_CMM_Additive'"() {
    bLoopAnimation=false
    groupName=JumpLand
    method=SyncGroup
    sequence=&JumpLand_NoWep_CMM_Additive
   }
   n23 = Blend Poses by bool(BlendPose_0=@n21, BlendPose_1=@n22, BlendTime_0=.25, BlendTime_1=.25, bActiveValue=Get bIsAboveMinimumLocomotionSpeed()) {blendType=Linear}
   n24 = "Sequence Player (additive) 'M_JumpLand_ARDS_Additive'"() {
    bLoopAnimation=false
    groupName=JumpLand
    method=SyncGroup
    sequence=&M_JumpLand_ARDS_Additive
   }
   n25 = Get bIsMeleeGuarding() -> (bIsMeleeGuarding:bool)
   n26 = "Sequence Player (additive) 'PupMinigun_Jump_Land'"() {
    bLoopAnimation=false
    groupName=JumpLand
    method=SyncGroup
    sequence=/Game/Animation/Game/MainPlayer/Combat/Guns/PupMinigun/PupMinigun_Jump_Land
   }
   n27 = "Blend Poses (EFortWeaponCoreAnimation)"!(
    BlendPose_0=@n20
    BlendPose_1=Modify Curve(SourcePose=Blend Poses by bool(BlendPose_0=@n24, BlendPose_1=@n23, bActiveValue=@n25) {blendType=Linear}, CurveValues_0=1, Alpha=1) {curveNames [LeftHandFK], curveValues [0]}
    BlendPose_2=@n26
    BlendPose_3:Pose Link Structure=Blend Poses by bool(BlendPose_0=@n10, BlendPose_1=@n9, BlendTime_0=.25, BlendTime_1=.25, bActiveValue=Get bIsAboveMinimumLocomotionSpeed()) {blendType=Linear}
    BlendTime_0=0
    BlendTime_1=0.1
    BlendTime_2=0.1
    BlendTime_3:float=0.1
    ActiveEnumValue=Get WeaponCoreAnim()
   ) -> (Pose) {
    blendPose [{linkId=-1, sourceLinkId=-1}, {linkId=-1, sourceLinkId=-1}, {linkId=-1, sourceLinkId=-1}, {linkId=-1, sourceLinkId=-1}]
    blendTime [0.1,0.1,0.1,0.1]
   }
   n28 = "Blend Poses (EFortWeaponCoreAnimation)"!(
    BlendPose_0="Blend Poses (EFortWeaponCoreAnimation)"(
     BlendPose_0=@n27
     BlendPose_1=Blend Poses by bool(BlendPose_0=@n6, BlendPose_1=@n7, BlendTime_0=.25, BlendTime_1=.25, bActiveValue=@n8) {blendType=Linear}
     BlendTime_0=0
     BlendTime_1=0.1
     ActiveEnumValue=Get WeaponCoreAnim()
    ) {blendPose [{linkId=-1, sourceLinkId=-1}, {linkId=-1, sourceLinkId=-1}], blendTime [0.1,0.1]}
    BlendPose_1=Blend Poses by bool(BlendPose_0=@n5, BlendPose_1=@n4, BlendTime_1=Get ADSToNonADSBlendTime() -> (ADSToNonADSBlendTime:float), bActiveValue=Get bIsTargeting()) {blendType=Linear}
    BlendTime_0=0
    BlendTime_1=0.1
    ActiveEnumValue=Get WeaponCoreAnim()
   ) -> (Pose) {blendPose [{linkId=-1, sourceLinkId=-1}, {linkId=-1, sourceLinkId=-1}], blendTime [0.1,0.1]}
   n29 = Sequence Player!(Sequence:Anim Sequence Base Object Reference=Break Fort Anim Input Player Anim Asset!(FortAnimInput_PlayerAnimAsset:"Fort Anim Input Player Anim Asset Structure (by ref)"=Get CurrentAnimSet() -> (CurrentAnimSet:Fort Anim Input Player Anim Asset Structure)) -> (
     bPlayUpperBodySlotOnFullBodyInAir:bool=~
     bDisableFullBodyAimOffsetDuringMelee:bool=~
     bShouldApplyAimOffsetFullBody:bool=~
     FullBodyAimOffsetAlpha:float=~
     UpperBodyAimOffsetAlpha:float=~
     bOverrideDisableArmsHeadAdditive:bool=~
     DisableArmsHeadAdditiveOverride:float=~
     SprintPlayrateClampRange:Fort Anim Playrate Range Structure
     SprintTargetingPlayrateClampRange:Fort Anim Playrate Range Structure
     CrouchSprintPlayrateClampRange:Fort Anim Playrate Range Structure
     CrouchSprintTargetingPlayrateClampRange:Fort Anim Playrate Range Structure
     TargetingAimOffset:Aim Offset Blend Space Object Reference
     NonTargetingAimOffset:Aim Offset Blend Space Object Reference
     RelaxedAimOffset:Aim Offset Blend Space Object Reference
     SwimTargetingPose:Anim Sequence Object Reference
     SwimNonTargetingPose:Anim Sequence Object Reference
     SwimJogAdditiveBlendSpace:Blend Space Object Reference
     SwimTargetingAimOffset:Aim Offset Blend Space Object Reference
     SwimIdleNoise:Anim Sequence Object Reference
     SwimJumpSurfaceEndAdditve:Anim Sequence Object Reference
     SwimJumpSurfaceLoopAdditve:Anim Sequence Object Reference
     SwimJumpFallLoopAdditve:Anim Sequence Object Reference
     FlyModeLoopAdditive:Anim Sequence Object Reference
     FlyModeStartAdditive:Anim Sequence Object Reference
     ZipLineStartAdditve:Anim Sequence Object Reference
     JetPackJumpAdditve:Anim Sequence Object Reference
     JetPackStartAdditve:Anim Sequence Object Reference
     FallAdditive:Anim Sequence Object Reference
     JumpLoopAdditive:Anim Sequence Object Reference
     JumpUpAdditive:Anim Sequence Object Reference
     FullBodyJumpAdditive:Anim Sequence Object Reference
     WeaponChargeLoop:Anim Sequence Object Reference
     LandingAdditiveOverride:Anim Sequence Object Reference
     TurnInPlaceLeftAdditiveOverride:Anim Sequence Object Reference
     TurnInPlaceRightAdditiveOverride:Anim Sequence Object Reference
     bOverridePitchAndYawOffsets:bool=~
     AnimSetOffsets:Fort Anim Input Adjusted Aim Offset Structure
     BonesToHide:Array of Names
     bOverrideHandIKRetargetingWeight:bool=~
     HandIKRetargetingWeight:float=~
     bForceKneeSliding:bool=~
     TargetingPose:Anim Sequence Object Reference
     NonTargetingPose:Anim Sequence Object Reference
     IdleNoise:Anim Sequence Object Reference
    ) {is=title K2Node_BreakStruct}.LandingAdditiveOverride) -> (Pose) {bLoopAnimation=false, groupName=JumpLand, method=SyncGroup}
   n30 = Get bHasLandingOverride() -> (bHasLandingOverride:bool)
   n31 = "Sequence Player (additive) 'JumpLand_NoWep_CMM_Additive_Core'"() {
    bLoopAnimation=false
    groupName=JumpLand
    method=SyncGroup
    sequence=/Game/Animation/Game/MainPlayer/Locomotion/Medium/Male/Jump/NoWep/JumpLand_NoWep_CMM_Additive_Core
   }
   n32 = "Sequence Player (additive) 'DualWield_Land_CMM'"() {
    bLoopAnimation=false
    groupName=JumpLand
    method=SyncGroup
    sequence=&DualWield_Land_CMM
   }
   n33 = Get WeaponCoreAnim()
   n34 = "Sequence Player (additive) 'M_JumpLand_ARDS_Additive'"() {
    bLoopAnimation=false
    groupName=JumpLand
    method=SyncGroup
    sequence=&M_JumpLand_ARDS_Additive
   }
   n35 = Blend Poses by bool(
    BlendPose_0=@n34
    BlendPose_1=Blend Poses by bool(
     BlendPose_0="Blend Poses (EFortWeaponCoreAnimation)"(BlendPose_0=@n31, BlendPose_1=@n32, BlendTime_0=0, BlendTime_1=0.1, ActiveEnumValue=@n33) {blendPose [{linkId=-1, sourceLinkId=-1}, {linkId=-1, sourceLinkId=-1}], blendTime [0.1,0.1]}
     BlendPose_1=Blend Poses by bool(BlendPose_0=@n29, BlendPose_1=@n28, bActiveValue=@n30) {blendType=Linear}
     BlendTime_0=0.2
     BlendTime_1=0.4
     bActiveValue=Get bIsPlayingMeleeAnim()
    ) {blendType=Linear}
    bActiveValue=Get bIsDBNOCarrying() -> (bIsDBNOCarrying:bool)
   ) {blendType=Linear}
   n36 = Get bUseMotionMatchingWithoutProceduralLayering() -> (bUseMotionMatchingWithoutProceduralLayering:bool)
   Output Pose(Result=Blend Poses by bool(
     BlendPose_0=Additive Identity Pose()
     BlendPose_1=Blend Poses by bool(BlendPose_0=@n3, BlendPose_1=@n35, BlendTime_0=.2, BlendTime_1=.2, bActiveValue=@n36) {blendType=Linear}
     bActiveValue=Get bIsDBNO()
    ) {blendType=Linear}) {layerGroup=None}
  ]}
 ItemAimOffset {nodes [
   Input Pose!() -> (Pose, AimOffsetAlpha:double, Yaw:double, Pitch:double) {name=InPose_AimOffset}
   Input Pose!() -> (Pose) {name=InPose_PreAimOffsets}
   Output Pose() {layerGroup=Item Override}
  ]}
 ItemFallAdditive {nodes [
   Input Pose() {name=InPoseItemFallAdditive}
   Output Pose() {layerGroup=Item Override}
  ]}
 ItemFallLandAdditive {nodes [Output Pose(Result=Input Pose() {name=InPoseItemFallLandAdditive}) {layerGroup=Item Override}]}
 ItemFallLevel2Override {nodes [
   Input Pose() {name=InPose_FallLevel2}
   Output Pose() {layerGroup=Item Override}
  ]}
 ItemFinalPoseOverride {nodes [
   Input Pose() {name=InFinalPose}
   Output Pose() {layerGroup=Item Override}
  ]}
 ItemFullBodyAimOffsets {nodes [
   Input Pose!() -> (Pose, FullBody_AimOffsetAlpha:double, Aim_Yaw:double, Aim_Pitch:double) {name=InPoseFullBodyAimOffsets}
   Input Pose!() -> (Pose) {name=InPosePreFullBodyAimOffsets}
   Output Pose() {layerGroup=Item Override}
  ]}
 ItemFullBodyOverride {nodes [
   Input Pose() {name=InPoseFullBody}
   Output Pose() {layerGroup=Item Override}
  ]}
 ItemIdleOverride {nodes [
   Input Pose() {name=InPose_Idle}
   Output Pose() {layerGroup=Item Override}
  ]}
 ItemJetPackJumpAdditive {nodes [Output Pose() {layerGroup=Item Override}]}
 ItemJetPackStartAdditive {nodes [Output Pose() {layerGroup=Item Override}]}
 ItemJumpLoopAdditive {nodes [
   Input Pose() {name=InPoseItemJumpLoopAdditive}
   Output Pose() {layerGroup=Item Override}
  ]}
 ItemJumpOverride {nodes [
   Input Pose() {name=InPoseJump}
   Output Pose() {layerGroup=Item Override}
  ]}
 ItemJumpUpAdditive {nodes [
   Input Pose() {name=InPoseItemJumpUpAdditive}
   Output Pose() {layerGroup=Item Override}
  ]}
 ItemPreAimOffsets {nodes [
   Input Pose() {name=InPosePreAimOffsets}
   Output Pose() {layerGroup=Item Override}
  ]}
 ItemPreIK {nodes [
   Input Pose() {name=InPose_PreIK}
   Output Pose() {layerGroup=Item Override}
  ]}
 ItemRightHandIK {nodes [
   Input Pose() {name=InPosePreRightHandIK}
   Input Pose() {name=InPosePostRightHandIK}
   Output Pose() {layerGroup=Item Override}
  ]}
 ItemSkeletalControl {nodes [
   Input Pose()
   Output Pose() {layerGroup=Item Override}
  ]}
 ItemSwimJumpFallLoop {nodes [
   n1 = Break Fort Anim Input Player Anim Asset!(FortAnimInput_PlayerAnimAsset=Get CurrentAnimSet()) -> (
    SwimJumpFallLoopAdditve
    FullBodyJumpAdditive
    LandingAdditiveOverride
    TurnInPlaceLeftAdditiveOverride
    TurnInPlaceRightAdditiveOverride
    bForceKneeSliding=~
   )
   Output Pose(Result=Blend Poses by bool(
     BlendPose_0=Additive Identity Pose()
     BlendPose_1=Sequence Player!(Sequence=@n1, bLoopAnimation:bool=false) -> (Pose) {startPosition=0.2}
     bActiveValue=Get bIsDBNO()
    ) {blendType=Linear}) {layerGroup=Item Override}
  ]}
 ItemSwimJumpSurfaceEnd {nodes [
   n1 = Break Fort Anim Input Player Anim Asset!(FortAnimInput_PlayerAnimAsset=Get CurrentAnimSet()) -> (
    SwimJumpSurfaceEndAdditve:Anim Sequence Object Reference
    FullBodyJumpAdditive
    LandingAdditiveOverride
    TurnInPlaceLeftAdditiveOverride
    TurnInPlaceRightAdditiveOverride
    bForceKneeSliding=~
   )
   Output Pose(Result=Blend Poses by bool(
     BlendPose_0=Additive Identity Pose()
     BlendPose_1=Sequence Player(Sequence=@n1, bLoopAnimation=False) {bLoopAnimation=false}
     bActiveValue=Get bIsDBNO()
    ) {blendType=Linear}) {layerGroup=Item Override}
  ]}
 ItemSwimJumpSurfaceLoop {nodes [
   n1 = Break Fort Anim Input Player Anim Asset!(FortAnimInput_PlayerAnimAsset=Get CurrentAnimSet()) -> (
    SwimJumpSurfaceLoopAdditve:Anim Sequence Object Reference
    FullBodyJumpAdditive
    LandingAdditiveOverride
    TurnInPlaceLeftAdditiveOverride
    TurnInPlaceRightAdditiveOverride
    bForceKneeSliding=~
   )
   Output Pose(Result=Blend Poses by bool(
     BlendPose_0=Additive Identity Pose()
     BlendPose_1=Sequence Player!(Sequence=@n1, PlayRate:float=1, StartPosition:float=0) -> (Pose)
     bActiveValue=Get bIsDBNO()
    ) {blendType=Linear}) {layerGroup=Item Override}
  ]}
 ItemUpperBody {nodes [
   Input Pose!() -> (Pose, InputParam:Fort Anim Input Adjusted Aim Structure) {name=InPoseUpperBody}
   Output Pose() {layerGroup=Item Override}
  ]}
 ItemZipLineStartAdditive {nodes [Output Pose() {layerGroup=Item Override}]}
 LocomotionLayer_Additives {nodes [Output Pose() {layerGroup=LocomotionOverride}]}
 LocomotionLayer_BaseLocomotion {nodes [
   Input Pose!() -> (Pose) {name=InPose_BaseLocomotion}
   Output Pose() {layerGroup=LocomotionOverride}
  ]}
 LocomotionLayer_FinalPose {nodes [
   Input Pose() {name=InPoseFinalPose}
   Output Pose() {layerGroup=LocomotionOverride}
  ]}
 LocomotionLayer_FullBody {nodes [
   Input Pose() {name=InPose_FullBody}
   Output Pose() {layerGroup=LocomotionOverride}
  ]}
 LocomotionLayer_FullLocomotion {nodes [
   Input Pose() {name=InPose_FullLocomotion}
   Output Pose() {layerGroup=LocomotionOverride}
  ]}
 LocomotionLayer_GroundLandAdditive {nodes [Output Pose(Result=Input Pose()) {layerGroup=LocomotionOverride}]}
 LocomotionLayer_InAir {nodes [
   Input Pose() {name=InPose_InAir}
   Input Pose() {name=InPose_InAirSourcePose}
   Output Pose() {layerGroup=LocomotionOverride}
  ]}
 LocomotionLayer_LowerBody {nodes [
   Input Pose() {name=InPose_LowerBody}
   Output Pose() {layerGroup=LocomotionOverride}
  ]}
 LocomotionLayer_UpperBody {nodes [
   Input Pose() {name=InPose_UpperBody}
   Output Pose() {layerGroup=LocomotionOverride}
  ]}
 OnLandingBecomesRelevant {nodes [
   OnLandingBecomesRelevant!() -> (then, Context:"Anim Update Context Structure (by ref)", Node:"Anim Node Reference Structure (by ref)") {is=title K2Node_FunctionEntry}
   > ResetCanExitLanding(self:Fort Player Landing Anim Instance Object Reference)
   > Return Node()
  ]}
 OnLandingUpdate {nodes [
   n1 = OnLandingUpdate!() -> (then, Context:"Anim Update Context Structure (by ref)", Node:"Anim Node Reference Structure (by ref)") {is=title K2Node_FunctionEntry}
   > n2 = Convert to Animation State!(execute, Node:"Anim Node Reference Structure (by ref)"=@n1.Node) -> (Succeeded:Exec, Failed:Exec, AnimationState:Animation State Reference Structure)
   > n3 = Branch(Condition="float > float"(
     A=0.25
     B=GetRelevantAnimTimeRemaining!(UpdateContext:"Anim Update Context Structure (by ref)"=@n1, Node:"Animation State Reference Structure (by ref)"=@n2) -> (ReturnValue:float)
    ))
   > n4 = SetCanExitLanding(self:Fort Player Landing Anim Instance Object Reference)
   Return Node(execute=[@n4.then, @n3.else])
  ]}
]
variables []
