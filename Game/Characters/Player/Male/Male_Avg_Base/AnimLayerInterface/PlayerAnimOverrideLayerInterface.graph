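# Animation layer interface that declares the override points for the player character's animation.
# It defines four layers in the PlayerOverride group: FullBodyOverride, UpperBody, PreLegIK and CharacterSkeletalControlPostLegIK.
# Each layer takes one input pose and returns an output pose, with no logic, variables or target skeleton.
# The layer names point to full-body and upper-body overrides plus hooks before and after leg IK.
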AnimBlueprint /Game/Characters/Player/Male/Male_Avg_Base/AnimLayerInterface/PlayerAnimOverrideLayerInterface
target_skeleton=null
graphs [
 CharacterSkeletalControlPostLegIK {nodes [
   Input Pose() {name=InPosePostLegIK}
   Output Pose() {layerGroup=PlayerOverride}
  ]}
 FullBodyOverride {nodes [
   Input Pose() {name=PassThroughFullBodyPose}
   Output Pose() {layerGroup=PlayerOverride}
  ]}
 PreLegIK {nodes [
   Input Pose()
   Output Pose() {layerGroup=PlayerOverride}
  ]}
 UpperBody {nodes [
   Input Pose() {name=InPoseUpperBody}
   Output Pose() {layerGroup=PlayerOverride}
  ]}
]
variables []
