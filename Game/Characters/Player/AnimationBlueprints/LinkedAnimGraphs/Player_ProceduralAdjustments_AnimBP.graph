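# Linked anim graph for the Fortnite player that applies procedural skeletal fix-ups to an incoming pose.
# Cached poses chain item pre-IK, hand IK, body-type pelvis adjustment, leg IK, cannon adjustment, override hooks, then glider or freefall root and pelvis offsets.
# Hand IK retargets IK bones through alpha-weighted virtual-bone copies.
# Leg IK adds slope warping, hoverboard foot attachment and emote offsets, with item and player layer interfaces allowing overrides.
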AnimBlueprint /Game/Characters/Player/AnimationBlueprints/LinkedAnimGraphs/Player_ProceduralAdjustments_AnimBP
target_skeleton=&Fortnite_M_Avg_Player_Skeleton
graphs [
 AnimGraph {nodes [
   n1 = Blend Poses by bool(
    BlendPose_0=ParaGliderAdjustmentsLayer!(InPoseParaGliderAdjustments:Pose Link Structure=Use cached pose 'PreParaGlideRoot'()) -> (Pose:Pose Link Structure) {is=title AnimGraphNode_LinkedAnimLayer, layer=ParaGliderAdjustmentsLayer}
    BlendPose_1=Use cached pose 'PreParaGlideRoot'()
    bActiveValue=Break Fort Anim Input Skydiving!(FortAnimInput_Skydiving:"Fort Anim Input Skydiving Structure (by ref)"=Get Skydiving() -> (Skydiving:Fort Anim Input Skydiving Structure)) -> (
     ParaGilderRootOffset:Vector=~
     ParaGliderRootRotationOverride:Rotator=~
     bUseParaGlideRootModifier:bool=~
     bIsSkydivingFromLaunchPad:bool=~
     bIsSkydivingFromBus:bool=~
     bIsInVortex:bool=~
     bIsDiving:bool=~
     bIsParachuteOpen:bool=~
     bIsSkydiving:bool=~
     bIsParachuteLeaning:bool=~
     bIsSkydiveLeaning:bool=~
     bIsLeaning:bool=~
     bIsSkydiveDiveMode:bool=~
     bPlaySkydiveDrift:bool=~
     bSkydiveDriftDelayActive:bool=~
     bSkydiveDriftAnimAllowed:bool=~
     bIsGliderRight:bool=~
     bIsGliderCenter:bool=~
     bIsGliderLeft:bool=~
     bIsGliderForward:bool=~
     bIsGliderBack:bool=~
     bHasDeployedAndIsStandingUpright:bool=~
     LocalAccelForward:float=~
     LocalAccelRight:float=~
     LocalVelocityRight:float=~
     SkydiveAimPitch:float=~
     SkydiveAimYaw:float=~
     DeployChuteAnimRate:float=~
     SkydiveDriftAnimRate:float=~
     SkydiveDriftAnimRateCurrent:float=~
     SkydiveFidgetAnimRate:float=~
     SkydiveFidgetAnimRateCurrent:float=~
     SkydiveAdditiveAlpha:float=~
     SkydiveDriftDelay:float=~
     SkydiveDriftAnim:int
     SkydiveDriftAnimMax:int
     LaunchpadAnim:int
     LocalAccelDir:ESkydivingDirection Enum
     DirectionLast:ESkydivingDirection Enum
     PoseAdditive:Anim Sequence Object Reference
    ) {is=title K2Node_BreakStruct}.bUseParaGlideRootModifier
   ) {blendType=Linear}
   n2 = "Transform (Modify) Bone - Bone: pelvis"(
    ComponentPose=Local To Component(LocalPose=Use cached pose 'PreParaGlideRoot'())
    Alpha=1
    Translation="0,0,0"
    Rotation=Get FreeFall_PelvisModRot() -> (FreeFall_PelvisModRot:Rotator)
    Scale="1,1,1"
   ) {rotationMode=BMM_Additive, alphaBoolBlend {blendOption=Linear}, boneToModify {boneName=pelvis}, rotation {roll=45}}
   n3 = Component To Local(ComponentPose=@n2)
   n4 = Get bFreeFall_UseDiveAngleMod() -> (bFreeFall_UseDiveAngleMod:bool)
   Output Pose(Result=Profiler Stat Event!(Source:Pose Link Structure=Blend Poses by bool(BlendPose_0=@n3, BlendPose_1=@n1, BlendTime_0=0.45, BlendTime_1=0.45, bActiveValue=@n4) {blendType=Linear}) -> (Pose:Pose Link Structure) {is=title FortAnimGraphNode_ProfilerStatEvent, profilerEventName=MainPlayerAnimBP_ProceduralAdjustments}) {layerGroup=None}
   Save cached pose 'PostTurnInPlace'(Pose=Rotate Root Bone!(BasePose:Pose Link Structure=Use cached pose 'InputPose'(), Yaw:float=0) -> (Pose:Pose Link Structure) {is=title AnimGraphNode_RotateRootBone}) {cacheName=PostTurnInPlace}
   Save cached pose 'InputPose'(Pose=Input Pose() {name=InPosePostTurnInPlace}) {cacheName=InputPose}
   Save cached pose 'PostHandIK'(Pose=HandIKFixUpLayer!(InPoseHandIk:Pose Link Structure=Use cached pose 'ProceduralFixupsBase'()) -> (Pose:Pose Link Structure) {is=title AnimGraphNode_LinkedAnimLayer, layer=HandIKFixUpLayer}) {cacheName=PostHandIK}
   n5 = Use cached pose 'PostLegIK'()
   n6 = PlayerAnimOverrideLayerInterface - CharacterSkeletalControlPostLegIK!(InPosePostLegIK:Pose Link Structure=Blend Poses by bool(
     BlendPose_0=PushCannonAdjustmentsLayer!(InPosePushCannonAdjustments:Pose Link Structure=@n5) -> (Pose:Pose Link Structure) {is=title AnimGraphNode_LinkedAnimLayer, layer=PushCannonAdjustmentsLayer}
     BlendPose_1=Use cached pose 'PostLegIK'()
     bActiveValue=Get bIsInCannon() -> (bIsInCannon:bool)
    ) {blendType=Linear}) -> (Pose:Pose Link Structure) {
    is=title AnimGraphNode_LinkedAnimLayer
    interface=/Game/Characters/Player/Male/Male_Avg_Base/AnimLayerInterface/PlayerAnimOverrideLayerInterface.PlayerAnimOverrideLayerInterface_C
    layer=CharacterSkeletalControlPostLegIK
   }
   n7 = ItemAnimLayerInterface - ItemSkeletalControl!(InPose:Pose Link Structure=@n6) -> (Pose:Pose Link Structure) {
    is=title AnimGraphNode_LinkedAnimLayer
    interface=/Game/Characters/Player/Male/Male_Avg_Base/AnimLayerInterface/ItemAnimLayerInterface.ItemAnimLayerInterface_C
    layer=ItemSkeletalControl
   }
   Save cached pose 'PreParaGlideRoot'(Pose=@n7) {cacheName=PreParaGlideRoot}
   Save cached pose 'PostLegIK'(Pose=LegIK!(InPosePreIKAdjustments:Pose Link Structure=Use cached pose 'PostLargeBodyCrouchAdjustments'()) -> (Pose:Pose Link Structure) {is=title AnimGraphNode_LinkedAnimLayer, layer=LegIK}) {cacheName=PostLegIK}
   n8 = "Blend Poses (EFortPlayerAnimBodyType)"!(
    BlendPose_0:Pose Link Structure=Use cached pose 'PostHandIK'()
    BlendPose_1:Pose Link Structure=PlevisAndFeetAdjustmentsLayer!(InPosePelvisAndFeetAdjustments:Pose Link Structure=Use cached pose 'PostHandIK'()) -> (Pose:Pose Link Structure) {is=title AnimGraphNode_LinkedAnimLayer, layer=PlevisAndFeetAdjustmentsLayer}
    BlendTime_0:float=0.1
    BlendTime_1:float=0.1
    ActiveEnumValue:Byte=Get AnimBodyType() -> (AnimBodyType:EFortPlayerAnimBodyType Enum=Small)
   ) -> (Pose:Pose Link Structure) {
    is=title AnimGraphNode_BlendListByEnum
    blendType=Linear
    blendPose [{linkId=-1, sourceLinkId=-1}, {linkId=-1, sourceLinkId=-1}]
    blendTime [0.1,0.1]
   }
   Save cached pose 'PostLargeBodyCrouchAdjustments'(Pose=@n8) {cacheName=PostLargeBodyCrouchAdjustments}
   n9 = Layered blend per bone(
    BasePose=Slot 'FullBodyNoAdditive'(Source=Use cached pose 'PostTurnInPlace'()) {slotName=FullBodyNoAdditive}
    BlendPoses_0=Use cached pose 'PostTurnInPlace'()
    BlendWeights_0=Get UpperBodyNoAdditivesMaskWeight() -> (UpperBodyNoAdditivesMaskWeight:float)
   ) {
    bMeshSpaceRotationBlend=true
    curveBlendOption=UseMaxValue
    layerSetup [{branchFilters [{boneName=Root}, {blendDepth=-1, boneName=Ik_hand_root}, {blendDepth=-4, boneName=Spine_01}]}]
   }
   n10 = ItemAnimLayerInterface - ItemPreIK!(InPose_PreIK:Pose Link Structure=@n9) -> (Pose:Pose Link Structure) {
    is=title AnimGraphNode_LinkedAnimLayer
    interface=&ItemAnimLayerInterface_C
    layer=ItemPreIK
   }
   Save cached pose 'ProceduralFixupsBase'(Pose=@n10) {cacheName=ProceduralFixupsBase}
  ]}
 CharacterSkeletalControlPostLegIK {nodes [Output Pose(Result=Input Pose() {name=InPosePostLegIK}) {layerGroup=PlayerOverride}]}
 EventGraph {links [], nodes []}
 FullBodyOverride {nodes [
   Input Pose() {name=PassThroughFullBodyPose}
   Output Pose() {layerGroup=PlayerOverride}
  ]}
 HandIKFixUpLayer {nodes [
   n1 = Use cached pose 'PreRightHandIK'()
   n2 = Get HandIKInput() -> (HandIKInput:Fort Anim Input Hand IK Structure)
   n3 = Leg IK!(
    ComponentPose:Component Space Pose Link Structure=Copy Bone - Source Bone: hand_r - Target Bone: ik_hand_r!(
     ComponentPose:Component Space Pose Link Structure=Local To Component(LocalPose=@n1)
     Alpha:float=Break Fort Anim Input Hand IK!(FortAnimInput_HandIK:"Fort Anim Input Hand IK Structure (by ref)"=@n2) -> (
      HandIKRetargetingWeight:float=~
      HandsInRootSpaceAlpha:float=~
      LeftHandIKAlpha:float=~
      RightHandIKAlpha:float=~
      RightHandFKRotationAlpha:float=~
     ) {is=title K2Node_BreakStruct}.RightHandFKRotationAlpha
     bCopyTranslation:bool=False
     bCopyRotation:bool=true
     bCopyScale:bool=False
    ) -> (Pose:Component Space Pose Link Structure) {is=title AnimGraphNode_CopyBone, alphaBoolBlend {blendOption=Linear}, sourceBone {boneName=hand_r}, targetBone {boneName=ik_hand_r}}
    Alpha:float=Get RightHandIKAlpha() -> (RightHandIKAlpha:float)
   ) -> (Pose:Component Space Pose Link Structure) {
    is=title AnimGraphNode_LegIK
    alphaBoolBlend {blendOption=Linear}
    legsDefinition [{
      bEnableKneeTwistCorrection=false
      bEnableRotationLimit=false
      footBoneForwardAxis=Y
      hingeRotationAxis=Z
      minRotationAngle=15
      numBonesInLimb=2
      twistOffsetCurveName=None
      fKFootBone {boneName=hand_r}
      iKFootBone {boneName=ik_hand_r}
     }]
   }
   n4 = ItemAnimLayerInterface - ItemRightHandIK!(
    InPosePreRightHandIK:Pose Link Structure=Use cached pose 'PreRightHandIK'()
    InPosePostRightHandIK:Pose Link Structure=Component To Local(ComponentPose=@n3)
   ) -> (Pose:Pose Link Structure) {
    is=title AnimGraphNode_LinkedAnimLayer
    interface=&ItemAnimLayerInterface_C
    layer=ItemRightHandIK
   }
   n5 = Copy Bone - Source Bone: VB spine_05_hand_attach_r - Target Bone: hand_attach_r!(
    ComponentPose:Component Space Pose Link Structure=Copy Bone - Source Bone: VB spine_05_weapon_r - Target Bone: weapon_r!(
     ComponentPose:Component Space Pose Link Structure=Local To Component(LocalPose=@n4)
     Alpha:float=Get WeaponRInChestSpace() -> (WeaponRInChestSpace:float)
     bCopyTranslation:bool=true
     bCopyRotation:bool=true
     bCopyScale:bool=true
    ) -> (Pose:Component Space Pose Link Structure) {
     is=title AnimGraphNode_CopyBone
     alphaBoolBlend {blendOption=Linear}
     sourceBone {boneName=VB spine_05_weapon_r}
     targetBone {boneName=weapon_r}
    }
    Alpha:float=Get HandAttachRInChestSpace() -> (HandAttachRInChestSpace:float)
    bCopyTranslation:bool=true
    bCopyRotation:bool=true
    bCopyScale:bool=true
   ) -> (Pose:Component Space Pose Link Structure) {
    is=title AnimGraphNode_CopyBone
    alphaBoolBlend {blendOption=Linear}
    sourceBone {boneName=VB spine_05_hand_attach_r}
    targetBone {boneName=hand_attach_r}
   }
   n6 = Copy Bone - Source Bone: weapon_l - Target Bone: weapon_r!(
    ComponentPose:Component Space Pose Link Structure=@n5
    Alpha:float=Get SnapWeapon_LHandAlphaSwitch() -> (SnapWeapon_LHandAlphaSwitch:float)
    bCopyTranslation:bool=true
    bCopyRotation:bool=true
    bCopyScale:bool=true
   ) -> (Pose:Component Space Pose Link Structure) {is=title AnimGraphNode_CopyBone, alphaBoolBlend {blendOption=Linear}, sourceBone {boneName=weapon_l}, targetBone {boneName=weapon_r}}
   n7 = Get WeaponLinFloorSpace() -> (WeaponLinFloorSpace:float)
   n8 = Copy Bone - Source Bone: VB ik_foot_root_hand_attach_l - Target Bone: hand_attach_l!(
    ComponentPose:Component Space Pose Link Structure=Copy Bone - Source Bone: VB ik_foot_root_weapon_r - Target Bone: weapon_r!(
     ComponentPose:Component Space Pose Link Structure=Copy Bone - Source Bone: VB ik_foot_root_weapon_l - Target Bone: weapon_l!(
      ComponentPose:Component Space Pose Link Structure=@n6
      Alpha:float=@n7
      bCopyTranslation:bool=true
      bCopyRotation:bool=true
      bCopyScale:bool=true
     ) -> (Pose:Component Space Pose Link Structure) {
      is=title AnimGraphNode_CopyBone
      alphaBoolBlend {blendOption=Linear}
      sourceBone {boneName=VB ik_foot_root_weapon_l}
      targetBone {boneName=weapon_l}
     }
     Alpha:float=Get WeaponRinFloorSpace() -> (WeaponRinFloorSpace:float)
     bCopyTranslation:bool=true
     bCopyRotation:bool=true
     bCopyScale:bool=true
    ) -> (Pose:Component Space Pose Link Structure) {
     is=title AnimGraphNode_CopyBone
     alphaBoolBlend {blendOption=Linear}
     sourceBone {boneName=VB ik_foot_root_weapon_r}
     targetBone {boneName=weapon_r}
    }
    Alpha:float=Get HandAttachLinFloorSpace() -> (HandAttachLinFloorSpace:float)
    bCopyTranslation:bool=true
    bCopyRotation:bool=true
    bCopyScale:bool=true
   ) -> (Pose:Component Space Pose Link Structure) {
    is=title AnimGraphNode_CopyBone
    alphaBoolBlend {blendOption=Linear}
    sourceBone {boneName=VB ik_foot_root_hand_attach_l}
    targetBone {boneName=hand_attach_l}
   }
   n9 = Copy Bone - Source Bone: VB ik_foot_root_hand_attach_r - Target Bone: hand_attach_r!(
    ComponentPose:Component Space Pose Link Structure=@n8
    Alpha:float=Get HandAttachRinFloorSpace() -> (HandAttachRinFloorSpace:float)
    bCopyTranslation:bool=true
    bCopyRotation:bool=true
    bCopyScale:bool=true
   ) -> (Pose:Component Space Pose Link Structure) {
    is=title AnimGraphNode_CopyBone
    alphaBoolBlend {blendOption=Linear}
    sourceBone {boneName=VB ik_foot_root_hand_attach_r}
    targetBone {boneName=hand_attach_r}
   }
   Output Pose(Result=Component To Local(ComponentPose=@n9)) {layerGroup=None}
   n10 = "Transform (Modify) Bone - Bone: pelvis"!(
    ComponentPose=Local To Component(LocalPose=Input Pose() {name=InPoseHandIk})
    Alpha=Get PelvisAdjustmentCrouchAlpha() -> (PelvisAdjustmentCrouchAlpha:float)
    Translation="0,10,0"
    Rotation="0,0,-5"
   ) -> (Pose) {
    rotationMode=BMM_Additive
    translationMode=BMM_Additive
    alphaBoolBlend {blendOption=Linear}
    boneToModify {boneName=pelvis}
    rotation {roll=-5}
    translation {y=10}
   }
   n11 = Break Fort Anim Input Pelvis Adjustment!(FortAnimInput_PelvisAdjustment:"Fort Anim Input Pelvis Adjustment Structure (by ref)"=Get PelvisAdjustmentInput() -> (PelvisAdjustmentInput:Fort Anim Input Pelvis Adjustment Structure)) -> (
    PawnMesh:Skeletal Mesh Component Object Reference
    DistanceToFeet:float=~
    LegLength:float=~
    DesiredLegLengthTreshold:float=~
    DotProductBetweenUpVectorsTreshold:float=~
    PelvisInterpSpeed:float=~
    EmotePelvisOffsetInterpSpeed:float=~
    EmotePelvisOffsetAlpha:float=~
    CurrentEmotePelvisOffset:Vector=~
   ) {is=title K2Node_BreakStruct}
   n12 = Copy Bone - Source Bone: VB VB spine_05_weapon_r_ik_hand_gun - Target Bone: ik_hand_gun!(
    ComponentPose:Component Space Pose Link Structure="Transform (Modify) Bone - Bone: pelvis"!(ComponentPose=@n10, Alpha=@n11.EmotePelvisOffsetAlpha, Translation=@n11.CurrentEmotePelvisOffset, Rotation="0,0,0", Scale:Vector="1,1,1") -> (Pose) {lODThreshold=1, translationMode=BMM_Additive, alphaBoolBlend {blendOption=Linear}, boneToModify {boneName=pelvis}}
    Alpha:float=Get WeaponRInChestSpace()
    bCopyTranslation:bool=true
    bCopyRotation:bool=true
    bCopyScale:bool=true
   ) -> (Pose:Component Space Pose Link Structure) {
    is=title AnimGraphNode_CopyBone
    alphaBoolBlend {blendOption=Linear}
    sourceBone {boneName=VB VB spine_05_weapon_r_ik_hand_gun}
    targetBone {boneName=ik_hand_gun}
   }
   n13 = Copy Bone - Source Bone: VB VB spine_05_hand_attach_r_ik_hand_gun - Target Bone: ik_hand_gun!(
    ComponentPose:Component Space Pose Link Structure=@n12
    Alpha:float=Get HandAttachRInChestSpace()
    bCopyTranslation:bool=true
    bCopyRotation:bool=true
    bCopyScale:bool=true
   ) -> (Pose:Component Space Pose Link Structure) {
    is=title AnimGraphNode_CopyBone
    alphaBoolBlend {blendOption=Linear}
    sourceBone {boneName=VB VB spine_05_hand_attach_r_ik_hand_gun}
    targetBone {boneName=ik_hand_gun}
   }
   n14 = Copy Bone - Source Bone: VB VB spine_05_weapon_r_ik_hand_l - Target Bone: ik_hand_l!(
    ComponentPose:Component Space Pose Link Structure=@n13
    Alpha:float=Get WeaponRInChestSpace()
    bCopyTranslation:bool=true
    bCopyRotation:bool=true
    bCopyScale:bool=true
   ) -> (Pose:Component Space Pose Link Structure) {
    is=title AnimGraphNode_CopyBone
    alphaBoolBlend {blendOption=Linear}
    sourceBone {boneName=VB VB spine_05_weapon_r_ik_hand_l}
    targetBone {boneName=ik_hand_l}
   }
   n15 = Copy Bone - Source Bone: VB VB spine_05_hand_attach_r_ik_hand_l - Target Bone: ik_hand_l!(
    ComponentPose:Component Space Pose Link Structure=@n14
    Alpha:float=Get HandAttachRInChestSpace()
    bCopyTranslation:bool=true
    bCopyRotation:bool=true
    bCopyScale:bool=true
   ) -> (Pose:Component Space Pose Link Structure) {
    is=title AnimGraphNode_CopyBone
    alphaBoolBlend {blendOption=Linear}
    sourceBone {boneName=VB VB spine_05_hand_attach_r_ik_hand_l}
    targetBone {boneName=ik_hand_l}
   }
   n16 = Copy Bone - Source Bone: VB VB attach_ik_hand_gun_ik_hand_l - Target Bone: ik_hand_l!(
    ComponentPose:Component Space Pose Link Structure=Copy Bone - Source Bone: VB attach_ik_hand_gun - Target Bone: ik_hand_gun!(
     ComponentPose:Component Space Pose Link Structure=@n15
     Alpha:float=Get HandIKInAttachBoneSpace() -> (HandIKInAttachBoneSpace:float)
     bCopyTranslation:bool=true
     bCopyRotation:bool=true
     bCopyScale:bool=true
    ) -> (Pose:Component Space Pose Link Structure) {
     is=title AnimGraphNode_CopyBone
     alphaBoolBlend {blendOption=Linear}
     sourceBone {boneName=VB attach_ik_hand_gun}
     targetBone {boneName=ik_hand_gun}
    }
    Alpha:float=Get HandIKInAttachBoneSpace()
    bCopyTranslation:bool=true
    bCopyRotation:bool=true
    bCopyScale:bool=true
   ) -> (Pose:Component Space Pose Link Structure) {
    is=title AnimGraphNode_CopyBone
    alphaBoolBlend {blendOption=Linear}
    sourceBone {boneName=VB VB attach_ik_hand_gun_ik_hand_l}
    targetBone {boneName=ik_hand_l}
   }
   n17 = Copy Bone - Source Bone: VB VB attach_ik_hand_gun_ik_hand_r - Target Bone: ik_hand_r!(
    ComponentPose:Component Space Pose Link Structure=@n16
    Alpha:float=Get HandIKInAttachBoneSpace()
    bCopyTranslation:bool=true
    bCopyRotation:bool=true
    bCopyScale:bool=true
   ) -> (Pose:Component Space Pose Link Structure) {
    is=title AnimGraphNode_CopyBone
    alphaBoolBlend {blendOption=Linear}
    sourceBone {boneName=VB VB attach_ik_hand_gun_ik_hand_r}
    targetBone {boneName=ik_hand_r}
   }
   n18 = Break Fort Anim Input Hand IK(FortAnimInput_HandIK=Get HandIKInput()) -> (HandIKRetargetingWeight=~, HandsInRootSpaceAlpha=~, LeftHandIKAlpha=~, RightHandIKAlpha=~, RightHandFKRotationAlpha=~)
   n19 = Copy Bone - Source Bone: VB Root_hand_l - Target Bone: ik_hand_l!(
    ComponentPose:Component Space Pose Link Structure=@n17
    Alpha:float=@n18.HandsInRootSpaceAlpha
    bCopyTranslation:bool=true
    bCopyRotation:bool=true
    bCopyScale:bool=true
   ) -> (Pose:Component Space Pose Link Structure) {is=title AnimGraphNode_CopyBone, alphaBoolBlend {blendOption=Linear}, sourceBone {boneName=VB Root_hand_l}, targetBone {boneName=ik_hand_l}}
   n20 = Get OverrideIKHands() -> (OverrideIKHands:float)
   n21 = "Transform (Modify) Bone - Bone: ik_hand_r"!(
    ComponentPose="Transform (Modify) Bone - Bone: ik_hand_l"!(
     ComponentPose=Copy Bone - Source Bone: VB Root_hand_r - Target Bone: ik_hand_r!(
      ComponentPose:Component Space Pose Link Structure=@n19
      Alpha:float=@n18.HandsInRootSpaceAlpha
      bCopyTranslation:bool=true
      bCopyRotation:bool=true
      bCopyScale:bool=true
     ) -> (Pose:Component Space Pose Link Structure) {is=title AnimGraphNode_CopyBone, alphaBoolBlend {blendOption=Linear}, sourceBone {boneName=VB Root_hand_r}, targetBone {boneName=ik_hand_r}}
     Alpha=@n20
     Translation=Get LeftHandOverride() -> (LeftHandOverride:Vector)
    ) -> (Pose) {translationMode=BMM_Additive, alphaBoolBlend {blendOption=Linear}, boneToModify {boneName=ik_hand_l}}
    Alpha=@n20
    Translation=Get RightHandOverride() -> (RightHandOverride:Vector)
   ) -> (Pose) {translationMode=BMM_Additive, alphaBoolBlend {blendOption=Linear}, boneToModify {boneName=ik_hand_r}}
   n22 = Get IKHandGunInHandSpace() -> (IKHandGunInHandSpace:float)
   n23 = Leg IK(
    ComponentPose=Copy Bone - Source Bone: VB pelvis_ik_hand_l - Target Bone: ik_hand_l!(
     ComponentPose:Component Space Pose Link Structure=Copy Bone - Source Bone: hand_r - Target Bone: ik_hand_gun!(
      ComponentPose:Component Space Pose Link Structure=@n21
      Alpha:float=@n22
      bCopyTranslation:bool=true
      bCopyRotation:bool=true
      bCopyScale:bool=true
     ) -> (Pose:Component Space Pose Link Structure) {is=title AnimGraphNode_CopyBone, alphaBoolBlend {blendOption=Linear}, sourceBone {boneName=hand_r}, targetBone {boneName=ik_hand_gun}}
     Alpha:float=Get IKHandLeftInPelvisSpaceAlpha() -> (IKHandLeftInPelvisSpaceAlpha:float)
     bCopyTranslation:bool=true
     bCopyRotation:bool=true
     bCopyScale:bool=true
    ) -> (Pose:Component Space Pose Link Structure) {
     is=title AnimGraphNode_CopyBone
     alphaBoolBlend {blendOption=Linear}
     sourceBone {boneName=VB pelvis_ik_hand_l}
     targetBone {boneName=ik_hand_l}
    }
    Alpha=Get LeftHandIKAlpha() -> (LeftHandIKAlpha:float)
   ) {
    alphaBoolBlend {blendOption=Linear}
    legsDefinition [{
      bEnableKneeTwistCorrection=false
      bEnableRotationLimit=false
      footBoneForwardAxis=Y
      hingeRotationAxis=Z
      minRotationAngle=15
      numBonesInLimb=2
      twistOffsetCurveName=None
      fKFootBone {boneName=hand_l}
      iKFootBone {boneName=ik_hand_l}
     }]
   }
   Save cached pose 'PreRightHandIK'(Pose=Component To Local(ComponentPose=@n23)) {cacheName=PreRightHandIK}
  ]}
 ItemAimOffset {nodes [
   Input Pose!() -> (Pose, AimOffsetAlpha:double, Yaw:double, Pitch:double) {name=InPose_AimOffset}
   Input Pose!() -> (Pose) {name=InPose_PreAimOffsets}
   Output Pose() {layerGroup=Item Override}
  ]}
 ItemFallAdditive {nodes [
   Input Pose() {name=InPoseItemFallAdditive}
   Output Pose() {layerGroup=Item Override}
  ]}
 ItemFallLandAdditive {nodes [
   Input Pose() {name=InPoseItemFallLandAdditive}
   Output Pose() {layerGroup=Item Override}
  ]}
 ItemFallLevel2Override {nodes [
   Input Pose() {name=InPose_FallLevel2}
   Output Pose() {layerGroup=Item Override}
  ]}
 ItemFinalPoseOverride {nodes [
   Input Pose() {name=InFinalPose}
   Output Pose() {layerGroup=Item Override}
  ]}
 ItemFullBodyAimOffsets {nodes [
   Input Pose!() -> (Pose, FullBody_AimOffsetAlpha:double, Aim_Yaw:double, Aim_Pitch:double) {name=InPoseFullBodyAimOffsets}
   Input Pose!() -> (Pose) {name=InPosePreFullBodyAimOffsets}
   Output Pose() {layerGroup=Item Override}
  ]}
 ItemFullBodyOverride {nodes [
   Input Pose() {name=InPoseFullBody}
   Output Pose() {layerGroup=Item Override}
  ]}
 ItemIdleOverride {nodes [
   Input Pose() {name=InPose_Idle}
   Output Pose() {layerGroup=Item Override}
  ]}
 ItemJetPackJumpAdditive {nodes [Output Pose() {layerGroup=Item Override}]}
 ItemJetPackStartAdditive {nodes [Output Pose() {layerGroup=Item Override}]}
 ItemJumpLoopAdditive {nodes [
   Input Pose() {name=InPoseItemJumpLoopAdditive}
   Output Pose() {layerGroup=Item Override}
  ]}
 ItemJumpOverride {nodes [
   Input Pose() {name=InPoseJump}
   Output Pose() {layerGroup=Item Override}
  ]}
 ItemJumpUpAdditive {nodes [
   Input Pose() {name=InPoseItemJumpUpAdditive}
   Output Pose() {layerGroup=Item Override}
  ]}
 ItemPreAimOffsets {nodes [
   Input Pose() {name=InPosePreAimOffsets}
   Output Pose() {layerGroup=Item Override}
  ]}
 ItemPreIK {nodes [Output Pose(Result=Input Pose() {name=InPose_PreIK}) {layerGroup=Item Override}]}
 ItemRightHandIK {nodes [
   Input Pose() {name=InPosePreRightHandIK}
   Output Pose(Result=Input Pose() {name=InPosePostRightHandIK}) {layerGroup=Item Override}
  ]}
 ItemSkeletalControl {nodes [Output Pose(Result=Input Pose()) {layerGroup=Item Override}]}
 ItemSwimJumpFallLoop {nodes [Output Pose() {layerGroup=Item Override}]}
 ItemSwimJumpSurfaceEnd {nodes [Output Pose() {layerGroup=Item Override}]}
 ItemSwimJumpSurfaceLoop {nodes [Output Pose() {layerGroup=Item Override}]}
 ItemUpperBody {nodes [
   Input Pose!() -> (Pose, InputParam:Fort Anim Input Adjusted Aim Structure) {name=InPoseUpperBody}
   Output Pose() {layerGroup=Item Override}
  ]}
 ItemZipLineStartAdditive {nodes [Output Pose() {layerGroup=Item Override}]}
 LegIK {nodes [
   n1 = "Transform (Modify) Bone - Bone: pelvis"(
    ComponentPose=Local To Component(LocalPose=Use cached pose 'PostRotateRoot'())
    Alpha=Get PreIkPelvisRotationAlpha() -> (PreIkPelvisRotationAlpha:float)
    Translation=Get TurnPelvisTranslationOffset() -> (TurnPelvisTranslationOffset:Vector)
    Rotation=Get TurnPelvisRotationOffset() -> (TurnPelvisRotationOffset:Rotator)
    Scale="1,1,1"
   ) {lODThreshold=1, rotationMode=BMM_Additive, translationMode=BMM_Additive, alphaBoolBlend {blendOption=Linear}, boneToModify {boneName=pelvis}}
   n2 = Layered blend per bone(
    BasePose=Component To Local(ComponentPose=@n1)
    BlendPoses_0=Use cached pose 'PostRotateRoot'()
    BlendWeights_0=Get PreIkPelvisRotationAlpha()
   ) {
    bMeshSpaceRotationBlend=true
    blendMode=BlendMask
    blendMasks [&Fortnite_M_Avg_Player_Skeleton:PelvisRotationCorrection]
    layerSetup []
   }
   n3 = Use cached pose 'PostRotateRoot'()
   n4 = Get bIsPlayingEmote()
   n5 = "Transform (Modify) Bone - Bone: pelvis"!(
    ComponentPose=Local To Component(LocalPose=Blend Poses by bool(BlendPose_0=@n3, BlendPose_1=@n2, bActiveValue=@n4) {blendType=Linear})
    Alpha=Get UseEmoteAnimOffset() -> (UseEmoteAnimOffset:float)
    Translation="0,0,0"
   ) -> (Pose) {translationMode=BMM_Additive, translationSpace=BCS_WorldSpace, alphaBoolBlend {blendOption=Linear}, boneToModify {boneName=pelvis}}
   n6 = "Transform (Modify) Bone - Bone: Root"(
    ComponentPose="Transform (Modify) Bone - Bone: pelvis"!(ComponentPose=@n5, Alpha=1, Rotation:Rotator=Get PelvisRotationAdjustment() -> (PelvisRotationAdjustment:Rotator)) -> (Pose) {rotationMode=BMM_Additive, alphaBoolBlend {blendOption=Linear}, boneToModify {boneName=pelvis}}
    Alpha=1
    Translation=Get WholeAnimOffset() -> (WholeAnimOffset:Vector)
    Rotation="0,0,0"
    Scale="1,1,1"
   ) {translationMode=BMM_Additive, translationSpace=BCS_WorldSpace, alphaBoolBlend {blendOption=Linear}, boneToModify {boneName=Root}}
   n7 = Biped Slope Warping!(
    ComponentPose:Component Space Pose Link Structure=@n6
    Alpha:float=Get SlopeWarpingAlpha() -> (SlopeWarpingAlpha:float)
    CustomFloorOffset:Vector=Get SlopeWarpingFloorOffset() -> (SlopeWarpingFloorOffset:Vector)
    bUseFloorOffsetInterpolator:bool=Get bUseFloorOffsetInterpolator() -> (bUseFloorOffsetInterpolator:bool)
    bUseCustomFloorOffset:bool=Get bUseCustomFloorOffset() -> (bUseCustomFloorOffset:bool)
    bShouldAdjustCameraBoneWithLegsRootBone:bool=True
   ) -> (Pose:Component Space Pose Link Structure) {
    is=title FortAnimGraphNode_SlopeWarping
    bShouldAdjustCameraBoneWithLegsRootBone=true
    lODThreshold=3
    maxStepHeight=35
    pullPelvisDownMaxDistanceWhenCrouching=20
    alphaBoolBlend {blendOption=Linear}
    cameraBone {boneName=camera_root}
    feetDefinitions [
     {numBonesInLimb=2, fKFootBone {boneName=foot_l}, iKFootBone {boneName=ik_foot_l}}
     {numBonesInLimb=2, fKFootBone {boneName=foot_r}, iKFootBone {boneName=ik_foot_r}}
    ]
    iKFootRootBone {boneName=ik_foot_root}
    pelvisBone {boneName=pelvis}
   }
   n8 = PlayerAnimOverrideLayerInterface - PreLegIK!(InPose:Pose Link Structure=Component To Local(ComponentPose=@n7)) -> (Pose:Pose Link Structure) {
    is=title AnimGraphNode_LinkedAnimLayer
    interface=&PlayerAnimOverrideLayerInterface_C
    layer=PreLegIK
   }
   n9 = Break Fort Anim Input Jackal Vehicle!(FortAnimInput_JackalVehicle:"Fort Anim Input Jackal Vehicle Structure (by ref)"=Get JackalVehicleInput() -> (JackalVehicleInput:Fort Anim Input Jackal Vehicle Structure)) -> (
    bIsUsingJackalVehicle:bool=~
    bIsSteeringLeft:bool=~
    bIsSteeringRight:bool=~
    SteerAngleDeadZoneDegrees:float=~
    SteerAngle:float=~
    LeanYaw:float=~
    QuantizedSteerAngle:float=~
    SteerAlpha:float=~
    RunForwardAlpha:float=~
    SlopePitchDegreeAngle:float=~
    SlopeRollDegreeAngle:float=~
    DistanceFromGround:float=~
    PivotPlayRate:float=~
    bIsReversing:bool=~
    bIsBraking:bool=~
    bIsMoving:bool=~
    bIsMovingForward:bool=~
    bIsSprinting:bool=~
    bInAir:bool=~
    bIsFalling:bool=~
    bIsJumping:bool=~
    bIsRelaxed:bool=~
    bIsBoosting:bool=~
    bHasReachedJumpChargeStartThreshold:bool=~
    bHasReachedJumpChargeMidThreshold:bool=~
    bHasReachedJumpChargeFullThreshold:bool=~
    bAdjustRootForFemaleRider:bool=~
    bIsPlayingEmoteOnHoverboard:bool=~
    bShouldApplyAdditive:bool=~
    bPlayPivotOnGroundAndNotBoosting:bool=~
    bIsOnLowerHill:bool=~
    bLowerBodyIdleToLoopTransition:bool=~
    bInterruptHoverboardFullybody:bool=~
    bEnterFullBodyHoverboardState:bool=~
    bDefaultToJumpStartTransition:bool=~
    bLocomotionPoseToJumpTransition:bool=~
    bBoostingToBoostingJumpTransition:bool=~
    bJumpToLocomotionPoseTransition:bool=~
    bJumpApexToJumpFallTransition:bool=~
    bIdleToMovementStartTransition:bool=~
    bIdleToMovementLoopTransition:bool=~
    bMovementLoopToMovementStopTransition:bool=~
    bMovementLoopToPivotTransition:bool=~
    bMovementLoopToIdleTransition:bool=~
    bIdleAdditiveToCollisionNTransition:bool=~
    bSplitBodyToHoverboardBRTransition:bool=~
    bHoverboardBRToSplitBodyTransition:bool=~
    bHoverboardBRMovementToJumpChargeTransition:bool=~
    bIdlesToJackalVehicleTransition:bool=~
    bPlayAdditiveLeans:bool=~
    bPlayBalloonLeans:bool=~
    bPlayJumpTrickVertical:bool=~
    bPlayJumpTrick:bool=~
    bPlayMovingFast:bool=~
    bPlayHipAdjustmentAdditive:bool=~
    bPlayDriveSouth:bool=~
    bPlayHeadAimOffset:bool=~
    bPlaySlopeAimOffset:bool=~
    JumpCombatAdditiveWeight:float=~
    MeleeTwistIdle:float=~
    MeleeTwistLocomotionLoop:float=~
    VelocityImpact:Fort Anim Input Velocity Impact Structure
    ImpactDisplacement:Vector=~
    SteerYaw:float=~
    EmoteHoverboardPosition:Vector=~
    EmoteHoverboardRotation:Rotator=~
    bShouldAttachFeetToHoverboard:bool=~
    FootLeftLocationOffset:Vector=~
    FootLeftRotationOffset:Rotator=~
    FootRightLocationOffset:Vector=~
    FootRightRotationOffset:Rotator=~
    StoppedThreshold:float=~
    MovingForwardThreshold:float=~
    MovingFowardFastThreshold:float=~
    DefaultToJumpStartTransitionThreshold:float=~
    JumpTrickAngularVelocityThreshold:float=~
    JumpDistanceFromGroundThreshold:float=~
    VelocityStartThreshold:float=~
    MovingFastThreshold:float=~
    RelaxedSpeedThreshold:float=~
    JumpChargeStartThreshold:float=~
    JumpChargeMidThreshold:float=~
    JumpChargeFullThreshold:float=~
    RotatingAngularVelocityThreshold:float=~
    IdleToLoopTransitionSpeedThreshold:float=~
    IdleToMovementLoopTransitionThreshold:float=~
    MeleeTwistIdleMultiplier:float=~
    MeleeTwistLocomotionLoopMultiplier:float=~
    LeanYawForMaxPivotPlayRate:float=~
    SteerAlphaForMaxPivotPlayRate:float=~
    PrePivotCardinalDirection:EFortCardinalDirection Enum
   ) {is=title K2Node_BreakStruct}
   n10 = "Transform (Modify) Bone - Bone: ik_foot_l"!(
    ComponentPose=Local To Component(LocalPose=@n8)
    bAlphaBoolEnabled:bool=@n9.bShouldAttachFeetToHoverboard
    Translation=@n9.FootLeftLocationOffset
    Rotation=@n9.FootLeftRotationOffset
   ) -> (Pose) {
    alphaInputType=Bool
    rotationMode=BMM_Replace
    rotationSpace=BCS_WorldSpace
    translationMode=BMM_Replace
    translationSpace=BCS_WorldSpace
    alphaBoolBlend {blendOption=Linear}
    boneToModify {boneName=ik_foot_l}
   }
   n11 = "Transform (Modify) Bone - Bone: ik_foot_r"!(
    ComponentPose=@n10
    bAlphaBoolEnabled:bool=@n9.bShouldAttachFeetToHoverboard
    Translation=@n9.FootRightLocationOffset
    Rotation=@n9.FootRightRotationOffset
   ) -> (Pose) {
    alphaInputType=Bool
    rotationMode=BMM_Replace
    rotationSpace=BCS_WorldSpace
    translationMode=BMM_Replace
    translationSpace=BCS_WorldSpace
    alphaBoolBlend {blendOption=Linear}
    boneToModify {boneName=ik_foot_r}
   }
   n12 = Get LegIKInAttachBoneSpace() -> (LegIKInAttachBoneSpace:float)
   n13 = Copy Bone - Source Bone: VB pelvis_attach_emote_01 - Target Bone: attach_emote_01!(
    ComponentPose:Component Space Pose Link Structure=Copy Bone - Source Bone: VB attach_ik_foot_r - Target Bone: ik_foot_r!(
     ComponentPose:Component Space Pose Link Structure=Copy Bone - Source Bone: VB attach_ik_foot_l - Target Bone: ik_foot_l!(
      ComponentPose:Component Space Pose Link Structure=@n11
      Alpha:float=@n12
      bCopyTranslation:bool=true
      bCopyRotation:bool=true
      bCopyScale:bool=true
     ) -> (Pose:Component Space Pose Link Structure) {
      is=title AnimGraphNode_CopyBone
      alphaBoolBlend {blendOption=Linear}
      sourceBone {boneName=VB attach_ik_foot_l}
      targetBone {boneName=ik_foot_l}
     }
     Alpha:float=@n12
     bCopyTranslation:bool=true
     bCopyRotation:bool=true
     bCopyScale:bool=true
    ) -> (Pose:Component Space Pose Link Structure) {
     is=title AnimGraphNode_CopyBone
     alphaBoolBlend {blendOption=Linear}
     sourceBone {boneName=VB attach_ik_foot_r}
     targetBone {boneName=ik_foot_r}
    }
    Alpha:float=Get AttachEmote1InPelvisSpace() -> (AttachEmote1InPelvisSpace:float)
    bCopyTranslation:bool=true
    bCopyRotation:bool=true
    bCopyScale:bool=true
   ) -> (Pose:Component Space Pose Link Structure) {
    is=title AnimGraphNode_CopyBone
    alphaBoolBlend {blendOption=Linear}
    sourceBone {boneName=VB pelvis_attach_emote_01}
    targetBone {boneName=attach_emote_01}
   }
   n14 = Copy Bone - Source Bone: VB pelvis_attach_emote_02 - Target Bone: attach_emote_02!(
    ComponentPose:Component Space Pose Link Structure=@n13
    Alpha:float=Get AttachEmote2InPelvisSpace() -> (AttachEmote2InPelvisSpace:float)
    bCopyTranslation:bool=true
    bCopyRotation:bool=true
    bCopyScale:bool=true
   ) -> (Pose:Component Space Pose Link Structure) {
    is=title AnimGraphNode_CopyBone
    alphaBoolBlend {blendOption=Linear}
    sourceBone {boneName=VB pelvis_attach_emote_02}
    targetBone {boneName=attach_emote_02}
   }
   n15 = Copy Bone - Source Bone: VB pelvis_attach_emote_03 - Target Bone: attach_emote_03!(
    ComponentPose:Component Space Pose Link Structure=@n14
    Alpha:float=Get AttachEmote3InPelvisSpace() -> (AttachEmote3InPelvisSpace:float)
    bCopyTranslation:bool=true
    bCopyRotation:bool=true
    bCopyScale:bool=true
   ) -> (Pose:Component Space Pose Link Structure) {
    is=title AnimGraphNode_CopyBone
    alphaBoolBlend {blendOption=Linear}
    sourceBone {boneName=VB pelvis_attach_emote_03}
    targetBone {boneName=attach_emote_03}
   }
   n16 = "Transform (Modify) Bone - Bone: attach_emote_03"!(
    ComponentPose=@n15
    Alpha=Get SlopeMatchAttachEmote3() -> (SlopeMatchAttachEmote3:float)
    Rotation=Get SmoothedSlopeRotationCS() -> (SmoothedSlopeRotationCS:Rotator)
   ) -> (Pose) {rotationMode=BMM_Replace, alphaBoolBlend {blendOption=Linear}, boneToModify {boneName=attach_emote_03}, translation {z=100}}
   n17 = Get IkSoftPercentLength() -> (IkSoftPercentLength:float)
   n18 = Leg IK!(ComponentPose=@n16, Alpha=Get RightLegIKAlpha() -> (RightLegIKAlpha:float), SoftPercentLength:float=@n17) -> (Pose) {
    softPercentLength=0.95
    alphaBoolBlend {blendOption=Linear}
    legsDefinition [{
      bEnableKneeTwistCorrection=false
      bEnableRotationLimit=false
      footBoneForwardAxis=Y
      hingeRotationAxis=Z
      minRotationAngle=15
      numBonesInLimb=2
      twistOffsetCurveName=None
      fKFootBone {boneName=foot_r}
      iKFootBone {boneName=ik_foot_r}
     }]
   }
   n19 = Leg IK(ComponentPose=@n18, Alpha=Get LeftLegIKAlpha() -> (LeftLegIKAlpha:float), SoftPercentLength=@n17) {
    softPercentLength=0.95
    alphaBoolBlend {blendOption=Linear}
    legsDefinition [{
      bEnableKneeTwistCorrection=false
      bEnableRotationLimit=false
      footBoneForwardAxis=Y
      hingeRotationAxis=Z
      minRotationAngle=15
      numBonesInLimb=2
      twistOffsetCurveName=None
      fKFootBone {boneName=foot_l}
      iKFootBone {boneName=ik_foot_l}
     }]
   }
   n20 = "Transform (Modify) Bone - Bone: Root"!(
    ComponentPose=@n19
    Alpha=Get RootSlopeRotationAlpha() -> (RootSlopeRotationAlpha:float)
    Rotation=Get RootSlopeRotationOffset() -> (RootSlopeRotationOffset:Rotator)
   ) -> (Pose) {
    rotationMode=BMM_Additive
    translationMode=BMM_Additive
    alphaBoolBlend {blendOption=Linear}
    boneToModify {boneName=Root}
    translation {z=-2.5}
   }
   n21 = "Transform (Modify) Bone - Bone: Root"!(
    ComponentPose=@n20
    Alpha=Get IdlePelvisOffsetAlpha() -> (IdlePelvisOffsetAlpha:float)
    Translation:Vector="0,0,1.5"
    Rotation="0,0,0"
    Scale:Vector="1,1,1"
   ) -> (Pose) {lODThreshold=2, translationMode=BMM_Additive, alphaBoolBlend {blendOption=Linear}, boneToModify {boneName=Root}, translation {z=1.5}}
   Output Pose(Result=Component To Local(ComponentPose=@n21)) {layerGroup=None}
   Save cached pose 'PreIKAdjustmentsInputPose'(Pose=Input Pose!() -> (Pose) {name=InPosePreIKAdjustments}) {cacheName=PreIKAdjustmentsInputPose}
   Save cached pose 'PostRotateRoot'(Pose=Rotate Root Bone!(
     BasePose=Use cached pose 'PreIKAdjustmentsInputPose'()
     Pitch:float=0
     Yaw=Get EmoteForcedRootYaw() -> (EmoteForcedRootYaw:float)
     MeshToComponent:Rotator="0,0,0"
    ) -> (Pose)) {cacheName=PostRotateRoot}
  ]}
 ParaGliderAdjustmentsLayer {nodes [
   n1 = Break Fort Anim Input Skydiving(FortAnimInput_Skydiving=Get Skydiving()) -> (
    ParaGilderRootOffset=~
    ParaGliderRootRotationOverride=~
    bUseParaGlideRootModifier=~
    bIsSkydivingFromLaunchPad=~
    bIsSkydivingFromBus=~
    bIsInVortex=~
    bIsDiving=~
    bIsParachuteOpen=~
    bIsSkydiving=~
    bIsParachuteLeaning=~
    bIsSkydiveLeaning=~
    bIsLeaning=~
    bIsSkydiveDiveMode=~
    bPlaySkydiveDrift=~
    bSkydiveDriftDelayActive=~
    bSkydiveDriftAnimAllowed=~
    bIsGliderRight=~
    bIsGliderCenter=~
    bIsGliderLeft=~
    bIsGliderForward=~
    bIsGliderBack=~
    bHasDeployedAndIsStandingUpright=~
    LocalAccelForward=~
    LocalAccelRight=~
    LocalVelocityRight=~
    SkydiveAimPitch=~
    SkydiveAimYaw=~
    DeployChuteAnimRate=~
    SkydiveDriftAnimRate=~
    SkydiveDriftAnimRateCurrent=~
    SkydiveFidgetAnimRate=~
    SkydiveFidgetAnimRateCurrent=~
    SkydiveAdditiveAlpha=~
    SkydiveDriftDelay=~
   )
   n2 = "Transform (Modify) Bone - Bone: Root"!(
    ComponentPose=Local To Component(LocalPose=Use cached pose 'ParaGliderAdjustmentsInputPose'())
    AlphaCurveName:Name=ParaGlideModifyRoot
    Translation=@n1
    Rotation=@n1.ParaGliderRootRotationOverride
    Scale="1,1,1"
   ) -> (Pose) {
    alphaInputType=Curve
    rotationMode=BMM_Replace
    translationMode=BMM_Additive
    alphaBoolBlend {blendOption=Linear}
    boneToModify {boneName=Root}
   }
   n3 = Break Fort Anim Input Spring Glider!(FortAnimInput_SpringGlider:"Fort Anim Input Spring Glider Structure (by ref)"=Get SpringGlider() -> (SpringGlider:Fort Anim Input Spring Glider Structure)) -> (GliderOffsetRotator:Rotator=~, GliderOffsetRotationRightLeg:Rotator=~, GliderOffsetRotationLeftLeg:Rotator=~, GliderOffsetLocation:Vector=~) {is=title K2Node_BreakStruct}
   n4 = "Transform (Modify) Bone - Bone: pelvis"!(
    ComponentPose=Local To Component(LocalPose=Use cached pose 'ParaGliderAdjustmentsInputPose'())
    AlphaCurveName:Name=ParaGlideModifyPelvis
    Translation:Vector=@n3.GliderOffsetLocation
    Rotation=@n3
    Scale:Vector="1,1,1"
   ) -> (Pose) {
    alphaInputType=Curve
    rotationMode=BMM_Additive
    translationMode=BMM_Additive
    alphaBoolBlend {blendOption=Linear}
    boneToModify {boneName=pelvis}
   }
   n5 = "Transform (Modify) Bone - Bone: thigh_r"!(ComponentPose=@n4, AlphaCurveName:Name=ParaGlideModifyLegs, Translation="0,0,0", Rotation=@n3.GliderOffsetRotationRightLeg, Scale="1,1,1") -> (Pose) {alphaInputType=Curve, rotationMode=BMM_Additive, alphaBoolBlend {blendOption=Linear}, boneToModify {boneName=thigh_r}}
   n6 = "Transform (Modify) Bone - Bone: thigh_l"!(ComponentPose=@n5, AlphaCurveName:Name=ParaGlideModifyLegs, Translation="0,0,0", Rotation=@n3.GliderOffsetRotationLeftLeg, Scale="1,1,1") -> (Pose) {alphaInputType=Curve, rotationMode=BMM_Additive, alphaBoolBlend {blendOption=Linear}, boneToModify {boneName=thigh_l}}
   n7 = Get CurrentGliderAnimSet() -> (CurrentGliderAnimSet:Fort Anim Input Player Glider Anim Asset Structure)
   Output Pose(Result=Blend Poses by bool(
     BlendPose_0=Component To Local(ComponentPose=@n6)
     BlendPose_1=Component To Local(ComponentPose=@n2)
     bActiveValue=Break Fort Anim Input Player Glider Anim Asset!(FortAnimInput_PlayerGliderAnimAsset:"Fort Anim Input Player Glider Anim Asset Structure (by ref)"=@n7) -> (
      Into_Base_MaleMedium_BS:Blend Space Object Reference
      bEnableSpringMods:bool=~
      bUseLeanPlayRateAdjust:bool=~
      bUseForwardAs_Y_BS:bool=~
      bUseDeployCurveForIntoPlayRateAdjust:bool=~
      LeanForwardValue:float=~
      IdleValue:float=~
      LeanBackValue:float=~
      IdleInterpToSpeed:float=~
      LeanForwardInterpToSpeed:float=~
      LeanBackInterpToSpeed:float=~
     ) {is=title K2Node_BreakStruct}.bEnableSpringMods
    ) {blendType=Linear}) {layerGroup=None}
   Save cached pose 'ParaGliderAdjustmentsInputPose'(Pose=Input Pose() {name=InPoseParaGliderAdjustments}) {cacheName=ParaGliderAdjustmentsInputPose}
  ]}
 PlevisAndFeetAdjustmentsLayer {nodes [
   n1 = "Transform (Modify) Bone - Bone: ik_foot_l"!(ComponentPose=Local To Component(LocalPose=Use cached pose 'PelvisAndFeetAdjustmentsInputPose'()), Alpha:float=1, Translation="4,0,0") -> (Pose) {translationMode=BMM_Additive, alphaBoolBlend {blendOption=Linear}, boneToModify {boneName=ik_foot_l}, translation {x=4}}
   n2 = "Transform (Modify) Bone - Bone: ik_foot_r"!(ComponentPose=@n1, Alpha:float=1, Translation="4,0,0") -> (Pose) {translationMode=BMM_Additive, alphaBoolBlend {blendOption=Linear}, boneToModify {boneName=ik_foot_r}, translation {x=4}}
   n3 = "Transform (Modify) Bone - Bone: pelvis"!(ComponentPose=@n2, Alpha:float=1, Translation="8,0,0") -> (Pose) {translationMode=BMM_Additive, alphaBoolBlend {blendOption=Linear}, boneToModify {boneName=pelvis}, translation {x=8}}
   n4 = "Transform (Modify) Bone - Bone: ik_foot_l"(ComponentPose=Local To Component(LocalPose=Use cached pose 'PelvisAndFeetAdjustmentsInputPose'()), Alpha=1, Translation="9,0,0") {translationMode=BMM_Additive, alphaBoolBlend {blendOption=Linear}, boneToModify {boneName=ik_foot_l}, translation {x=9}}
   n5 = "Transform (Modify) Bone - Bone: ik_foot_r"(ComponentPose=@n4, Alpha=1, Translation="7,0,0") {translationMode=BMM_Additive, alphaBoolBlend {blendOption=Linear}, boneToModify {boneName=ik_foot_r}, translation {x=7}}
   n6 = "Transform (Modify) Bone - Bone: pelvis"(ComponentPose=@n5, Alpha=1, Translation="9,0,0") {translationMode=BMM_Additive, alphaBoolBlend {blendOption=Linear}, boneToModify {boneName=pelvis}, translation {x=9}}
   n7 = Blend Poses by bool(
    BlendPose_0=Component To Local(ComponentPose=@n6)
    BlendPose_1=Blend Poses by bool(
     BlendPose_0=Component To Local(ComponentPose=@n3)
     BlendPose_1=Use cached pose 'PelvisAndFeetAdjustmentsInputPose'()
     BlendTime_0=0.2
     BlendTime_1=0.2
     bActiveValue=Get bIsCrouching()
    ) {blendType=Linear}
    BlendTime_0=0.2
    BlendTime_1=0.2
    bActiveValue=Get bIsCrouchMoving() -> (bIsCrouchMoving:bool)
   ) {blendType=Linear}
   n8 = "Transform (Modify) Bone - Bone: ik_foot_l"(ComponentPose=Local To Component(LocalPose=Use cached pose 'PelvisAndFeetAdjustmentsInputPose'()), Alpha=1, Translation="10,0,0") {translationMode=BMM_Additive, alphaBoolBlend {blendOption=Linear}, boneToModify {boneName=ik_foot_l}, translation {x=10}}
   n9 = "Transform (Modify) Bone - Bone: ik_foot_r"(ComponentPose=@n8, Alpha=1, Translation="4,0,0") {translationMode=BMM_Additive, alphaBoolBlend {blendOption=Linear}, boneToModify {boneName=ik_foot_r}, translation {x=4}}
   n10 = "Transform (Modify) Bone - Bone: pelvis"(ComponentPose=@n9, Alpha=1, Translation="9,0,0") {translationMode=BMM_Additive, alphaBoolBlend {blendOption=Linear}, boneToModify {boneName=pelvis}, translation {x=9}}
   n11 = Component To Local(ComponentPose=@n10)
   n12 = Get bIsCrouchSprinting()
   Output Pose(Result=Blend Poses by bool(
     BlendPose_0=Use cached pose 'PelvisAndFeetAdjustmentsInputPose'()
     BlendPose_1=Blend Poses by bool(BlendPose_0=@n11, BlendPose_1=@n7, BlendTime_0=0.2, BlendTime_1=0.2, bActiveValue=@n12) {blendType=Linear}
     BlendTime_0=0.2
     BlendTime_1=0.2
     bActiveValue=Get bIsRiding() -> (bIsRiding:bool)
    ) {blendType=Linear}) {layerGroup=None}
   Save cached pose 'PelvisAndFeetAdjustmentsInputPose'(Pose=Input Pose() {name=InPosePelvisAndFeetAdjustments}) {cacheName=PelvisAndFeetAdjustmentsInputPose}
  ]}
 PreLegIK {nodes [Output Pose(Result=Input Pose()) {layerGroup=PlayerOverride}]}
 PushCannonAdjustmentsLayer {nodes [
   n1 = Get bIsInCannon()
   n2 = "Transform (Modify) Bone - Bone: Root"!(
    ComponentPose=Local To Component(LocalPose=Input Pose() {name=InPosePushCannonAdjustments})
    bAlphaBoolEnabled:bool=@n1
    Translation="0,0,15"
    Scale="0.8,0.8,0.8"
   ) -> (Pose) {
    alphaInputType=Bool
    scaleMode=BMM_Replace
    scaleSpace=BCS_BoneSpace
    translationMode=BMM_Additive
    translationSpace=BCS_BoneSpace
    alphaBoolBlend {blendOption=Linear}
    boneToModify {boneName=Root}
    scale=<0.8 0.8 0.8>
    translation {z=15}
   }
   n3 = "Transform (Modify) Bone - Bone: head"!(ComponentPose=@n2, bAlphaBoolEnabled:bool=@n1, Scale="1.2,1.2,1.2") -> (Pose) {
    alphaInputType=Bool
    scaleMode=BMM_Replace
    scaleSpace=BCS_BoneSpace
    alphaBoolBlend {blendOption=Linear}
    boneToModify {boneName=head}
    scale=<1.2 1.2 1.2>
   }
   Output Pose(Result=Component To Local(ComponentPose=@n3)) {layerGroup=None}
  ]}
 UpperBody {nodes [
   Input Pose() {name=InPoseUpperBody}
   Output Pose() {layerGroup=PlayerOverride}
  ]}
]
variables [
 EmoteAnimOffsetCurveName {type=string, default=UseEmoteAnimOffset}
 IsHidingInProp {type=boolean}
 LeftHandOverride {type=Vector}
 RightHandOverride {type=Vector}
]
