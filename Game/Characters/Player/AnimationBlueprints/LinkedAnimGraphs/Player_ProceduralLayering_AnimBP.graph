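# Player linked anim graph that procedurally layers item poses over locomotion, with a first-person camera variant.
# ProceduralLayeredPose bone-masks the item pose onto locomotion; ProceduralAdditivePose runs a Blend Stack with dead blending and pose history.
# Each stack entry retargets hand IK, applies locomotion as a dynamic additive against a base pose, then transfers motion to IK hands and pelvis via Copy Motion.
# Post-processing solves arms with Leg IK and syncs virtual bones.
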AnimBlueprint /Game/Characters/Player/AnimationBlueprints/LinkedAnimGraphs/Player_ProceduralLayering_AnimBP
target_skeleton=&Fortnite_M_Avg_Player_Skeleton
graphs [
 AdditiveLayering {nodes [
   n1 = Use cached pose 'ItemAdditivePose'()
   n2 = Use cached pose 'InSourcePose'()
   n3 = Get AdditiveBoneMask() -> (AdditiveBoneMask:Bone Mask Structure)
   n4 = Layered blend per bone(
    BasePose=Blend Poses by bool(
     BlendPose_0=Use cached pose 'InSourcePose'()
     BlendPose_1=Bone Mask!(
      BasePose:Pose Link Structure=@n2
      BlendPoses_0:Pose Link Structure=@n1
      BlendWeights_0:float=1
      BoneMask:Bone Mask Structure=@n3
      BoneMaskDefinitionDataAsset:Bone Mask Definition Data Asset Object Reference=/Game/Characters/Player/DataAssets/Layering_BodyParts_DA
     ) -> (Pose:Pose Link Structure) {is=title AnimGraphNode_BoneMask, bProcessMeshSpaceAsRootSpace=true}
     bActiveValue=Get bIsMovingOnFootWithMeleeTwistCurve() -> (bIsMovingOnFootWithMeleeTwistCurve:bool)
    ) {blendType=Linear}
    BlendPoses_0=Use cached pose 'InSourcePose'()
    BlendWeights_0=Get HandAttachRInChestSpaceCurveValue() -> (HandAttachRInChestSpaceCurveValue:float)
   ) {bMeshSpaceRotationBlend=true, layerSetup [{branchFilters [{boneName=VB spine_05_hand_attach_r}, {boneName=VB spine_05_weapon_r}]}]}
   n5 = Layered blend per bone(
    BasePose=@n4
    BlendPoses_0=Use cached pose 'InSourcePose'()
    BlendWeights_0=Get WeaponRInChestSpaceCurveValue() -> (WeaponRInChestSpaceCurveValue:float)
   ) {bMeshSpaceRotationBlend=true, layerSetup [{branchFilters [{boneName=VB spine_05_weapon_r}, {boneName=VB spine_05_weapon_r}]}]}
   AnimGraphNode_Root_0 = Output Pose(Result=Two Way Blend(A=Use cached pose 'InSourcePose'(), B=@n5, Alpha=Get AdditiveLayeringAlpha() -> (AdditiveLayeringAlpha:float)) {blendNode {alphaBoolBlend {blendOption=Linear}}}) {
    binding=&self:AdditiveLayering.AnimGraphNode_Root_0.Binding
    layerGroup=None
   }
   AnimGraphNode_LinkedInputPose_2 = Input Pose() {
    name=AddLayer_InBasePose
    binding=&self:AdditiveLayering.AnimGraphNode_LinkedInputPose_2.Binding
   }
   Save cached pose 'AddLayerBasePose'(Pose=@AnimGraphNode_LinkedInputPose_2) {cacheName=AddLayerBasePose}
   AnimGraphNode_LinkedInputPose_0 = Input Pose() {
    name=AddLayer_InLocomotionPose
    binding=&self:AdditiveLayering.AnimGraphNode_LinkedInputPose_0.Binding
   }
   Save cached pose 'InLocomotionPose'(Pose=@AnimGraphNode_LinkedInputPose_0) {cacheName=InLocomotionPose}
   Save cached pose 'InSourcePose'(Pose=Input Pose() {name=InPoseSourcePose}) {cacheName=InSourcePose}
   n6 = Make Dynamic Additive!(Base:Pose Link Structure=Use cached pose 'AddLayerBasePose'(), Additive:Pose Link Structure=Use cached pose 'InLocomotionPose'()) -> (Pose:Pose Link Structure) {is=title AnimGraphNode_MakeDynamicAdditive}
   n7 = Use cached pose 'InSourcePose'()
   n8 = Make Dynamic Additive(Base=Use cached pose 'AddLayerBasePose'(), Additive=Use cached pose 'InLocomotionPose'()) {bMeshSpaceAdditive=true}
   n9 = Use cached pose 'InSourcePose'()
   Save cached pose 'ItemAdditivePose'(Pose=Blend Poses by bool(
     BlendPose_0=Apply Additive(Base=@n7, Additive=@n6) {alphaBoolBlend {blendOption=Linear}}
     BlendPose_1=Apply Mesh Space Additive!(Base:Pose Link Structure=@n9, Additive:Pose Link Structure=@n8, Alpha:float=1) -> (Pose:Pose Link Structure) {is=title AnimGraphNode_ApplyMeshSpaceAdditive, bRootSpaceAdditive=true}
     BlendTime_0=0
     BlendTime_1=0
     bActiveValue=Get bGenerateLocalSpaceAdditive() -> (bGenerateLocalSpaceAdditive:bool)
    )) {cacheName=ItemAdditivePose}
  ]}
 AdditivePosePostProcess {nodes [
   n1 = Use cached pose 'InputPose'()
   n2 = Get HandIKInput() -> (HandIKInput:Fort Anim Input Hand IK Structure)
   n3 = Leg IK!(
    ComponentPose:Component Space Pose Link Structure=Copy Bone - Source Bone: hand_r - Target Bone: ik_hand_r!(
     ComponentPose:Component Space Pose Link Structure=Local To Component(LocalPose=@n1)
     Alpha:float=Break Fort Anim Input Hand IK!(FortAnimInput_HandIK:"Fort Anim Input Hand IK Structure (by ref)"=@n2) -> (
      HandIKRetargetingWeight:float=~
      HandsInRootSpaceAlpha:float=~
      LeftHandIKAlpha:float=~
      RightHandIKAlpha:float=~
      RightHandFKRotationAlpha:float=~
     ) {is=title K2Node_BreakStruct}.RightHandFKRotationAlpha
     bCopyTranslation:bool=False
     bCopyRotation:bool=true
     bCopyScale:bool=False
    ) -> (Pose:Component Space Pose Link Structure) {is=title AnimGraphNode_CopyBone, alphaBoolBlend {blendOption=Linear}, sourceBone {boneName=hand_r}, targetBone {boneName=ik_hand_r}}
    Alpha:float=Get RightHandIKAlpha() -> (RightHandIKAlpha:float)
   ) -> (Pose:Component Space Pose Link Structure) {
    is=title AnimGraphNode_LegIK
    alphaBoolBlend {blendOption=Linear}
    legsDefinition [{
      bEnableKneeTwistCorrection=false
      bEnableRotationLimit=false
      footBoneForwardAxis=Y
      hingeRotationAxis=Z
      minRotationAngle=15
      numBonesInLimb=2
      twistOffsetCurveName=RightElbowTwist
      fKFootBone {boneName=hand_r}
      iKFootBone {boneName=ik_hand_r}
     }]
   }
   n4 = Leg IK(ComponentPose=@n3, Alpha=Get LeftHandIKAlpha() -> (LeftHandIKAlpha:float)) {
    alphaBoolBlend {blendOption=Linear}
    legsDefinition [{
      bEnableKneeTwistCorrection=false
      bEnableRotationLimit=false
      footBoneForwardAxis=Y
      hingeRotationAxis=Z
      minRotationAngle=15
      numBonesInLimb=2
      twistOffsetCurveName=LeftElbowTwist
      fKFootBone {boneName=hand_l}
      iKFootBone {boneName=ik_hand_l}
     }]
   }
   n5 = Copy Bone - Source Bone: ik_hand_gun - Target Bone: VB VB spine_05_weapon_r_ik_hand_gun!(
    ComponentPose:Component Space Pose Link Structure=Copy Bone - Source Bone: weapon_r - Target Bone: VB spine_05_weapon_r!(
     ComponentPose:Component Space Pose Link Structure=@n4
     Alpha:float=1
     bCopyTranslation:bool=true
     bCopyRotation:bool=true
     bCopyScale:bool=true
    ) -> (Pose:Component Space Pose Link Structure) {
     is=title AnimGraphNode_CopyBone
     alphaBoolBlend {blendOption=Linear}
     sourceBone {boneName=weapon_r}
     targetBone {boneName=VB spine_05_weapon_r}
    }
    Alpha:float=1
    bCopyTranslation:bool=true
    bCopyRotation:bool=true
    bCopyScale:bool=true
   ) -> (Pose:Component Space Pose Link Structure) {
    is=title AnimGraphNode_CopyBone
    alphaBoolBlend {blendOption=Linear}
    sourceBone {boneName=ik_hand_gun}
    targetBone {boneName=VB VB spine_05_weapon_r_ik_hand_gun}
   }
   n6 = Copy Bone - Source Bone: ik_hand_l - Target Bone: VB VB spine_05_weapon_r_ik_hand_l!(
    ComponentPose:Component Space Pose Link Structure=@n5
    Alpha:float=1
    bCopyTranslation:bool=true
    bCopyRotation:bool=true
    bCopyScale:bool=true
   ) -> (Pose:Component Space Pose Link Structure) {
    is=title AnimGraphNode_CopyBone
    alphaBoolBlend {blendOption=Linear}
    sourceBone {boneName=ik_hand_l}
    targetBone {boneName=VB VB spine_05_weapon_r_ik_hand_l}
   }
   n7 = Copy Bone - Source Bone: hand_attach_r - Target Bone: VB spine_05_hand_attach_r!(
    ComponentPose:Component Space Pose Link Structure=@n6
    Alpha:float=1
    bCopyTranslation:bool=true
    bCopyRotation:bool=true
    bCopyScale:bool=true
   ) -> (Pose:Component Space Pose Link Structure) {
    is=title AnimGraphNode_CopyBone
    alphaBoolBlend {blendOption=Linear}
    sourceBone {boneName=hand_attach_r}
    targetBone {boneName=VB spine_05_hand_attach_r}
   }
   n8 = Copy Bone - Source Bone: ik_hand_gun - Target Bone: VB VB spine_05_hand_attach_r_ik_hand_gun!(
    ComponentPose:Component Space Pose Link Structure=@n7
    Alpha:float=1
    bCopyTranslation:bool=true
    bCopyRotation:bool=true
    bCopyScale:bool=true
   ) -> (Pose:Component Space Pose Link Structure) {
    is=title AnimGraphNode_CopyBone
    alphaBoolBlend {blendOption=Linear}
    sourceBone {boneName=ik_hand_gun}
    targetBone {boneName=VB VB spine_05_hand_attach_r_ik_hand_gun}
   }
   n9 = Copy Bone - Source Bone: ik_hand_l - Target Bone: VB VB spine_05_hand_attach_r_ik_hand_l!(
    ComponentPose:Component Space Pose Link Structure=@n8
    Alpha:float=1
    bCopyTranslation:bool=true
    bCopyRotation:bool=true
    bCopyScale:bool=true
   ) -> (Pose:Component Space Pose Link Structure) {
    is=title AnimGraphNode_CopyBone
    alphaBoolBlend {blendOption=Linear}
    sourceBone {boneName=ik_hand_l}
    targetBone {boneName=VB VB spine_05_hand_attach_r_ik_hand_l}
   }
   n10 = Copy Bone - Source Bone: ik_hand_r - Target Bone: VB Root_hand_r!(
    ComponentPose:Component Space Pose Link Structure=Copy Bone - Source Bone: ik_hand_l - Target Bone: VB Root_hand_l!(
     ComponentPose:Component Space Pose Link Structure=Copy Bone - Source Bone: ik_hand_l - Target Bone: VB pelvis_ik_hand_l!(
      ComponentPose:Component Space Pose Link Structure=@n9
      Alpha:float=1
      bCopyTranslation:bool=true
      bCopyRotation:bool=true
      bCopyScale:bool=true
     ) -> (Pose:Component Space Pose Link Structure) {
      is=title AnimGraphNode_CopyBone
      alphaBoolBlend {blendOption=Linear}
      sourceBone {boneName=ik_hand_l}
      targetBone {boneName=VB pelvis_ik_hand_l}
     }
     Alpha:float=1
     bCopyTranslation:bool=true
     bCopyRotation:bool=true
     bCopyScale:bool=true
    ) -> (Pose:Component Space Pose Link Structure) {is=title AnimGraphNode_CopyBone, alphaBoolBlend {blendOption=Linear}, sourceBone {boneName=ik_hand_l}, targetBone {boneName=VB Root_hand_l}}
    Alpha:float=1
    bCopyTranslation:bool=true
    bCopyRotation:bool=true
    bCopyScale:bool=true
   ) -> (Pose:Component Space Pose Link Structure) {is=title AnimGraphNode_CopyBone, alphaBoolBlend {blendOption=Linear}, sourceBone {boneName=ik_hand_r}, targetBone {boneName=VB Root_hand_r}}
   AnimGraphNode_UseCachedPose_1 = Use cached pose 'InputPose'()
   AnimGraphNode_Root_0 = Output Pose(Result=Two Way Blend(A=@AnimGraphNode_UseCachedPose_1, B=Component To Local(ComponentPose=@n10), Alpha=Get PostProcessAlpha() -> (PostProcessAlpha:float)) {blendNode {alphaBoolBlend {blendOption=Linear}}}) {layerGroup=None}
   AnimGraphNode_LinkedInputPose_0 = Input Pose()
   AnimGraphNode_SaveCachedPose_0 = Save cached pose 'InputPose'(Pose=@AnimGraphNode_LinkedInputPose_0) {cacheName=InputPose}
  ]}
 AnimGraph {nodes [AnimGraphNode_Root_0 = Output Pose() {
    binding=&self:AnimGraph.AnimGraphNode_Root_0.Binding
    layerGroup=None
   }]}
 BasePose {nodes [
   n1 = Cached Sequence Evaluator!(ExplicitTime:float=0) -> (Pose:Pose Link Structure) {
    is=title FortAnimGraphNode_CachedSequenceEvaluator
    node {bShouldLoop=false, cacheName=BasePose, updateFunction {functionName=OnUpdateNoWeaponBasePose}}
    updateFunction {bSelfContext=true, memberName=OnUpdateNoWeaponBasePose}
   }
   AnimGraphNode_Root_0 = Output Pose(Result=@n1) {
    binding=&self:BasePose.AnimGraphNode_Root_0.Binding
    layerGroup=None
   }
  ]}
 BlendStack {nodes [
   n1 = Layered blend per bone(BasePose=Use cached pose 'FinalPose'(), BlendPoses_0=Use cached pose 'BS_SourcePose'()) {layerSetup [{branchFilters [{boneName=clavicle_r}, {boneName=clavicle_l}, {blendDepth=3, boneName=Spine_01}, {boneName=root_fp}]}]}
   n2 = Blend Bone By Channel!(A:Pose Link Structure=@n1, B:Pose Link Structure=Use cached pose 'BS_SourcePose'(), Alpha:float=1) -> (Pose:Pose Link Structure) {
    is=title AnimGraphNode_BlendBoneByChannel
    blendNode {
     transformsSpace=BCS_ComponentSpace
     boneDefinitions [
      {bBlendRotation=true, bBlendScale=true, bBlendTranslation=true, sourceBone {boneName=clavicle_r}, targetBone {boneName=clavicle_r}}
      {bBlendRotation=true, bBlendScale=true, bBlendTranslation=true, sourceBone {boneName=clavicle_l}, targetBone {boneName=clavicle_l}}
     ]
    }
   }
   n3 = Blend Poses by bool(
    BlendPose_0=@n2
    BlendPose_1=Use cached pose 'FinalPose'()
    bActiveValue=Get bEnableCameraBoneSpace1stPersonAnims() -> (bEnableCameraBoneSpace1stPersonAnims:bool)
   ) {blendType=Linear}
   n4 = Cached Sequence Evaluator!(Sequence:Anim Sequence Base Object Reference) -> (Pose) {cacheName=SourcePose}
   AnimGraphNode_Root_0 = Output Pose(Result=Blend Poses by bool(
     BlendPose_0=Layered blend per bone(BasePose=@n3, BlendPoses_0=@n4) {layerSetup [{branchFilters [{boneName=camera_root}, {blendDepth=-1, boneName=camera}]}]}
     BlendPose_1=Use cached pose 'FinalPose'()
     bActiveValue=Get bIsInFirstPerson() -> (bIsInFirstPerson:bool)
    ) {blendType=Linear}) {layerGroup=None}
   n5 = Use cached pose 'BS_SourceLocoPose'()
   n6 = Hand IK Retargeting!(
    ComponentPose:Component Space Pose Link Structure=Copy Bone - Source Bone: VB pelvis_ik_hand_l - Target Bone: ik_hand_l!(
     ComponentPose:Component Space Pose Link Structure=Local To Component(LocalPose=@n5)
     Alpha:float=Get IKHandLeftInPelvisSpaceAlpha() -> (IKHandLeftInPelvisSpaceAlpha:float)
     bCopyTranslation:bool=true
     bCopyRotation:bool=true
     bCopyScale:bool=true
    ) -> (Pose:Component Space Pose Link Structure) {
     is=title AnimGraphNode_CopyBone
     alphaBoolBlend {blendOption=Linear}
     sourceBone {boneName=VB pelvis_ik_hand_l}
     targetBone {boneName=ik_hand_l}
    }
    AlphaCurveName:Name=DisableHandIKRetargeting
    PerAxisAlpha:Vector="1,1,1"
    HandFKWeight:float=Get HandIKRetargetingWeight() -> (HandIKRetargetingWeight:float)
   ) -> (Pose:Component Space Pose Link Structure) {
    is=title AnimGraphNode_HandIKRetargeting
    alphaInputType=Curve
    alphaBoolBlend {blendOption=Linear}
    alphaScaleBiasClamp {bias=1, scale=-1}
    iKBonesToMove [{boneName=ik_hand_gun}]
    leftHandFK {boneName=hand_l}
    leftHandIK {boneName=ik_hand_l}
    rightHandFK {boneName=hand_r}
    rightHandIK {boneName=ik_hand_r}
   }
   AnimGraphNode_LinkedAnimLayer_2 = AdditiveLayering!(
    AddLayer_InLocomotionPose:Pose Link Structure=Use cached pose 'BS_BaseLocomotionPose'()
    AddLayer_InBasePose:Pose Link Structure=Use cached pose 'BS_BasePose'()
    InPoseSourcePose:Pose Link Structure=Component To Local(ComponentPose=@n6)
   ) -> (Pose:Pose Link Structure) {
    is=title AnimGraphNode_LinkedAnimLayer
    binding=&self:BlendStack.AnimGraphNode_LinkedAnimLayer_2.Binding
    layer=AdditiveLayering
   }
   AnimGraphNode_LinkedAnimLayer_3 = ProceduralLayering!(
    ProcLayer_InBaseLocomotionPose:Pose Link Structure=Use cached pose 'BS_BaseLocomotionPose'()
    ProcLayer_InLayeredLocomotionPose:Pose Link Structure=@AnimGraphNode_LinkedAnimLayer_2
    ProcLayer_InBasePose:Pose Link Structure=Use cached pose 'BS_BasePose'()
   ) -> (Pose:Pose Link Structure) {
    is=title AnimGraphNode_LinkedAnimLayer
    binding=&self:BlendStack.AnimGraphNode_LinkedAnimLayer_3.Binding
    layer=ProceduralLayering
   }
   AnimGraphNode_LinkedAnimLayer_0 = AdditivePosePostProcess!(InPose:Pose Link Structure=@AnimGraphNode_LinkedAnimLayer_3) -> (Pose:Pose Link Structure) {is=title AnimGraphNode_LinkedAnimLayer, layer=AdditivePosePostProcess}
   AnimGraphNode_UseCachedPose_1 = Use cached pose 'BS_SourceLocoPose'()
   AnimGraphNode_SaveCachedPose_0 = Save cached pose 'FinalPose'(Pose=Blend Poses by bool(
     BlendPose_0=@AnimGraphNode_LinkedAnimLayer_0
     BlendPose_1=@AnimGraphNode_UseCachedPose_1
     BlendTime_0=0
     BlendTime_1=0
     bActiveValue=Get bUseProceduralLayering() -> (bUseProceduralLayering:bool)
    ) {blendType=Linear}) {cacheName=FinalPose}
   AnimGraphNode_LinkedInputPose_1 = Input Pose() {name=InSourceLocoPose}
   Save cached pose 'BS_SourceLocoPose'(Pose=@AnimGraphNode_LinkedInputPose_1) {cacheName=BS_SourceLocoPose}
   AnimGraphNode_LinkedInputPose_0 = Input Pose() {name=SourcePose}
   Save cached pose 'BS_SourcePose'(Pose=@AnimGraphNode_LinkedInputPose_0) {cacheName=BS_SourcePose}
   AnimGraphNode_LinkedInputPose_2 = Input Pose() {name=BasePose}
   Save cached pose 'BS_BasePose'(Pose=@AnimGraphNode_LinkedInputPose_2) {cacheName=BS_BasePose}
   Save cached pose 'BS_BaseLocomotionPose'(Pose=Input Pose() {name=BaseLocomotionPose}) {cacheName=BS_BaseLocomotionPose}
  ]}
 CopyBoneLayer {nodes [
   AnimGraphNode_UseCachedPose_1 = Use cached pose 'InPose'()
   n1 = Copy Bone Advanced - Source Bone: hand_r - Target Bone: ik_hand_r!(
    ComponentPose:Component Space Pose Link Structure=Local To Component(LocalPose=@AnimGraphNode_UseCachedPose_1)
    Alpha:float=0
    TranslationWeight:Vector="1,1,1"
    RotationWeight:float=1
   ) -> (Pose:Component Space Pose Link Structure) {
    is=title AnimGraphNode_CopyBoneAdvanced
    bTranslationInCustomBoneSpace=false
    alphaBoolBlend {blendOption=Linear}
    sourceBone {boneName=hand_r}
    targetBone {boneName=ik_hand_r}
   }
   n2 = Copy Bone Advanced - Source Bone: hand_l - Target Bone: ik_hand_l!(ComponentPose:Component Space Pose Link Structure=@n1, Alpha:float=0, TranslationWeight:Vector="1,1,1", RotationWeight:float=1) -> (Pose:Component Space Pose Link Structure) {
    is=title AnimGraphNode_CopyBoneAdvanced
    bTranslationInCustomBoneSpace=false
    alphaBoolBlend {blendOption=Linear}
    sourceBone {boneName=hand_l}
    targetBone {boneName=ik_hand_l}
   }
   n3 = Copy Bone Advanced - Source Bone: hand_r - Target Bone: ik_hand_gun!(
    ComponentPose:Component Space Pose Link Structure=Local To Component(LocalPose=Use cached pose 'InPose'())
    Alpha:float=0
    RotationWeight:float=1
   ) -> (Pose:Component Space Pose Link Structure) {
    is=title AnimGraphNode_CopyBoneAdvanced
    bTranslationInCustomBoneSpace=false
    alphaBoolBlend {blendOption=Linear}
    sourceBone {boneName=hand_r}
    targetBone {boneName=ik_hand_gun}
    translationWeight=<0 0 0>
   }
   n4 = Hand IK Retargeting!(ComponentPose=@n3, Alpha:float=1, PerAxisAlpha="1,1,1", HandFKWeight=Get HandIKRetargetingWeight()) -> (Pose) {
    alphaBoolBlend {blendOption=Linear}
    iKBonesToMove [{boneName=ik_hand_gun}]
    leftHandFK {boneName=hand_l}
    leftHandIK {boneName=ik_hand_l}
    rightHandFK {boneName=hand_r}
    rightHandIK {boneName=ik_hand_r}
   }
   AnimGraphNode_Root_0 = Output Pose(Result=Blend Poses by bool(
     BlendPose_0=Component To Local(ComponentPose=@n2)
     BlendPose_1=Component To Local(ComponentPose=@n4)
     bActiveValue=Get bBothHandsIKInHandSpace() -> (bBothHandsIKInHandSpace:bool)
    ) {blendType=Linear}) {
    binding=&self:CopyBoneLayer.AnimGraphNode_Root_0.Binding
    layerGroup=None
   }
   AnimGraphNode_LinkedInputPose_0 = Input Pose() {
    name=InCopyBoneLayerPose
    binding=&self:CopyBoneLayer.AnimGraphNode_LinkedInputPose_0.Binding
   }
   AnimGraphNode_SaveCachedPose_0 = Save cached pose 'InPose'(Pose=@AnimGraphNode_LinkedInputPose_0) {cacheName=InPose}
  ]}
 CopyMotionLayer {nodes [
   AnimGraphNode_UseCachedPose_1 = Use cached pose 'CachedBasePose'()
   n1 = Copy Motion!(
    ComponentPose:Component Space Pose Link Structure=Local To Component(LocalPose=Use cached pose 'CachedLayeredLoco'())
    Alpha:float=1
    BasePose:Component Space Pose Link Structure=Local To Component(LocalPose=Use cached pose 'CachedBaseLoco'())
    BasePoseReference:Component Space Pose Link Structure=Local To Component(LocalPose=Use cached pose 'CachedBasePose'())
    Delay:float=0
    DynamicSourceBone:Name=None
    DynamicCopySpace:Name=None
    DynamicApplySpace:Name=None
    TranslationOffset:Rotator="0,0,0"
    RotationOffset:Rotator="0,0,0"
    RotationPivot:Vector="0,0,0"
    TargetCurveScale:float=1
    bUseLegacyBehavior:bool=False
    TranslationScale:Vector="0,0,0"
    TranslationSmoothingTime:float=0
    TranslationRemapCurve:Curve Vector Object Reference
    RotationScale:float=0
    RotationSmoothingTime:float=0
    RotationRemapCurve:Curve Float Object Reference
   ) -> (Pose:Component Space Pose Link Structure) {
    is=title AnimGraphNode_CopyMotion
    bUseBasePose=true
    bUseRootSpace=true
    poseHistoryTag=PoseHistory
    rotationPitch_CurveName=None_rotation_Y
    rotationRoll_CurveName=None_rotation_X
    rotationScale=0
    rotationYaw_CurveName=None_rotation_Z
    targetCurveName=RightElbowTwist
    translationX_CurveName=None_translation_X
    translationY_CurveName=None_translation_Y
    translationZ_CurveName=None_translation_Z
    alphaBoolBlend {blendOption=Linear}
    applySpace {boneName=pelvis}
    boneToModify {boneName=ik_hand_r}
    copySpace {boneName=spine_05}
    sourceBone {boneName=ik_hand_r}
    translationScale=<0 0 0>
   }
   n2 = Copy Motion(
    ComponentPose=@n1
    Alpha=1
    BasePose=Local To Component(LocalPose=Use cached pose 'CachedBaseLoco'())
    BasePoseReference=Local To Component(LocalPose=Use cached pose 'CachedBasePose'())
    Delay=0
    DynamicSourceBone=None
    DynamicCopySpace=None
    DynamicApplySpace=None
    TranslationOffset="0,0,0"
    RotationOffset="0,0,0"
    RotationPivot="0,0,0"
    TargetCurveScale=1
    bUseLegacyBehavior=False
    TranslationScale="0,0,0"
    TranslationSmoothingTime=0
    RotationScale=0
    RotationSmoothingTime=0
   ) {
    bUseBasePose=true
    bUseRootSpace=true
    poseHistoryTag=PoseHistory
    rotationPitch_CurveName=None_rotation_Y
    rotationRoll_CurveName=None_rotation_X
    rotationScale=0
    rotationYaw_CurveName=None_rotation_Z
    targetCurveName=LeftElbowTwist
    translationX_CurveName=None_translation_X
    translationY_CurveName=None_translation_Y
    translationZ_CurveName=None_translation_Z
    alphaBoolBlend {blendOption=Linear}
    applySpace {boneName=pelvis}
    boneToModify {boneName=ik_hand_l}
    copySpace {boneName=spine_05}
    sourceBone {boneName=VB ik_hand_root_ik_hand_l}
    translationScale=<0 0 0>
   }
   n3 = Component To Local(ComponentPose=@n2)
   n4 = Get bBothHandsIKInHandSpace()
   n5 = Copy Motion(
    ComponentPose=Local To Component(LocalPose=Use cached pose 'CachedLayeredLoco'())
    Alpha=1
    BasePose=Local To Component(LocalPose=Use cached pose 'CachedBaseLoco'())
    BasePoseReference=Local To Component(LocalPose=Use cached pose 'CachedBasePose'())
    Delay=0
    DynamicSourceBone=None
    DynamicCopySpace=None
    DynamicApplySpace=None
    TranslationOffset="0,0,0"
    RotationOffset="0,0,0"
    RotationPivot="0,0,0"
    TargetCurveScale=1
    bUseLegacyBehavior=False
    TranslationScale="0,0,0"
    TranslationSmoothingTime=0
    RotationScale=0
    RotationSmoothingTime=0
   ) {
    bUseBasePose=true
    bUseRootSpace=true
    poseHistoryTag=PoseHistory
    rotationPitch_CurveName=None_rotation_Y
    rotationRoll_CurveName=None_rotation_X
    rotationScale=0
    rotationYaw_CurveName=None_rotation_Z
    targetCurveName=RightElbowTwist
    translationX_CurveName=None_translation_X
    translationY_CurveName=None_translation_Y
    translationZ_CurveName=None_translation_Z
    alphaBoolBlend {blendOption=Linear}
    applySpace {boneName=pelvis}
    boneToModify {boneName=ik_hand_gun}
    copySpace {boneName=spine_05}
    sourceBone {boneName=ik_hand_gun}
    translationScale=<0 0 0>
   }
   n6 = Component To Local(ComponentPose=@n5)
   n7 = Copy Motion!(
    ComponentPose=Local To Component(LocalPose=Blend Poses by bool(BlendPose_0=@n3, BlendPose_1=@n6, bActiveValue=@n4) {blendType=Linear})
    Alpha=Get CopyMotionHipAlpha() -> (CopyMotionHipAlpha:float)
    BasePose=Local To Component(LocalPose=Use cached pose 'CachedBaseLoco'())
    BasePoseReference=Local To Component(LocalPose=@AnimGraphNode_UseCachedPose_1)
    TranslationRemapCurve=Get HipRemapCurve() -> (HipRemapCurve:Curve Vector Object Reference)
   ) -> (Pose) {
    bUseBasePose=true
    bUseRootSpace=true
    delay=0.1
    rotationPitch_CurveName=None_rotation_Y
    rotationRoll_CurveName=None_rotation_X
    rotationScale=0.3
    rotationYaw_CurveName=None_rotation_Z
    translationX_CurveName=None_translation_X
    translationY_CurveName=None_translation_Y
    translationZ_CurveName=None_translation_Z
    alphaBoolBlend {blendOption=Linear}
    boneToModify {boneName=pelvis}
    sourceBone {boneName=pelvis}
   }
   AnimGraphNode_Root_0 = Output Pose(Result=Blend Poses by bool(
     BlendPose_0=Component To Local(ComponentPose=@n7)
     BlendPose_1=Use cached pose 'CachedLayeredLoco'()
     BlendTime_1=0
     bActiveValue=Get bEnableCopyMotion() -> (bEnableCopyMotion:bool)
    ) {blendType=Linear}) {
    binding=&self:CopyMotionLayer.AnimGraphNode_Root_0.Binding
    layerGroup=None
   }
   AnimGraphNode_LinkedInputPose_0 = Input Pose() {
    name=CopyMotion_InBaseLocomotionPose
    binding=&self:CopyMotionLayer.AnimGraphNode_LinkedInputPose_0.Binding
   }
   AnimGraphNode_SaveCachedPose_0 = Save cached pose 'CachedBaseLoco'(Pose=@AnimGraphNode_LinkedInputPose_0) {cacheName=CachedBaseLoco}
   AnimGraphNode_LinkedInputPose_1 = Input Pose() {
    name=CopyMotion_InBasePose
    binding=&self:CopyMotionLayer.AnimGraphNode_LinkedInputPose_1.Binding
   }
   Save cached pose 'CachedBasePose'(Pose=@AnimGraphNode_LinkedInputPose_1) {cacheName=CachedBasePose}
   AnimGraphNode_LinkedInputPose_2 = Input Pose() {
    name=CopyMotion_InLayeredLocomotionPose
    binding=&self:CopyMotionLayer.AnimGraphNode_LinkedInputPose_2.Binding
   }
   Save cached pose 'CachedLayeredLoco'(Pose=Modify Curve(
     SourcePose=Modify Curve(SourcePose=@AnimGraphNode_LinkedInputPose_2, CurveValues_0=0, Alpha=1) {curveNames [RightElbowTwist], curveValues [0]}
     CurveValues_0=0
     Alpha=1
    ) {curveNames [LeftElbowTwist], curveValues [0]}) {cacheName=CachedLayeredLoco}
  ]}
 EventGraph {links [], nodes []}
 ProceduralAdditivePose {nodes [
   AnimGraphNode_UseCachedPose_1 = Use cached pose 'PostDeadBlending'()
   n1 = Layered blend per bone(BasePose=Use cached pose 'PostBlendStack'(), BlendPoses_0=@AnimGraphNode_UseCachedPose_1) {blendMode=BlendMask, blendMasks [&Fortnite_M_Avg_Player_Skeleton:LowerBody], layerSetup []}
   AnimGraphNode_Root_0 = Output Pose(Result=Blend Poses by bool(
     BlendPose_0=@n1
     BlendPose_1=Use cached pose 'PostDeadBlending'()
     BlendTime_0=0.05
     bActiveValue=Get bDisableDeadBlendOnUpperBody() -> (bDisableDeadBlendOnUpperBody:bool)
    ) {blendType=Linear}) {layerGroup=None}
   AnimGraphNode_LinkedInputPose_1 = Input Pose() {name=InMotionMatchingPose}
   AnimGraphNode_PoseSearchHistoryCollector_0 = Pose History!(Source:Pose Link Structure=@AnimGraphNode_LinkedInputPose_1, TransformTrajectory:Transform Trajectory Structure=(Samples=)) -> (Pose:Pose Link Structure) {
    is=title AnimGraphNode_PoseSearchHistoryCollector
    binding=&self:ProceduralAdditivePose.AnimGraphNode_PoseSearchHistoryCollector_0.Binding
    tag=PoseHistory
    poseCount=10
    collectedBones [{boneName=ik_hand_gun}, {boneName=spine_05}, {boneName=pelvis}, {boneName=VB ik_hand_root_ik_hand_l}, {boneName=ik_hand_r}]
   }
   AnimGraphNode_SaveCachedPose_0 = Save cached pose 'BaseLocomotionPose'(Pose=Remove Curve!(SourcePose:Pose Link Structure=@AnimGraphNode_PoseSearchHistoryCollector_0, bEnabled:bool=True) -> (Pose:Pose Link Structure) {is=title AnimGraphNode_RemoveCurve, removeMode=KeepOnlySpecifiedCurves}) {
    binding=&self:ProceduralAdditivePose.AnimGraphNode_SaveCachedPose_0.Binding
    cacheName=BaseLocomotionPose
   }
   AnimGraphNode_LinkedAnimLayer_0 = BasePose!() -> (Pose:Pose Link Structure) {
    is=title AnimGraphNode_LinkedAnimLayer
    binding=&self:ProceduralAdditivePose.AnimGraphNode_LinkedAnimLayer_0.Binding
    layer=BasePose
   }
   Save cached pose 'BasePose'(Pose=@AnimGraphNode_LinkedAnimLayer_0) {cacheName=BasePose}
   AnimGraphNode_LinkedInputPose_0 = Input Pose() {name=InSourceLocomotionPose}
   Save cached pose 'SourceLocoPose'(Pose=@AnimGraphNode_LinkedInputPose_0) {cacheName=SourceLocoPose}
   n2 = Use cached pose 'PostBlendStack'()
   Save cached pose 'PostDeadBlending'(Pose=Rotate Root Bone!(
     BasePose:Pose Link Structure=Dead Blending!(Source:Pose Link Structure=@n2) -> (Pose:Pose Link Structure) {is=title AnimGraphNode_DeadBlending, tag=BlendStackDeadBlend, defaultBlendMode=Linear}
     Pitch:float=0
     Yaw:float=Get InverseShuffleRootBoneCorrection() -> (InverseShuffleRootBoneCorrection:float)
     MeshToComponent:Rotator="0,0,0"
    ) -> (Pose:Pose Link Structure) {is=title AnimGraphNode_RotateRootBone}) {cacheName=PostDeadBlending}
   AnimGraphNode_BlendStack_0 = Blend Stack!(
    AnimationAsset:Animation Asset Object Reference=Get SourcePoseAnimSequence() -> (SourcePoseAnimSequence:Anim Sequence Object Reference)
    BlendTime:float=Get BlendStackBlendTime() -> (BlendStackBlendTime:float)
    BlendProfile:Blend Profile Object Reference
    BlendOption:EAlphaBlendOption Enum=HermiteCubic
   ) -> (Pose:Pose Link Structure) {
    is=title AnimGraphNode_BlendStack
    binding=&self:ProceduralAdditivePose.AnimGraphNode_BlendStack_0.Binding
    tag=BlendStack
    node {
     bUseInertialBlend=true
     blendOption=Linear
     inertialBlendNodeTag=BlendStackDeadBlend
     maxActiveBlends=0
     updateFunction {functionName=OnBlendStackUpdate}
    }
    updateFunction {bSelfContext=true, memberName=OnBlendStackUpdate}
   }
   Save cached pose 'PostBlendStack'(Pose=@AnimGraphNode_BlendStack_0) {cacheName=PostBlendStack}
   AnimGraphNode_LinkedInputPose_2 = Input Pose() {name=InSourcePose}
   Save cached pose 'SourcePose'(Pose=@AnimGraphNode_LinkedInputPose_2) {cacheName=SourcePose}
  ]}
 AnimationBlendStackGraph_0 {
  path=ProceduralAdditivePose.AnimGraphNode_BlendStack_0.AnimationBlendStackGraph_0
  nodes [
   n1 = Use cached pose 'SourcePose'()
   AnimGraphNode_UseCachedPose_1 = Use cached pose 'BaseLocomotionPose'() {binding=&self:ProceduralAdditivePose.AnimGraphNode_BlendStack_0.AnimationBlendStackGraph_0.AnimGraphNode_UseCachedPose_1.Binding}
   n2 = Use cached pose 'BasePose'()
   n3 = Use cached pose 'SourceLocoPose'()
   AnimGraphNode_BlendStackResult_0 = Output Pose!(Result:Pose Link Structure=Rotate Root Bone(
     BasePose=BlendStack!(
      SourcePose:Pose Link Structure=@n1
      BasePose:Pose Link Structure=@n2
      BaseLocomotionPose:Pose Link Structure=@AnimGraphNode_UseCachedPose_1
      InSourceLocoPose:Pose Link Structure=@n3
     ) -> (Pose:Pose Link Structure) {is=title AnimGraphNode_LinkedAnimLayer, layer=BlendStack}
     Pitch=0
     Yaw=Get ShuffleRootBoneCorrection() -> (ShuffleRootBoneCorrection:float)
     MeshToComponent="0,0,0"
    )) {
    is=title AnimGraphNode_BlendStackResult
    binding=&self:ProceduralAdditivePose.AnimGraphNode_BlendStack_0.AnimationBlendStackGraph_0.AnimGraphNode_BlendStackResult_0.Binding
    layerGroup=None
   }
  ]
 }
 ProceduralLayeredPose {nodes [
   AnimGraphNode_UseCachedPose_1 = Use cached pose 'InLocoPose'()
   n1 = Layered blend per bone(
    BasePose=Bone Mask(
     BasePose=@AnimGraphNode_UseCachedPose_1
     BlendPoses_0=Use cached pose 'LayerItemPose'()
     BlendWeights_0=1
     BoneMask=Get LayeringBoneMask() -> (LayeringBoneMask:Bone Mask Structure)
     BoneMaskDefinitionDataAsset=&Layering_BodyParts_DA
    ) {bProcessMeshSpaceAsRootSpace=true}
    BlendPoses_0=Use cached pose 'LayerItemPose'()
   ) {
    bMeshSpaceRotationBlend=true
    bRootSpaceRotationBlend=true
    curveBlendOption=UseBasePose
    layerSetup [{branchFilters [{boneName=ik_hand_root}, {boneName=VB pelvis_ik_hand_gun}, {boneName=VB pelvis_ik_hand_r}, {boneName=VB pelvis_ik_hand_l}]}]
   }
   n2 = Use cached pose 'LayerItemPose'()
   n3 = Get bUseFullLayeringPose() -> (bUseFullLayeringPose:bool)
   AnimGraphNode_Root_0 = Output Pose(Result=Blend Poses by bool(
     BlendPose_0=Blend Poses by bool(BlendPose_0=@n2, BlendPose_1=@n1, BlendTime_0=0.2, BlendTime_1=0.2, bActiveValue=@n3) {blendType=Linear}
     BlendPose_1=Use cached pose 'InLocoPose'()
     BlendTime_0=0.2
     BlendTime_1=0
     bActiveValue=Get bUseProceduralLayering()
    ) {blendType=Linear}) {is=title AnimGraphNode_Root, layerGroup=None}
   AnimGraphNode_LinkedInputPose_0 = Input Pose() {name=InLocomotionPose}
   AnimGraphNode_SaveCachedPose_0 = Save cached pose 'InLocoPose'(Pose=@AnimGraphNode_LinkedInputPose_0) {cacheName=InLocoPose}
   Save cached pose 'LayerItemPose'(Pose=Cached Sequence Evaluator() {cacheName=SourcePose}) {cacheName=LayerItemPose}
  ]}
 ProceduralLayering {nodes [
   AnimGraphNode_Root_0 = Output Pose(Result=Two Way Blend(
     A=Use cached pose 'CachedPoseLayeredLocomotion'()
     B=Use cached pose 'PostProceduralLayering'()
     Alpha=Get ProceduralLayeringAlpha() -> (ProceduralLayeringAlpha:float)
    ) {blendNode {alphaBoolBlend {blendOption=Linear}}}) {
    binding=&self:ProceduralLayering.AnimGraphNode_Root_0.Binding
    layerGroup=None
   }
   AnimGraphNode_LinkedInputPose_1 = Input Pose() {
    name=ProcLayer_InLayeredLocomotionPose
    binding=&self:ProceduralLayering.AnimGraphNode_LinkedInputPose_1.Binding
   }
   AnimGraphNode_SaveCachedPose_0 = Save cached pose 'CachedPoseLayeredLocomotion'(Pose=@AnimGraphNode_LinkedInputPose_1) {cacheName=CachedPoseLayeredLocomotion}
   AnimGraphNode_LinkedAnimLayer_0 = CopyBoneLayer!(InCopyBoneLayerPose:Pose Link Structure=Use cached pose 'CachedPoseLayeredLocomotion'()) -> (Pose:Pose Link Structure) {is=title AnimGraphNode_LinkedAnimLayer, layer=CopyBoneLayer}
   AnimGraphNode_LinkedInputPose_0 = Input Pose() {
    name=ProcLayer_InBaseLocomotionPose
    binding=&self:ProceduralLayering.AnimGraphNode_LinkedInputPose_0.Binding
   }
   AnimGraphNode_LinkedInputPose_2 = Input Pose() {name=ProcLayer_InBasePose}
   n1 = "Transform (Modify) Bone - Bone: ik_hand_gun"!(
    ComponentPose=Local To Component(LocalPose=CopyMotionLayer!(
      CopyMotion_InBaseLocomotionPose:Pose Link Structure=@AnimGraphNode_LinkedInputPose_0
      CopyMotion_InBasePose:Pose Link Structure=@AnimGraphNode_LinkedInputPose_2
      CopyMotion_InLayeredLocomotionPose:Pose Link Structure=@AnimGraphNode_LinkedAnimLayer_0
     ) -> (Pose:Pose Link Structure) {is=title AnimGraphNode_LinkedAnimLayer, layer=CopyMotionLayer})
    bAlphaBoolEnabled:bool=Get bApplyIKHandGunOffset() -> (bApplyIKHandGunOffset:bool)
    Translation="0,0,0"
    Rotation="0,0,0"
   ) -> (Pose) {
    alphaInputType=Bool
    rotationMode=BMM_Additive
    translationMode=BMM_Additive
    alphaBoolBlend {blendInTime=0.2, blendOutTime=0.2}
    boneToModify {boneName=ik_hand_gun}
   }
   n2 = Get bApplyIKHandGunOffsetInHandSpace() -> (bApplyIKHandGunOffsetInHandSpace:bool)
   n3 = "Transform (Modify) Bone - Bone: ik_hand_r"!(ComponentPose=@n1, bAlphaBoolEnabled:bool=@n2, Translation="0,0,0", Rotation="0,0,0") -> (Pose) {
    alphaInputType=Bool
    rotationMode=BMM_Additive
    translationMode=BMM_Additive
    alphaBoolBlend {blendInTime=0.2, blendOutTime=0.2}
    boneToModify {boneName=ik_hand_r}
   }
   n4 = "Transform (Modify) Bone - Bone: ik_hand_l"!(ComponentPose=@n3, bAlphaBoolEnabled:bool=@n2, Translation="0,0,0", Rotation="0,0,0") -> (Pose) {
    alphaInputType=Bool
    rotationMode=BMM_Additive
    translationMode=BMM_Additive
    alphaBoolBlend {blendInTime=0.2, blendOutTime=0.2}
    boneToModify {boneName=ik_hand_l}
   }
   Save cached pose 'PostProceduralLayering'(Pose=Component To Local(ComponentPose=@n4)) {cacheName=PostProceduralLayering}
  ]}
 UpdateNoWeaponBasePose_PreNativize {nodes [
   n1 = UpdateNoWeaponBasePose_PreNativize!() -> (then, Context:"Anim Update Context Structure (by ref)", Node:"Anim Node Reference Structure (by ref)") {is=title K2Node_FunctionEntry}
   > n2 = Evaluate Proxy: NoWepPoseBase_PA!(
    execute:Exec
    AnimInstance:Anim Instance Object Reference
    ProxyTable:Proxy Table Object Reference=Get ProxyTableNoWeaponBasePose() -> (ProxyTableNoWeaponBasePose:Proxy Table Object Reference)
   ) -> (execute:Exec, Result:Anim Sequence Object Reference) {is=title K2Node_EvaluateProxy2, proxy=/Game/Characters/Player/ProxyAssets/Pose/NoWepPoseBase_PA}
   > n3 = ConvertToSequenceEvaluator!(execute, Node:"Anim Node Reference Structure (by ref)"=@n1.Node) -> (Succeeded:Exec, Failed:Exec, ReturnValue:Sequence Evaluator Reference Structure)
   > SetSequence!(execute, SequenceEvaluator:"Sequence Evaluator Reference Structure (by ref)"=@n3, Sequence:Anim Sequence Base Object Reference=@n2) -> (then, ReturnValue:Sequence Evaluator Reference Structure)
   > Return Node()
  ]}
]
variables [HipRemapCurve {type=CurveVector, default=null}]
