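# Player linked anim graph that animates skydiving and gliding, including launchpad takeoff.
# An outer state machine uses inertialized transitions into nested machines switching between freefall dive or glide and paragliding.
# Freefall blendspaces from the current skydive asset set follow aim, with turn and lean additives via cached poses or curve-driven blends.
# Paragliding selects a sub-machine by glider type, then layers curve-weighted lean and pose additives.
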AnimBlueprint /Game/Characters/Player/AnimationBlueprints/LinkedAnimGraphs/Player_SkyDiveAndGlide_AnimBP
target_skeleton=&Fortnite_M_Avg_Player_Skeleton
graphs [
 AnimGraph {nodes [
   AnimGraphNode_StateMachine_0 = SkyDiveState() {is=title AnimGraphNode_StateMachine, bAllowConduitEntryStates=true, maxTransitionsPerFrame=1}
   Output Pose(Result=@AnimGraphNode_StateMachine_0) {layerGroup=None}
   AnimGraphNode_StateMachine_1 = FreeFall_TurnAndLean_Glide() {is=title AnimGraphNode_StateMachine}
   Save cached pose 'FreeFall_TurnAndLean_Glide_Pose'(Pose=@AnimGraphNode_StateMachine_1) {cacheName=FreeFall_TurnAndLean_Glide_Pose}
   AnimGraphNode_StateMachine_8 = FreeFall_TurnAndLean_Dive() {is=title AnimGraphNode_StateMachine}
   Save cached pose 'FreeFall_TurnAndLean_Dive_Pose'(Pose=@AnimGraphNode_StateMachine_8) {cacheName=FreeFall_TurnAndLean_Dive_Pose}
  ]}
 SkyDiveState {
  entry=Conduit
  path=AnimGraph.AnimGraphNode_StateMachine_0.SkyDiveState
  ids=id
  nodes [
   AnimStateConduitNode_0 = Conduit()
   AnimStateNode_1 = DiveOrGlide() {is=title AnimStateNode}
   AnimStateNode_2 = Launchpad_TakeOff() {is=title AnimStateNode}
   AnimStateNode_3 = LunchPad_Transition() {is=title AnimStateNode}
   AnimStateTransitionNode_1 = AnimStateTransitionNode() {logicType=TLT_Inertialization}
   AnimStateTransitionNode_14 = AnimStateTransitionNode() {logicType=TLT_Inertialization, priorityOrder=100}
   AnimStateTransitionNode_5 = AnimStateTransitionNode() {bAutomaticRuleBasedOnSequencePlayerInState=true, logicType=TLT_Inertialization, priorityOrder=100}
   AnimStateTransitionNode_6 = AnimStateTransitionNode() {bAutomaticRuleBasedOnSequencePlayerInState=true, logicType=TLT_Inertialization, priorityOrder=100}
   AnimStateTransitionNode_7 = AnimStateTransitionNode() {logicType=TLT_Inertialization}
   AnimStateTransitionNode_8 = AnimStateTransitionNode() {logicType=TLT_Inertialization}
  ]
  links [
   {
    dst=DiveOrGlide
    rule=AnimGraph.AnimGraphNode_StateMachine_0.SkyDiveState.AnimStateTransitionNode_14.Transition
    src=Conduit
    via=AnimStateTransitionNode_14
   }
   {
    dst=Launchpad_TakeOff
    rule=AnimGraph.AnimGraphNode_StateMachine_0.SkyDiveState.AnimStateTransitionNode_1.Transition
    src=Conduit
    via=AnimStateTransitionNode_1
   }
   {
    dst=DiveOrGlide
    rule=AnimGraph.AnimGraphNode_StateMachine_0.SkyDiveState.AnimStateTransitionNode_7.Transition
    src=Launchpad_TakeOff
    via=AnimStateTransitionNode_7
   }
   {
    dst=LunchPad_Transition
    rule=AnimGraph.AnimGraphNode_StateMachine_0.SkyDiveState.AnimStateTransitionNode_6.Transition
    src=Launchpad_TakeOff
    via=AnimStateTransitionNode_6
   }
   {
    dst=DiveOrGlide
    rule=AnimGraph.AnimGraphNode_StateMachine_0.SkyDiveState.AnimStateTransitionNode_5.Transition
    src=LunchPad_Transition
    via=AnimStateTransitionNode_5
   }
   {
    dst=DiveOrGlide
    rule=AnimGraph.AnimGraphNode_StateMachine_0.SkyDiveState.AnimStateTransitionNode_8.Transition
    src=LunchPad_Transition
    via=AnimStateTransitionNode_8
   }
  ]
 }
 Conduit {path=AnimGraph.AnimGraphNode_StateMachine_0.SkyDiveState.AnimStateConduitNode_0.Conduit, nodes [Result(bCanEnterTransition=true)]}
 DiveOrGlide {
  path=AnimGraph.AnimGraphNode_StateMachine_0.SkyDiveState.AnimStateNode_1.DiveOrGlide
  nodes [
   AnimGraphNode_StateMachine_212 = SkydiveOrGlide() {is=title AnimGraphNode_StateMachine}
   Output Animation Pose(Result=@AnimGraphNode_StateMachine_212)
  ]
 }
 SkydiveOrGlide {
  entry=SkyDive
  path=AnimGraph.AnimGraphNode_StateMachine_0.SkyDiveState.AnimStateNode_1.DiveOrGlide.AnimGraphNode_StateMachine_212.SkydiveOrGlide
  ids=id
  nodes [
   AnimStateNode_2 = SkyDive() {is=title AnimStateNode}
   AnimStateNode_3 = ParaGlide() {is=title AnimStateNode}
   AnimStateNode_4 = ParaGlide_Out() {is=title AnimStateNode}
   AnimStateTransitionNode_0 = AnimStateTransitionNode() {
    crossfadeDuration=0.1
    blendProfileWrapper {blendProfile=&Fortnite_M_Avg_Player_Skeleton:WeaponR_Snap}
   }
   AnimStateTransitionNode_1 = AnimStateTransitionNode() {bAutomaticRuleBasedOnSequencePlayerInState=true, crossfadeDuration=0.1}
   AnimStateTransitionNode_2 = AnimStateTransitionNode() {crossfadeDuration=0.1, priorityOrder=2}
   AnimStateTransitionNode_3 = AnimStateTransitionNode() {crossfadeDuration=0.1}
  ]
  links [
   {
    dst=ParaGlide_Out
    rule=AnimGraph.AnimGraphNode_StateMachine_0.SkyDiveState.AnimStateNode_1.DiveOrGlide.AnimGraphNode_StateMachine_212.SkydiveOrGlide.AnimStateTransitionNode_2.AnimationTransitionGraph_1
    src=ParaGlide
    via=AnimStateTransitionNode_2
   }
   {
    dst=ParaGlide
    rule=AnimGraph.AnimGraphNode_StateMachine_0.SkyDiveState.AnimStateNode_1.DiveOrGlide.AnimGraphNode_StateMachine_212.SkydiveOrGlide.AnimStateTransitionNode_3.AnimationTransitionGraph_1
    src=ParaGlide_Out
    via=AnimStateTransitionNode_3
   }
   {
    dst=SkyDive
    rule=AnimGraph.AnimGraphNode_StateMachine_0.SkyDiveState.AnimStateNode_1.DiveOrGlide.AnimGraphNode_StateMachine_212.SkydiveOrGlide.AnimStateTransitionNode_1.AnimationTransitionGraph_1
    src=ParaGlide_Out
    via=AnimStateTransitionNode_1
   }
   {
    dst=ParaGlide
    rule=AnimGraph.AnimGraphNode_StateMachine_0.SkyDiveState.AnimStateNode_1.DiveOrGlide.AnimGraphNode_StateMachine_212.SkydiveOrGlide.AnimStateTransitionNode_0.AnimationTransitionGraph_1
    src=SkyDive
    via=AnimStateTransitionNode_0
   }
  ]
 }
 SkyDive {
  path=AnimGraph.AnimGraphNode_StateMachine_0.SkyDiveState.AnimStateNode_1.DiveOrGlide.AnimGraphNode_StateMachine_212.SkydiveOrGlide.AnimStateNode_2.SkyDive
  nodes [
   AnimGraphNode_StateMachine_0 = Skydiving() {is=title AnimGraphNode_StateMachine}
   Output Animation Pose(Result=@AnimGraphNode_StateMachine_0)
  ]
 }
 Skydiving {
  entry=Passthrough
  path=AnimGraph.AnimGraphNode_StateMachine_0.SkyDiveState.AnimStateNode_1.DiveOrGlide.AnimGraphNode_StateMachine_212.SkydiveOrGlide.AnimStateNode_2.SkyDive.AnimGraphNode_StateMachine_0.Skydiving
  ids=id
  nodes [
   AnimStateNode_1 = Dive() {is=title AnimStateNode}
   AnimStateNode_2 = to_Dive() {is=title AnimStateNode}
   AnimStateNode_3 = to_Glide() {is=title AnimStateNode}
   AnimStateNode_4 = Passthrough() {is=title AnimStateNode}
   AnimStateNode_6 = Glide() {is=title AnimStateNode}
   AnimStateTransitionNode_0 = AnimStateTransitionNode() {crossfadeDuration=0.15}
   AnimStateTransitionNode_1 = AnimStateTransitionNode() {bAutomaticRuleBasedOnSequencePlayerInState=true, blendMode=Linear}
   AnimStateTransitionNode_2 = AnimStateTransitionNode() {crossfadeDuration=0.15}
   AnimStateTransitionNode_3 = AnimStateTransitionNode() {bAutomaticRuleBasedOnSequencePlayerInState=true}
   AnimStateTransitionNode_4 = AnimStateTransitionNode()
   AnimStateTransitionNode_5 = AnimStateTransitionNode()
   AnimStateTransitionNode_7 = AnimStateTransitionNode() {crossfadeDuration=0}
   AnimStateTransitionNode_8 = AnimStateTransitionNode() {crossfadeDuration=0}
  ]
  links [
   {
    dst=to_Glide
    rule=AnimGraph.AnimGraphNode_StateMachine_0.SkyDiveState.AnimStateNode_1.DiveOrGlide.AnimGraphNode_StateMachine_212.SkydiveOrGlide.AnimStateNode_2.SkyDive.AnimGraphNode_StateMachine_0.Skydiving.AnimStateTransitionNode_2.Transition
    src=Dive
    via=AnimStateTransitionNode_2
   }
   {
    dst=to_Dive
    rule=AnimGraph.AnimGraphNode_StateMachine_0.SkyDiveState.AnimStateNode_1.DiveOrGlide.AnimGraphNode_StateMachine_212.SkydiveOrGlide.AnimStateNode_2.SkyDive.AnimGraphNode_StateMachine_0.Skydiving.AnimStateTransitionNode_0.Transition
    src=Glide
    via=AnimStateTransitionNode_0
   }
   {
    dst=Dive
    rule=AnimGraph.AnimGraphNode_StateMachine_0.SkyDiveState.AnimStateNode_1.DiveOrGlide.AnimGraphNode_StateMachine_212.SkydiveOrGlide.AnimStateNode_2.SkyDive.AnimGraphNode_StateMachine_0.Skydiving.AnimStateTransitionNode_8.Transition
    src=Passthrough
    via=AnimStateTransitionNode_8
   }
   {
    dst=Glide
    rule=AnimGraph.AnimGraphNode_StateMachine_0.SkyDiveState.AnimStateNode_1.DiveOrGlide.AnimGraphNode_StateMachine_212.SkydiveOrGlide.AnimStateNode_2.SkyDive.AnimGraphNode_StateMachine_0.Skydiving.AnimStateTransitionNode_7.Transition
    src=Passthrough
    via=AnimStateTransitionNode_7
   }
   {
    dst=Dive
    rule=AnimGraph.AnimGraphNode_StateMachine_0.SkyDiveState.AnimStateNode_1.DiveOrGlide.AnimGraphNode_StateMachine_212.SkydiveOrGlide.AnimStateNode_2.SkyDive.AnimGraphNode_StateMachine_0.Skydiving.AnimStateTransitionNode_1.Transition
    src=to_Dive
    via=AnimStateTransitionNode_1
   }
   {
    dst=to_Glide
    rule=AnimGraph.AnimGraphNode_StateMachine_0.SkyDiveState.AnimStateNode_1.DiveOrGlide.AnimGraphNode_StateMachine_212.SkydiveOrGlide.AnimStateNode_2.SkyDive.AnimGraphNode_StateMachine_0.Skydiving.AnimStateTransitionNode_4.Transition
    src=to_Dive
    via=AnimStateTransitionNode_4
   }
   {
    dst=Glide
    rule=AnimGraph.AnimGraphNode_StateMachine_0.SkyDiveState.AnimStateNode_1.DiveOrGlide.AnimGraphNode_StateMachine_212.SkydiveOrGlide.AnimStateNode_2.SkyDive.AnimGraphNode_StateMachine_0.Skydiving.AnimStateTransitionNode_3.Transition
    src=to_Glide
    via=AnimStateTransitionNode_3
   }
   {
    dst=to_Dive
    rule=AnimGraph.AnimGraphNode_StateMachine_0.SkyDiveState.AnimStateNode_1.DiveOrGlide.AnimGraphNode_StateMachine_212.SkydiveOrGlide.AnimStateNode_2.SkyDive.AnimGraphNode_StateMachine_0.Skydiving.AnimStateTransitionNode_5.Transition
    src=to_Glide
    via=AnimStateTransitionNode_5
   }
  ]
 }
 Dive {
  path=AnimGraph.AnimGraphNode_StateMachine_0.SkyDiveState.AnimStateNode_1.DiveOrGlide.AnimGraphNode_StateMachine_212.SkydiveOrGlide.AnimStateNode_2.SkyDive.AnimGraphNode_StateMachine_0.Skydiving.AnimStateNode_1.Dive
  nodes [
   n1 = Break Fort Anim Input Skydiving!(FortAnimInput_Skydiving:"Fort Anim Input Skydiving Structure (by ref)"=Get Skydiving() -> (Skydiving:Fort Anim Input Skydiving Structure)) -> (
    ParaGilderRootOffset:Vector=~
    ParaGliderRootRotationOverride:Rotator=~
    bUseParaGlideRootModifier:bool=~
    bIsSkydivingFromLaunchPad:bool=~
    bIsSkydivingFromBus:bool=~
    bIsInVortex:bool=~
    bIsDiving:bool=~
    bIsParachuteOpen:bool=~
    bIsSkydiving:bool=~
    bIsParachuteLeaning:bool=~
    bIsSkydiveLeaning:bool=~
    bIsLeaning:bool=~
    bIsSkydiveDiveMode:bool=~
    bPlaySkydiveDrift:bool=~
    bSkydiveDriftDelayActive:bool=~
    bSkydiveDriftAnimAllowed:bool=~
    bIsGliderRight:bool=~
    bIsGliderCenter:bool=~
    bIsGliderLeft:bool=~
    bIsGliderForward:bool=~
    bIsGliderBack:bool=~
    bHasDeployedAndIsStandingUpright:bool=~
    LocalAccelForward:float=~
    LocalAccelRight:float=~
    LocalVelocityRight:float=~
    SkydiveAimPitch:float=~
    SkydiveAimYaw:float=~
    DeployChuteAnimRate:float=~
    SkydiveDriftAnimRate:float=~
    SkydiveDriftAnimRateCurrent:float=~
    SkydiveFidgetAnimRate:float=~
    SkydiveFidgetAnimRateCurrent:float=~
    SkydiveAdditiveAlpha:float=~
    SkydiveDriftDelay:float=~
    SkydiveDriftAnim:int
    SkydiveDriftAnimMax:int
    LaunchpadAnim:int
    LocalAccelDir:ESkydivingDirection Enum
    DirectionLast:ESkydivingDirection Enum
    PoseAdditive:Anim Sequence Object Reference
   ) {is=title K2Node_BreakStruct}
   n2 = Break Fort Anim Input Player Skydive Anim Asset!(FortAnimInput_PlayerSkydiveAnimAsset:"Fort Anim Input Player Skydive Anim Asset Structure (by ref)"=Get CurrentSkydiveAnimSet() -> (CurrentSkydiveAnimSet:Fort Anim Input Player Skydive Anim Asset Structure)) -> (
    Dive_BS_Female:Blend Space Object Reference
    Dive_BS_Male:Blend Space Object Reference
    Dive_ToGlide_BS_Female:Blend Space Object Reference
    Dive_ToGlide_BS_Male:Blend Space Object Reference
    Dive_Lean_Additive_BS_Female:Blend Space Object Reference
    Dive_Lean_Additive_BS_Male:Blend Space Object Reference
    Dive_Lean_Center_Pose:Anim Sequence Object Reference
    Dive_Turn_Additive_BS_Female:Blend Space Object Reference
    Dive_Turn_Additive_BS_Male:Blend Space Object Reference
    Dive_DriftA_Additive_BS:Blend Space Object Reference
    Dive_DriftB_Additive_BS:Blend Space Object Reference
    Dive_DriftC_Additive_BS:Blend Space Object Reference
    Dive_FidgetA_Additive_BS:Blend Space Object Reference
    Dive_FidgetB_Additive_BS:Blend Space Object Reference
    Dive_FidgetC_Additive_BS:Blend Space Object Reference
    Glide_BS_Female:Blend Space Object Reference
    Glide_BS_Male:Blend Space Object Reference
    Glide_ToDive_BS_Female:Blend Space Object Reference
    Glide_ToDive_BS_Male:Blend Space Object Reference
    Glide_Lean_Additive_BS_Female:Blend Space Object Reference
    Glide_Lean_Additive_BS_Male:Blend Space Object Reference
    Glide_Turn_Additive_BS_Female:Blend Space Object Reference
    Glide_Turn_Additive_BS_Male:Blend Space Object Reference
    Glide_DriftA_Additive:Anim Sequence Object Reference
   ) {is=title K2Node_BreakStruct}
   n3 = Get FreeFall_AimPitch() -> (FreeFall_AimPitch:float)
   n4 = "Blend Poses (EFortCustomGender)"!(
    BlendPose_0:Pose Link Structure=Blendspace Player 'BS_Dive_2D'!(X=@n1.SkydiveAimYaw, Y=@n3, BlendSpace:Blend Space Object Reference=@n2.Dive_BS_Male) -> (Pose)
    BlendPose_1:Pose Link Structure="Sequence Player (additive) 'Dive_Lean_Center_Pose'"!(Sequence:Anim Sequence Base Object Reference=@n2.Dive_Lean_Center_Pose) -> (Pose)
    BlendTime_0:float=0
    BlendTime_1:float=0
    ActiveEnumValue:Byte=Get Gender() -> (Gender:EFortCustomGender Enum=Invalid)
   ) -> (Pose:Pose Link Structure) {
    is=title AnimGraphNode_BlendListByEnum
    blendType=Linear
    blendPose [{linkId=-1, sourceLinkId=-1}, {linkId=-1, sourceLinkId=-1}]
    blendTime [0.1,0.1]
   }
   n5 = Blendspace Player 'BS_Dive_2D'(X=@n1.SkydiveAimYaw, Y=@n3, BlendSpace=@n2)
   Output Animation Pose(Result=Apply Additive(
     Base=Apply Additive(Base=@n5, Additive=@n4) {alphaBoolBlend {blendOption=Linear}}
     Additive=Use cached pose 'FreeFall_TurnAndLean_Dive_Pose'()
    ) {alphaBoolBlend {blendOption=Linear}})
  ]
 }
 to_Dive {
  path=AnimGraph.AnimGraphNode_StateMachine_0.SkyDiveState.AnimStateNode_1.DiveOrGlide.AnimGraphNode_StateMachine_212.SkydiveOrGlide.AnimStateNode_2.SkyDive.AnimGraphNode_StateMachine_0.Skydiving.AnimStateNode_2.to_Dive
  nodes [
   n1 = Break Fort Anim Input Player Skydive Anim Asset(FortAnimInput_PlayerSkydiveAnimAsset=Get CurrentSkydiveAnimSet())
   n2 = Break Fort Anim Input Skydiving(FortAnimInput_Skydiving=Get Skydiving()) -> (
    ParaGilderRootOffset=~
    ParaGliderRootRotationOverride=~
    bUseParaGlideRootModifier=~
    bIsSkydivingFromLaunchPad=~
    bIsSkydivingFromBus=~
    bIsInVortex=~
    bIsDiving=~
    bIsParachuteOpen=~
    bIsSkydiving=~
    bIsParachuteLeaning=~
    bIsSkydiveLeaning=~
    bIsLeaning=~
    bIsSkydiveDiveMode=~
    bPlaySkydiveDrift=~
    bSkydiveDriftDelayActive=~
    bSkydiveDriftAnimAllowed=~
    bIsGliderRight=~
    bIsGliderCenter=~
    bIsGliderLeft=~
    bIsGliderForward=~
    bIsGliderBack=~
    bHasDeployedAndIsStandingUpright=~
    LocalAccelForward=~
    LocalAccelRight=~
    LocalVelocityRight=~
    SkydiveAimPitch=~
    SkydiveAimYaw=~
    DeployChuteAnimRate=~
    SkydiveDriftAnimRate=~
    SkydiveDriftAnimRateCurrent=~
    SkydiveFidgetAnimRate=~
    SkydiveFidgetAnimRateCurrent=~
    SkydiveAdditiveAlpha=~
    SkydiveDriftDelay=~
   )
   n3 = Get FreeFall_AimPitch()
   n4 = "Blend Poses (EFortCustomGender)"(
    BlendPose_0=Blendspace Player 'BS_Glide_to_Dive'!(X=@n2.SkydiveAimYaw, Y=@n3, BlendSpace:Blend Space Object Reference=@n1.Glide_ToDive_BS_Male) -> (Pose) {bLoop=false, bResetPlayTimeWhenBlendSpaceChanges=false}
    BlendPose_1=Blendspace Player 'BS_Glide_to_Dive'(X=@n2.SkydiveAimYaw, Y=@n3, BlendSpace=@n1.Glide_ToDive_BS_Female) {bLoop=false, bResetPlayTimeWhenBlendSpaceChanges=false, groupName=to_Dive, method=SyncGroup}
    BlendTime_0=0
    BlendTime_1=0
    ActiveEnumValue=Get Gender() -> (Gender=Invalid)
   ) {blendType=Linear, blendPose [{linkId=-1, sourceLinkId=-1}, {linkId=-1, sourceLinkId=-1}], blendTime [0.1,0.1]}
   n5 = Get FreeFall_AimPitch()
   n6 = Two Way Blend!(
    A=Blendspace Player 'BS_Glide_2D_Turn_Additive'!(X=@n2.SkydiveAimYaw, Y=@n5, BlendSpace:Blend Space Object Reference=@n1.Glide_Turn_Additive_BS_Female) -> (Pose) {bIgnoreForRelevancyTest=true}
    B=Blendspace Player 'BS_Dive_2D_Turn_Additive'!(X=@n2.SkydiveAimYaw, Y=@n5, BlendSpace:Blend Space Object Reference=@n1.Dive_Turn_Additive_BS_Female) -> (Pose) {bIgnoreForRelevancyTest=true}
    AlphaCurveName:Name=BlendIn
   ) -> (Pose) {blendNode {alphaInputType=Curve, alphaBoolBlend {blendOption=Linear}}}
   n7 = Get FreeFall_AimPitch()
   n8 = "Blend Poses (EFortCustomGender)"(
    BlendPose_0=Blendspace Player 'BS_Dive_Lean_Additive'!(X=@n2.LocalVelocityRight, Y=@n7, BlendSpace:Blend Space Object Reference=@n1.Dive_Lean_Additive_BS_Male) -> (Pose) {bIgnoreForRelevancyTest=true}
    BlendPose_1=Blendspace Player 'BS_Dive_Lean_Additive'(X=@n2.LocalVelocityRight, Y=@n7, BlendSpace=@n1.Dive_Lean_Additive_BS_Female) {bIgnoreForRelevancyTest=true}
    BlendTime_0=0
    BlendTime_1=0
    ActiveEnumValue=Get Gender() -> (Gender=Invalid)
   ) {blendType=Linear, blendPose [{linkId=-1, sourceLinkId=-1}, {linkId=-1, sourceLinkId=-1}], blendTime [0.1,0.1]}
   n9 = Blendspace Player 'BS_Glide_Lean_Additive'!(X=@n2.LocalVelocityRight, Y=Get FreeFall_AimPitch(), BlendSpace:Blend Space Object Reference=@n1.Glide_Lean_Additive_BS_Female) -> (Pose) {bIgnoreForRelevancyTest=true}
   Output Animation Pose(Result=Apply Additive!(
     Base=Apply Additive!(Base=@n4, Additive=@n6, Alpha:float=0.8) -> (Pose) {alphaBoolBlend {blendOption=Linear}}
     Additive=Two Way Blend(A=@n9, B=@n8, AlphaCurveName=BlendIn) {blendNode {alphaInputType=Curve, alphaBoolBlend {blendOption=Linear}}}
     bAlphaBoolEnabled:bool=@n2.bIsSkydiveLeaning
    ) -> (Pose) {alphaInputType=Bool, alphaBoolBlend {blendInTime=0.5, blendOption=Linear, blendOutTime=0.5}})
  ]
 }
 to_Glide {
  path=AnimGraph.AnimGraphNode_StateMachine_0.SkyDiveState.AnimStateNode_1.DiveOrGlide.AnimGraphNode_StateMachine_212.SkydiveOrGlide.AnimStateNode_2.SkyDive.AnimGraphNode_StateMachine_0.Skydiving.AnimStateNode_3.to_Glide
  nodes [
   n1 = Break Fort Anim Input Player Skydive Anim Asset(FortAnimInput_PlayerSkydiveAnimAsset=Get CurrentSkydiveAnimSet())
   n2 = Break Fort Anim Input Skydiving(FortAnimInput_Skydiving=Get Skydiving()) -> (
    ParaGilderRootOffset=~
    ParaGliderRootRotationOverride=~
    bUseParaGlideRootModifier=~
    bIsSkydivingFromLaunchPad=~
    bIsSkydivingFromBus=~
    bIsInVortex=~
    bIsDiving=~
    bIsParachuteOpen=~
    bIsSkydiving=~
    bIsParachuteLeaning=~
    bIsSkydiveLeaning=~
    bIsLeaning=~
    bIsSkydiveDiveMode=~
    bPlaySkydiveDrift=~
    bSkydiveDriftDelayActive=~
    bSkydiveDriftAnimAllowed=~
    bIsGliderRight=~
    bIsGliderCenter=~
    bIsGliderLeft=~
    bIsGliderForward=~
    bIsGliderBack=~
    bHasDeployedAndIsStandingUpright=~
    LocalAccelForward=~
    LocalAccelRight=~
    LocalVelocityRight=~
    SkydiveAimPitch=~
    SkydiveAimYaw=~
    DeployChuteAnimRate=~
    SkydiveDriftAnimRate=~
    SkydiveDriftAnimRateCurrent=~
    SkydiveFidgetAnimRate=~
    SkydiveFidgetAnimRateCurrent=~
    SkydiveAdditiveAlpha=~
    SkydiveDriftDelay=~
   )
   n3 = Blendspace Player 'BS_Dive_to_Glide'!(X=@n2.SkydiveAimYaw, Y=Get FreeFall_AimPitch(), BlendSpace:Blend Space Object Reference=@n1.Dive_ToGlide_BS_Female) -> (Pose) {bLoop=false, bResetPlayTimeWhenBlendSpaceChanges=false, groupName=to_Gliding, method=SyncGroup}
   n4 = Get FreeFall_AimPitch()
   n5 = Two Way Blend(
    A=Blendspace Player 'BS_Dive_2D_Turn_Additive'(X=@n2.SkydiveAimYaw, Y=@n4, BlendSpace=@n1.Dive_Turn_Additive_BS_Female)
    B=Blendspace Player 'BS_Glide_2D_Turn_Additive'(X=@n2.SkydiveAimYaw, Y=@n4, BlendSpace=@n1.Glide_Turn_Additive_BS_Female)
    AlphaCurveName=BlendIn
   ) {blendNode {alphaInputType=Curve, alphaBoolBlend {blendOption=Linear}}}
   n6 = Get FreeFall_AimPitch()
   n7 = Blendspace Player 'BS_Glide_Lean_Additive'(X=@n2.LocalVelocityRight, Y=@n6, BlendSpace=@n1.Glide_Lean_Additive_BS_Female)
   n8 = Blendspace Player 'BS_Dive_Lean_Additive'(X=@n2.LocalVelocityRight, Y=@n6, BlendSpace=@n1.Dive_Lean_Additive_BS_Female)
   Output Animation Pose(Result=Apply Additive!(
     Base=Apply Additive!(Base=@n3, Additive=@n5, Alpha:float=0.8) -> (Pose) {alphaBoolBlend {blendOption=Linear}}
     Additive=Two Way Blend(A=@n8, B=@n7, AlphaCurveName=BlendIn) {blendNode {alphaInputType=Curve, alphaBoolBlend {blendOption=Linear}}}
     bAlphaBoolEnabled:bool=@n2.bIsSkydiveLeaning
    ) -> (Pose) {alphaInputType=Bool, alphaBoolBlend {blendInTime=0.5, blendOption=Linear, blendOutTime=0.5}})
  ]
 }
 Passthrough {
  path=AnimGraph.AnimGraphNode_StateMachine_0.SkyDiveState.AnimStateNode_1.DiveOrGlide.AnimGraphNode_StateMachine_212.SkydiveOrGlide.AnimStateNode_2.SkyDive.AnimGraphNode_StateMachine_0.Skydiving.AnimStateNode_4.Passthrough
  nodes [Output Animation Pose(Result=Local Space Ref Pose())]
 }
 Glide {
  path=AnimGraph.AnimGraphNode_StateMachine_0.SkyDiveState.AnimStateNode_1.DiveOrGlide.AnimGraphNode_StateMachine_212.SkydiveOrGlide.AnimStateNode_2.SkyDive.AnimGraphNode_StateMachine_0.Skydiving.AnimStateNode_6.Glide
  nodes [
   n1 = Break Fort Anim Input Player Skydive Anim Asset(FortAnimInput_PlayerSkydiveAnimAsset=Get CurrentSkydiveAnimSet())
   n2 = Break Fort Anim Input Skydiving(FortAnimInput_Skydiving=Get Skydiving()) -> (
    ParaGilderRootOffset=~
    ParaGliderRootRotationOverride=~
    bUseParaGlideRootModifier=~
    bIsSkydivingFromLaunchPad=~
    bIsSkydivingFromBus=~
    bIsInVortex=~
    bIsDiving=~
    bIsParachuteOpen=~
    bIsSkydiving=~
    bIsParachuteLeaning=~
    bIsSkydiveLeaning=~
    bIsLeaning=~
    bIsSkydiveDiveMode=~
    bPlaySkydiveDrift=~
    bSkydiveDriftDelayActive=~
    bSkydiveDriftAnimAllowed=~
    bIsGliderRight=~
    bIsGliderCenter=~
    bIsGliderLeft=~
    bIsGliderForward=~
    bIsGliderBack=~
    bHasDeployedAndIsStandingUpright=~
    LocalAccelForward=~
    LocalAccelRight=~
    LocalVelocityRight=~
    SkydiveAimPitch=~
    SkydiveAimYaw=~
    DeployChuteAnimRate=~
    SkydiveDriftAnimRate=~
    SkydiveDriftAnimRateCurrent=~
    SkydiveFidgetAnimRate=~
    SkydiveFidgetAnimRateCurrent=~
    SkydiveAdditiveAlpha=~
    SkydiveDriftDelay=~
   )
   n3 = Get AimPitch()
   Output Animation Pose(Result=Apply Additive(
     Base=Blendspace Player 'BS_Glide_2D'!(X=@n2.SkydiveAimYaw, Y=@n3, BlendSpace:Blend Space Object Reference=@n1.Glide_BS_Female) -> (Pose)
     Additive=Use cached pose 'FreeFall_TurnAndLean_Glide_Pose'()
    ) {alphaBoolBlend {blendOption=Linear}})
  ]
 }
 Transition {
  path=AnimGraph.AnimGraphNode_StateMachine_0.SkyDiveState.AnimStateNode_1.DiveOrGlide.AnimGraphNode_StateMachine_212.SkydiveOrGlide.AnimStateNode_2.SkyDive.AnimGraphNode_StateMachine_0.Skydiving.AnimStateTransitionNode_0.Transition
  nodes [Result(bCanEnterTransition=Break Fort Anim Input Skydiving(FortAnimInput_Skydiving=Get Skydiving()) -> (
     ParaGilderRootOffset=~
     ParaGliderRootRotationOverride=~
     bUseParaGlideRootModifier=~
     bIsSkydivingFromLaunchPad=~
     bIsSkydivingFromBus=~
     bIsInVortex=~
     bIsDiving=~
     bIsParachuteOpen=~
     bIsSkydiving=~
     bIsParachuteLeaning=~
     bIsSkydiveLeaning=~
     bIsLeaning=~
     bIsSkydiveDiveMode=~
     bPlaySkydiveDrift=~
     bSkydiveDriftDelayActive=~
     bSkydiveDriftAnimAllowed=~
     bIsGliderRight=~
     bIsGliderCenter=~
     bIsGliderLeft=~
     bIsGliderForward=~
     bIsGliderBack=~
     bHasDeployedAndIsStandingUpright=~
     LocalAccelForward=~
     LocalAccelRight=~
     LocalVelocityRight=~
     SkydiveAimPitch=~
     SkydiveAimYaw=~
     DeployChuteAnimRate=~
     SkydiveDriftAnimRate=~
     SkydiveDriftAnimRateCurrent=~
     SkydiveFidgetAnimRate=~
     SkydiveFidgetAnimRateCurrent=~
     SkydiveAdditiveAlpha=~
     SkydiveDriftDelay=~
    ).bIsSkydiveDiveMode)]
 }
 Transition {
  path=AnimGraph.AnimGraphNode_StateMachine_0.SkyDiveState.AnimStateNode_1.DiveOrGlide.AnimGraphNode_StateMachine_212.SkydiveOrGlide.AnimStateNode_2.SkyDive.AnimGraphNode_StateMachine_0.Skydiving.AnimStateTransitionNode_1.Transition
  nodes [Result()]
 }
 Transition {
  path=AnimGraph.AnimGraphNode_StateMachine_0.SkyDiveState.AnimStateNode_1.DiveOrGlide.AnimGraphNode_StateMachine_212.SkydiveOrGlide.AnimStateNode_2.SkyDive.AnimGraphNode_StateMachine_0.Skydiving.AnimStateTransitionNode_2.Transition
  nodes [
   n1 = Get Skydiving()
   Result(bCanEnterTransition=NOT Boolean(A=Break Fort Anim Input Skydiving(FortAnimInput_Skydiving=@n1) -> (
      ParaGilderRootOffset=~
      ParaGliderRootRotationOverride=~
      bUseParaGlideRootModifier=~
      bIsSkydivingFromLaunchPad=~
      bIsSkydivingFromBus=~
      bIsInVortex=~
      bIsDiving=~
      bIsParachuteOpen=~
      bIsSkydiving=~
      bIsParachuteLeaning=~
      bIsSkydiveLeaning=~
      bIsLeaning=~
      bIsSkydiveDiveMode=~
      bPlaySkydiveDrift=~
      bSkydiveDriftDelayActive=~
      bSkydiveDriftAnimAllowed=~
      bIsGliderRight=~
      bIsGliderCenter=~
      bIsGliderLeft=~
      bIsGliderForward=~
      bIsGliderBack=~
      bHasDeployedAndIsStandingUpright=~
      LocalAccelForward=~
      LocalAccelRight=~
      LocalVelocityRight=~
      SkydiveAimPitch=~
      SkydiveAimYaw=~
      DeployChuteAnimRate=~
      SkydiveDriftAnimRate=~
      SkydiveDriftAnimRateCurrent=~
      SkydiveFidgetAnimRate=~
      SkydiveFidgetAnimRateCurrent=~
      SkydiveAdditiveAlpha=~
      SkydiveDriftDelay=~
     ).bIsSkydiveDiveMode))
  ]
 }
 Transition {
  path=AnimGraph.AnimGraphNode_StateMachine_0.SkyDiveState.AnimStateNode_1.DiveOrGlide.AnimGraphNode_StateMachine_212.SkydiveOrGlide.AnimStateNode_2.SkyDive.AnimGraphNode_StateMachine_0.Skydiving.AnimStateTransitionNode_3.Transition
  nodes [Result()]
 }
 Transition {
  path=AnimGraph.AnimGraphNode_StateMachine_0.SkyDiveState.AnimStateNode_1.DiveOrGlide.AnimGraphNode_StateMachine_212.SkydiveOrGlide.AnimStateNode_2.SkyDive.AnimGraphNode_StateMachine_0.Skydiving.AnimStateTransitionNode_4.Transition
  nodes [
   n1 = Get Skydiving()
   Result(bCanEnterTransition=NOT Boolean(A=Break Fort Anim Input Skydiving(FortAnimInput_Skydiving=@n1) -> (
      ParaGilderRootOffset=~
      ParaGliderRootRotationOverride=~
      bUseParaGlideRootModifier=~
      bIsSkydivingFromLaunchPad=~
      bIsSkydivingFromBus=~
      bIsInVortex=~
      bIsDiving=~
      bIsParachuteOpen=~
      bIsSkydiving=~
      bIsParachuteLeaning=~
      bIsSkydiveLeaning=~
      bIsLeaning=~
      bIsSkydiveDiveMode=~
      bPlaySkydiveDrift=~
      bSkydiveDriftDelayActive=~
      bSkydiveDriftAnimAllowed=~
      bIsGliderRight=~
      bIsGliderCenter=~
      bIsGliderLeft=~
      bIsGliderForward=~
      bIsGliderBack=~
      bHasDeployedAndIsStandingUpright=~
      LocalAccelForward=~
      LocalAccelRight=~
      LocalVelocityRight=~
      SkydiveAimPitch=~
      SkydiveAimYaw=~
      DeployChuteAnimRate=~
      SkydiveDriftAnimRate=~
      SkydiveDriftAnimRateCurrent=~
      SkydiveFidgetAnimRate=~
      SkydiveFidgetAnimRateCurrent=~
      SkydiveAdditiveAlpha=~
      SkydiveDriftDelay=~
     ).bIsSkydiveDiveMode))
  ]
 }
 Transition {
  path=AnimGraph.AnimGraphNode_StateMachine_0.SkyDiveState.AnimStateNode_1.DiveOrGlide.AnimGraphNode_StateMachine_212.SkydiveOrGlide.AnimStateNode_2.SkyDive.AnimGraphNode_StateMachine_0.Skydiving.AnimStateTransitionNode_5.Transition
  nodes [Result(bCanEnterTransition=Break Fort Anim Input Skydiving(FortAnimInput_Skydiving=Get Skydiving()) -> (
     ParaGilderRootOffset=~
     ParaGliderRootRotationOverride=~
     bUseParaGlideRootModifier=~
     bIsSkydivingFromLaunchPad=~
     bIsSkydivingFromBus=~
     bIsInVortex=~
     bIsDiving=~
     bIsParachuteOpen=~
     bIsSkydiving=~
     bIsParachuteLeaning=~
     bIsSkydiveLeaning=~
     bIsLeaning=~
     bIsSkydiveDiveMode=~
     bPlaySkydiveDrift=~
     bSkydiveDriftDelayActive=~
     bSkydiveDriftAnimAllowed=~
     bIsGliderRight=~
     bIsGliderCenter=~
     bIsGliderLeft=~
     bIsGliderForward=~
     bIsGliderBack=~
     bHasDeployedAndIsStandingUpright=~
     LocalAccelForward=~
     LocalAccelRight=~
     LocalVelocityRight=~
     SkydiveAimPitch=~
     SkydiveAimYaw=~
     DeployChuteAnimRate=~
     SkydiveDriftAnimRate=~
     SkydiveDriftAnimRateCurrent=~
     SkydiveFidgetAnimRate=~
     SkydiveFidgetAnimRateCurrent=~
     SkydiveAdditiveAlpha=~
     SkydiveDriftDelay=~
    ).bIsSkydiveDiveMode)]
 }
 Transition {
  path=AnimGraph.AnimGraphNode_StateMachine_0.SkyDiveState.AnimStateNode_1.DiveOrGlide.AnimGraphNode_StateMachine_212.SkydiveOrGlide.AnimStateNode_2.SkyDive.AnimGraphNode_StateMachine_0.Skydiving.AnimStateTransitionNode_7.Transition
  nodes [
   n1 = Get Skydiving()
   Result(bCanEnterTransition=NOT Boolean(A=Break Fort Anim Input Skydiving(FortAnimInput_Skydiving=@n1) -> (
      ParaGilderRootOffset=~
      ParaGliderRootRotationOverride=~
      bUseParaGlideRootModifier=~
      bIsSkydivingFromLaunchPad=~
      bIsSkydivingFromBus=~
      bIsInVortex=~
      bIsDiving=~
      bIsParachuteOpen=~
      bIsSkydiving=~
      bIsParachuteLeaning=~
      bIsSkydiveLeaning=~
      bIsLeaning=~
      bIsSkydiveDiveMode=~
      bPlaySkydiveDrift=~
      bSkydiveDriftDelayActive=~
      bSkydiveDriftAnimAllowed=~
      bIsGliderRight=~
      bIsGliderCenter=~
      bIsGliderLeft=~
      bIsGliderForward=~
      bIsGliderBack=~
      bHasDeployedAndIsStandingUpright=~
      LocalAccelForward=~
      LocalAccelRight=~
      LocalVelocityRight=~
      SkydiveAimPitch=~
      SkydiveAimYaw=~
      DeployChuteAnimRate=~
      SkydiveDriftAnimRate=~
      SkydiveDriftAnimRateCurrent=~
      SkydiveFidgetAnimRate=~
      SkydiveFidgetAnimRateCurrent=~
      SkydiveAdditiveAlpha=~
      SkydiveDriftDelay=~
     ).bIsSkydiveDiveMode))
  ]
 }
 Transition {
  path=AnimGraph.AnimGraphNode_StateMachine_0.SkyDiveState.AnimStateNode_1.DiveOrGlide.AnimGraphNode_StateMachine_212.SkydiveOrGlide.AnimStateNode_2.SkyDive.AnimGraphNode_StateMachine_0.Skydiving.AnimStateTransitionNode_8.Transition
  nodes [Result(bCanEnterTransition=Break Fort Anim Input Skydiving(FortAnimInput_Skydiving=Get Skydiving()) -> (
     ParaGilderRootOffset=~
     ParaGliderRootRotationOverride=~
     bUseParaGlideRootModifier=~
     bIsSkydivingFromLaunchPad=~
     bIsSkydivingFromBus=~
     bIsInVortex=~
     bIsDiving=~
     bIsParachuteOpen=~
     bIsSkydiving=~
     bIsParachuteLeaning=~
     bIsSkydiveLeaning=~
     bIsLeaning=~
     bIsSkydiveDiveMode=~
     bPlaySkydiveDrift=~
     bSkydiveDriftDelayActive=~
     bSkydiveDriftAnimAllowed=~
     bIsGliderRight=~
     bIsGliderCenter=~
     bIsGliderLeft=~
     bIsGliderForward=~
     bIsGliderBack=~
     bHasDeployedAndIsStandingUpright=~
     LocalAccelForward=~
     LocalAccelRight=~
     LocalVelocityRight=~
     SkydiveAimPitch=~
     SkydiveAimYaw=~
     DeployChuteAnimRate=~
     SkydiveDriftAnimRate=~
     SkydiveDriftAnimRateCurrent=~
     SkydiveFidgetAnimRate=~
     SkydiveFidgetAnimRateCurrent=~
     SkydiveAdditiveAlpha=~
     SkydiveDriftDelay=~
    ).bIsSkydiveDiveMode)]
 }
 ParaGlide {
  path=AnimGraph.AnimGraphNode_StateMachine_0.SkyDiveState.AnimStateNode_1.DiveOrGlide.AnimGraphNode_StateMachine_212.SkydiveOrGlide.AnimStateNode_3.ParaGlide
  nodes [
   "Blend Poses (EGliderType)"!(
    BlendPose_0:Pose Link Structure
    BlendPose_1:Pose Link Structure
    BlendTime_0:float=0
    BlendTime_1:float=0
    ActiveEnumValue:Byte=Break Fort Anim Input Player Glider Anim Asset!(FortAnimInput_PlayerGliderAnimAsset:"Fort Anim Input Player Glider Anim Asset Structure (by ref)"=Get CurrentGliderAnimSet() -> (CurrentGliderAnimSet:Fort Anim Input Player Glider Anim Asset Structure)) -> (
     Into_Base_MaleMedium_BS:Blend Space Object Reference
     PelvisModVertical:Pelvis Mod Vertical Input Spring Structure
     PelvisModLateral:Pelvis Mod Lateral Input Spring Structure
     PelvisModBankAngle:Pelvis Mod Bank Angle Input Spring Structure
     ThighModLegAngle:Thigh Mod Leg Angle Input Structure
     ThighModLegPitch:Thigh Mod Leg Pitch Spring Structure
     ThighModLegBank:Thigh Mod Leg Bank Spring Structure
     ThighModLeftLegDrag:Thigh Mod Left Leg Drag Structure
     PlayerGliderType:EGliderType Enum
     bEnableSpringMods:bool=~
     bUseLeanPlayRateAdjust:bool=~
     bUseForwardAs_Y_BS:bool=~
     bUseDeployCurveForIntoPlayRateAdjust:bool=~
     LeanForwardValue:float=~
     IdleValue:float=~
     LeanBackValue:float=~
     IdleInterpToSpeed:float=~
     LeanForwardInterpToSpeed:float=~
     LeanBackInterpToSpeed:float=~
    ) {is=title K2Node_BreakStruct}.PlayerGliderType
   ) -> (Pose:Pose Link Structure) {
    is=title AnimGraphNode_BlendListByEnum
    blendType=Linear
    blendPose [{linkId=-1, sourceLinkId=-1}, {linkId=-1, sourceLinkId=-1}]
    blendTime [0.1,0.1]
   }
   AnimGraphNode_StateMachine_2 = Glider_Lean() {is=title AnimGraphNode_StateMachine}
   AnimGraphNode_StateMachine_1 = Glider_SurfStyleAlt() {is=title AnimGraphNode_StateMachine}
   AnimGraphNode_StateMachine_3 = Glider_or_Unmbrella() {is=title AnimGraphNode_StateMachine}
   AnimGraphNode_StateMachine_4 = Glider_SurfStyle() {is=title AnimGraphNode_StateMachine}
   AnimGraphNode_StateMachine_73 = NoGlider() {is=title AnimGraphNode_StateMachine}
   n1 = "Blend Poses (EGliderType)"!(
    BlendPose_0=@AnimGraphNode_StateMachine_3
    BlendPose_1=@AnimGraphNode_StateMachine_4
    BlendPose_2:Pose Link Structure=@AnimGraphNode_StateMachine_73
    BlendPose_3:Pose Link Structure=@AnimGraphNode_StateMachine_1
    BlendTime_0=0
    BlendTime_1=0
    BlendTime_2:float=0.1
    BlendTime_3:float=0.1
    ActiveEnumValue=Break Fort Anim Input Player Glider Anim Asset(FortAnimInput_PlayerGliderAnimAsset=Get CurrentGliderAnimSet()) -> (
     bEnableSpringMods=~
     bUseLeanPlayRateAdjust=~
     bUseForwardAs_Y_BS=~
     bUseDeployCurveForIntoPlayRateAdjust=~
     LeanForwardValue=~
     IdleValue=~
     LeanBackValue=~
     IdleInterpToSpeed=~
     LeanForwardInterpToSpeed=~
     LeanBackInterpToSpeed=~
    ).PlayerGliderType
   ) -> (Pose) {
    blendType=Linear
    blendPose [{linkId=-1, sourceLinkId=-1}, {linkId=-1, sourceLinkId=-1}, {linkId=-1, sourceLinkId=-1}, {linkId=-1, sourceLinkId=-1}]
    blendTime [0.1,0.1,0.1,0.1]
   }
   AnimGraphNode_StateMachine_0 = NoGlider_Lean() {is=title AnimGraphNode_StateMachine}
   n2 = Get Skydiving()
   n3 = Apply Additive!(
    Base=Apply Additive(Base=@n1, Additive=@AnimGraphNode_StateMachine_0, AlphaCurveName=ParaGlideLeanAlpha) {alphaInputType=Curve, alphaBoolBlend {blendOption=Linear}}
    Additive=Sequence Player!(Sequence:Anim Sequence Base Object Reference=Break Fort Anim Input Skydiving(FortAnimInput_Skydiving=@n2) -> (
      ParaGilderRootOffset=~
      ParaGliderRootRotationOverride=~
      bUseParaGlideRootModifier=~
      bIsSkydivingFromLaunchPad=~
      bIsSkydivingFromBus=~
      bIsInVortex=~
      bIsDiving=~
      bIsParachuteOpen=~
      bIsSkydiving=~
      bIsParachuteLeaning=~
      bIsSkydiveLeaning=~
      bIsLeaning=~
      bIsSkydiveDiveMode=~
      bPlaySkydiveDrift=~
      bSkydiveDriftDelayActive=~
      bSkydiveDriftAnimAllowed=~
      bIsGliderRight=~
      bIsGliderCenter=~
      bIsGliderLeft=~
      bIsGliderForward=~
      bIsGliderBack=~
      bHasDeployedAndIsStandingUpright=~
      LocalAccelForward=~
      LocalAccelRight=~
      LocalVelocityRight=~
      SkydiveAimPitch=~
      SkydiveAimYaw=~
      DeployChuteAnimRate=~
      SkydiveDriftAnimRate=~
      SkydiveDriftAnimRateCurrent=~
      SkydiveFidgetAnimRate=~
      SkydiveFidgetAnimRateCurrent=~
      SkydiveAdditiveAlpha=~
      SkydiveDriftDelay=~
     ).PoseAdditive) -> (Pose)
    AlphaCurveName:Name=DisableGliderPoseAdditive
   ) -> (Pose) {alphaInputType=Curve, alphaBoolBlend {blendOption=Linear}, alphaScaleBiasClamp {bClampResult=true, bias=1, scale=-1}}
   Output Animation Pose(Result=@n3)
  ]
 }
 NoGlider_Lean {
  entry=Lean_Zero
  path=AnimGraph.AnimGraphNode_StateMachine_0.SkyDiveState.AnimStateNode_1.DiveOrGlide.AnimGraphNode_StateMachine_212.SkydiveOrGlide.AnimStateNode_3.ParaGlide.AnimGraphNode_StateMachine_0.NoGlider_Lean
  ids=id
  nodes [
   AnimStateNode_10 = Lean_Zero() {is=title AnimStateNode}
   AnimStateNode_13 = Lean() {is=title AnimStateNode}
   AnimStateNode_14 = Lean_Out() {is=title AnimStateNode}
   AnimStateNode_17 = Lean_Into() {is=title AnimStateNode}
   AnimStateTransitionNode_25 = AnimStateTransitionNode()
   AnimStateTransitionNode_26 = AnimStateTransitionNode() {bAutomaticRuleBasedOnSequencePlayerInState=true, crossfadeDuration=0.1}
   AnimStateTransitionNode_37 = AnimStateTransitionNode()
   AnimStateTransitionNode_38 = AnimStateTransitionNode()
   AnimStateTransitionNode_39 = AnimStateTransitionNode() {crossfadeDuration=0.1}
   AnimStateTransitionNode_40 = AnimStateTransitionNode()
  ]
  links [
   {
    dst=Lean_Out
    rule=AnimGraph.AnimGraphNode_StateMachine_0.SkyDiveState.AnimStateNode_1.DiveOrGlide.AnimGraphNode_StateMachine_212.SkydiveOrGlide.AnimStateNode_3.ParaGlide.AnimGraphNode_StateMachine_0.NoGlider_Lean.AnimStateTransitionNode_25.AnimationTransitionGraph_1
    src=Lean
    via=AnimStateTransitionNode_25
   }
   {
    dst=Lean
    rule=AnimGraph.AnimGraphNode_StateMachine_0.SkyDiveState.AnimStateNode_1.DiveOrGlide.AnimGraphNode_StateMachine_212.SkydiveOrGlide.AnimStateNode_3.ParaGlide.AnimGraphNode_StateMachine_0.NoGlider_Lean.AnimStateTransitionNode_39.AnimationTransitionGraph_1
    src=Lean_Into
    via=AnimStateTransitionNode_39
   }
   {
    dst=Lean_Out
    rule=AnimGraph.AnimGraphNode_StateMachine_0.SkyDiveState.AnimStateNode_1.DiveOrGlide.AnimGraphNode_StateMachine_212.SkydiveOrGlide.AnimStateNode_3.ParaGlide.AnimGraphNode_StateMachine_0.NoGlider_Lean.AnimStateTransitionNode_38.AnimationTransitionGraph_1
    src=Lean_Into
    via=AnimStateTransitionNode_38
   }
   {
    dst=Lean_Into
    rule=AnimGraph.AnimGraphNode_StateMachine_0.SkyDiveState.AnimStateNode_1.DiveOrGlide.AnimGraphNode_StateMachine_212.SkydiveOrGlide.AnimStateNode_3.ParaGlide.AnimGraphNode_StateMachine_0.NoGlider_Lean.AnimStateTransitionNode_40.Transition
    src=Lean_Out
    via=AnimStateTransitionNode_40
   }
   {
    dst=Lean_Zero
    rule=AnimGraph.AnimGraphNode_StateMachine_0.SkyDiveState.AnimStateNode_1.DiveOrGlide.AnimGraphNode_StateMachine_212.SkydiveOrGlide.AnimStateNode_3.ParaGlide.AnimGraphNode_StateMachine_0.NoGlider_Lean.AnimStateTransitionNode_26.AnimationTransitionGraph_1
    src=Lean_Out
    via=AnimStateTransitionNode_26
   }
   {
    dst=Lean_Into
    rule=AnimGraph.AnimGraphNode_StateMachine_0.SkyDiveState.AnimStateNode_1.DiveOrGlide.AnimGraphNode_StateMachine_212.SkydiveOrGlide.AnimStateNode_3.ParaGlide.AnimGraphNode_StateMachine_0.NoGlider_Lean.AnimStateTransitionNode_37.Transition
    src=Lean_Zero
    via=AnimStateTransitionNode_37
   }
  ]
 }
 Lean_Zero {
  path=AnimGraph.AnimGraphNode_StateMachine_0.SkyDiveState.AnimStateNode_1.DiveOrGlide.AnimGraphNode_StateMachine_212.SkydiveOrGlide.AnimStateNode_3.ParaGlide.AnimGraphNode_StateMachine_0.NoGlider_Lean.AnimStateNode_10.Lean_Zero
  nodes [
   n1 = Break Fort Anim Input Player Glider Anim Asset!(FortAnimInput_PlayerGliderAnimAsset=Get CurrentGliderAnimSet()) -> (
    Default_Base_BS:Blend Space Object Reference
    Default_BodyAdditive_MaleMedium_BS:Blend Space Object Reference
    Default_BodyAdditive_MaleLarge_BS:Blend Space Object Reference
    Default_BodyAdditive_FemaleSmall_BS:Blend Space Object Reference
    Default_BodyAdditive_FemaleMedium_BS:Blend Space Object Reference
    Default_BodyAdditive_FemaleLarge_BS:Blend Space Object Reference
    Default_TurnAdditive_MaleMedium_BS:Blend Space Object Reference
    Default_TurnAdditive_MaleLarge_BS:Blend Space Object Reference
    Default_TurnAdditive_FemaleSmall_BS:Blend Space Object Reference
    Default_TurnAdditive_FemaleMedium_BS:Blend Space Object Reference
    Default_TurnAdditive_FemaleLarge_BS:Blend Space Object Reference
    Into_Base_BS:Blend Space Object Reference
    Into_Base_MaleMedium_BS
    Into_BodyAdditive_MaleMedium_BS:Blend Space Object Reference
    Into_BodyAdditive_MaleLarge_BS:Blend Space Object Reference
    Into_BodyAdditive_FemaleSmall_BS:Blend Space Object Reference
    Into_BodyAdditive_FemaleMedium_BS:Blend Space Object Reference
    Into_BodyAdditive_FemaleLarge_BS:Blend Space Object Reference
    Into_TurnAdditive_MaleMedium_BS:Blend Space Object Reference
    Into_TurnAdditive_MaleLarge_BS:Blend Space Object Reference
    Into_TurnAdditive_FemaleSmall_BS:Blend Space Object Reference
    Into_TurnAdditive_FemaleMedium_BS:Blend Space Object Reference
    Into_TurnAdditive_FemaleLarge_BS:Blend Space Object Reference
    Lean_MaleMedium_BS:Blend Space Object Reference
    Lean_MaleLarge_BS:Blend Space Object Reference
    Lean_FemaleSmall_BS:Blend Space Object Reference
    Lean_FemaleMedium_BS:Blend Space Object Reference
    Lean_FemaleLarge_BS:Blend Space Object Reference
    LeanAdditive_Center_MaleMedium_Pose:Anim Sequence Object Reference
    LeanAdditive_Center_MaleLarge_Pose:Anim Sequence Object Reference
    LeanAdditive_Center_FemaleSmall_Pose:Anim Sequence Object Reference
    LeanAdditive_Center_FemaleMedium_Pose:Anim Sequence Object Reference
    LeanAdditive_Center_FemaleLarge_Pose:Anim Sequence Object Reference
    LeanAdditive_Into_BS:Blend Space Object Reference
    LeanAdditive_ForwardInto_Anim:Anim Sequence Object Reference
    LeanAdditive_ForwardInto_FromDeploy_Anim:Anim Sequence Object Reference
    LeanAdditive_BackInto_Anim:Anim Sequence Object Reference
    LeanAdditive_LeftInto_Anim:Anim Sequence Object Reference
    LeanAdditive_RightInto_Anim:Anim Sequence Object Reference
    LeanAdditive_ForwardOut_Anim:Anim Sequence Object Reference
    LeanAdditive_BackOut_Anim:Anim Sequence Object Reference
    LeanAdditive_LeftOut_Anim:Anim Sequence Object Reference
    LeanAdditive_RightOut_Anim:Anim Sequence Object Reference
    ToGlide_BS:Blend Space Object Reference
    ToGlide_Lean_BS:Blend Space Object Reference
    ToDive_BS:Blend Space Object Reference
    ToDive_Lean_BS:Blend Space Object Reference
    Dive_WeaponR_Additive_BS:Blend Space Object Reference
    Glide_WeaponR_Additive_Anim:Anim Sequence Object Reference
    GenericAdditive_Male_BS:Blend Space Object Reference
    GenericAdditive_Female_BS:Blend Space Object Reference
    RootModPitchMin:float=~
    RootModPitchMax:float=~
    RootModYOffsetMin:float=~
    RootModYOffsetMax:float=~
    PelvisModVertical
    PelvisModLateral
    PelvisModBankAngle
    ThighModLegAngle
    ThighModLegPitch
    ThighModLegBank
    ThighModLeftLegDrag
    PlayerGliderType
    bEnableSpringMods=~
    bAllowPlayerDeployRootMod:bool=~
    bUseSurfStyle:bool=~
    bUseLeanPlayRateAdjust=~
    bUseForwardAs_Y_BS=~
    bUseDeployCurveForIntoPlayRateAdjust=~
    LeanForwardValue=~
    IdleValue=~
    LeanBackValue=~
    IdleInterpToSpeed=~
    LeanForwardInterpToSpeed=~
    LeanBackInterpToSpeed=~
   )
   n2 = "Blend Poses (EFortPlayerAnimBodyType)"!(
    BlendPose_0:Pose Link Structure=Sequence Player(Sequence=@n1.LeanAdditive_Center_MaleMedium_Pose)
    BlendPose_1:Pose Link Structure=Sequence Player(Sequence=@n1.LeanAdditive_Center_MaleLarge_Pose)
    BlendTime_0:float=0.1
    BlendTime_1:float=0.1
    ActiveEnumValue:Byte=Get AnimBodyType() -> (AnimBodyType:EFortPlayerAnimBodyType Enum=Small)
   ) -> (Pose:Pose Link Structure) {
    is=title AnimGraphNode_BlendListByEnum
    blendType=Linear
    blendPose [{linkId=-1, sourceLinkId=-1}, {linkId=-1, sourceLinkId=-1}]
    blendTime [0.1,0.1]
   }
   n3 = "Blend Poses (EFortPlayerAnimBodyType)"!(
    BlendPose_0=Sequence Player(Sequence=@n1.LeanAdditive_Center_FemaleMedium_Pose)
    BlendPose_1=Sequence Player(Sequence=@n1.LeanAdditive_Center_FemaleSmall_Pose)
    BlendPose_2:Pose Link Structure=Sequence Player(Sequence=@n1.LeanAdditive_Center_FemaleLarge_Pose)
    BlendTime_0=0.1
    BlendTime_1=0.1
    BlendTime_2:float=0.1
    ActiveEnumValue=Get AnimBodyType() -> (AnimBodyType=Small)
   ) -> (Pose) {
    blendType=Linear
    blendPose [{linkId=-1, sourceLinkId=-1}, {linkId=-1, sourceLinkId=-1}, {linkId=-1, sourceLinkId=-1}]
    blendTime [0.1,0.1,0.1]
   }
   "Blend Poses (EFortCustomGender)"(BlendPose_0=@n2, BlendPose_1=@n3, BlendTime_0=0, BlendTime_1=0, ActiveEnumValue=Get Gender() -> (Gender=Invalid)) {blendType=Linear, blendPose [{linkId=-1, sourceLinkId=-1}, {linkId=-1, sourceLinkId=-1}], blendTime [0.1,0.1]}
   Output Animation Pose(Result=Additive Identity Pose())
  ]
 }
 Lean {
  path=AnimGraph.AnimGraphNode_StateMachine_0.SkyDiveState.AnimStateNode_1.DiveOrGlide.AnimGraphNode_StateMachine_212.SkydiveOrGlide.AnimStateNode_3.ParaGlide.AnimGraphNode_StateMachine_0.NoGlider_Lean.AnimStateNode_13.Lean
  nodes [
   n1 = Break Fort Anim Input Skydiving(FortAnimInput_Skydiving=Get Skydiving()) -> (
    ParaGilderRootOffset=~
    ParaGliderRootRotationOverride=~
    bUseParaGlideRootModifier=~
    bIsSkydivingFromLaunchPad=~
    bIsSkydivingFromBus=~
    bIsInVortex=~
    bIsDiving=~
    bIsParachuteOpen=~
    bIsSkydiving=~
    bIsParachuteLeaning=~
    bIsSkydiveLeaning=~
    bIsLeaning=~
    bIsSkydiveDiveMode=~
    bPlaySkydiveDrift=~
    bSkydiveDriftDelayActive=~
    bSkydiveDriftAnimAllowed=~
    bIsGliderRight=~
    bIsGliderCenter=~
    bIsGliderLeft=~
    bIsGliderForward=~
    bIsGliderBack=~
    bHasDeployedAndIsStandingUpright=~
    LocalAccelForward=~
    LocalAccelRight=~
    LocalVelocityRight=~
    SkydiveAimPitch=~
    SkydiveAimYaw=~
    DeployChuteAnimRate=~
    SkydiveDriftAnimRate=~
    SkydiveDriftAnimRateCurrent=~
    SkydiveFidgetAnimRate=~
    SkydiveFidgetAnimRateCurrent=~
    SkydiveAdditiveAlpha=~
    SkydiveDriftDelay=~
   )
   n2 = Break Fort Anim Input Player Glider Anim Asset(FortAnimInput_PlayerGliderAnimAsset=Get CurrentGliderAnimSet()) -> (
    RootModPitchMin=~
    RootModPitchMax=~
    RootModYOffsetMin=~
    RootModYOffsetMax=~
    bEnableSpringMods=~
    bAllowPlayerDeployRootMod=~
    bUseSurfStyle=~
    bUseLeanPlayRateAdjust=~
    bUseForwardAs_Y_BS=~
    bUseDeployCurveForIntoPlayRateAdjust=~
    LeanForwardValue=~
    IdleValue=~
    LeanBackValue=~
    IdleInterpToSpeed=~
    LeanForwardInterpToSpeed=~
    LeanBackInterpToSpeed=~
   )
   n3 = "Blend Poses (EFortPlayerAnimBodyType)"!(
    BlendPose_0=Blendspace Player!(X=@n1.LocalAccelRight, Y=@n1.LocalAccelForward, BlendSpace:Blend Space Object Reference=@n2.Lean_MaleMedium_BS) -> (Pose)
    BlendPose_1=Blendspace Player(X=@n1.LocalAccelRight, Y=@n1.LocalAccelForward, BlendSpace=@n2.Lean_MaleLarge_BS)
    BlendTime_0=0.1
    BlendTime_1=0.1
    ActiveEnumValue=0
   ) -> (Pose) {blendType=Linear, blendPose [{linkId=-1, sourceLinkId=-1}, {linkId=-1, sourceLinkId=-1}], blendTime [0.1,0.1]}
   n4 = "Blend Poses (EFortPlayerAnimBodyType)"!(
    BlendPose_0=Blendspace Player(X=@n1.LocalAccelRight, Y=@n1.LocalAccelForward, BlendSpace=@n2.Lean_FemaleMedium_BS)
    BlendPose_1=Blendspace Player(X=@n1.LocalAccelRight, Y=@n1.LocalAccelForward, BlendSpace=@n2.Lean_FemaleSmall_BS)
    BlendPose_2:Pose Link Structure=Blendspace Player(X=@n1.LocalAccelRight, Y=@n1.LocalAccelForward, BlendSpace=@n2.Lean_FemaleLarge_BS)
    BlendTime_0=0.1
    BlendTime_1=0.1
    BlendTime_2:float=0.1
    ActiveEnumValue=Get AnimBodyType() -> (AnimBodyType=Small)
   ) -> (Pose) {
    blendType=Linear
    blendPose [{linkId=-1, sourceLinkId=-1}, {linkId=-1, sourceLinkId=-1}, {linkId=-1, sourceLinkId=-1}]
    blendTime [0.1,0.1,0.1]
   }
   n5 = "Blend Poses (EFortCustomGender)"(BlendPose_0=@n3, BlendPose_1=@n4, BlendTime_0=0, BlendTime_1=0, ActiveEnumValue=Get Gender() -> (Gender=Invalid)) {blendType=Linear, blendPose [{linkId=-1, sourceLinkId=-1}, {linkId=-1, sourceLinkId=-1}], blendTime [0.1,0.1]}
   Output Animation Pose(Result=@n5)
   Two Way Blend(AlphaCurveName=ParaGlideLeanTransition) {blendNode {alphaInputType=Curve, alphaBoolBlend {blendOption=Linear}, alphaScaleBiasClamp {bias=1, scale=-1}}}
  ]
 }
 Lean_Out {
  path=AnimGraph.AnimGraphNode_StateMachine_0.SkyDiveState.AnimStateNode_1.DiveOrGlide.AnimGraphNode_StateMachine_212.SkydiveOrGlide.AnimStateNode_3.ParaGlide.AnimGraphNode_StateMachine_0.NoGlider_Lean.AnimStateNode_14.Lean_Out
  nodes [
   n1 = Break Fort Anim Input Player Glider Anim Asset(FortAnimInput_PlayerGliderAnimAsset=Get CurrentGliderAnimSet()) -> (
    RootModPitchMin=~
    RootModPitchMax=~
    RootModYOffsetMin=~
    RootModYOffsetMax=~
    bEnableSpringMods=~
    bAllowPlayerDeployRootMod=~
    bUseSurfStyle=~
    bUseLeanPlayRateAdjust=~
    bUseForwardAs_Y_BS=~
    bUseDeployCurveForIntoPlayRateAdjust=~
    LeanForwardValue=~
    IdleValue=~
    LeanBackValue=~
    IdleInterpToSpeed=~
    LeanForwardInterpToSpeed=~
    LeanBackInterpToSpeed=~
   )
   n2 = "Blend Poses (ESkydivingDirection)"!(
    BlendPose_0:Pose Link Structure=Sequence Player(Sequence=@n1.LeanAdditive_Center_MaleMedium_Pose)
    BlendPose_1:Pose Link Structure=Sequence Player(Sequence=@n1.LeanAdditive_ForwardOut_Anim) {bLoopAnimation=false}
    BlendPose_2:Pose Link Structure=Sequence Player(Sequence=@n1.LeanAdditive_BackOut_Anim) {bLoopAnimation=false}
    BlendPose_3:Pose Link Structure=Sequence Player(Sequence=@n1.LeanAdditive_RightOut_Anim) {bLoopAnimation=false}
    BlendPose_4:Pose Link Structure=Sequence Player(Sequence=@n1.LeanAdditive_LeftOut_Anim) {bLoopAnimation=false}
    BlendTime_0:float=0.5
    BlendTime_1:float=0.5
    BlendTime_2:float=0.5
    BlendTime_3:float=0.5
    BlendTime_4:float=0.5
    ActiveEnumValue:Byte=Break Fort Anim Input Skydiving(FortAnimInput_Skydiving=Get Skydiving()) -> (
     ParaGilderRootOffset=~
     ParaGliderRootRotationOverride=~
     bUseParaGlideRootModifier=~
     bIsSkydivingFromLaunchPad=~
     bIsSkydivingFromBus=~
     bIsInVortex=~
     bIsDiving=~
     bIsParachuteOpen=~
     bIsSkydiving=~
     bIsParachuteLeaning=~
     bIsSkydiveLeaning=~
     bIsLeaning=~
     bIsSkydiveDiveMode=~
     bPlaySkydiveDrift=~
     bSkydiveDriftDelayActive=~
     bSkydiveDriftAnimAllowed=~
     bIsGliderRight=~
     bIsGliderCenter=~
     bIsGliderLeft=~
     bIsGliderForward=~
     bIsGliderBack=~
     bHasDeployedAndIsStandingUpright=~
     LocalAccelForward=~
     LocalAccelRight=~
     LocalVelocityRight=~
     SkydiveAimPitch=~
     SkydiveAimYaw=~
     DeployChuteAnimRate=~
     SkydiveDriftAnimRate=~
     SkydiveDriftAnimRateCurrent=~
     SkydiveFidgetAnimRate=~
     SkydiveFidgetAnimRateCurrent=~
     SkydiveAdditiveAlpha=~
     SkydiveDriftDelay=~
    ).DirectionLast
   ) -> (Pose:Pose Link Structure) {
    is=title AnimGraphNode_BlendListByEnum
    blendType=Linear
    blendPose [
     {linkId=-1, sourceLinkId=-1}
     {linkId=-1, sourceLinkId=-1}
     {linkId=-1, sourceLinkId=-1}
     {linkId=-1, sourceLinkId=-1}
     {linkId=-1, sourceLinkId=-1}
    ]
    blendTime [0.1,0.1,0.1,0.1,0.1]
   }
   Output Animation Pose(Result=@n2)
  ]
 }
 Lean_Into {
  path=AnimGraph.AnimGraphNode_StateMachine_0.SkyDiveState.AnimStateNode_1.DiveOrGlide.AnimGraphNode_StateMachine_212.SkydiveOrGlide.AnimStateNode_3.ParaGlide.AnimGraphNode_StateMachine_0.NoGlider_Lean.AnimStateNode_17.Lean_Into
  nodes [
   AnimGraphNode_StateMachine_0 = Lean_Into() {is=title AnimGraphNode_StateMachine}
   Output Animation Pose(Result=@AnimGraphNode_StateMachine_0)
  ]
 }
 Lean_Into {
  entry=Lean_Zero
  path=AnimGraph.AnimGraphNode_StateMachine_0.SkyDiveState.AnimStateNode_1.DiveOrGlide.AnimGraphNode_StateMachine_212.SkydiveOrGlide.AnimStateNode_3.ParaGlide.AnimGraphNode_StateMachine_0.NoGlider_Lean.AnimStateNode_17.Lean_Into.AnimGraphNode_StateMachine_0.Lean_Into
  ids=id
  nodes [
   AnimStateNode_0 = Lean_Zero()
   AnimStateNode_3 = Left_Into() {is=title AnimStateNode}
   AnimStateNode_4 = Right_Into() {is=title AnimStateNode}
   AnimStateNode_5 = Forward_Into() {is=title AnimStateNode}
   AnimStateNode_6 = Back_Into() {is=title AnimStateNode}
   AnimStateTransitionNode_19 = AnimStateTransitionNode()
   AnimStateTransitionNode_20 = AnimStateTransitionNode()
   AnimStateTransitionNode_21 = AnimStateTransitionNode()
   AnimStateTransitionNode_22 = AnimStateTransitionNode()
   AnimStateTransitionNode_23 = AnimStateTransitionNode() {crossfadeDuration=0.35}
   AnimStateTransitionNode_24 = AnimStateTransitionNode() {crossfadeDuration=0.35}
   AnimStateTransitionNode_25 = AnimStateTransitionNode() {crossfadeDuration=0.5}
   AnimStateTransitionNode_26 = AnimStateTransitionNode() {crossfadeDuration=0.5}
   AnimStateTransitionNode_27 = AnimStateTransitionNode() {crossfadeDuration=0.5}
   AnimStateTransitionNode_28 = AnimStateTransitionNode() {crossfadeDuration=0.5}
   AnimStateTransitionNode_29 = AnimStateTransitionNode() {crossfadeDuration=0.5}
   AnimStateTransitionNode_30 = AnimStateTransitionNode() {crossfadeDuration=0.5}
   AnimStateTransitionNode_31 = AnimStateTransitionNode() {crossfadeDuration=0.5}
   AnimStateTransitionNode_32 = AnimStateTransitionNode() {crossfadeDuration=0.5}
   AnimStateTransitionNode_33 = AnimStateTransitionNode() {crossfadeDuration=0.5}
   AnimStateTransitionNode_34 = AnimStateTransitionNode() {crossfadeDuration=0.5}
  ]
  links [
   {
    dst=Forward_Into
    rule=AnimGraph.AnimGraphNode_StateMachine_0.SkyDiveState.AnimStateNode_1.DiveOrGlide.AnimGraphNode_StateMachine_212.SkydiveOrGlide.AnimStateNode_3.ParaGlide.AnimGraphNode_StateMachine_0.NoGlider_Lean.AnimStateNode_17.Lean_Into.AnimGraphNode_StateMachine_0.Lean_Into.AnimStateTransitionNode_32.Transition
    src=Back_Into
    via=AnimStateTransitionNode_32
   }
   {
    dst=Left_Into
    rule=AnimGraph.AnimGraphNode_StateMachine_0.SkyDiveState.AnimStateNode_1.DiveOrGlide.AnimGraphNode_StateMachine_212.SkydiveOrGlide.AnimStateNode_3.ParaGlide.AnimGraphNode_StateMachine_0.NoGlider_Lean.AnimStateNode_17.Lean_Into.AnimGraphNode_StateMachine_0.Lean_Into.AnimStateTransitionNode_33.Transition
    src=Back_Into
    via=AnimStateTransitionNode_33
   }
   {
    dst=Right_Into
    rule=AnimGraph.AnimGraphNode_StateMachine_0.SkyDiveState.AnimStateNode_1.DiveOrGlide.AnimGraphNode_StateMachine_212.SkydiveOrGlide.AnimStateNode_3.ParaGlide.AnimGraphNode_StateMachine_0.NoGlider_Lean.AnimStateNode_17.Lean_Into.AnimGraphNode_StateMachine_0.Lean_Into.AnimStateTransitionNode_34.Transition
    src=Back_Into
    via=AnimStateTransitionNode_34
   }
   {
    dst=Back_Into
    rule=AnimGraph.AnimGraphNode_StateMachine_0.SkyDiveState.AnimStateNode_1.DiveOrGlide.AnimGraphNode_StateMachine_212.SkydiveOrGlide.AnimStateNode_3.ParaGlide.AnimGraphNode_StateMachine_0.NoGlider_Lean.AnimStateNode_17.Lean_Into.AnimGraphNode_StateMachine_0.Lean_Into.AnimStateTransitionNode_31.Transition
    src=Forward_Into
    via=AnimStateTransitionNode_31
   }
   {
    dst=Left_Into
    rule=AnimGraph.AnimGraphNode_StateMachine_0.SkyDiveState.AnimStateNode_1.DiveOrGlide.AnimGraphNode_StateMachine_212.SkydiveOrGlide.AnimStateNode_3.ParaGlide.AnimGraphNode_StateMachine_0.NoGlider_Lean.AnimStateNode_17.Lean_Into.AnimGraphNode_StateMachine_0.Lean_Into.AnimStateTransitionNode_29.Transition
    src=Forward_Into
    via=AnimStateTransitionNode_29
   }
   {
    dst=Right_Into
    rule=AnimGraph.AnimGraphNode_StateMachine_0.SkyDiveState.AnimStateNode_1.DiveOrGlide.AnimGraphNode_StateMachine_212.SkydiveOrGlide.AnimStateNode_3.ParaGlide.AnimGraphNode_StateMachine_0.NoGlider_Lean.AnimStateNode_17.Lean_Into.AnimGraphNode_StateMachine_0.Lean_Into.AnimStateTransitionNode_30.Transition
    src=Forward_Into
    via=AnimStateTransitionNode_30
   }
   {
    dst=Back_Into
    rule=AnimGraph.AnimGraphNode_StateMachine_0.SkyDiveState.AnimStateNode_1.DiveOrGlide.AnimGraphNode_StateMachine_212.SkydiveOrGlide.AnimStateNode_3.ParaGlide.AnimGraphNode_StateMachine_0.NoGlider_Lean.AnimStateNode_17.Lean_Into.AnimGraphNode_StateMachine_0.Lean_Into.AnimStateTransitionNode_20.Transition
    src=Lean_Zero
    via=AnimStateTransitionNode_20
   }
   {
    dst=Forward_Into
    rule=AnimGraph.AnimGraphNode_StateMachine_0.SkyDiveState.AnimStateNode_1.DiveOrGlide.AnimGraphNode_StateMachine_212.SkydiveOrGlide.AnimStateNode_3.ParaGlide.AnimGraphNode_StateMachine_0.NoGlider_Lean.AnimStateNode_17.Lean_Into.AnimGraphNode_StateMachine_0.Lean_Into.AnimStateTransitionNode_19.Transition
    src=Lean_Zero
    via=AnimStateTransitionNode_19
   }
   {
    dst=Left_Into
    rule=AnimGraph.AnimGraphNode_StateMachine_0.SkyDiveState.AnimStateNode_1.DiveOrGlide.AnimGraphNode_StateMachine_212.SkydiveOrGlide.AnimStateNode_3.ParaGlide.AnimGraphNode_StateMachine_0.NoGlider_Lean.AnimStateNode_17.Lean_Into.AnimGraphNode_StateMachine_0.Lean_Into.AnimStateTransitionNode_22.Transition
    src=Lean_Zero
    via=AnimStateTransitionNode_22
   }
   {
    dst=Right_Into
    rule=AnimGraph.AnimGraphNode_StateMachine_0.SkyDiveState.AnimStateNode_1.DiveOrGlide.AnimGraphNode_StateMachine_212.SkydiveOrGlide.AnimStateNode_3.ParaGlide.AnimGraphNode_StateMachine_0.NoGlider_Lean.AnimStateNode_17.Lean_Into.AnimGraphNode_StateMachine_0.Lean_Into.AnimStateTransitionNode_21.Transition
    src=Lean_Zero
    via=AnimStateTransitionNode_21
   }
   {
    dst=Back_Into
    rule=AnimGraph.AnimGraphNode_StateMachine_0.SkyDiveState.AnimStateNode_1.DiveOrGlide.AnimGraphNode_StateMachine_212.SkydiveOrGlide.AnimStateNode_3.ParaGlide.AnimGraphNode_StateMachine_0.NoGlider_Lean.AnimStateNode_17.Lean_Into.AnimGraphNode_StateMachine_0.Lean_Into.AnimStateTransitionNode_26.Transition
    src=Left_Into
    via=AnimStateTransitionNode_26
   }
   {
    dst=Forward_Into
    rule=AnimGraph.AnimGraphNode_StateMachine_0.SkyDiveState.AnimStateNode_1.DiveOrGlide.AnimGraphNode_StateMachine_212.SkydiveOrGlide.AnimStateNode_3.ParaGlide.AnimGraphNode_StateMachine_0.NoGlider_Lean.AnimStateNode_17.Lean_Into.AnimGraphNode_StateMachine_0.Lean_Into.AnimStateTransitionNode_25.Transition
    src=Left_Into
    via=AnimStateTransitionNode_25
   }
   {
    dst=Right_Into
    rule=AnimGraph.AnimGraphNode_StateMachine_0.SkyDiveState.AnimStateNode_1.DiveOrGlide.AnimGraphNode_StateMachine_212.SkydiveOrGlide.AnimStateNode_3.ParaGlide.AnimGraphNode_StateMachine_0.NoGlider_Lean.AnimStateNode_17.Lean_Into.AnimGraphNode_StateMachine_0.Lean_Into.AnimStateTransitionNode_24.Transition
    src=Left_Into
    via=AnimStateTransitionNode_24
   }
   {
    dst=Back_Into
    rule=AnimGraph.AnimGraphNode_StateMachine_0.SkyDiveState.AnimStateNode_1.DiveOrGlide.AnimGraphNode_StateMachine_212.SkydiveOrGlide.AnimStateNode_3.ParaGlide.AnimGraphNode_StateMachine_0.NoGlider_Lean.AnimStateNode_17.Lean_Into.AnimGraphNode_StateMachine_0.Lean_Into.AnimStateTransitionNode_28.Transition
    src=Right_Into
    via=AnimStateTransitionNode_28
   }
   {
    dst=Forward_Into
    rule=AnimGraph.AnimGraphNode_StateMachine_0.SkyDiveState.AnimStateNode_1.DiveOrGlide.AnimGraphNode_StateMachine_212.SkydiveOrGlide.AnimStateNode_3.ParaGlide.AnimGraphNode_StateMachine_0.NoGlider_Lean.AnimStateNode_17.Lean_Into.AnimGraphNode_StateMachine_0.Lean_Into.AnimStateTransitionNode_27.Transition
    src=Right_Into
    via=AnimStateTransitionNode_27
   }
   {
    dst=Left_Into
    rule=AnimGraph.AnimGraphNode_StateMachine_0.SkyDiveState.AnimStateNode_1.DiveOrGlide.AnimGraphNode_StateMachine_212.SkydiveOrGlide.AnimStateNode_3.ParaGlide.AnimGraphNode_StateMachine_0.NoGlider_Lean.AnimStateNode_17.Lean_Into.AnimGraphNode_StateMachine_0.Lean_Into.AnimStateTransitionNode_23.Transition
    src=Right_Into
    via=AnimStateTransitionNode_23
   }
  ]
 }
 Lean_Zero {
  path=AnimGraph.AnimGraphNode_StateMachine_0.SkyDiveState.AnimStateNode_1.DiveOrGlide.AnimGraphNode_StateMachine_212.SkydiveOrGlide.AnimStateNode_3.ParaGlide.AnimGraphNode_StateMachine_0.NoGlider_Lean.AnimStateNode_17.Lean_Into.AnimGraphNode_StateMachine_0.Lean_Into.AnimStateNode_0.Lean_Zero
  nodes [
   n1 = Break Fort Anim Input Player Glider Anim Asset(FortAnimInput_PlayerGliderAnimAsset=Get CurrentGliderAnimSet()) -> (
    RootModPitchMin=~
    RootModPitchMax=~
    RootModYOffsetMin=~
    RootModYOffsetMax=~
    bEnableSpringMods=~
    bAllowPlayerDeployRootMod=~
    bUseSurfStyle=~
    bUseLeanPlayRateAdjust=~
    bUseForwardAs_Y_BS=~
    bUseDeployCurveForIntoPlayRateAdjust=~
    LeanForwardValue=~
    IdleValue=~
    LeanBackValue=~
    IdleInterpToSpeed=~
    LeanForwardInterpToSpeed=~
    LeanBackInterpToSpeed=~
   )
   n2 = "Blend Poses (EFortPlayerAnimBodyType)"!(
    BlendPose_0=Sequence Player(Sequence=@n1.LeanAdditive_Center_MaleMedium_Pose)
    BlendPose_1=Sequence Player(Sequence=@n1.LeanAdditive_Center_MaleLarge_Pose)
    BlendTime_0=0.1
    BlendTime_1=0.1
    ActiveEnumValue=Get AnimBodyType() -> (AnimBodyType=Small)
   ) -> (Pose) {blendType=Linear, blendPose [{linkId=-1, sourceLinkId=-1}, {linkId=-1, sourceLinkId=-1}], blendTime [0.1,0.1]}
   n3 = "Blend Poses (EFortPlayerAnimBodyType)"!(
    BlendPose_0=Sequence Player(Sequence=@n1.LeanAdditive_Center_FemaleMedium_Pose)
    BlendPose_1=Sequence Player(Sequence=@n1.LeanAdditive_Center_FemaleSmall_Pose)
    BlendPose_2:Pose Link Structure=Sequence Player(Sequence=@n1.LeanAdditive_Center_FemaleLarge_Pose)
    BlendTime_0=0.1
    BlendTime_1=0.1
    BlendTime_2:float=0.1
    ActiveEnumValue=Get AnimBodyType() -> (AnimBodyType=Small)
   ) -> (Pose) {
    blendType=Linear
    blendPose [{linkId=-1, sourceLinkId=-1}, {linkId=-1, sourceLinkId=-1}, {linkId=-1, sourceLinkId=-1}]
    blendTime [0.1,0.1,0.1]
   }
   "Blend Poses (EFortCustomGender)"(BlendPose_0=@n2, BlendPose_1=@n3, BlendTime_0=0, BlendTime_1=0, ActiveEnumValue=Get Gender() -> (Gender=Invalid)) {blendType=Linear, blendPose [{linkId=-1, sourceLinkId=-1}, {linkId=-1, sourceLinkId=-1}], blendTime [0.1,0.1]}
   Output Animation Pose(Result=Additive Identity Pose())
  ]
 }
 Left_Into {
  path=AnimGraph.AnimGraphNode_StateMachine_0.SkyDiveState.AnimStateNode_1.DiveOrGlide.AnimGraphNode_StateMachine_212.SkydiveOrGlide.AnimStateNode_3.ParaGlide.AnimGraphNode_StateMachine_0.NoGlider_Lean.AnimStateNode_17.Lean_Into.AnimGraphNode_StateMachine_0.Lean_Into.AnimStateNode_3.Left_Into
  nodes [
   n1 = Get CurrentGliderAnimSet()
   Output Animation Pose(Result=Sequence Player!(
     Sequence=Break Fort Anim Input Player Glider Anim Asset(FortAnimInput_PlayerGliderAnimAsset=@n1) -> (
      RootModPitchMin=~
      RootModPitchMax=~
      RootModYOffsetMin=~
      RootModYOffsetMax=~
      bEnableSpringMods=~
      bAllowPlayerDeployRootMod=~
      bUseSurfStyle=~
      bUseLeanPlayRateAdjust=~
      bUseForwardAs_Y_BS=~
      bUseDeployCurveForIntoPlayRateAdjust=~
      LeanForwardValue=~
      IdleValue=~
      LeanBackValue=~
      IdleInterpToSpeed=~
      LeanForwardInterpToSpeed=~
      LeanBackInterpToSpeed=~
     ).LeanAdditive_LeftInto_Anim
     PlayRate:float=1
    ) -> (Pose) {bLoopAnimation=false})
  ]
 }
 Right_Into {
  path=AnimGraph.AnimGraphNode_StateMachine_0.SkyDiveState.AnimStateNode_1.DiveOrGlide.AnimGraphNode_StateMachine_212.SkydiveOrGlide.AnimStateNode_3.ParaGlide.AnimGraphNode_StateMachine_0.NoGlider_Lean.AnimStateNode_17.Lean_Into.AnimGraphNode_StateMachine_0.Lean_Into.AnimStateNode_4.Right_Into
  nodes [
   n1 = Get CurrentGliderAnimSet()
   Output Animation Pose(Result=Sequence Player(
     Sequence=Break Fort Anim Input Player Glider Anim Asset(FortAnimInput_PlayerGliderAnimAsset=@n1) -> (
      RootModPitchMin=~
      RootModPitchMax=~
      RootModYOffsetMin=~
      RootModYOffsetMax=~
      bEnableSpringMods=~
      bAllowPlayerDeployRootMod=~
      bUseSurfStyle=~
      bUseLeanPlayRateAdjust=~
      bUseForwardAs_Y_BS=~
      bUseDeployCurveForIntoPlayRateAdjust=~
      LeanForwardValue=~
      IdleValue=~
      LeanBackValue=~
      IdleInterpToSpeed=~
      LeanForwardInterpToSpeed=~
      LeanBackInterpToSpeed=~
     ).LeanAdditive_RightInto_Anim
     PlayRate=1
    ) {bLoopAnimation=false, playRate=0.75})
  ]
 }
 Forward_Into {
  path=AnimGraph.AnimGraphNode_StateMachine_0.SkyDiveState.AnimStateNode_1.DiveOrGlide.AnimGraphNode_StateMachine_212.SkydiveOrGlide.AnimStateNode_3.ParaGlide.AnimGraphNode_StateMachine_0.NoGlider_Lean.AnimStateNode_17.Lean_Into.AnimGraphNode_StateMachine_0.Lean_Into.AnimStateNode_5.Forward_Into
  nodes [
   n1 = Break Fort Anim Input Player Glider Anim Asset(FortAnimInput_PlayerGliderAnimAsset=Get CurrentGliderAnimSet()) -> (
    RootModPitchMin=~
    RootModPitchMax=~
    RootModYOffsetMin=~
    RootModYOffsetMax=~
    bEnableSpringMods=~
    bAllowPlayerDeployRootMod=~
    bUseSurfStyle=~
    bUseLeanPlayRateAdjust=~
    bUseForwardAs_Y_BS=~
    bUseDeployCurveForIntoPlayRateAdjust=~
    LeanForwardValue=~
    IdleValue=~
    LeanBackValue=~
    IdleInterpToSpeed=~
    LeanForwardInterpToSpeed=~
    LeanBackInterpToSpeed=~
   )
   Output Animation Pose(Result=Sequence Player(Sequence=@n1.LeanAdditive_ForwardInto_Anim, PlayRate=1) {bLoopAnimation=false, groupName=LeanForward_Into, method=SyncGroup})
   Two Way Blend(
    B=Sequence Player(Sequence=@n1.LeanAdditive_ForwardInto_FromDeploy_Anim, PlayRate=1) {bLoopAnimation=false, groupName=LeanForward_Into, method=SyncGroup}
    AlphaCurveName=ParaGlideDeploy
   ) {blendNode {alphaInputType=Curve, alphaBoolBlend {blendOption=Linear}}}
  ]
 }
 Back_Into {
  path=AnimGraph.AnimGraphNode_StateMachine_0.SkyDiveState.AnimStateNode_1.DiveOrGlide.AnimGraphNode_StateMachine_212.SkydiveOrGlide.AnimStateNode_3.ParaGlide.AnimGraphNode_StateMachine_0.NoGlider_Lean.AnimStateNode_17.Lean_Into.AnimGraphNode_StateMachine_0.Lean_Into.AnimStateNode_6.Back_Into
  nodes [
   n1 = Get CurrentGliderAnimSet()
   Output Animation Pose(Result=Sequence Player(
     Sequence=Break Fort Anim Input Player Glider Anim Asset(FortAnimInput_PlayerGliderAnimAsset=@n1) -> (
      RootModPitchMin=~
      RootModPitchMax=~
      RootModYOffsetMin=~
      RootModYOffsetMax=~
      bEnableSpringMods=~
      bAllowPlayerDeployRootMod=~
      bUseSurfStyle=~
      bUseLeanPlayRateAdjust=~
      bUseForwardAs_Y_BS=~
      bUseDeployCurveForIntoPlayRateAdjust=~
      LeanForwardValue=~
      IdleValue=~
      LeanBackValue=~
      IdleInterpToSpeed=~
      LeanForwardInterpToSpeed=~
      LeanBackInterpToSpeed=~
     ).LeanAdditive_BackInto_Anim
     PlayRate=1
    ) {bLoopAnimation=false})
  ]
 }
 Transition {
  path=AnimGraph.AnimGraphNode_StateMachine_0.SkyDiveState.AnimStateNode_1.DiveOrGlide.AnimGraphNode_StateMachine_212.SkydiveOrGlide.AnimStateNode_3.ParaGlide.AnimGraphNode_StateMachine_0.NoGlider_Lean.AnimStateNode_17.Lean_Into.AnimGraphNode_StateMachine_0.Lean_Into.AnimStateTransitionNode_19.Transition
  nodes [Result(bCanEnterTransition=Break Fort Anim Input Skydiving!(FortAnimInput_Skydiving=Get Skydiving()) -> (
     bIsGliderRight=~
     bIsGliderCenter=~
     bIsGliderLeft=~
     bIsGliderForward=~
     bIsGliderBack=~
     bHasDeployedAndIsStandingUpright=~
     LocalAccelDir
     PoseAdditive
    ).bIsGliderForward)]
 }
 Transition {
  path=AnimGraph.AnimGraphNode_StateMachine_0.SkyDiveState.AnimStateNode_1.DiveOrGlide.AnimGraphNode_StateMachine_212.SkydiveOrGlide.AnimStateNode_3.ParaGlide.AnimGraphNode_StateMachine_0.NoGlider_Lean.AnimStateNode_17.Lean_Into.AnimGraphNode_StateMachine_0.Lean_Into.AnimStateTransitionNode_20.Transition
  nodes [Result(bCanEnterTransition=Break Fort Anim Input Skydiving(FortAnimInput_Skydiving=Get Skydiving()) -> (bIsGliderRight=~, bIsGliderCenter=~, bIsGliderLeft=~, bIsGliderForward=~, bIsGliderBack=~, bHasDeployedAndIsStandingUpright=~).bIsGliderBack)]
 }
 Transition {
  path=AnimGraph.AnimGraphNode_StateMachine_0.SkyDiveState.AnimStateNode_1.DiveOrGlide.AnimGraphNode_StateMachine_212.SkydiveOrGlide.AnimStateNode_3.ParaGlide.AnimGraphNode_StateMachine_0.NoGlider_Lean.AnimStateNode_17.Lean_Into.AnimGraphNode_StateMachine_0.Lean_Into.AnimStateTransitionNode_21.Transition
  nodes [
   n1 = Break Fort Anim Input Skydiving(FortAnimInput_Skydiving=Get Skydiving()) -> (bIsGliderRight=~, bIsGliderCenter=~, bIsGliderLeft=~, bIsGliderForward=~, bIsGliderBack=~, bHasDeployedAndIsStandingUpright=~)
   Result(bCanEnterTransition=@n1)
   "Equal (Enum)"(A:ESkydivingDirection Enum=@n1.LocalAccelDir, B:ESkydivingDirection Enum=Right)
  ]
 }
 Transition {
  path=AnimGraph.AnimGraphNode_StateMachine_0.SkyDiveState.AnimStateNode_1.DiveOrGlide.AnimGraphNode_StateMachine_212.SkydiveOrGlide.AnimStateNode_3.ParaGlide.AnimGraphNode_StateMachine_0.NoGlider_Lean.AnimStateNode_17.Lean_Into.AnimGraphNode_StateMachine_0.Lean_Into.AnimStateTransitionNode_22.Transition
  nodes [Result(bCanEnterTransition=Break Fort Anim Input Skydiving(FortAnimInput_Skydiving=Get Skydiving()) -> (bIsGliderRight=~, bIsGliderCenter=~, bIsGliderLeft=~, bIsGliderForward=~, bIsGliderBack=~, bHasDeployedAndIsStandingUpright=~).bIsGliderLeft)]
 }
 Transition {
  path=AnimGraph.AnimGraphNode_StateMachine_0.SkyDiveState.AnimStateNode_1.DiveOrGlide.AnimGraphNode_StateMachine_212.SkydiveOrGlide.AnimStateNode_3.ParaGlide.AnimGraphNode_StateMachine_0.NoGlider_Lean.AnimStateNode_17.Lean_Into.AnimGraphNode_StateMachine_0.Lean_Into.AnimStateTransitionNode_23.Transition
  nodes [Result(bCanEnterTransition=Break Fort Anim Input Skydiving(FortAnimInput_Skydiving=Get Skydiving()) -> (bIsGliderRight=~, bIsGliderCenter=~, bIsGliderLeft=~, bIsGliderForward=~, bIsGliderBack=~, bHasDeployedAndIsStandingUpright=~).bIsGliderLeft)]
 }
 Transition {
  path=AnimGraph.AnimGraphNode_StateMachine_0.SkyDiveState.AnimStateNode_1.DiveOrGlide.AnimGraphNode_StateMachine_212.SkydiveOrGlide.AnimStateNode_3.ParaGlide.AnimGraphNode_StateMachine_0.NoGlider_Lean.AnimStateNode_17.Lean_Into.AnimGraphNode_StateMachine_0.Lean_Into.AnimStateTransitionNode_24.Transition
  nodes [
   n1 = Break Fort Anim Input Skydiving(FortAnimInput_Skydiving=Get Skydiving()) -> (bIsGliderRight=~, bIsGliderCenter=~, bIsGliderLeft=~, bIsGliderForward=~, bIsGliderBack=~, bHasDeployedAndIsStandingUpright=~)
   Result(bCanEnterTransition=@n1)
   "Equal (Enum)"(A=@n1.LocalAccelDir, B=Right)
  ]
 }
 Transition {
  path=AnimGraph.AnimGraphNode_StateMachine_0.SkyDiveState.AnimStateNode_1.DiveOrGlide.AnimGraphNode_StateMachine_212.SkydiveOrGlide.AnimStateNode_3.ParaGlide.AnimGraphNode_StateMachine_0.NoGlider_Lean.AnimStateNode_17.Lean_Into.AnimGraphNode_StateMachine_0.Lean_Into.AnimStateTransitionNode_25.Transition
  nodes [Result(bCanEnterTransition=Break Fort Anim Input Skydiving(FortAnimInput_Skydiving=Get Skydiving()) -> (bIsGliderRight=~, bIsGliderCenter=~, bIsGliderLeft=~, bIsGliderForward=~, bIsGliderBack=~, bHasDeployedAndIsStandingUpright=~).bIsGliderForward)]
 }
 Transition {
  path=AnimGraph.AnimGraphNode_StateMachine_0.SkyDiveState.AnimStateNode_1.DiveOrGlide.AnimGraphNode_StateMachine_212.SkydiveOrGlide.AnimStateNode_3.ParaGlide.AnimGraphNode_StateMachine_0.NoGlider_Lean.AnimStateNode_17.Lean_Into.AnimGraphNode_StateMachine_0.Lean_Into.AnimStateTransitionNode_26.Transition
  nodes [Result(bCanEnterTransition=Break Fort Anim Input Skydiving(FortAnimInput_Skydiving=Get Skydiving()) -> (bIsGliderRight=~, bIsGliderCenter=~, bIsGliderLeft=~, bIsGliderForward=~, bIsGliderBack=~, bHasDeployedAndIsStandingUpright=~).bIsGliderBack)]
 }
 Transition {
  path=AnimGraph.AnimGraphNode_StateMachine_0.SkyDiveState.AnimStateNode_1.DiveOrGlide.AnimGraphNode_StateMachine_212.SkydiveOrGlide.AnimStateNode_3.ParaGlide.AnimGraphNode_StateMachine_0.NoGlider_Lean.AnimStateNode_17.Lean_Into.AnimGraphNode_StateMachine_0.Lean_Into.AnimStateTransitionNode_27.Transition
  nodes [Result(bCanEnterTransition=Break Fort Anim Input Skydiving(FortAnimInput_Skydiving=Get Skydiving()) -> (bIsGliderRight=~, bIsGliderCenter=~, bIsGliderLeft=~, bIsGliderForward=~, bIsGliderBack=~, bHasDeployedAndIsStandingUpright=~).bIsGliderForward)]
 }
 Transition {
  path=AnimGraph.AnimGraphNode_StateMachine_0.SkyDiveState.AnimStateNode_1.DiveOrGlide.AnimGraphNode_StateMachine_212.SkydiveOrGlide.AnimStateNode_3.ParaGlide.AnimGraphNode_StateMachine_0.NoGlider_Lean.AnimStateNode_17.Lean_Into.AnimGraphNode_StateMachine_0.Lean_Into.AnimStateTransitionNode_28.Transition
  nodes [Result(bCanEnterTransition=Break Fort Anim Input Skydiving(FortAnimInput_Skydiving=Get Skydiving()) -> (bIsGliderRight=~, bIsGliderCenter=~, bIsGliderLeft=~, bIsGliderForward=~, bIsGliderBack=~, bHasDeployedAndIsStandingUpright=~).bIsGliderBack)]
 }
 Transition {
  path=AnimGraph.AnimGraphNode_StateMachine_0.SkyDiveState.AnimStateNode_1.DiveOrGlide.AnimGraphNode_StateMachine_212.SkydiveOrGlide.AnimStateNode_3.ParaGlide.AnimGraphNode_StateMachine_0.NoGlider_Lean.AnimStateNode_17.Lean_Into.AnimGraphNode_StateMachine_0.Lean_Into.AnimStateTransitionNode_29.Transition
  nodes [Result(bCanEnterTransition=Break Fort Anim Input Skydiving(FortAnimInput_Skydiving=Get Skydiving()) -> (bIsGliderRight=~, bIsGliderCenter=~, bIsGliderLeft=~, bIsGliderForward=~, bIsGliderBack=~, bHasDeployedAndIsStandingUpright=~).bIsGliderLeft)]
 }
 Transition {
  path=AnimGraph.AnimGraphNode_StateMachine_0.SkyDiveState.AnimStateNode_1.DiveOrGlide.AnimGraphNode_StateMachine_212.SkydiveOrGlide.AnimStateNode_3.ParaGlide.AnimGraphNode_StateMachine_0.NoGlider_Lean.AnimStateNode_17.Lean_Into.AnimGraphNode_StateMachine_0.Lean_Into.AnimStateTransitionNode_30.Transition
  nodes [
   n1 = Break Fort Anim Input Skydiving(FortAnimInput_Skydiving=Get Skydiving()) -> (bIsGliderRight=~, bIsGliderCenter=~, bIsGliderLeft=~, bIsGliderForward=~, bIsGliderBack=~, bHasDeployedAndIsStandingUpright=~)
   Result(bCanEnterTransition=@n1)
   "Equal (Enum)"(A=@n1.LocalAccelDir, B=Right)
  ]
 }
 Transition {
  path=AnimGraph.AnimGraphNode_StateMachine_0.SkyDiveState.AnimStateNode_1.DiveOrGlide.AnimGraphNode_StateMachine_212.SkydiveOrGlide.AnimStateNode_3.ParaGlide.AnimGraphNode_StateMachine_0.NoGlider_Lean.AnimStateNode_17.Lean_Into.AnimGraphNode_StateMachine_0.Lean_Into.AnimStateTransitionNode_31.Transition
  nodes [Result(bCanEnterTransition=Break Fort Anim Input Skydiving(FortAnimInput_Skydiving=Get Skydiving()) -> (bIsGliderRight=~, bIsGliderCenter=~, bIsGliderLeft=~, bIsGliderForward=~, bIsGliderBack=~, bHasDeployedAndIsStandingUpright=~).bIsGliderBack)]
 }
 Transition {
  path=AnimGraph.AnimGraphNode_StateMachine_0.SkyDiveState.AnimStateNode_1.DiveOrGlide.AnimGraphNode_StateMachine_212.SkydiveOrGlide.AnimStateNode_3.ParaGlide.AnimGraphNode_StateMachine_0.NoGlider_Lean.AnimStateNode_17.Lean_Into.AnimGraphNode_StateMachine_0.Lean_Into.AnimStateTransitionNode_32.Transition
  nodes [Result(bCanEnterTransition=Break Fort Anim Input Skydiving(FortAnimInput_Skydiving=Get Skydiving()) -> (bIsGliderRight=~, bIsGliderCenter=~, bIsGliderLeft=~, bIsGliderForward=~, bIsGliderBack=~, bHasDeployedAndIsStandingUpright=~).bIsGliderForward)]
 }
 Transition {
  path=AnimGraph.AnimGraphNode_StateMachine_0.SkyDiveState.AnimStateNode_1.DiveOrGlide.AnimGraphNode_StateMachine_212.SkydiveOrGlide.AnimStateNode_3.ParaGlide.AnimGraphNode_StateMachine_0.NoGlider_Lean.AnimStateNode_17.Lean_Into.AnimGraphNode_StateMachine_0.Lean_Into.AnimStateTransitionNode_33.Transition
  nodes [Result(bCanEnterTransition=Break Fort Anim Input Skydiving(FortAnimInput_Skydiving=Get Skydiving()) -> (bIsGliderRight=~, bIsGliderCenter=~, bIsGliderLeft=~, bIsGliderForward=~, bIsGliderBack=~, bHasDeployedAndIsStandingUpright=~).bIsGliderLeft)]
 }
 Transition {
  path=AnimGraph.AnimGraphNode_StateMachine_0.SkyDiveState.AnimStateNode_1.DiveOrGlide.AnimGraphNode_StateMachine_212.SkydiveOrGlide.AnimStateNode_3.ParaGlide.AnimGraphNode_StateMachine_0.NoGlider_Lean.AnimStateNode_17.Lean_Into.AnimGraphNode_StateMachine_0.Lean_Into.AnimStateTransitionNode_34.Transition
  nodes [
   n1 = Break Fort Anim Input Skydiving(FortAnimInput_Skydiving=Get Skydiving()) -> (bIsGliderRight=~, bIsGliderCenter=~, bIsGliderLeft=~, bIsGliderForward=~, bIsGliderBack=~, bHasDeployedAndIsStandingUpright=~)
   Result(bCanEnterTransition=@n1)
   "Equal (Enum)"(A=@n1.LocalAccelDir, B=Right)
  ]
 }
 AnimationTransitionGraph_1 {
  path=AnimGraph.AnimGraphNode_StateMachine_0.SkyDiveState.AnimStateNode_1.DiveOrGlide.AnimGraphNode_StateMachine_212.SkydiveOrGlide.AnimStateNode_3.ParaGlide.AnimGraphNode_StateMachine_0.NoGlider_Lean.AnimStateTransitionNode_25.AnimationTransitionGraph_1
  nodes [
   n1 = NOT Boolean(A=Get bIsAccelerating2D())
   Result(bCanEnterTransition=@n1)
   n2 = Break Fort Anim Input Skydiving(FortAnimInput_Skydiving=Get Skydiving()) -> (bIsGliderRight=~, bIsGliderCenter=~, bIsGliderLeft=~, bIsGliderForward=~, bIsGliderBack=~, bHasDeployedAndIsStandingUpright=~)
   OR Boolean(A=@n2.bIsGliderCenter, B=@n1)
   "Equal (Enum)"(A=@n2.LocalAccelDir, B=Center)
  ]
 }
 AnimationTransitionGraph_1 {
  path=AnimGraph.AnimGraphNode_StateMachine_0.SkyDiveState.AnimStateNode_1.DiveOrGlide.AnimGraphNode_StateMachine_212.SkydiveOrGlide.AnimStateNode_3.ParaGlide.AnimGraphNode_StateMachine_0.NoGlider_Lean.AnimStateTransitionNode_26.AnimationTransitionGraph_1
  nodes [Result()]
 }
 Transition {
  path=AnimGraph.AnimGraphNode_StateMachine_0.SkyDiveState.AnimStateNode_1.DiveOrGlide.AnimGraphNode_StateMachine_212.SkydiveOrGlide.AnimStateNode_3.ParaGlide.AnimGraphNode_StateMachine_0.NoGlider_Lean.AnimStateTransitionNode_37.Transition
  nodes [
   n1 = Get bIsAccelerating2D()
   Result(bCanEnterTransition=@n1)
   n2 = Get Skydiving()
   OR Boolean(
    A="Equal (Enum)"(
     A=Break Fort Anim Input Skydiving(FortAnimInput_Skydiving=@n2) -> (bIsGliderRight=~, bIsGliderCenter=~, bIsGliderLeft=~, bIsGliderForward=~, bIsGliderBack=~, bHasDeployedAndIsStandingUpright=~).LocalAccelDir
     B=Center
    )
    B=NOT Boolean(A=@n1)
   )
  ]
 }
 AnimationTransitionGraph_1 {
  path=AnimGraph.AnimGraphNode_StateMachine_0.SkyDiveState.AnimStateNode_1.DiveOrGlide.AnimGraphNode_StateMachine_212.SkydiveOrGlide.AnimStateNode_3.ParaGlide.AnimGraphNode_StateMachine_0.NoGlider_Lean.AnimStateTransitionNode_38.AnimationTransitionGraph_1
  nodes [
   n1 = NOT Boolean(A=Get bIsAccelerating2D())
   Result(bCanEnterTransition=@n1)
   n2 = Break Fort Anim Input Skydiving(FortAnimInput_Skydiving=Get Skydiving()) -> (bIsGliderRight=~, bIsGliderCenter=~, bIsGliderLeft=~, bIsGliderForward=~, bIsGliderBack=~, bHasDeployedAndIsStandingUpright=~)
   OR Boolean(A=@n2.bIsGliderCenter, B=@n1)
   "Equal (Enum)"(A=@n2.LocalAccelDir, B=Center)
  ]
 }
 AnimationTransitionGraph_1 {
  path=AnimGraph.AnimGraphNode_StateMachine_0.SkyDiveState.AnimStateNode_1.DiveOrGlide.AnimGraphNode_StateMachine_212.SkydiveOrGlide.AnimStateNode_3.ParaGlide.AnimGraphNode_StateMachine_0.NoGlider_Lean.AnimStateTransitionNode_39.AnimationTransitionGraph_1
  nodes [Result(bCanEnterTransition=Get BlockAnimRule() -> (BlockAnimRule:bool))]
 }
 Transition {
  path=AnimGraph.AnimGraphNode_StateMachine_0.SkyDiveState.AnimStateNode_1.DiveOrGlide.AnimGraphNode_StateMachine_212.SkydiveOrGlide.AnimStateNode_3.ParaGlide.AnimGraphNode_StateMachine_0.NoGlider_Lean.AnimStateTransitionNode_40.Transition
  nodes [Result(bCanEnterTransition=Get bIsAccelerating2D())]
 }
 Glider_SurfStyleAlt {
  entry=ParaGlide_Into_Deploy
  path=AnimGraph.AnimGraphNode_StateMachine_0.SkyDiveState.AnimStateNode_1.DiveOrGlide.AnimGraphNode_StateMachine_212.SkydiveOrGlide.AnimStateNode_3.ParaGlide.AnimGraphNode_StateMachine_1.Glider_SurfStyleAlt
  ids=id
  nodes [
   AnimStateNode_0 = ParaGlide_SurfStyle() {is=title AnimStateNode}
   AnimStateNode_1 = ParaGlide_Into_Deploy() {is=title AnimStateNode}
   AnimStateTransitionNode_1 = AnimStateTransitionNode() {bAutomaticRuleBasedOnSequencePlayerInState=true, blendMode=Linear, crossfadeDuration=0}
   AnimStateTransitionNode_3 = AnimStateTransitionNode() {crossfadeDuration=0}
  ]
  links [
   {
    dst=ParaGlide_SurfStyle
    rule=AnimGraph.AnimGraphNode_StateMachine_0.SkyDiveState.AnimStateNode_1.DiveOrGlide.AnimGraphNode_StateMachine_212.SkydiveOrGlide.AnimStateNode_3.ParaGlide.AnimGraphNode_StateMachine_1.Glider_SurfStyleAlt.AnimStateTransitionNode_1.AnimationTransitionGraph_1
    src=ParaGlide_Into_Deploy
    via=AnimStateTransitionNode_1
   }
   {
    dst=ParaGlide_SurfStyle
    rule=AnimGraph.AnimGraphNode_StateMachine_0.SkyDiveState.AnimStateNode_1.DiveOrGlide.AnimGraphNode_StateMachine_212.SkydiveOrGlide.AnimStateNode_3.ParaGlide.AnimGraphNode_StateMachine_1.Glider_SurfStyleAlt.AnimStateTransitionNode_3.Transition
    src=ParaGlide_Into_Deploy
    via=AnimStateTransitionNode_3
   }
  ]
 }
 ParaGlide_SurfStyle {
  path=AnimGraph.AnimGraphNode_StateMachine_0.SkyDiveState.AnimStateNode_1.DiveOrGlide.AnimGraphNode_StateMachine_212.SkydiveOrGlide.AnimStateNode_3.ParaGlide.AnimGraphNode_StateMachine_1.Glider_SurfStyleAlt.AnimStateNode_0.ParaGlide_SurfStyle
  nodes [
   n1 = Break Fort Anim Input Player Glider Anim Asset(FortAnimInput_PlayerGliderAnimAsset=Get CurrentGliderAnimSet()) -> (
    RootModPitchMin=~
    RootModPitchMax=~
    RootModYOffsetMin=~
    RootModYOffsetMax=~
    bEnableSpringMods=~
    bAllowPlayerDeployRootMod=~
    bUseSurfStyle=~
    bUseLeanPlayRateAdjust=~
    bUseForwardAs_Y_BS=~
    bUseDeployCurveForIntoPlayRateAdjust=~
    LeanForwardValue=~
    IdleValue=~
    LeanBackValue=~
    IdleInterpToSpeed=~
    LeanForwardInterpToSpeed=~
    LeanBackInterpToSpeed=~
   )
   n2 = Break Fort Anim Input Skydiving!(FortAnimInput_Skydiving=Get Skydiving()) -> (
    ParaGilderRootOffset:Vector=~
    ParaGliderRootRotationOverride:Rotator=~
    bUseParaGlideRootModifier:bool=~
    bIsSkydivingFromLaunchPad:bool=~
    bIsSkydivingFromBus:bool=~
    bIsInVortex:bool=~
    bIsDiving:bool=~
    bIsParachuteOpen:bool=~
    bIsSkydiving:bool=~
    bIsParachuteLeaning:bool=~
    bIsSkydiveLeaning:bool=~
    bIsLeaning:bool=~
    bIsSkydiveDiveMode:bool=~
    bPlaySkydiveDrift:bool=~
    bSkydiveDriftDelayActive:bool=~
    bSkydiveDriftAnimAllowed:bool=~
    bIsGliderRight=~
    bIsGliderCenter=~
    bIsGliderLeft=~
    bIsGliderForward=~
    bIsGliderBack=~
    bHasDeployedAndIsStandingUpright=~
    LocalAccelForward:float=~
    LocalAccelRight:float=~
    LocalVelocityRight:float=~
    SkydiveAimPitch:float=~
    SkydiveAimYaw:float=~
    DeployChuteAnimRate:float=~
    SkydiveDriftAnimRate:float=~
    SkydiveDriftAnimRateCurrent:float=~
    SkydiveFidgetAnimRate:float=~
    SkydiveFidgetAnimRateCurrent:float=~
    SkydiveAdditiveAlpha:float=~
    SkydiveDriftDelay:float=~
    SkydiveDriftAnim:int
    SkydiveDriftAnimMax:int
    LaunchpadAnim:int
    LocalAccelDir
    DirectionLast:ESkydivingDirection Enum
    PoseAdditive
   )
   Apply Additive!(
    Base
    Additive=Blendspace Player(X=@n2.SkydiveAimYaw, Y=@n2.LocalAccelForward, BlendSpace=@n1.Default_BodyAdditive_FemaleMedium_BS) {bResetPlayTimeWhenBlendSpaceChanges=false, groupName=SurfGlide, groupRole=AlwaysFollower, method=SyncGroup}
    Alpha:float=Get ParaGliderAddAlpha() -> (ParaGliderAddAlpha:float)
   ) -> (Pose) {alphaBoolBlend {blendOption=Linear}, alphaScaleBiasClamp {outRange {max=0, min=1}}}
   n3 = Blendspace Player(X=@n2.SkydiveAimYaw, Y=@n2.LocalAccelForward, BlendSpace=@n1.Default_TurnAdditive_FemaleMedium_BS)
   n4 = Blendspace Player(X=@n2.LocalAccelRight, Y=@n2.LocalAccelForward, BlendSpace=@n1.Lean_FemaleMedium_BS) {groupName=Glide, groupRole=AlwaysFollower, method=SyncGroup}
   Output Animation Pose(Result=Apply Additive!(
     Base=Blendspace Player(X=@n2.SkydiveAimYaw, Y=@n2.LocalAccelForward, BlendSpace=@n1) {groupName=Glide, groupRole=AlwaysLeader, method=SyncGroup}
     Additive=Apply Additive!(Base=@n3, Additive=@n4, Alpha:float) -> (Pose) {alphaBoolBlend {blendOption=Linear}, alphaScaleBiasClamp {outRange {max=0, min=1}}}
     AlphaCurveName:Name=ParaGlideLeanAlpha
    ) -> (Pose) {alphaInputType=Curve, alphaBoolBlend {blendOption=Linear}})
  ]
 }
 ParaGlide_Into_Deploy {
  path=AnimGraph.AnimGraphNode_StateMachine_0.SkyDiveState.AnimStateNode_1.DiveOrGlide.AnimGraphNode_StateMachine_212.SkydiveOrGlide.AnimStateNode_3.ParaGlide.AnimGraphNode_StateMachine_1.Glider_SurfStyleAlt.AnimStateNode_1.ParaGlide_Into_Deploy
  nodes [
   n1 = Break Fort Anim Input Player Glider Anim Asset!(FortAnimInput_PlayerGliderAnimAsset=Get CurrentGliderAnimSet()) -> (
    Into_Base_MaleMedium_BS
    Into_BodyAdditive_MaleMedium_BS
    Into_BodyAdditive_FemaleMedium_BS
    PelvisModVertical
    PelvisModLateral
    PelvisModBankAngle
    ThighModLegAngle
    ThighModLegPitch
    ThighModLegBank
    ThighModLeftLegDrag
    PlayerGliderType
    bEnableSpringMods=~
    bUseLeanPlayRateAdjust=~
    bUseForwardAs_Y_BS=~
    bUseDeployCurveForIntoPlayRateAdjust=~
    LeanForwardValue=~
    IdleValue=~
    LeanBackValue=~
    IdleInterpToSpeed=~
    LeanForwardInterpToSpeed=~
    LeanBackInterpToSpeed=~
   )
   n2 = Break Fort Anim Input Skydiving(FortAnimInput_Skydiving=Get Skydiving()) -> (
    ParaGilderRootOffset=~
    ParaGliderRootRotationOverride=~
    bUseParaGlideRootModifier=~
    bIsSkydivingFromLaunchPad=~
    bIsSkydivingFromBus=~
    bIsInVortex=~
    bIsDiving=~
    bIsParachuteOpen=~
    bIsSkydiving=~
    bIsParachuteLeaning=~
    bIsSkydiveLeaning=~
    bIsLeaning=~
    bIsSkydiveDiveMode=~
    bPlaySkydiveDrift=~
    bSkydiveDriftDelayActive=~
    bSkydiveDriftAnimAllowed=~
    bIsGliderRight=~
    bIsGliderCenter=~
    bIsGliderLeft=~
    bIsGliderForward=~
    bIsGliderBack=~
    bHasDeployedAndIsStandingUpright=~
    LocalAccelForward=~
    LocalAccelRight=~
    LocalVelocityRight=~
    SkydiveAimPitch=~
    SkydiveAimYaw=~
    DeployChuteAnimRate=~
    SkydiveDriftAnimRate=~
    SkydiveDriftAnimRateCurrent=~
    SkydiveFidgetAnimRate=~
    SkydiveFidgetAnimRateCurrent=~
    SkydiveAdditiveAlpha=~
    SkydiveDriftDelay=~
   )
   n3 = "Blend Poses (EFortCustomGender)"(
    BlendPose_0=Blendspace Player(X=@n2.SkydiveAimYaw, Y=@n2.SkydiveAimPitch, BlendSpace=@n1.Into_BodyAdditive_MaleMedium_BS) {bIgnoreForRelevancyTest=true, bLoop=false}
    BlendPose_1=Blendspace Player(X=@n2.SkydiveAimYaw, Y=@n2.SkydiveAimPitch, BlendSpace=@n1.Into_BodyAdditive_FemaleMedium_BS) {bIgnoreForRelevancyTest=true, bLoop=false}
    BlendTime_0=0
    BlendTime_1=0
    ActiveEnumValue=Get Gender() -> (Gender=Invalid)
   ) {blendType=Linear, blendPose [{linkId=-1, sourceLinkId=-1}, {linkId=-1, sourceLinkId=-1}], blendTime [0.1,0.1]}
   n4 = Apply Additive!(Base, Additive=@n3, AlphaCurveName:Name=ParaGlideLeanAlpha) -> (Pose) {alphaInputType=Curve, alphaBoolBlend {blendOption=Linear}, alphaScaleBiasClamp {bMapRange=true, inRange {max=0, min=1}}}
   n5 = Blendspace Player(
    X=@n2.SkydiveAimYaw
    Y=@n2.LocalAccelForward
    BlendSpace=Break Fort Anim Input Player Glider Anim Asset!(FortAnimInput_PlayerGliderAnimAsset=Get CurrentGliderAnimSet()) -> (
     Default_BodyAdditive_FemaleMedium_BS:Blend Space Object Reference
     Into_Base_MaleMedium_BS
     PelvisModVertical
     PelvisModLateral
     PelvisModBankAngle
     ThighModLegAngle
     ThighModLegPitch
     ThighModLegBank
     ThighModLeftLegDrag
     PlayerGliderType
     bEnableSpringMods=~
     bUseLeanPlayRateAdjust=~
     bUseForwardAs_Y_BS=~
     bUseDeployCurveForIntoPlayRateAdjust=~
     LeanForwardValue=~
     IdleValue=~
     LeanBackValue=~
     IdleInterpToSpeed=~
     LeanForwardInterpToSpeed=~
     LeanBackInterpToSpeed=~
    )
   ) {bIgnoreForRelevancyTest=true, bResetPlayTimeWhenBlendSpaceChanges=false, groupName=SurfGlide, groupRole=AlwaysLeader, method=SyncGroup}
   Apply Additive(Base=@n4, Additive=@n5, AlphaCurveName=ParaGlideLeanAlpha) {alphaInputType=Curve, alphaBoolBlend {blendOption=Linear}, alphaScaleBiasClamp {inRange {max=0, min=1}}}
   n6 = Break Fort Anim Input Player Glider Anim Asset!(FortAnimInput_PlayerGliderAnimAsset=Get CurrentGliderAnimSet()) -> (
    Into_Base_BS:Blend Space Object Reference
    Into_Base_MaleMedium_BS
    PelvisModVertical
    PelvisModLateral
    PelvisModBankAngle
    ThighModLegAngle
    ThighModLegPitch
    ThighModLegBank
    ThighModLeftLegDrag
    PlayerGliderType
    bEnableSpringMods=~
    bUseLeanPlayRateAdjust=~
    bUseForwardAs_Y_BS=~
    bUseDeployCurveForIntoPlayRateAdjust=~
    LeanForwardValue=~
    IdleValue=~
    LeanBackValue=~
    IdleInterpToSpeed=~
    LeanForwardInterpToSpeed=~
    LeanBackInterpToSpeed=~
   )
   n7 = Break Fort Anim Input Player Glider Anim Asset!(FortAnimInput_PlayerGliderAnimAsset=Get CurrentGliderAnimSet()) -> (
    Default_Base_BS:Blend Space Object Reference
    Default_BodyAdditive_MaleMedium_BS:Blend Space Object Reference
    Default_BodyAdditive_MaleLarge_BS:Blend Space Object Reference
    Default_BodyAdditive_FemaleSmall_BS:Blend Space Object Reference
    Default_BodyAdditive_FemaleMedium_BS:Blend Space Object Reference
    Default_BodyAdditive_FemaleLarge_BS:Blend Space Object Reference
    Default_TurnAdditive_MaleMedium_BS:Blend Space Object Reference
    Default_TurnAdditive_MaleLarge_BS:Blend Space Object Reference
    Default_TurnAdditive_FemaleSmall_BS:Blend Space Object Reference
    Default_TurnAdditive_FemaleMedium_BS:Blend Space Object Reference
    Default_TurnAdditive_FemaleLarge_BS:Blend Space Object Reference
    Into_Base_BS
    Into_Base_MaleMedium_BS
    Into_BodyAdditive_MaleMedium_BS:Blend Space Object Reference
    Into_BodyAdditive_MaleLarge_BS:Blend Space Object Reference
    Into_BodyAdditive_FemaleSmall_BS:Blend Space Object Reference
    Into_BodyAdditive_FemaleMedium_BS:Blend Space Object Reference
    Into_BodyAdditive_FemaleLarge_BS:Blend Space Object Reference
    Into_TurnAdditive_MaleMedium_BS:Blend Space Object Reference
    Into_TurnAdditive_MaleLarge_BS:Blend Space Object Reference
    Into_TurnAdditive_FemaleSmall_BS:Blend Space Object Reference
    Into_TurnAdditive_FemaleMedium_BS:Blend Space Object Reference
    Into_TurnAdditive_FemaleLarge_BS:Blend Space Object Reference
    Lean_MaleMedium_BS:Blend Space Object Reference
    Lean_MaleLarge_BS:Blend Space Object Reference
    Lean_FemaleSmall_BS:Blend Space Object Reference
    Lean_FemaleMedium_BS:Blend Space Object Reference
    Lean_FemaleLarge_BS:Blend Space Object Reference
    LeanAdditive_Center_MaleMedium_Pose:Anim Sequence Object Reference
    LeanAdditive_Center_MaleLarge_Pose:Anim Sequence Object Reference
    LeanAdditive_Center_FemaleSmall_Pose:Anim Sequence Object Reference
    LeanAdditive_Center_FemaleMedium_Pose:Anim Sequence Object Reference
    LeanAdditive_Center_FemaleLarge_Pose:Anim Sequence Object Reference
    LeanAdditive_Into_BS:Blend Space Object Reference
    LeanAdditive_ForwardInto_Anim:Anim Sequence Object Reference
    LeanAdditive_ForwardInto_FromDeploy_Anim:Anim Sequence Object Reference
    LeanAdditive_BackInto_Anim:Anim Sequence Object Reference
    LeanAdditive_LeftInto_Anim:Anim Sequence Object Reference
    LeanAdditive_RightInto_Anim:Anim Sequence Object Reference
    LeanAdditive_ForwardOut_Anim:Anim Sequence Object Reference
    LeanAdditive_BackOut_Anim:Anim Sequence Object Reference
    LeanAdditive_LeftOut_Anim:Anim Sequence Object Reference
    LeanAdditive_RightOut_Anim:Anim Sequence Object Reference
    ToGlide_BS:Blend Space Object Reference
    ToGlide_Lean_BS:Blend Space Object Reference
    ToDive_BS:Blend Space Object Reference
    ToDive_Lean_BS:Blend Space Object Reference
    Dive_WeaponR_Additive_BS:Blend Space Object Reference
    Glide_WeaponR_Additive_Anim:Anim Sequence Object Reference
    GenericAdditive_Male_BS:Blend Space Object Reference
    GenericAdditive_Female_BS:Blend Space Object Reference
    RootModPitchMin:float=~
    RootModPitchMax:float=~
    RootModYOffsetMin:float=~
    RootModYOffsetMax:float=~
    PelvisModVertical
    PelvisModLateral
    PelvisModBankAngle
    ThighModLegAngle
    ThighModLegPitch
    ThighModLegBank
    ThighModLeftLegDrag
    PlayerGliderType
    bEnableSpringMods=~
    bAllowPlayerDeployRootMod:bool=~
    bUseSurfStyle:bool=~
    bUseLeanPlayRateAdjust=~
    bUseForwardAs_Y_BS=~
    bUseDeployCurveForIntoPlayRateAdjust=~
    LeanForwardValue=~
    IdleValue=~
    LeanBackValue=~
    IdleInterpToSpeed=~
    LeanForwardInterpToSpeed=~
    LeanBackInterpToSpeed=~
   )
   n8 = Break Fort Anim Input Skydiving(FortAnimInput_Skydiving=Get Skydiving()) -> (
    ParaGilderRootOffset=~
    ParaGliderRootRotationOverride=~
    bUseParaGlideRootModifier=~
    bIsSkydivingFromLaunchPad=~
    bIsSkydivingFromBus=~
    bIsInVortex=~
    bIsDiving=~
    bIsParachuteOpen=~
    bIsSkydiving=~
    bIsParachuteLeaning=~
    bIsSkydiveLeaning=~
    bIsLeaning=~
    bIsSkydiveDiveMode=~
    bPlaySkydiveDrift=~
    bSkydiveDriftDelayActive=~
    bSkydiveDriftAnimAllowed=~
    bIsGliderRight=~
    bIsGliderCenter=~
    bIsGliderLeft=~
    bIsGliderForward=~
    bIsGliderBack=~
    bHasDeployedAndIsStandingUpright=~
    LocalAccelForward=~
    LocalAccelRight=~
    LocalVelocityRight=~
    SkydiveAimPitch=~
    SkydiveAimYaw=~
    DeployChuteAnimRate=~
    SkydiveDriftAnimRate=~
    SkydiveDriftAnimRateCurrent=~
    SkydiveFidgetAnimRate=~
    SkydiveFidgetAnimRateCurrent=~
    SkydiveAdditiveAlpha=~
    SkydiveDriftDelay=~
   )
   Output Animation Pose(Result=Apply Additive(
     Base=Blendspace Player(X=@n2.SkydiveAimYaw, Y=@n2.LocalAccelForward, BlendSpace=@n6) {bLoop=false}
     Additive=Blendspace Player(X=@n8.SkydiveAimYaw, Y=@n8.LocalAccelForward, BlendSpace=@n7.Default_TurnAdditive_FemaleMedium_BS)
     AlphaCurveName=ParaGlideLeanAlpha
    ) {alphaInputType=Curve, alphaBoolBlend {blendOption=Linear}})
  ]
 }
 AnimationTransitionGraph_1 {
  path=AnimGraph.AnimGraphNode_StateMachine_0.SkyDiveState.AnimStateNode_1.DiveOrGlide.AnimGraphNode_StateMachine_212.SkydiveOrGlide.AnimStateNode_3.ParaGlide.AnimGraphNode_StateMachine_1.Glider_SurfStyleAlt.AnimStateTransitionNode_1.AnimationTransitionGraph_1
  nodes [Result()]
 }
 Transition {
  path=AnimGraph.AnimGraphNode_StateMachine_0.SkyDiveState.AnimStateNode_1.DiveOrGlide.AnimGraphNode_StateMachine_212.SkydiveOrGlide.AnimStateNode_3.ParaGlide.AnimGraphNode_StateMachine_1.Glider_SurfStyleAlt.AnimStateTransitionNode_3.Transition
  nodes [Result(bCanEnterTransition=Get bTimeInParachutingAboveIntroLength() -> (bTimeInParachutingAboveIntroLength:bool))]
 }
 Glider_Lean {
  entry=Lean_Zero
  path=AnimGraph.AnimGraphNode_StateMachine_0.SkyDiveState.AnimStateNode_1.DiveOrGlide.AnimGraphNode_StateMachine_212.SkydiveOrGlide.AnimStateNode_3.ParaGlide.AnimGraphNode_StateMachine_2.Glider_Lean
  ids=id
  nodes [
   AnimStateNode_0 = Lean_Zero()
   AnimStateNode_1 = Lean_Out()
   AnimStateNode_4 = Lean_Into() {is=title AnimStateNode}
   AnimStateTransitionNode_4 = AnimStateTransitionNode() {blendMode=Linear, crossfadeDuration=0.15}
   AnimStateTransitionNode_5 = AnimStateTransitionNode() {blendMode=Linear}
   AnimStateTransitionNode_7 = AnimStateTransitionNode() {bAutomaticRuleBasedOnSequencePlayerInState=true, blendMode=Linear, priorityOrder=2}
   AnimStateTransitionNode_9 = AnimStateTransitionNode() {blendMode=Linear, priorityOrder=2}
  ]
  links [
   {
    dst=Lean_Out
    rule=AnimGraph.AnimGraphNode_StateMachine_0.SkyDiveState.AnimStateNode_1.DiveOrGlide.AnimGraphNode_StateMachine_212.SkydiveOrGlide.AnimStateNode_3.ParaGlide.AnimGraphNode_StateMachine_2.Glider_Lean.AnimStateTransitionNode_9.AnimationTransitionGraph_1
    src=Lean_Into
    via=AnimStateTransitionNode_9
   }
   {
    dst=Lean_Zero
    rule=AnimGraph.AnimGraphNode_StateMachine_0.SkyDiveState.AnimStateNode_1.DiveOrGlide.AnimGraphNode_StateMachine_212.SkydiveOrGlide.AnimStateNode_3.ParaGlide.AnimGraphNode_StateMachine_2.Glider_Lean.AnimStateTransitionNode_5.AnimationTransitionGraph_1
    src=Lean_Out
    via=AnimStateTransitionNode_5
   }
   {
    dst=Lean_Zero
    rule=AnimGraph.AnimGraphNode_StateMachine_0.SkyDiveState.AnimStateNode_1.DiveOrGlide.AnimGraphNode_StateMachine_212.SkydiveOrGlide.AnimStateNode_3.ParaGlide.AnimGraphNode_StateMachine_2.Glider_Lean.AnimStateTransitionNode_7.AnimationTransitionGraph_1
    src=Lean_Out
    via=AnimStateTransitionNode_7
   }
   {
    dst=Lean_Into
    rule=AnimGraph.AnimGraphNode_StateMachine_0.SkyDiveState.AnimStateNode_1.DiveOrGlide.AnimGraphNode_StateMachine_212.SkydiveOrGlide.AnimStateNode_3.ParaGlide.AnimGraphNode_StateMachine_2.Glider_Lean.AnimStateTransitionNode_4.AnimationTransitionGraph_1
    src=Lean_Zero
    via=AnimStateTransitionNode_4
   }
  ]
 }
 Lean_Zero {
  path=AnimGraph.AnimGraphNode_StateMachine_0.SkyDiveState.AnimStateNode_1.DiveOrGlide.AnimGraphNode_StateMachine_212.SkydiveOrGlide.AnimStateNode_3.ParaGlide.AnimGraphNode_StateMachine_2.Glider_Lean.AnimStateNode_0.Lean_Zero
  nodes [
   n1 = Break Fort Anim Input Player Glider Anim Asset(FortAnimInput_PlayerGliderAnimAsset=Get CurrentGliderAnimSet()) -> (
    RootModPitchMin=~
    RootModPitchMax=~
    RootModYOffsetMin=~
    RootModYOffsetMax=~
    bEnableSpringMods=~
    bAllowPlayerDeployRootMod=~
    bUseSurfStyle=~
    bUseLeanPlayRateAdjust=~
    bUseForwardAs_Y_BS=~
    bUseDeployCurveForIntoPlayRateAdjust=~
    LeanForwardValue=~
    IdleValue=~
    LeanBackValue=~
    IdleInterpToSpeed=~
    LeanForwardInterpToSpeed=~
    LeanBackInterpToSpeed=~
   )
   n2 = "Blend Poses (EFortPlayerAnimBodyType)"!(
    BlendPose_0=Sequence Player!(Sequence=@n1.LeanAdditive_Center_MaleMedium_Pose) -> (Pose)
    BlendPose_1=Sequence Player(Sequence=@n1.LeanAdditive_Center_MaleLarge_Pose)
    BlendTime_0=0.1
    BlendTime_1=0.1
    ActiveEnumValue=Get AnimBodyType() -> (AnimBodyType=Small)
   ) -> (Pose) {blendType=Linear, blendPose [{linkId=-1, sourceLinkId=-1}, {linkId=-1, sourceLinkId=-1}], blendTime [0.1,0.1]}
   n3 = "Blend Poses (EFortPlayerAnimBodyType)"!(
    BlendPose_0=Sequence Player(Sequence=@n1.LeanAdditive_Center_FemaleMedium_Pose)
    BlendPose_1=Sequence Player(Sequence=@n1.LeanAdditive_Center_FemaleSmall_Pose)
    BlendPose_2:Pose Link Structure=Sequence Player(Sequence=@n1.LeanAdditive_Center_FemaleLarge_Pose)
    BlendTime_0=0.1
    BlendTime_1=0.1
    BlendTime_2:float=0.1
    ActiveEnumValue=Get AnimBodyType() -> (AnimBodyType=Small)
   ) -> (Pose) {
    blendType=Linear
    blendPose [{linkId=-1, sourceLinkId=-1}, {linkId=-1, sourceLinkId=-1}, {linkId=-1, sourceLinkId=-1}]
    blendTime [0.1,0.1,0.1]
   }
   n4 = "Blend Poses (EFortCustomGender)"(BlendPose_0=@n2, BlendPose_1=@n3, BlendTime_0=0, BlendTime_1=0, ActiveEnumValue=Get Gender() -> (Gender=Invalid)) {blendType=Linear, blendPose [{linkId=-1, sourceLinkId=-1}, {linkId=-1, sourceLinkId=-1}], blendTime [0.1,0.1]}
   Output Animation Pose(Result=@n4)
  ]
 }
 Lean_Out {
  path=AnimGraph.AnimGraphNode_StateMachine_0.SkyDiveState.AnimStateNode_1.DiveOrGlide.AnimGraphNode_StateMachine_212.SkydiveOrGlide.AnimStateNode_3.ParaGlide.AnimGraphNode_StateMachine_2.Glider_Lean.AnimStateNode_1.Lean_Out
  nodes [
   n1 = Break Fort Anim Input Player Glider Anim Asset(FortAnimInput_PlayerGliderAnimAsset=Get CurrentGliderAnimSet()) -> (
    RootModPitchMin=~
    RootModPitchMax=~
    RootModYOffsetMin=~
    RootModYOffsetMax=~
    bEnableSpringMods=~
    bAllowPlayerDeployRootMod=~
    bUseSurfStyle=~
    bUseLeanPlayRateAdjust=~
    bUseForwardAs_Y_BS=~
    bUseDeployCurveForIntoPlayRateAdjust=~
    LeanForwardValue=~
    IdleValue=~
    LeanBackValue=~
    IdleInterpToSpeed=~
    LeanForwardInterpToSpeed=~
    LeanBackInterpToSpeed=~
   )
   n2 = "Blend Poses (ESkydivingDirection)"(
    BlendPose_0=Sequence Player(Sequence=@n1.LeanAdditive_Center_MaleMedium_Pose)
    BlendPose_1=Sequence Player(Sequence=@n1.LeanAdditive_ForwardOut_Anim) {bLoopAnimation=false}
    BlendPose_2=Sequence Player(Sequence=@n1.LeanAdditive_BackOut_Anim) {bLoopAnimation=false}
    BlendPose_3=Sequence Player(Sequence=@n1.LeanAdditive_RightOut_Anim) {bLoopAnimation=false}
    BlendPose_4=Sequence Player(Sequence=@n1.LeanAdditive_LeftOut_Anim) {bLoopAnimation=false}
    BlendTime_0=0.5
    BlendTime_1=0.5
    BlendTime_2=0.5
    BlendTime_3=0.5
    BlendTime_4=0.5
    ActiveEnumValue=Break Fort Anim Input Skydiving(FortAnimInput_Skydiving=Get Skydiving()) -> (
     ParaGilderRootOffset=~
     ParaGliderRootRotationOverride=~
     bUseParaGlideRootModifier=~
     bIsSkydivingFromLaunchPad=~
     bIsSkydivingFromBus=~
     bIsInVortex=~
     bIsDiving=~
     bIsParachuteOpen=~
     bIsSkydiving=~
     bIsParachuteLeaning=~
     bIsSkydiveLeaning=~
     bIsLeaning=~
     bIsSkydiveDiveMode=~
     bPlaySkydiveDrift=~
     bSkydiveDriftDelayActive=~
     bSkydiveDriftAnimAllowed=~
     bIsGliderRight=~
     bIsGliderCenter=~
     bIsGliderLeft=~
     bIsGliderForward=~
     bIsGliderBack=~
     bHasDeployedAndIsStandingUpright=~
     LocalAccelForward=~
     LocalAccelRight=~
     LocalVelocityRight=~
     SkydiveAimPitch=~
     SkydiveAimYaw=~
     DeployChuteAnimRate=~
     SkydiveDriftAnimRate=~
     SkydiveDriftAnimRateCurrent=~
     SkydiveFidgetAnimRate=~
     SkydiveFidgetAnimRateCurrent=~
     SkydiveAdditiveAlpha=~
     SkydiveDriftDelay=~
    ).DirectionLast
   ) {
    blendType=Linear
    blendPose [
     {linkId=-1, sourceLinkId=-1}
     {linkId=-1, sourceLinkId=-1}
     {linkId=-1, sourceLinkId=-1}
     {linkId=-1, sourceLinkId=-1}
     {linkId=-1, sourceLinkId=-1}
    ]
    blendTime [0.1,0.1,0.1,0.1,0.1]
   }
   Output Animation Pose(Result=@n2)
  ]
 }
 Lean_Into {
  path=AnimGraph.AnimGraphNode_StateMachine_0.SkyDiveState.AnimStateNode_1.DiveOrGlide.AnimGraphNode_StateMachine_212.SkydiveOrGlide.AnimStateNode_3.ParaGlide.AnimGraphNode_StateMachine_2.Glider_Lean.AnimStateNode_4.Lean_Into
  nodes [
   n1 = Break Fort Anim Input Player Glider Anim Asset(FortAnimInput_PlayerGliderAnimAsset=Get CurrentGliderAnimSet()) -> (
    RootModPitchMin=~
    RootModPitchMax=~
    RootModYOffsetMin=~
    RootModYOffsetMax=~
    bEnableSpringMods=~
    bAllowPlayerDeployRootMod=~
    bUseSurfStyle=~
    bUseLeanPlayRateAdjust=~
    bUseForwardAs_Y_BS=~
    bUseDeployCurveForIntoPlayRateAdjust=~
    LeanForwardValue=~
    IdleValue=~
    LeanBackValue=~
    IdleInterpToSpeed=~
    LeanForwardInterpToSpeed=~
    LeanBackInterpToSpeed=~
   )
   n2 = "Blend Poses (ESkydivingDirection)"(
    BlendPose_0=Sequence Player(Sequence=@n1.LeanAdditive_Center_MaleMedium_Pose) {bLoopAnimation=false}
    BlendPose_1=Two Way Blend(
     A=Sequence Player!(Sequence=@n1.LeanAdditive_ForwardInto_Anim, PlayRate:float=1) -> (Pose) {bLoopAnimation=false, groupName=LeanForward_Into, method=SyncGroup}
     B=Sequence Player(Sequence=@n1.LeanAdditive_ForwardInto_FromDeploy_Anim, PlayRate=1) {bLoopAnimation=false, groupName=LeanForward_Into, method=SyncGroup}
     AlphaCurveName=ParaGlideDeploy
    ) {blendNode {alphaInputType=Curve, alphaBoolBlend {blendOption=Linear}}}
    BlendPose_2=Sequence Player(Sequence=@n1.LeanAdditive_BackInto_Anim, PlayRate=1) {bLoopAnimation=false}
    BlendPose_3=Sequence Player(Sequence=@n1.LeanAdditive_LeftInto_Anim, PlayRate=1) {bLoopAnimation=false, groupName=LeftRight, method=SyncGroup}
    BlendPose_4=Sequence Player(Sequence=@n1.LeanAdditive_RightInto_Anim, PlayRate=1) {bLoopAnimation=false, groupName=LeftRight, method=SyncGroup, playRate=0.75}
    BlendTime_0=0.2
    BlendTime_1=0.2
    BlendTime_2=0.2
    BlendTime_3=0.2
    BlendTime_4=0.2
    ActiveEnumValue=Break Fort Anim Input Skydiving(FortAnimInput_Skydiving=Get Skydiving()) -> (
     ParaGilderRootOffset=~
     ParaGliderRootRotationOverride=~
     bUseParaGlideRootModifier=~
     bIsSkydivingFromLaunchPad=~
     bIsSkydivingFromBus=~
     bIsInVortex=~
     bIsDiving=~
     bIsParachuteOpen=~
     bIsSkydiving=~
     bIsParachuteLeaning=~
     bIsSkydiveLeaning=~
     bIsLeaning=~
     bIsSkydiveDiveMode=~
     bPlaySkydiveDrift=~
     bSkydiveDriftDelayActive=~
     bSkydiveDriftAnimAllowed=~
     bIsGliderRight=~
     bIsGliderCenter=~
     bIsGliderLeft=~
     bIsGliderForward=~
     bIsGliderBack=~
     bHasDeployedAndIsStandingUpright=~
     LocalAccelForward=~
     LocalAccelRight=~
     LocalVelocityRight=~
     SkydiveAimPitch=~
     SkydiveAimYaw=~
     DeployChuteAnimRate=~
     SkydiveDriftAnimRate=~
     SkydiveDriftAnimRateCurrent=~
     SkydiveFidgetAnimRate=~
     SkydiveFidgetAnimRateCurrent=~
     SkydiveAdditiveAlpha=~
     SkydiveDriftDelay=~
    ).LocalAccelDir
   ) {
    blendType=Linear
    blendPose [
     {linkId=-1, sourceLinkId=-1}
     {linkId=-1, sourceLinkId=-1}
     {linkId=-1, sourceLinkId=-1}
     {linkId=-1, sourceLinkId=-1}
     {linkId=-1, sourceLinkId=-1}
    ]
    blendTime [0.1,0.1,0.1,0.1,0.1]
   }
   n3 = Break Fort Anim Input Skydiving(FortAnimInput_Skydiving=Get Skydiving()) -> (
    ParaGilderRootOffset=~
    ParaGliderRootRotationOverride=~
    bUseParaGlideRootModifier=~
    bIsSkydivingFromLaunchPad=~
    bIsSkydivingFromBus=~
    bIsInVortex=~
    bIsDiving=~
    bIsParachuteOpen=~
    bIsSkydiving=~
    bIsParachuteLeaning=~
    bIsSkydiveLeaning=~
    bIsLeaning=~
    bIsSkydiveDiveMode=~
    bPlaySkydiveDrift=~
    bSkydiveDriftDelayActive=~
    bSkydiveDriftAnimAllowed=~
    bIsGliderRight=~
    bIsGliderCenter=~
    bIsGliderLeft=~
    bIsGliderForward=~
    bIsGliderBack=~
    bHasDeployedAndIsStandingUpright=~
    LocalAccelForward=~
    LocalAccelRight=~
    LocalVelocityRight=~
    SkydiveAimPitch=~
    SkydiveAimYaw=~
    DeployChuteAnimRate=~
    SkydiveDriftAnimRate=~
    SkydiveDriftAnimRateCurrent=~
    SkydiveFidgetAnimRate=~
    SkydiveFidgetAnimRateCurrent=~
    SkydiveAdditiveAlpha=~
    SkydiveDriftDelay=~
   )
   n4 = Get CurrentGliderAnimSet()
   n5 = "Blend Poses (EGliderType)"!(
    BlendPose_0=@n2
    BlendPose_1=Blendspace Player(
     X=@n3.LocalAccelRight
     Y=@n3.LocalAccelForward
     BlendSpace=Break Fort Anim Input Player Glider Anim Asset!(FortAnimInput_PlayerGliderAnimAsset=@n4) -> (
      Into_Base_MaleMedium_BS
      LeanAdditive_Into_BS
      GenericAdditive_Male_BS
      GenericAdditive_Female_BS
      PelvisModVertical
      PelvisModLateral
      PelvisModBankAngle
      ThighModLegAngle
      ThighModLegPitch
      ThighModLegBank
      ThighModLeftLegDrag
      PlayerGliderType
      bEnableSpringMods=~
      bUseLeanPlayRateAdjust=~
      bUseForwardAs_Y_BS=~
      bUseDeployCurveForIntoPlayRateAdjust=~
      LeanForwardValue=~
      IdleValue=~
      LeanBackValue=~
      IdleInterpToSpeed=~
      LeanForwardInterpToSpeed=~
      LeanBackInterpToSpeed=~
     ).LeanAdditive_Into_BS
    ) {bLoop=false}
    BlendTime_0=0
    BlendTime_1=0
    ActiveEnumValue=Break Fort Anim Input Player Glider Anim Asset!(FortAnimInput_PlayerGliderAnimAsset=Get CurrentGliderAnimSet()) -> (
     Into_Base_MaleMedium_BS
     PelvisModVertical
     PelvisModLateral
     PelvisModBankAngle
     ThighModLegAngle
     ThighModLegPitch
     ThighModLegBank
     ThighModLeftLegDrag
     PlayerGliderType
     bEnableSpringMods=~
     bUseLeanPlayRateAdjust=~
     bUseForwardAs_Y_BS=~
     bUseDeployCurveForIntoPlayRateAdjust=~
     LeanForwardValue=~
     IdleValue=~
     LeanBackValue=~
     IdleInterpToSpeed=~
     LeanForwardInterpToSpeed=~
     LeanBackInterpToSpeed=~
    ).PlayerGliderType
   ) -> (Pose) {blendType=Linear, blendPose [{linkId=-1, sourceLinkId=-1}, {linkId=-1, sourceLinkId=-1}], blendTime [0.1,0.1]}
   n6 = Break Fort Anim Input Skydiving(FortAnimInput_Skydiving=Get Skydiving()) -> (
    ParaGilderRootOffset=~
    ParaGliderRootRotationOverride=~
    bUseParaGlideRootModifier=~
    bIsSkydivingFromLaunchPad=~
    bIsSkydivingFromBus=~
    bIsInVortex=~
    bIsDiving=~
    bIsParachuteOpen=~
    bIsSkydiving=~
    bIsParachuteLeaning=~
    bIsSkydiveLeaning=~
    bIsLeaning=~
    bIsSkydiveDiveMode=~
    bPlaySkydiveDrift=~
    bSkydiveDriftDelayActive=~
    bSkydiveDriftAnimAllowed=~
    bIsGliderRight=~
    bIsGliderCenter=~
    bIsGliderLeft=~
    bIsGliderForward=~
    bIsGliderBack=~
    bHasDeployedAndIsStandingUpright=~
    LocalAccelForward=~
    LocalAccelRight=~
    LocalVelocityRight=~
    SkydiveAimPitch=~
    SkydiveAimYaw=~
    DeployChuteAnimRate=~
    SkydiveDriftAnimRate=~
    SkydiveDriftAnimRateCurrent=~
    SkydiveFidgetAnimRate=~
    SkydiveFidgetAnimRateCurrent=~
    SkydiveAdditiveAlpha=~
    SkydiveDriftDelay=~
   )
   n7 = "Blend Poses (EFortPlayerAnimBodyType)"!(
    BlendPose_0=Blendspace Player(X=@n6.LocalAccelRight, Y=@n6.LocalAccelForward, BlendSpace=@n1.Lean_MaleMedium_BS)
    BlendPose_1=Blendspace Player(X=@n6.LocalAccelRight, Y=@n6.LocalAccelForward, BlendSpace=@n1.Lean_MaleLarge_BS)
    BlendTime_0=0.1
    BlendTime_1=0.1
    ActiveEnumValue=Get AnimBodyType() -> (AnimBodyType=Small)
   ) -> (Pose) {blendType=Linear, blendPose [{linkId=-1, sourceLinkId=-1}, {linkId=-1, sourceLinkId=-1}], blendTime [0.1,0.1]}
   n8 = "Blend Poses (EFortPlayerAnimBodyType)"!(
    BlendPose_0=Blendspace Player(X=@n6.LocalAccelRight, Y=@n6.LocalAccelForward, BlendSpace=@n1.Lean_FemaleMedium_BS)
    BlendPose_1=Blendspace Player(X=@n6.LocalAccelRight, Y=@n6.LocalAccelForward, BlendSpace=@n1.Lean_FemaleSmall_BS)
    BlendPose_2:Pose Link Structure=Blendspace Player(X=@n6.LocalAccelRight, Y=@n6.LocalAccelForward, BlendSpace=@n1.Lean_FemaleLarge_BS)
    BlendTime_0=0.1
    BlendTime_1=0.1
    BlendTime_2:float=0.1
    ActiveEnumValue=Get AnimBodyType() -> (AnimBodyType=Small)
   ) -> (Pose) {
    blendType=Linear
    blendPose [{linkId=-1, sourceLinkId=-1}, {linkId=-1, sourceLinkId=-1}, {linkId=-1, sourceLinkId=-1}]
    blendTime [0.1,0.1,0.1]
   }
   n9 = "Blend Poses (EFortCustomGender)"(BlendPose_0=@n7, BlendPose_1=@n8, BlendTime_0=0, BlendTime_1=0, ActiveEnumValue=Get Gender() -> (Gender=Invalid)) {blendType=Linear, blendPose [{linkId=-1, sourceLinkId=-1}, {linkId=-1, sourceLinkId=-1}], blendTime [0.1,0.1]}
   Output Animation Pose(Result=Two Way Blend(A=@n9, B=@n5, AlphaCurveName=ParaGlideLeanTransition) {blendNode {alphaInputType=Curve, alphaBoolBlend {blendOption=Linear}, alphaScaleBiasClamp {bias=1, scale=-1}}})
  ]
 }
 AnimationTransitionGraph_1 {
  path=AnimGraph.AnimGraphNode_StateMachine_0.SkyDiveState.AnimStateNode_1.DiveOrGlide.AnimGraphNode_StateMachine_212.SkydiveOrGlide.AnimStateNode_3.ParaGlide.AnimGraphNode_StateMachine_2.Glider_Lean.AnimStateTransitionNode_4.AnimationTransitionGraph_1
  nodes [Result(bCanEnterTransition=Get bIsAccelerating2D())]
 }
 AnimationTransitionGraph_1 {
  path=AnimGraph.AnimGraphNode_StateMachine_0.SkyDiveState.AnimStateNode_1.DiveOrGlide.AnimGraphNode_StateMachine_212.SkydiveOrGlide.AnimStateNode_3.ParaGlide.AnimGraphNode_StateMachine_2.Glider_Lean.AnimStateTransitionNode_5.AnimationTransitionGraph_1
  nodes [Result(bCanEnterTransition=Get bIsAccelerating2D())]
 }
 AnimationTransitionGraph_1 {
  path=AnimGraph.AnimGraphNode_StateMachine_0.SkyDiveState.AnimStateNode_1.DiveOrGlide.AnimGraphNode_StateMachine_212.SkydiveOrGlide.AnimStateNode_3.ParaGlide.AnimGraphNode_StateMachine_2.Glider_Lean.AnimStateTransitionNode_7.AnimationTransitionGraph_1
  nodes [Result()]
 }
 AnimationTransitionGraph_1 {
  path=AnimGraph.AnimGraphNode_StateMachine_0.SkyDiveState.AnimStateNode_1.DiveOrGlide.AnimGraphNode_StateMachine_212.SkydiveOrGlide.AnimStateNode_3.ParaGlide.AnimGraphNode_StateMachine_2.Glider_Lean.AnimStateTransitionNode_9.AnimationTransitionGraph_1
  nodes [Result(bCanEnterTransition=NOT Boolean(A=Get bIsAccelerating2D()))]
 }
 Glider_or_Unmbrella {
  entry=ParaGlide_Into_Deploy
  path=AnimGraph.AnimGraphNode_StateMachine_0.SkyDiveState.AnimStateNode_1.DiveOrGlide.AnimGraphNode_StateMachine_212.SkydiveOrGlide.AnimStateNode_3.ParaGlide.AnimGraphNode_StateMachine_3.Glider_or_Unmbrella
  ids=id
  nodes [
   AnimStateNode_1 = ParaGlide_Into_Deploy()
   AnimStateNode_3 = ParaGlide()
   AnimStateTransitionNode_0 = AnimStateTransitionNode() {crossfadeDuration=0}
   AnimStateTransitionNode_2 = AnimStateTransitionNode() {bAutomaticRuleBasedOnSequencePlayerInState=true, crossfadeDuration=0}
  ]
  links [
   {
    dst=ParaGlide
    rule=AnimGraph.AnimGraphNode_StateMachine_0.SkyDiveState.AnimStateNode_1.DiveOrGlide.AnimGraphNode_StateMachine_212.SkydiveOrGlide.AnimStateNode_3.ParaGlide.AnimGraphNode_StateMachine_3.Glider_or_Unmbrella.AnimStateTransitionNode_0.Transition
    src=ParaGlide_Into_Deploy
    via=AnimStateTransitionNode_0
   }
   {
    dst=ParaGlide
    rule=AnimGraph.AnimGraphNode_StateMachine_0.SkyDiveState.AnimStateNode_1.DiveOrGlide.AnimGraphNode_StateMachine_212.SkydiveOrGlide.AnimStateNode_3.ParaGlide.AnimGraphNode_StateMachine_3.Glider_or_Unmbrella.AnimStateTransitionNode_2.AnimationTransitionGraph_1
    src=ParaGlide_Into_Deploy
    via=AnimStateTransitionNode_2
   }
  ]
 }
 ParaGlide_Into_Deploy {
  path=AnimGraph.AnimGraphNode_StateMachine_0.SkyDiveState.AnimStateNode_1.DiveOrGlide.AnimGraphNode_StateMachine_212.SkydiveOrGlide.AnimStateNode_3.ParaGlide.AnimGraphNode_StateMachine_3.Glider_or_Unmbrella.AnimStateNode_1.ParaGlide_Into_Deploy
  nodes [
   n1 = Break Fort Anim Input Player Glider Anim Asset!(FortAnimInput_PlayerGliderAnimAsset=Get CurrentGliderAnimSet()) -> (
    Default_Base_BS:Blend Space Object Reference
    Default_BodyAdditive_MaleMedium_BS:Blend Space Object Reference
    Default_BodyAdditive_MaleLarge_BS:Blend Space Object Reference
    Default_BodyAdditive_FemaleSmall_BS:Blend Space Object Reference
    Default_BodyAdditive_FemaleMedium_BS:Blend Space Object Reference
    Default_BodyAdditive_FemaleLarge_BS:Blend Space Object Reference
    Default_TurnAdditive_MaleMedium_BS:Blend Space Object Reference
    Default_TurnAdditive_MaleLarge_BS:Blend Space Object Reference
    Default_TurnAdditive_FemaleSmall_BS:Blend Space Object Reference
    Default_TurnAdditive_FemaleMedium_BS:Blend Space Object Reference
    Default_TurnAdditive_FemaleLarge_BS:Blend Space Object Reference
    Into_Base_BS:Blend Space Object Reference
    Into_Base_MaleMedium_BS
    Into_BodyAdditive_MaleMedium_BS:Blend Space Object Reference
    Into_BodyAdditive_MaleLarge_BS:Blend Space Object Reference
    Into_BodyAdditive_FemaleSmall_BS:Blend Space Object Reference
    Into_BodyAdditive_FemaleMedium_BS:Blend Space Object Reference
    Into_BodyAdditive_FemaleLarge_BS:Blend Space Object Reference
    Into_TurnAdditive_MaleMedium_BS:Blend Space Object Reference
    Into_TurnAdditive_MaleLarge_BS:Blend Space Object Reference
    Into_TurnAdditive_FemaleSmall_BS:Blend Space Object Reference
    Into_TurnAdditive_FemaleMedium_BS:Blend Space Object Reference
    Into_TurnAdditive_FemaleLarge_BS:Blend Space Object Reference
    Lean_MaleMedium_BS:Blend Space Object Reference
    Lean_MaleLarge_BS:Blend Space Object Reference
    Lean_FemaleSmall_BS:Blend Space Object Reference
    Lean_FemaleMedium_BS:Blend Space Object Reference
    Lean_FemaleLarge_BS:Blend Space Object Reference
    LeanAdditive_Center_MaleMedium_Pose:Anim Sequence Object Reference
    LeanAdditive_Center_MaleLarge_Pose:Anim Sequence Object Reference
    LeanAdditive_Center_FemaleSmall_Pose:Anim Sequence Object Reference
    LeanAdditive_Center_FemaleMedium_Pose:Anim Sequence Object Reference
    LeanAdditive_Center_FemaleLarge_Pose:Anim Sequence Object Reference
    LeanAdditive_Into_BS:Blend Space Object Reference
    LeanAdditive_ForwardInto_Anim:Anim Sequence Object Reference
    LeanAdditive_ForwardInto_FromDeploy_Anim:Anim Sequence Object Reference
    LeanAdditive_BackInto_Anim:Anim Sequence Object Reference
    LeanAdditive_LeftInto_Anim:Anim Sequence Object Reference
    LeanAdditive_RightInto_Anim:Anim Sequence Object Reference
    LeanAdditive_ForwardOut_Anim:Anim Sequence Object Reference
    LeanAdditive_BackOut_Anim:Anim Sequence Object Reference
    LeanAdditive_LeftOut_Anim:Anim Sequence Object Reference
    LeanAdditive_RightOut_Anim:Anim Sequence Object Reference
    ToGlide_BS:Blend Space Object Reference
    ToGlide_Lean_BS:Blend Space Object Reference
    ToDive_BS:Blend Space Object Reference
    ToDive_Lean_BS:Blend Space Object Reference
    Dive_WeaponR_Additive_BS:Blend Space Object Reference
    Glide_WeaponR_Additive_Anim:Anim Sequence Object Reference
    GenericAdditive_Male_BS:Blend Space Object Reference
    GenericAdditive_Female_BS:Blend Space Object Reference
    RootModPitchMin:float=~
    RootModPitchMax:float=~
    RootModYOffsetMin:float=~
    RootModYOffsetMax:float=~
    PelvisModVertical
    PelvisModLateral
    PelvisModBankAngle
    ThighModLegAngle
    ThighModLegPitch
    ThighModLegBank
    ThighModLeftLegDrag
    PlayerGliderType
    bEnableSpringMods=~
    bAllowPlayerDeployRootMod:bool=~
    bUseSurfStyle:bool=~
    bUseLeanPlayRateAdjust=~
    bUseForwardAs_Y_BS=~
    bUseDeployCurveForIntoPlayRateAdjust=~
    LeanForwardValue=~
    IdleValue=~
    LeanBackValue=~
    IdleInterpToSpeed=~
    LeanForwardInterpToSpeed=~
    LeanBackInterpToSpeed=~
   )
   n2 = Break Fort Anim Input Skydiving(FortAnimInput_Skydiving=Get Skydiving()) -> (
    ParaGilderRootOffset=~
    ParaGliderRootRotationOverride=~
    bUseParaGlideRootModifier=~
    bIsSkydivingFromLaunchPad=~
    bIsSkydivingFromBus=~
    bIsInVortex=~
    bIsDiving=~
    bIsParachuteOpen=~
    bIsSkydiving=~
    bIsParachuteLeaning=~
    bIsSkydiveLeaning=~
    bIsLeaning=~
    bIsSkydiveDiveMode=~
    bPlaySkydiveDrift=~
    bSkydiveDriftDelayActive=~
    bSkydiveDriftAnimAllowed=~
    bIsGliderRight=~
    bIsGliderCenter=~
    bIsGliderLeft=~
    bIsGliderForward=~
    bIsGliderBack=~
    bHasDeployedAndIsStandingUpright=~
    LocalAccelForward=~
    LocalAccelRight=~
    LocalVelocityRight=~
    SkydiveAimPitch=~
    SkydiveAimYaw=~
    DeployChuteAnimRate=~
    SkydiveDriftAnimRate=~
    SkydiveDriftAnimRateCurrent=~
    SkydiveFidgetAnimRate=~
    SkydiveFidgetAnimRateCurrent=~
    SkydiveAdditiveAlpha=~
    SkydiveDriftDelay=~
   )
   n3 = Get AimPitch()
   n4 = "Blend Poses (EFortPlayerAnimBodyType)"!(
    BlendPose_0=Blendspace Player(X=@n2.SkydiveAimYaw, Y=@n3, BlendSpace=@n1.Into_BodyAdditive_MaleMedium_BS) {bIgnoreForRelevancyTest=true}
    BlendPose_1=Blendspace Player(X=@n2.SkydiveAimYaw, Y=@n3, BlendSpace=@n1.Into_BodyAdditive_MaleLarge_BS) {bIgnoreForRelevancyTest=true}
    BlendTime_0=0.1
    BlendTime_1=0.1
    ActiveEnumValue=Get AnimBodyType() -> (AnimBodyType=Small)
   ) -> (Pose) {blendType=Linear, blendPose [{linkId=-1, sourceLinkId=-1}, {linkId=-1, sourceLinkId=-1}], blendTime [0.1,0.1]}
   n5 = "Blend Poses (EFortPlayerAnimBodyType)"!(
    BlendPose_0=Blendspace Player(X=@n2.SkydiveAimYaw, Y=@n3, BlendSpace=@n1.Into_BodyAdditive_FemaleMedium_BS) {bIgnoreForRelevancyTest=true}
    BlendPose_1=Blendspace Player(X=@n2.SkydiveAimYaw, Y=@n3, BlendSpace=@n1.Into_BodyAdditive_FemaleSmall_BS) {bIgnoreForRelevancyTest=true}
    BlendPose_2:Pose Link Structure=Blendspace Player(X=@n2.SkydiveAimYaw, Y=@n3, BlendSpace=@n1.Into_BodyAdditive_FemaleLarge_BS) {bIgnoreForRelevancyTest=true}
    BlendTime_0=0.1
    BlendTime_1=0.1
    BlendTime_2:float=0.1
    ActiveEnumValue=Get AnimBodyType() -> (AnimBodyType=Small)
   ) -> (Pose) {
    blendType=Linear
    blendPose [{linkId=-1, sourceLinkId=-1}, {linkId=-1, sourceLinkId=-1}, {linkId=-1, sourceLinkId=-1}]
    blendTime [0.1,0.1,0.1]
   }
   n6 = "Blend Poses (EFortCustomGender)"(BlendPose_0=@n4, BlendPose_1=@n5, BlendTime_0=0, BlendTime_1=0, ActiveEnumValue=Get Gender() -> (Gender=Invalid)) {blendType=Linear, blendPose [{linkId=-1, sourceLinkId=-1}, {linkId=-1, sourceLinkId=-1}], blendTime [0.1,0.1]}
   n7 = Break Fort Anim Input Skydiving(FortAnimInput_Skydiving=Get Skydiving()) -> (
    ParaGilderRootOffset=~
    ParaGliderRootRotationOverride=~
    bUseParaGlideRootModifier=~
    bIsSkydivingFromLaunchPad=~
    bIsSkydivingFromBus=~
    bIsInVortex=~
    bIsDiving=~
    bIsParachuteOpen=~
    bIsSkydiving=~
    bIsParachuteLeaning=~
    bIsSkydiveLeaning=~
    bIsLeaning=~
    bIsSkydiveDiveMode=~
    bPlaySkydiveDrift=~
    bSkydiveDriftDelayActive=~
    bSkydiveDriftAnimAllowed=~
    bIsGliderRight=~
    bIsGliderCenter=~
    bIsGliderLeft=~
    bIsGliderForward=~
    bIsGliderBack=~
    bHasDeployedAndIsStandingUpright=~
    LocalAccelForward=~
    LocalAccelRight=~
    LocalVelocityRight=~
    SkydiveAimPitch=~
    SkydiveAimYaw=~
    DeployChuteAnimRate=~
    SkydiveDriftAnimRate=~
    SkydiveDriftAnimRateCurrent=~
    SkydiveFidgetAnimRate=~
    SkydiveFidgetAnimRateCurrent=~
    SkydiveAdditiveAlpha=~
    SkydiveDriftDelay=~
   )
   n8 = Break Fort Anim Input Player Glider Anim Asset(FortAnimInput_PlayerGliderAnimAsset=Get CurrentGliderAnimSet()) -> (
    RootModPitchMin=~
    RootModPitchMax=~
    RootModYOffsetMin=~
    RootModYOffsetMax=~
    bEnableSpringMods=~
    bAllowPlayerDeployRootMod=~
    bUseSurfStyle=~
    bUseLeanPlayRateAdjust=~
    bUseForwardAs_Y_BS=~
    bUseDeployCurveForIntoPlayRateAdjust=~
    LeanForwardValue=~
    IdleValue=~
    LeanBackValue=~
    IdleInterpToSpeed=~
    LeanForwardInterpToSpeed=~
    LeanBackInterpToSpeed=~
   )
   n9 = Get Glider_IntoBS_PlayRate() -> (Glider_IntoBS_PlayRate:float)
   n10 = Get Glider_BaseBS_Y_Axis() -> (Glider_BaseBS_Y_Axis:float)
   n11 = Break Fort Anim Input Player Glider Anim Asset(FortAnimInput_PlayerGliderAnimAsset=Get CurrentGliderAnimSet()) -> (
    RootModPitchMin=~
    RootModPitchMax=~
    RootModYOffsetMin=~
    RootModYOffsetMax=~
    bEnableSpringMods=~
    bAllowPlayerDeployRootMod=~
    bUseSurfStyle=~
    bUseLeanPlayRateAdjust=~
    bUseForwardAs_Y_BS=~
    bUseDeployCurveForIntoPlayRateAdjust=~
    LeanForwardValue=~
    IdleValue=~
    LeanBackValue=~
    IdleInterpToSpeed=~
    LeanForwardInterpToSpeed=~
    LeanBackInterpToSpeed=~
   )
   n12 = Break Fort Anim Input Skydiving(FortAnimInput_Skydiving=Get Skydiving()) -> (
    ParaGilderRootOffset=~
    ParaGliderRootRotationOverride=~
    bUseParaGlideRootModifier=~
    bIsSkydivingFromLaunchPad=~
    bIsSkydivingFromBus=~
    bIsInVortex=~
    bIsDiving=~
    bIsParachuteOpen=~
    bIsSkydiving=~
    bIsParachuteLeaning=~
    bIsSkydiveLeaning=~
    bIsLeaning=~
    bIsSkydiveDiveMode=~
    bPlaySkydiveDrift=~
    bSkydiveDriftDelayActive=~
    bSkydiveDriftAnimAllowed=~
    bIsGliderRight=~
    bIsGliderCenter=~
    bIsGliderLeft=~
    bIsGliderForward=~
    bIsGliderBack=~
    bHasDeployedAndIsStandingUpright=~
    LocalAccelForward=~
    LocalAccelRight=~
    LocalVelocityRight=~
    SkydiveAimPitch=~
    SkydiveAimYaw=~
    DeployChuteAnimRate=~
    SkydiveDriftAnimRate=~
    SkydiveDriftAnimRateCurrent=~
    SkydiveFidgetAnimRate=~
    SkydiveFidgetAnimRateCurrent=~
    SkydiveAdditiveAlpha=~
    SkydiveDriftDelay=~
   )
   n13 = Get AimPitch()
   n14 = "Blend Poses (EFortPlayerAnimBodyType)"!(
    BlendPose_0=Blendspace Player(X=@n12.SkydiveAimYaw, Y=@n13, BlendSpace=@n11.Default_TurnAdditive_MaleMedium_BS) {bIgnoreForRelevancyTest=true}
    BlendPose_1=Blendspace Player(X=@n12.SkydiveAimYaw, Y=@n13, BlendSpace=@n11.Default_TurnAdditive_MaleLarge_BS) {bIgnoreForRelevancyTest=true}
    BlendTime_0=0.1
    BlendTime_1=0.1
    ActiveEnumValue=Get AnimBodyType() -> (AnimBodyType=Small)
   ) -> (Pose) {blendType=Linear, blendPose [{linkId=-1, sourceLinkId=-1}, {linkId=-1, sourceLinkId=-1}], blendTime [0.1,0.1]}
   n15 = "Blend Poses (EFortPlayerAnimBodyType)"!(
    BlendPose_0=Blendspace Player(X=@n12.SkydiveAimYaw, Y=@n13, BlendSpace=@n11.Default_TurnAdditive_FemaleMedium_BS) {bIgnoreForRelevancyTest=true}
    BlendPose_1=Blendspace Player(X=@n12.SkydiveAimYaw, Y=@n13, BlendSpace=@n11.Default_TurnAdditive_FemaleSmall_BS) {bIgnoreForRelevancyTest=true}
    BlendPose_2:Pose Link Structure=Blendspace Player(X=@n12.SkydiveAimYaw, Y=@n13, BlendSpace=@n11.Default_TurnAdditive_FemaleLarge_BS) {bIgnoreForRelevancyTest=true}
    BlendTime_0=0.1
    BlendTime_1=0.1
    BlendTime_2:float=0.1
    ActiveEnumValue=Get AnimBodyType() -> (AnimBodyType=Small)
   ) -> (Pose) {
    blendType=Linear
    blendPose [{linkId=-1, sourceLinkId=-1}, {linkId=-1, sourceLinkId=-1}, {linkId=-1, sourceLinkId=-1}]
    blendTime [0.1,0.1,0.1]
   }
   n16 = "Blend Poses (EFortCustomGender)"(BlendPose_0=@n14, BlendPose_1=@n15, BlendTime_0=0, BlendTime_1=0, ActiveEnumValue=Get Gender() -> (Gender=Invalid)) {blendType=Linear, blendPose [{linkId=-1, sourceLinkId=-1}, {linkId=-1, sourceLinkId=-1}], blendTime [0.1,0.1]}
   n17 = Break Fort Anim Input Player Glider Anim Asset(FortAnimInput_PlayerGliderAnimAsset=Get CurrentGliderAnimSet()) -> (
    RootModPitchMin=~
    RootModPitchMax=~
    RootModYOffsetMin=~
    RootModYOffsetMax=~
    bEnableSpringMods=~
    bAllowPlayerDeployRootMod=~
    bUseSurfStyle=~
    bUseLeanPlayRateAdjust=~
    bUseForwardAs_Y_BS=~
    bUseDeployCurveForIntoPlayRateAdjust=~
    LeanForwardValue=~
    IdleValue=~
    LeanBackValue=~
    IdleInterpToSpeed=~
    LeanForwardInterpToSpeed=~
    LeanBackInterpToSpeed=~
   )
   n18 = Break Fort Anim Input Skydiving(FortAnimInput_Skydiving=Get Skydiving()) -> (
    ParaGilderRootOffset=~
    ParaGliderRootRotationOverride=~
    bUseParaGlideRootModifier=~
    bIsSkydivingFromLaunchPad=~
    bIsSkydivingFromBus=~
    bIsInVortex=~
    bIsDiving=~
    bIsParachuteOpen=~
    bIsSkydiving=~
    bIsParachuteLeaning=~
    bIsSkydiveLeaning=~
    bIsLeaning=~
    bIsSkydiveDiveMode=~
    bPlaySkydiveDrift=~
    bSkydiveDriftDelayActive=~
    bSkydiveDriftAnimAllowed=~
    bIsGliderRight=~
    bIsGliderCenter=~
    bIsGliderLeft=~
    bIsGliderForward=~
    bIsGliderBack=~
    bHasDeployedAndIsStandingUpright=~
    LocalAccelForward=~
    LocalAccelRight=~
    LocalVelocityRight=~
    SkydiveAimPitch=~
    SkydiveAimYaw=~
    DeployChuteAnimRate=~
    SkydiveDriftAnimRate=~
    SkydiveDriftAnimRateCurrent=~
    SkydiveFidgetAnimRate=~
    SkydiveFidgetAnimRateCurrent=~
    SkydiveAdditiveAlpha=~
    SkydiveDriftDelay=~
   )
   n19 = Get AimPitch()
   n20 = "Blend Poses (EFortPlayerAnimBodyType)"!(
    BlendPose_0=Blendspace Player(X=@n18.SkydiveAimYaw, Y=@n19, BlendSpace=@n17.Into_TurnAdditive_MaleMedium_BS) {bIgnoreForRelevancyTest=true}
    BlendPose_1=Blendspace Player(X=@n18.SkydiveAimYaw, Y=@n19, BlendSpace=@n17.Into_TurnAdditive_MaleLarge_BS) {bIgnoreForRelevancyTest=true}
    BlendTime_0=0.1
    BlendTime_1=0.1
    ActiveEnumValue=Get AnimBodyType() -> (AnimBodyType=Small)
   ) -> (Pose) {blendType=Linear, blendPose [{linkId=-1, sourceLinkId=-1}, {linkId=-1, sourceLinkId=-1}], blendTime [0.1,0.1]}
   n21 = "Blend Poses (EFortPlayerAnimBodyType)"!(
    BlendPose_0=Blendspace Player(X=@n18.SkydiveAimYaw, Y=@n19, BlendSpace=@n17.Into_TurnAdditive_FemaleMedium_BS) {bIgnoreForRelevancyTest=true}
    BlendPose_1=Blendspace Player(X=@n18.SkydiveAimYaw, Y=@n19, BlendSpace=@n17.Into_TurnAdditive_FemaleSmall_BS) {bIgnoreForRelevancyTest=true}
    BlendPose_2:Pose Link Structure=Blendspace Player(X=@n18.SkydiveAimYaw, Y=@n19, BlendSpace=@n17.Into_TurnAdditive_FemaleLarge_BS) {bIgnoreForRelevancyTest=true}
    BlendTime_0=0.1
    BlendTime_1=0.1
    BlendTime_2:float=0.1
    ActiveEnumValue=Get AnimBodyType() -> (AnimBodyType=Small)
   ) -> (Pose) {
    blendType=Linear
    blendPose [{linkId=-1, sourceLinkId=-1}, {linkId=-1, sourceLinkId=-1}, {linkId=-1, sourceLinkId=-1}]
    blendTime [0.1,0.1,0.1]
   }
   n22 = "Blend Poses (EFortCustomGender)"(BlendPose_0=@n20, BlendPose_1=@n21, BlendTime_0=0, BlendTime_1=0, ActiveEnumValue=Get Gender() -> (Gender=Invalid)) {blendType=Linear, blendPose [{linkId=-1, sourceLinkId=-1}, {linkId=-1, sourceLinkId=-1}], blendTime [0.1,0.1]}
   n23 = Apply Additive(
    Base=Apply Additive(
     Base=Blendspace Player!(X=@n7.SkydiveAimYaw, Y=@n10, PlayRate:float=@n9, BlendSpace=@n8.Into_Base_BS) -> (Pose) {bLoop=false}
     Additive=@n6
     AlphaCurveName=ParaGlideLeanAlpha
    ) {alphaInputType=Curve, alphaBoolBlend {blendOption=Linear}}
    Additive=Two Way Blend(A=@n22, B=@n16, AlphaCurveName=BlendIn) {blendNode {alphaInputType=Curve, alphaBoolBlend {blendOption=Linear}, alphaScaleBiasClamp {bInterpResult=true}}}
    AlphaCurveName=ParaGlideTurn
   ) {alphaInputType=Curve, alphaBoolBlend {blendOption=Linear}, alphaScaleBiasClamp {bMapRange=true, outRange {max=0, min=1}}}
   Output Animation Pose(Result=@n23)
  ]
 }
 ParaGlide {
  path=AnimGraph.AnimGraphNode_StateMachine_0.SkyDiveState.AnimStateNode_1.DiveOrGlide.AnimGraphNode_StateMachine_212.SkydiveOrGlide.AnimStateNode_3.ParaGlide.AnimGraphNode_StateMachine_3.Glider_or_Unmbrella.AnimStateNode_3.ParaGlide
  nodes [
   n1 = Break Fort Anim Input Skydiving!(FortAnimInput_Skydiving=Get Skydiving()) -> (bHasDeployedAndIsStandingUpright=~, LocalAccelForward=~, SkydiveAimYaw=~, PoseAdditive)
   n2 = Break Fort Anim Input Player Glider Anim Asset(FortAnimInput_PlayerGliderAnimAsset=Get CurrentGliderAnimSet()) -> (
    RootModPitchMin=~
    RootModPitchMax=~
    RootModYOffsetMin=~
    RootModYOffsetMax=~
    bEnableSpringMods=~
    bAllowPlayerDeployRootMod=~
    bUseSurfStyle=~
    bUseLeanPlayRateAdjust=~
    bUseForwardAs_Y_BS=~
    bUseDeployCurveForIntoPlayRateAdjust=~
    LeanForwardValue=~
    IdleValue=~
    LeanBackValue=~
    IdleInterpToSpeed=~
    LeanForwardInterpToSpeed=~
    LeanBackInterpToSpeed=~
   )
   n3 = Get AimPitch()
   n4 = "Blend Poses (EFortPlayerAnimBodyType)"!(
    BlendPose_0=Blendspace Player!(X=@n1.SkydiveAimYaw, Y=@n3, BlendSpace=@n2.Default_BodyAdditive_MaleMedium_BS) -> (Pose)
    BlendPose_1=Blendspace Player(X=@n1.SkydiveAimYaw, Y=@n3, BlendSpace=@n2.Default_BodyAdditive_MaleLarge_BS)
    BlendTime_0=0.1
    BlendTime_1=0.1
    ActiveEnumValue=Get AnimBodyType() -> (AnimBodyType=Small)
   ) -> (Pose) {blendType=Linear, blendPose [{linkId=-1, sourceLinkId=-1}, {linkId=-1, sourceLinkId=-1}], blendTime [0.1,0.1]}
   n5 = "Blend Poses (EFortPlayerAnimBodyType)"!(
    BlendPose_0=Blendspace Player(X=@n1.SkydiveAimYaw, Y=@n3, BlendSpace=@n2.Default_BodyAdditive_FemaleMedium_BS)
    BlendPose_1=Blendspace Player(X=@n1.SkydiveAimYaw, Y=@n3, BlendSpace=@n2.Default_BodyAdditive_FemaleSmall_BS)
    BlendPose_2:Pose Link Structure=Blendspace Player(X=@n1.SkydiveAimYaw, Y=@n3, BlendSpace=@n2.Default_BodyAdditive_FemaleLarge_BS)
    BlendTime_0=0.1
    BlendTime_1=0.1
    BlendTime_2:float=0.1
    ActiveEnumValue=Get AnimBodyType() -> (AnimBodyType=Small)
   ) -> (Pose) {
    blendType=Linear
    blendPose [{linkId=-1, sourceLinkId=-1}, {linkId=-1, sourceLinkId=-1}, {linkId=-1, sourceLinkId=-1}]
    blendTime [0.1,0.1,0.1]
   }
   n6 = "Blend Poses (EFortCustomGender)"(BlendPose_0=@n4, BlendPose_1=@n5, BlendTime_0=0, BlendTime_1=0, ActiveEnumValue=Get Gender() -> (Gender=Invalid)) {blendType=Linear, blendPose [{linkId=-1, sourceLinkId=-1}, {linkId=-1, sourceLinkId=-1}], blendTime [0.1,0.1]}
   n7 = Break Fort Anim Input Player Glider Anim Asset(FortAnimInput_PlayerGliderAnimAsset=Get CurrentGliderAnimSet()) -> (
    RootModPitchMin=~
    RootModPitchMax=~
    RootModYOffsetMin=~
    RootModYOffsetMax=~
    bEnableSpringMods=~
    bAllowPlayerDeployRootMod=~
    bUseSurfStyle=~
    bUseLeanPlayRateAdjust=~
    bUseForwardAs_Y_BS=~
    bUseDeployCurveForIntoPlayRateAdjust=~
    LeanForwardValue=~
    IdleValue=~
    LeanBackValue=~
    IdleInterpToSpeed=~
    LeanForwardInterpToSpeed=~
    LeanBackInterpToSpeed=~
   )
   n8 = Get Glider_BaseBS_PlayRate() -> (Glider_BaseBS_PlayRate:float)
   n9 = Get Glider_BaseBS_Y_Axis()
   n10 = "Blend Poses (EFortPlayerAnimBodyType)"!(
    BlendPose_0=Blendspace Player(X=@n1.SkydiveAimYaw, Y=@n3, BlendSpace=@n2.Default_TurnAdditive_MaleMedium_BS)
    BlendPose_1=Blendspace Player(X=@n1.SkydiveAimYaw, Y=@n3, BlendSpace=@n2.Default_TurnAdditive_MaleLarge_BS)
    BlendTime_0=0.1
    BlendTime_1=0.1
    ActiveEnumValue=Get AnimBodyType() -> (AnimBodyType=Small)
   ) -> (Pose) {blendType=Linear, blendPose [{linkId=-1, sourceLinkId=-1}, {linkId=-1, sourceLinkId=-1}], blendTime [0.1,0.1]}
   n11 = "Blend Poses (EFortPlayerAnimBodyType)"!(
    BlendPose_0=Blendspace Player(X=@n1.SkydiveAimYaw, Y=@n3, BlendSpace=@n2.Default_TurnAdditive_FemaleMedium_BS)
    BlendPose_1=Blendspace Player(X=@n1.SkydiveAimYaw, Y=@n3, BlendSpace=@n2.Default_TurnAdditive_FemaleSmall_BS)
    BlendPose_2:Pose Link Structure=Blendspace Player(X=@n1.SkydiveAimYaw, Y=@n3, BlendSpace=@n2.Default_TurnAdditive_FemaleLarge_BS)
    BlendTime_0=0.1
    BlendTime_1=0.1
    BlendTime_2:float=0.1
    ActiveEnumValue=Get AnimBodyType() -> (AnimBodyType=Small)
   ) -> (Pose) {
    blendType=Linear
    blendPose [{linkId=-1, sourceLinkId=-1}, {linkId=-1, sourceLinkId=-1}, {linkId=-1, sourceLinkId=-1}]
    blendTime [0.1,0.1,0.1]
   }
   n12 = "Blend Poses (EFortCustomGender)"(BlendPose_0=@n10, BlendPose_1=@n11, BlendTime_0=0, BlendTime_1=0, ActiveEnumValue=Get Gender() -> (Gender=Invalid)) {blendType=Linear, blendPose [{linkId=-1, sourceLinkId=-1}, {linkId=-1, sourceLinkId=-1}], blendTime [0.1,0.1]}
   n13 = Apply Additive(
    Base=Apply Additive!(Base=Blendspace Player!(X=@n1.SkydiveAimYaw, Y=@n9, PlayRate:float=@n8, BlendSpace=@n7) -> (Pose), Additive=@n6, Alpha:float) -> (Pose) {alphaBoolBlend {blendOption=Linear}}
    Additive=@n12
    AlphaCurveName=ParaGlideTurn
   ) {alphaInputType=Curve, alphaBoolBlend {blendOption=Linear}, alphaScaleBiasClamp {bMapRange=true, outRange {max=0, min=1}}}
   Output Animation Pose(Result=@n13)
  ]
 }
 Transition {
  path=AnimGraph.AnimGraphNode_StateMachine_0.SkyDiveState.AnimStateNode_1.DiveOrGlide.AnimGraphNode_StateMachine_212.SkydiveOrGlide.AnimStateNode_3.ParaGlide.AnimGraphNode_StateMachine_3.Glider_or_Unmbrella.AnimStateTransitionNode_0.Transition
  nodes [Result(bCanEnterTransition=Get bTimeInParachutingAboveIntroLength())]
 }
 AnimationTransitionGraph_1 {
  path=AnimGraph.AnimGraphNode_StateMachine_0.SkyDiveState.AnimStateNode_1.DiveOrGlide.AnimGraphNode_StateMachine_212.SkydiveOrGlide.AnimStateNode_3.ParaGlide.AnimGraphNode_StateMachine_3.Glider_or_Unmbrella.AnimStateTransitionNode_2.AnimationTransitionGraph_1
  nodes [Result()]
 }
 Glider_SurfStyle {
  entry=ParaGlide_Into_Deploy
  path=AnimGraph.AnimGraphNode_StateMachine_0.SkyDiveState.AnimStateNode_1.DiveOrGlide.AnimGraphNode_StateMachine_212.SkydiveOrGlide.AnimStateNode_3.ParaGlide.AnimGraphNode_StateMachine_4.Glider_SurfStyle
  ids=id
  nodes [
   AnimStateNode_0 = ParaGlide_SurfStyle()
   AnimStateNode_1 = ParaGlide_Into_Deploy()
   AnimStateTransitionNode_0 = AnimStateTransitionNode()
   AnimStateTransitionNode_1 = AnimStateTransitionNode() {bAutomaticRuleBasedOnSequencePlayerInState=true, blendMode=Linear, crossfadeDuration=0.3}
  ]
  links [
   {
    dst=ParaGlide_SurfStyle
    rule=AnimGraph.AnimGraphNode_StateMachine_0.SkyDiveState.AnimStateNode_1.DiveOrGlide.AnimGraphNode_StateMachine_212.SkydiveOrGlide.AnimStateNode_3.ParaGlide.AnimGraphNode_StateMachine_4.Glider_SurfStyle.AnimStateTransitionNode_0.Transition
    src=ParaGlide_Into_Deploy
    via=AnimStateTransitionNode_0
   }
   {
    dst=ParaGlide_SurfStyle
    rule=AnimGraph.AnimGraphNode_StateMachine_0.SkyDiveState.AnimStateNode_1.DiveOrGlide.AnimGraphNode_StateMachine_212.SkydiveOrGlide.AnimStateNode_3.ParaGlide.AnimGraphNode_StateMachine_4.Glider_SurfStyle.AnimStateTransitionNode_1.AnimationTransitionGraph_1
    src=ParaGlide_Into_Deploy
    via=AnimStateTransitionNode_1
   }
  ]
 }
 ParaGlide_SurfStyle {
  path=AnimGraph.AnimGraphNode_StateMachine_0.SkyDiveState.AnimStateNode_1.DiveOrGlide.AnimGraphNode_StateMachine_212.SkydiveOrGlide.AnimStateNode_3.ParaGlide.AnimGraphNode_StateMachine_4.Glider_SurfStyle.AnimStateNode_0.ParaGlide_SurfStyle
  nodes [
   Apply Additive!(
    Base
    Additive="Sequence Player (additive) 'TemplateSurfer_TurnLeanLow_Pose'"() {sequence=/Game/Animation/Game/MainPlayer/Skydive/ParaGlide/_TemplateSurfer/Medium/Female/TemplateSurfer_TurnLeanLow_Pose}
    bAlphaBoolEnabled:bool=Get bParaGlideSurferHipsAreLow() -> (bParaGlideSurferHipsAreLow:bool)
   ) -> (Pose) {alphaInputType=Bool, alphaBoolBlend {blendInTime=0.125, blendOption=Linear, blendOutTime=0.125}}
   n1 = Break Fort Anim Input Skydiving!(FortAnimInput_Skydiving=Get Skydiving()) -> (
    ParaGilderRootOffset:Vector=~
    ParaGliderRootRotationOverride:Rotator=~
    bUseParaGlideRootModifier:bool=~
    bIsSkydivingFromLaunchPad:bool=~
    bIsSkydivingFromBus:bool=~
    bIsInVortex:bool=~
    bIsDiving:bool=~
    bIsParachuteOpen:bool=~
    bIsSkydiving:bool=~
    bIsParachuteLeaning:bool=~
    bIsSkydiveLeaning:bool=~
    bIsLeaning:bool=~
    bIsSkydiveDiveMode:bool=~
    bPlaySkydiveDrift:bool=~
    bSkydiveDriftDelayActive:bool=~
    bSkydiveDriftAnimAllowed:bool=~
    bIsGliderRight:bool=~
    bIsGliderCenter:bool=~
    bIsGliderLeft:bool=~
    bIsGliderForward:bool=~
    bIsGliderBack:bool=~
    bHasDeployedAndIsStandingUpright=~
    LocalAccelForward=~
    LocalAccelRight:float=~
    LocalVelocityRight:float=~
    SkydiveAimPitch:float=~
    SkydiveAimYaw=~
    DeployChuteAnimRate:float=~
    SkydiveDriftAnimRate:float=~
    SkydiveDriftAnimRateCurrent:float=~
    SkydiveFidgetAnimRate:float=~
    SkydiveFidgetAnimRateCurrent:float=~
    SkydiveAdditiveAlpha:float=~
    SkydiveDriftDelay:float=~
    SkydiveDriftAnim:int
    SkydiveDriftAnimMax:int
    LaunchpadAnim:int
    LocalAccelDir:ESkydivingDirection Enum
    DirectionLast:ESkydivingDirection Enum
    PoseAdditive
   )
   n2 = Blendspace Player 'BS_TemplateSurfer_Glide_Add'(X=@n1.SkydiveAimYaw, Y=@n1.LocalAccelForward) {
    bResetPlayTimeWhenBlendSpaceChanges=false
    blendSpace=/Game/Animation/Game/MainPlayer/Skydive/ParaGlide/_TemplateSurfer/Medium/Female/BS_TemplateSurfer_Glide_Add
    groupName=SurfGlide
    groupRole=AlwaysFollower
    method=SyncGroup
   }
   Apply Additive!(
    Base=Apply Additive(
     Base=Blendspace Player 'BS_TemplateSurfer'(X=@n1.SkydiveAimYaw, Y=@n1.LocalAccelForward) {blendSpace=/Game/Animation/Game/MainPlayer/Skydive/ParaGlide/_TemplateSurfer/Medium/Female/BS_TemplateSurfer}
     Additive=@n2
     Alpha=Get ParaGliderAddAlpha()
    ) {alphaBoolBlend {blendOption=Linear}, alphaScaleBiasClamp {outRange {max=0, min=1}}}
    Additive=Blendspace Player 'BS_TemplateSurfer_Turn'(X=@n1.SkydiveAimYaw, Y=@n1.LocalAccelForward) {blendSpace=/Game/Animation/Game/MainPlayer/Skydive/ParaGlide/_TemplateSurfer/Medium/Female/BS_TemplateSurfer_Turn}
    Alpha:float=Get ParaGliderTurnAlpha() -> (ParaGliderTurnAlpha:float)
   ) -> (Pose) {alphaBoolBlend {blendOption=Linear}, alphaScaleBiasClamp {outRange {max=0, min=1}}}
   Apply Additive!(Base, Additive, AlphaCurveName:Name=ParaGlideAddAlpha) -> (Pose) {alphaInputType=Curve, alphaBoolBlend {blendOption=Linear}, alphaScaleBiasClamp {bMapRange=true, outRange {max=0, min=1}}}
   Apply Additive(AlphaCurveName=ParaGlideAddAlpha) {alphaInputType=Curve, alphaBoolBlend {blendOption=Linear}, alphaScaleBiasClamp {bMapRange=true, outRange {max=0, min=1}}}
   n3 = Break Fort Anim Input Player Glider Anim Asset(FortAnimInput_PlayerGliderAnimAsset=Get CurrentGliderAnimSet()) -> (
    RootModPitchMin=~
    RootModPitchMax=~
    RootModYOffsetMin=~
    RootModYOffsetMax=~
    bEnableSpringMods=~
    bAllowPlayerDeployRootMod=~
    bUseSurfStyle=~
    bUseLeanPlayRateAdjust=~
    bUseForwardAs_Y_BS=~
    bUseDeployCurveForIntoPlayRateAdjust=~
    LeanForwardValue=~
    IdleValue=~
    LeanBackValue=~
    IdleInterpToSpeed=~
    LeanForwardInterpToSpeed=~
    LeanBackInterpToSpeed=~
   )
   n4 = Break Fort Anim Input Skydiving(FortAnimInput_Skydiving=Get Skydiving()) -> (
    ParaGilderRootOffset=~
    ParaGliderRootRotationOverride=~
    bUseParaGlideRootModifier=~
    bIsSkydivingFromLaunchPad=~
    bIsSkydivingFromBus=~
    bIsInVortex=~
    bIsDiving=~
    bIsParachuteOpen=~
    bIsSkydiving=~
    bIsParachuteLeaning=~
    bIsSkydiveLeaning=~
    bIsLeaning=~
    bIsSkydiveDiveMode=~
    bPlaySkydiveDrift=~
    bSkydiveDriftDelayActive=~
    bSkydiveDriftAnimAllowed=~
    bIsGliderRight=~
    bIsGliderCenter=~
    bIsGliderLeft=~
    bIsGliderForward=~
    bIsGliderBack=~
    bHasDeployedAndIsStandingUpright=~
    LocalAccelForward=~
    LocalAccelRight=~
    LocalVelocityRight=~
    SkydiveAimPitch=~
    SkydiveAimYaw=~
    DeployChuteAnimRate=~
    SkydiveDriftAnimRate=~
    SkydiveDriftAnimRateCurrent=~
    SkydiveFidgetAnimRate=~
    SkydiveFidgetAnimRateCurrent=~
    SkydiveAdditiveAlpha=~
    SkydiveDriftDelay=~
   )
   n5 = Apply Additive!(
    Base=Blendspace Player!(X=@n4.SkydiveAimYaw, Y=@n4.LocalAccelForward, BlendSpace=@n3) -> (Pose)
    Additive=Blendspace Player(X=@n4.SkydiveAimYaw, Y=@n4.LocalAccelForward, BlendSpace=@n3.Default_BodyAdditive_FemaleMedium_BS) {bResetPlayTimeWhenBlendSpaceChanges=false, groupName=SurfGlide, groupRole=AlwaysFollower, method=SyncGroup}
    Alpha:float=Get ParaGliderAddAlpha()
   ) -> (Pose) {alphaBoolBlend {blendOption=Linear}, alphaScaleBiasClamp {outRange {max=0, min=1}}}
   n6 = Blendspace Player(X=@n4.SkydiveAimYaw, Y=@n4.LocalAccelForward, BlendSpace=@n3.Default_TurnAdditive_FemaleMedium_BS)
   n7 = Get ParaGliderTurnAlpha()
   AnimGraphNode_StateMachine_0 = TriggerAnim() {is=title AnimGraphNode_StateMachine}
   Output Animation Pose(Result=Apply Additive!(
     Base=Apply Additive!(Base=@n5, Additive=@n6, Alpha:float=@n7) -> (Pose) {alphaBoolBlend {blendOption=Linear}, alphaScaleBiasClamp {outRange {max=0, min=1}}}
     Additive=@AnimGraphNode_StateMachine_0
     bAlphaBoolEnabled:bool=Get ParaGlide_TriggerAnim() -> (ParaGlide_TriggerAnim:bool)
    ) -> (Pose) {alphaInputType=Bool, alphaBoolBlend {blendInTime=0.1, blendOption=Linear, blendOutTime=0.1}})
  ]
 }
 TriggerAnim {
  entry=TriggerAnim
  path=AnimGraph.AnimGraphNode_StateMachine_0.SkyDiveState.AnimStateNode_1.DiveOrGlide.AnimGraphNode_StateMachine_212.SkydiveOrGlide.AnimStateNode_3.ParaGlide.AnimGraphNode_StateMachine_4.Glider_SurfStyle.AnimStateNode_0.ParaGlide_SurfStyle.AnimGraphNode_StateMachine_0.TriggerAnim
  ids=id
  nodes [AnimStateNode_0 = TriggerAnim() {is=title AnimStateNode}]
 }
 TriggerAnim {
  path=AnimGraph.AnimGraphNode_StateMachine_0.SkyDiveState.AnimStateNode_1.DiveOrGlide.AnimGraphNode_StateMachine_212.SkydiveOrGlide.AnimStateNode_3.ParaGlide.AnimGraphNode_StateMachine_4.Glider_SurfStyle.AnimStateNode_0.ParaGlide_SurfStyle.AnimGraphNode_StateMachine_0.TriggerAnim.AnimStateNode_0.TriggerAnim
  nodes [
   n1 = Get CurrentGliderAnimSet()
   n2 = Break Fort Anim Input Skydiving(FortAnimInput_Skydiving=Get Skydiving()) -> (
    ParaGilderRootOffset=~
    ParaGliderRootRotationOverride=~
    bUseParaGlideRootModifier=~
    bIsSkydivingFromLaunchPad=~
    bIsSkydivingFromBus=~
    bIsInVortex=~
    bIsDiving=~
    bIsParachuteOpen=~
    bIsSkydiving=~
    bIsParachuteLeaning=~
    bIsSkydiveLeaning=~
    bIsLeaning=~
    bIsSkydiveDiveMode=~
    bPlaySkydiveDrift=~
    bSkydiveDriftDelayActive=~
    bSkydiveDriftAnimAllowed=~
    bIsGliderRight=~
    bIsGliderCenter=~
    bIsGliderLeft=~
    bIsGliderForward=~
    bIsGliderBack=~
    bHasDeployedAndIsStandingUpright=~
    LocalAccelForward=~
    LocalAccelRight=~
    LocalVelocityRight=~
    SkydiveAimPitch=~
    SkydiveAimYaw=~
    DeployChuteAnimRate=~
    SkydiveDriftAnimRate=~
    SkydiveDriftAnimRateCurrent=~
    SkydiveFidgetAnimRate=~
    SkydiveFidgetAnimRateCurrent=~
    SkydiveAdditiveAlpha=~
    SkydiveDriftDelay=~
   )
   Output Animation Pose(Result=Blendspace Player(
     X=@n2.SkydiveAimYaw
     Y=@n2.LocalAccelForward
     BlendSpace=Break Fort Anim Input Player Glider Anim Asset(FortAnimInput_PlayerGliderAnimAsset=@n1) -> (
      RootModPitchMin=~
      RootModPitchMax=~
      RootModYOffsetMin=~
      RootModYOffsetMax=~
      bEnableSpringMods=~
      bAllowPlayerDeployRootMod=~
      bUseSurfStyle=~
      bUseLeanPlayRateAdjust=~
      bUseForwardAs_Y_BS=~
      bUseDeployCurveForIntoPlayRateAdjust=~
      LeanForwardValue=~
      IdleValue=~
      LeanBackValue=~
      IdleInterpToSpeed=~
      LeanForwardInterpToSpeed=~
      LeanBackInterpToSpeed=~
     ).GenericAdditive_Female_BS
    ))
  ]
 }
 ParaGlide_Into_Deploy {
  path=AnimGraph.AnimGraphNode_StateMachine_0.SkyDiveState.AnimStateNode_1.DiveOrGlide.AnimGraphNode_StateMachine_212.SkydiveOrGlide.AnimStateNode_3.ParaGlide.AnimGraphNode_StateMachine_4.Glider_SurfStyle.AnimStateNode_1.ParaGlide_Into_Deploy
  nodes [
   n1 = Break Fort Anim Input Skydiving(FortAnimInput_Skydiving=Get Skydiving()) -> (
    ParaGilderRootOffset=~
    ParaGliderRootRotationOverride=~
    bUseParaGlideRootModifier=~
    bIsSkydivingFromLaunchPad=~
    bIsSkydivingFromBus=~
    bIsInVortex=~
    bIsDiving=~
    bIsParachuteOpen=~
    bIsSkydiving=~
    bIsParachuteLeaning=~
    bIsSkydiveLeaning=~
    bIsLeaning=~
    bIsSkydiveDiveMode=~
    bPlaySkydiveDrift=~
    bSkydiveDriftDelayActive=~
    bSkydiveDriftAnimAllowed=~
    bIsGliderRight=~
    bIsGliderCenter=~
    bIsGliderLeft=~
    bIsGliderForward=~
    bIsGliderBack=~
    bHasDeployedAndIsStandingUpright=~
    LocalAccelForward=~
    LocalAccelRight=~
    LocalVelocityRight=~
    SkydiveAimPitch=~
    SkydiveAimYaw=~
    DeployChuteAnimRate=~
    SkydiveDriftAnimRate=~
    SkydiveDriftAnimRateCurrent=~
    SkydiveFidgetAnimRate=~
    SkydiveFidgetAnimRateCurrent=~
    SkydiveAdditiveAlpha=~
    SkydiveDriftDelay=~
   )
   n2 = Blendspace Player 'BS_TemplateSurfer_Glide_Add'(X=@n1.SkydiveAimYaw, Y=@n1.LocalAccelForward) {
    bResetPlayTimeWhenBlendSpaceChanges=false
    blendSpace=&BS_TemplateSurfer_Glide_Add
    groupName=SurfGlide
    groupRole=AlwaysLeader
    method=SyncGroup
   }
   Apply Additive!(
    Base=Blendspace Player 'BS_TemplateSurfer_Into'!(
     X=@n1.SkydiveAimYaw
     Y=@n1.SkydiveAimPitch
     BlendSpace:Blend Space Object Reference=/Game/Animation/Game/MainPlayer/Skydive/ParaGlide/_TemplateSurfer/Medium/Female/BS_TemplateSurfer_Into
    ) -> (Pose) {bLoop=false}
    Additive=@n2
    AlphaCurveName:Name=ParaGlideLeanAlpha
   ) -> (Pose) {alphaInputType=Curve, alphaBoolBlend {blendOption=Linear}}
   n3 = Break Fort Anim Input Player Glider Anim Asset!(FortAnimInput_PlayerGliderAnimAsset=Get CurrentGliderAnimSet()) -> (
    Into_Base_MaleMedium_BS
    Into_BodyAdditive_MaleMedium_BS
    Into_BodyAdditive_FemaleMedium_BS
    PelvisModVertical
    PelvisModLateral
    PelvisModBankAngle
    ThighModLegAngle
    ThighModLegPitch
    ThighModLegBank
    ThighModLeftLegDrag
    PlayerGliderType
    bEnableSpringMods=~
    bUseLeanPlayRateAdjust=~
    bUseForwardAs_Y_BS=~
    bUseDeployCurveForIntoPlayRateAdjust=~
    LeanForwardValue=~
    IdleValue=~
    LeanBackValue=~
    IdleInterpToSpeed=~
    LeanForwardInterpToSpeed=~
    LeanBackInterpToSpeed=~
   )
   n4 = Break Fort Anim Input Skydiving(FortAnimInput_Skydiving=Get Skydiving()) -> (
    ParaGilderRootOffset=~
    ParaGliderRootRotationOverride=~
    bUseParaGlideRootModifier=~
    bIsSkydivingFromLaunchPad=~
    bIsSkydivingFromBus=~
    bIsInVortex=~
    bIsDiving=~
    bIsParachuteOpen=~
    bIsSkydiving=~
    bIsParachuteLeaning=~
    bIsSkydiveLeaning=~
    bIsLeaning=~
    bIsSkydiveDiveMode=~
    bPlaySkydiveDrift=~
    bSkydiveDriftDelayActive=~
    bSkydiveDriftAnimAllowed=~
    bIsGliderRight=~
    bIsGliderCenter=~
    bIsGliderLeft=~
    bIsGliderForward=~
    bIsGliderBack=~
    bHasDeployedAndIsStandingUpright=~
    LocalAccelForward=~
    LocalAccelRight=~
    LocalVelocityRight=~
    SkydiveAimPitch=~
    SkydiveAimYaw=~
    DeployChuteAnimRate=~
    SkydiveDriftAnimRate=~
    SkydiveDriftAnimRateCurrent=~
    SkydiveFidgetAnimRate=~
    SkydiveFidgetAnimRateCurrent=~
    SkydiveAdditiveAlpha=~
    SkydiveDriftDelay=~
   )
   n5 = "Blend Poses (EFortCustomGender)"(
    BlendPose_0=Blendspace Player(X=@n4.SkydiveAimYaw, Y=@n4.SkydiveAimPitch, BlendSpace=@n3.Into_BodyAdditive_MaleMedium_BS) {bIgnoreForRelevancyTest=true, bLoop=false}
    BlendPose_1=Blendspace Player(X=@n4.SkydiveAimYaw, Y=@n4.SkydiveAimPitch, BlendSpace=@n3.Into_BodyAdditive_FemaleMedium_BS) {bIgnoreForRelevancyTest=true, bLoop=false}
    BlendTime_0=0
    BlendTime_1=0
    ActiveEnumValue=Get Gender() -> (Gender=Invalid)
   ) {blendType=Linear, blendPose [{linkId=-1, sourceLinkId=-1}, {linkId=-1, sourceLinkId=-1}], blendTime [0.1,0.1]}
   n6 = Get CurrentGliderAnimSet()
   n7 = Apply Additive(
    Base=Blendspace Player(
     X=@n4.SkydiveAimYaw
     Y=@n4.SkydiveAimPitch
     BlendSpace=Break Fort Anim Input Player Glider Anim Asset!(FortAnimInput_PlayerGliderAnimAsset=@n6) -> (
      Into_Base_BS:Blend Space Object Reference
      Into_Base_MaleMedium_BS
      PelvisModVertical
      PelvisModLateral
      PelvisModBankAngle
      ThighModLegAngle
      ThighModLegPitch
      ThighModLegBank
      ThighModLeftLegDrag
      PlayerGliderType
      bEnableSpringMods=~
      bUseLeanPlayRateAdjust=~
      bUseForwardAs_Y_BS=~
      bUseDeployCurveForIntoPlayRateAdjust=~
      LeanForwardValue=~
      IdleValue=~
      LeanBackValue=~
      IdleInterpToSpeed=~
      LeanForwardInterpToSpeed=~
      LeanBackInterpToSpeed=~
     )
    ) {bLoop=false}
    Additive=@n5
    AlphaCurveName=ParaGlideLeanAlpha
   ) {alphaInputType=Curve, alphaBoolBlend {blendOption=Linear}, alphaScaleBiasClamp {bMapRange=true, inRange {max=0, min=1}}}
   n8 = Blendspace Player(
    X=@n4.SkydiveAimYaw
    Y=@n4.LocalAccelForward
    BlendSpace=Break Fort Anim Input Player Glider Anim Asset!(FortAnimInput_PlayerGliderAnimAsset=Get CurrentGliderAnimSet()) -> (
     Default_BodyAdditive_FemaleMedium_BS:Blend Space Object Reference
     Into_Base_MaleMedium_BS
     PelvisModVertical
     PelvisModLateral
     PelvisModBankAngle
     ThighModLegAngle
     ThighModLegPitch
     ThighModLegBank
     ThighModLeftLegDrag
     PlayerGliderType
     bEnableSpringMods=~
     bUseLeanPlayRateAdjust=~
     bUseForwardAs_Y_BS=~
     bUseDeployCurveForIntoPlayRateAdjust=~
     LeanForwardValue=~
     IdleValue=~
     LeanBackValue=~
     IdleInterpToSpeed=~
     LeanForwardInterpToSpeed=~
     LeanBackInterpToSpeed=~
    )
   ) {bIgnoreForRelevancyTest=true, bResetPlayTimeWhenBlendSpaceChanges=false, groupName=SurfGlide, groupRole=AlwaysLeader, method=SyncGroup}
   Output Animation Pose(Result=Apply Additive(Base=@n7, Additive=@n8, AlphaCurveName=ParaGlideLeanAlpha) {alphaInputType=Curve, alphaBoolBlend {blendOption=Linear}, alphaScaleBiasClamp {inRange {max=0, min=1}}})
  ]
 }
 Transition {
  path=AnimGraph.AnimGraphNode_StateMachine_0.SkyDiveState.AnimStateNode_1.DiveOrGlide.AnimGraphNode_StateMachine_212.SkydiveOrGlide.AnimStateNode_3.ParaGlide.AnimGraphNode_StateMachine_4.Glider_SurfStyle.AnimStateTransitionNode_0.Transition
  nodes [Result(bCanEnterTransition=Get bTimeInParachutingAboveIntroLength())]
 }
 AnimationTransitionGraph_1 {
  path=AnimGraph.AnimGraphNode_StateMachine_0.SkyDiveState.AnimStateNode_1.DiveOrGlide.AnimGraphNode_StateMachine_212.SkydiveOrGlide.AnimStateNode_3.ParaGlide.AnimGraphNode_StateMachine_4.Glider_SurfStyle.AnimStateTransitionNode_1.AnimationTransitionGraph_1
  nodes [Result()]
 }
 NoGlider {
  entry=ParaGlide_Into_Deploy
  path=AnimGraph.AnimGraphNode_StateMachine_0.SkyDiveState.AnimStateNode_1.DiveOrGlide.AnimGraphNode_StateMachine_212.SkydiveOrGlide.AnimStateNode_3.ParaGlide.AnimGraphNode_StateMachine_73.NoGlider
  ids=id
  nodes [
   AnimStateNode_1 = ParaGlide_Into_Deploy()
   AnimStateNode_3 = ParaGlide()
   AnimStateTransitionNode_0 = AnimStateTransitionNode()
   AnimStateTransitionNode_2 = AnimStateTransitionNode() {bAutomaticRuleBasedOnSequencePlayerInState=true, crossfadeDuration=0}
  ]
  links [
   {
    dst=ParaGlide
    rule=AnimGraph.AnimGraphNode_StateMachine_0.SkyDiveState.AnimStateNode_1.DiveOrGlide.AnimGraphNode_StateMachine_212.SkydiveOrGlide.AnimStateNode_3.ParaGlide.AnimGraphNode_StateMachine_73.NoGlider.AnimStateTransitionNode_0.Transition
    src=ParaGlide_Into_Deploy
    via=AnimStateTransitionNode_0
   }
   {
    dst=ParaGlide
    rule=AnimGraph.AnimGraphNode_StateMachine_0.SkyDiveState.AnimStateNode_1.DiveOrGlide.AnimGraphNode_StateMachine_212.SkydiveOrGlide.AnimStateNode_3.ParaGlide.AnimGraphNode_StateMachine_73.NoGlider.AnimStateTransitionNode_2.AnimationTransitionGraph_1
    src=ParaGlide_Into_Deploy
    via=AnimStateTransitionNode_2
   }
  ]
 }
 ParaGlide_Into_Deploy {
  path=AnimGraph.AnimGraphNode_StateMachine_0.SkyDiveState.AnimStateNode_1.DiveOrGlide.AnimGraphNode_StateMachine_212.SkydiveOrGlide.AnimStateNode_3.ParaGlide.AnimGraphNode_StateMachine_73.NoGlider.AnimStateNode_1.ParaGlide_Into_Deploy
  nodes [
   n1 = Break Fort Anim Input Player Glider Anim Asset!(FortAnimInput_PlayerGliderAnimAsset=Get CurrentGliderAnimSet()) -> (
    Default_Base_BS:Blend Space Object Reference
    Default_BodyAdditive_MaleMedium_BS:Blend Space Object Reference
    Default_BodyAdditive_MaleLarge_BS:Blend Space Object Reference
    Default_BodyAdditive_FemaleSmall_BS:Blend Space Object Reference
    Default_BodyAdditive_FemaleMedium_BS
    Default_BodyAdditive_FemaleLarge_BS:Blend Space Object Reference
    Default_TurnAdditive_MaleMedium_BS:Blend Space Object Reference
    Default_TurnAdditive_MaleLarge_BS:Blend Space Object Reference
    Default_TurnAdditive_FemaleSmall_BS:Blend Space Object Reference
    Default_TurnAdditive_FemaleMedium_BS:Blend Space Object Reference
    Default_TurnAdditive_FemaleLarge_BS:Blend Space Object Reference
    Into_Base_BS:Blend Space Object Reference
    Into_Base_MaleMedium_BS
    Into_BodyAdditive_MaleMedium_BS:Blend Space Object Reference
    Into_BodyAdditive_MaleLarge_BS:Blend Space Object Reference
    Into_BodyAdditive_FemaleSmall_BS:Blend Space Object Reference
    Into_BodyAdditive_FemaleMedium_BS:Blend Space Object Reference
    Into_BodyAdditive_FemaleLarge_BS:Blend Space Object Reference
    Into_TurnAdditive_MaleMedium_BS:Blend Space Object Reference
    Into_TurnAdditive_MaleLarge_BS:Blend Space Object Reference
    Into_TurnAdditive_FemaleSmall_BS:Blend Space Object Reference
    Into_TurnAdditive_FemaleMedium_BS:Blend Space Object Reference
    Into_TurnAdditive_FemaleLarge_BS:Blend Space Object Reference
    Lean_MaleMedium_BS:Blend Space Object Reference
    Lean_MaleLarge_BS:Blend Space Object Reference
    Lean_FemaleSmall_BS:Blend Space Object Reference
    Lean_FemaleMedium_BS:Blend Space Object Reference
    Lean_FemaleLarge_BS:Blend Space Object Reference
    LeanAdditive_Center_MaleMedium_Pose:Anim Sequence Object Reference
    LeanAdditive_Center_MaleLarge_Pose:Anim Sequence Object Reference
    LeanAdditive_Center_FemaleSmall_Pose:Anim Sequence Object Reference
    LeanAdditive_Center_FemaleMedium_Pose:Anim Sequence Object Reference
    LeanAdditive_Center_FemaleLarge_Pose:Anim Sequence Object Reference
    LeanAdditive_Into_BS:Blend Space Object Reference
    LeanAdditive_ForwardInto_Anim:Anim Sequence Object Reference
    LeanAdditive_ForwardInto_FromDeploy_Anim:Anim Sequence Object Reference
    LeanAdditive_BackInto_Anim:Anim Sequence Object Reference
    LeanAdditive_LeftInto_Anim:Anim Sequence Object Reference
    LeanAdditive_RightInto_Anim:Anim Sequence Object Reference
    LeanAdditive_ForwardOut_Anim:Anim Sequence Object Reference
    LeanAdditive_BackOut_Anim:Anim Sequence Object Reference
    LeanAdditive_LeftOut_Anim:Anim Sequence Object Reference
    LeanAdditive_RightOut_Anim:Anim Sequence Object Reference
    ToGlide_BS:Blend Space Object Reference
    ToGlide_Lean_BS:Blend Space Object Reference
    ToDive_BS:Blend Space Object Reference
    ToDive_Lean_BS:Blend Space Object Reference
    Dive_WeaponR_Additive_BS:Blend Space Object Reference
    Glide_WeaponR_Additive_Anim:Anim Sequence Object Reference
    GenericAdditive_Male_BS:Blend Space Object Reference
    GenericAdditive_Female_BS:Blend Space Object Reference
    RootModPitchMin:float=~
    RootModPitchMax:float=~
    RootModYOffsetMin:float=~
    RootModYOffsetMax:float=~
    PelvisModVertical
    PelvisModLateral
    PelvisModBankAngle
    ThighModLegAngle
    ThighModLegPitch
    ThighModLegBank
    ThighModLeftLegDrag
    PlayerGliderType
    bEnableSpringMods=~
    bAllowPlayerDeployRootMod:bool=~
    bUseSurfStyle:bool=~
    bUseLeanPlayRateAdjust=~
    bUseForwardAs_Y_BS=~
    bUseDeployCurveForIntoPlayRateAdjust=~
    LeanForwardValue=~
    IdleValue=~
    LeanBackValue=~
    IdleInterpToSpeed=~
    LeanForwardInterpToSpeed=~
    LeanBackInterpToSpeed=~
   )
   n2 = Break Fort Anim Input Skydiving(FortAnimInput_Skydiving=Get Skydiving()) -> (
    ParaGilderRootOffset=~
    ParaGliderRootRotationOverride=~
    bUseParaGlideRootModifier=~
    bIsSkydivingFromLaunchPad=~
    bIsSkydivingFromBus=~
    bIsInVortex=~
    bIsDiving=~
    bIsParachuteOpen=~
    bIsSkydiving=~
    bIsParachuteLeaning=~
    bIsSkydiveLeaning=~
    bIsLeaning=~
    bIsSkydiveDiveMode=~
    bPlaySkydiveDrift=~
    bSkydiveDriftDelayActive=~
    bSkydiveDriftAnimAllowed=~
    bIsGliderRight=~
    bIsGliderCenter=~
    bIsGliderLeft=~
    bIsGliderForward=~
    bIsGliderBack=~
    bHasDeployedAndIsStandingUpright=~
    LocalAccelForward=~
    LocalAccelRight=~
    LocalVelocityRight=~
    SkydiveAimPitch=~
    SkydiveAimYaw=~
    DeployChuteAnimRate=~
    SkydiveDriftAnimRate=~
    SkydiveDriftAnimRateCurrent=~
    SkydiveFidgetAnimRate=~
    SkydiveFidgetAnimRateCurrent=~
    SkydiveAdditiveAlpha=~
    SkydiveDriftDelay=~
   )
   n3 = Get AimPitch()
   n4 = "Blend Poses (EFortPlayerAnimBodyType)"!(
    BlendPose_0=Blendspace Player(X=@n2.SkydiveAimYaw, Y=@n3, BlendSpace=@n1.Default_BodyAdditive_MaleMedium_BS) {bIgnoreForRelevancyTest=true}
    BlendPose_1=Blendspace Player(X=@n2.SkydiveAimYaw, Y=@n3, BlendSpace=@n1.Default_BodyAdditive_MaleLarge_BS) {bIgnoreForRelevancyTest=true}
    BlendTime_0=0.1
    BlendTime_1=0.1
    ActiveEnumValue=Get AnimBodyType() -> (AnimBodyType=Small)
   ) -> (Pose) {blendType=Linear, blendPose [{linkId=-1, sourceLinkId=-1}, {linkId=-1, sourceLinkId=-1}], blendTime [0.1,0.1]}
   n5 = "Blend Poses (EFortPlayerAnimBodyType)"!(
    BlendPose_0=Blendspace Player(X=@n2.SkydiveAimYaw, Y=@n3, BlendSpace=@n1.Default_BodyAdditive_FemaleMedium_BS) {bIgnoreForRelevancyTest=true}
    BlendPose_1=Blendspace Player(X=@n2.SkydiveAimYaw, Y=@n3, BlendSpace=@n1.Default_BodyAdditive_FemaleSmall_BS) {bIgnoreForRelevancyTest=true}
    BlendPose_2:Pose Link Structure=Blendspace Player(X=@n2.SkydiveAimYaw, Y=@n3, BlendSpace=@n1.Default_BodyAdditive_FemaleLarge_BS) {bIgnoreForRelevancyTest=true}
    BlendTime_0=0.1
    BlendTime_1=0.1
    BlendTime_2:float=0.1
    ActiveEnumValue=Get AnimBodyType() -> (AnimBodyType=Small)
   ) -> (Pose) {
    blendType=Linear
    blendPose [{linkId=-1, sourceLinkId=-1}, {linkId=-1, sourceLinkId=-1}, {linkId=-1, sourceLinkId=-1}]
    blendTime [0.1,0.1,0.1]
   }
   n6 = "Blend Poses (EFortCustomGender)"(BlendPose_0=@n4, BlendPose_1=@n5, BlendTime_0=0, BlendTime_1=0, ActiveEnumValue=Get Gender() -> (Gender=Invalid)) {blendType=Linear, blendPose [{linkId=-1, sourceLinkId=-1}, {linkId=-1, sourceLinkId=-1}], blendTime [0.1,0.1]}
   n7 = Break Fort Anim Input Skydiving(FortAnimInput_Skydiving=Get Skydiving()) -> (
    ParaGilderRootOffset=~
    ParaGliderRootRotationOverride=~
    bUseParaGlideRootModifier=~
    bIsSkydivingFromLaunchPad=~
    bIsSkydivingFromBus=~
    bIsInVortex=~
    bIsDiving=~
    bIsParachuteOpen=~
    bIsSkydiving=~
    bIsParachuteLeaning=~
    bIsSkydiveLeaning=~
    bIsLeaning=~
    bIsSkydiveDiveMode=~
    bPlaySkydiveDrift=~
    bSkydiveDriftDelayActive=~
    bSkydiveDriftAnimAllowed=~
    bIsGliderRight=~
    bIsGliderCenter=~
    bIsGliderLeft=~
    bIsGliderForward=~
    bIsGliderBack=~
    bHasDeployedAndIsStandingUpright=~
    LocalAccelForward=~
    LocalAccelRight=~
    LocalVelocityRight=~
    SkydiveAimPitch=~
    SkydiveAimYaw=~
    DeployChuteAnimRate=~
    SkydiveDriftAnimRate=~
    SkydiveDriftAnimRateCurrent=~
    SkydiveFidgetAnimRate=~
    SkydiveFidgetAnimRateCurrent=~
    SkydiveAdditiveAlpha=~
    SkydiveDriftDelay=~
   )
   n8 = Break Fort Anim Input Player Glider Anim Asset(FortAnimInput_PlayerGliderAnimAsset=Get CurrentGliderAnimSet()) -> (
    RootModPitchMin=~
    RootModPitchMax=~
    RootModYOffsetMin=~
    RootModYOffsetMax=~
    bEnableSpringMods=~
    bAllowPlayerDeployRootMod=~
    bUseSurfStyle=~
    bUseLeanPlayRateAdjust=~
    bUseForwardAs_Y_BS=~
    bUseDeployCurveForIntoPlayRateAdjust=~
    LeanForwardValue=~
    IdleValue=~
    LeanBackValue=~
    IdleInterpToSpeed=~
    LeanForwardInterpToSpeed=~
    LeanBackInterpToSpeed=~
   )
   n9 = Get Glider_BaseBS_Y_Axis()
   n10 = Break Fort Anim Input Player Glider Anim Asset(FortAnimInput_PlayerGliderAnimAsset=Get CurrentGliderAnimSet()) -> (
    RootModPitchMin=~
    RootModPitchMax=~
    RootModYOffsetMin=~
    RootModYOffsetMax=~
    bEnableSpringMods=~
    bAllowPlayerDeployRootMod=~
    bUseSurfStyle=~
    bUseLeanPlayRateAdjust=~
    bUseForwardAs_Y_BS=~
    bUseDeployCurveForIntoPlayRateAdjust=~
    LeanForwardValue=~
    IdleValue=~
    LeanBackValue=~
    IdleInterpToSpeed=~
    LeanForwardInterpToSpeed=~
    LeanBackInterpToSpeed=~
   )
   n11 = Break Fort Anim Input Skydiving(FortAnimInput_Skydiving=Get Skydiving()) -> (
    ParaGilderRootOffset=~
    ParaGliderRootRotationOverride=~
    bUseParaGlideRootModifier=~
    bIsSkydivingFromLaunchPad=~
    bIsSkydivingFromBus=~
    bIsInVortex=~
    bIsDiving=~
    bIsParachuteOpen=~
    bIsSkydiving=~
    bIsParachuteLeaning=~
    bIsSkydiveLeaning=~
    bIsLeaning=~
    bIsSkydiveDiveMode=~
    bPlaySkydiveDrift=~
    bSkydiveDriftDelayActive=~
    bSkydiveDriftAnimAllowed=~
    bIsGliderRight=~
    bIsGliderCenter=~
    bIsGliderLeft=~
    bIsGliderForward=~
    bIsGliderBack=~
    bHasDeployedAndIsStandingUpright=~
    LocalAccelForward=~
    LocalAccelRight=~
    LocalVelocityRight=~
    SkydiveAimPitch=~
    SkydiveAimYaw=~
    DeployChuteAnimRate=~
    SkydiveDriftAnimRate=~
    SkydiveDriftAnimRateCurrent=~
    SkydiveFidgetAnimRate=~
    SkydiveFidgetAnimRateCurrent=~
    SkydiveAdditiveAlpha=~
    SkydiveDriftDelay=~
   )
   n12 = Get AimPitch()
   n13 = "Blend Poses (EFortPlayerAnimBodyType)"!(
    BlendPose_0=Blendspace Player(X=@n11.SkydiveAimYaw, Y=@n12, BlendSpace=@n10.Default_TurnAdditive_MaleMedium_BS) {bIgnoreForRelevancyTest=true}
    BlendPose_1=Blendspace Player(X=@n11.SkydiveAimYaw, Y=@n12, BlendSpace=@n10.Default_TurnAdditive_MaleLarge_BS) {bIgnoreForRelevancyTest=true}
    BlendTime_0=0.1
    BlendTime_1=0.1
    ActiveEnumValue=Get AnimBodyType() -> (AnimBodyType=Small)
   ) -> (Pose) {blendType=Linear, blendPose [{linkId=-1, sourceLinkId=-1}, {linkId=-1, sourceLinkId=-1}], blendTime [0.1,0.1]}
   n14 = "Blend Poses (EFortPlayerAnimBodyType)"!(
    BlendPose_0=Blendspace Player(X=@n11.SkydiveAimYaw, Y=@n12, BlendSpace=@n10.Default_TurnAdditive_FemaleMedium_BS) {bIgnoreForRelevancyTest=true}
    BlendPose_1=Blendspace Player(X=@n11.SkydiveAimYaw, Y=@n12, BlendSpace=@n10.Default_TurnAdditive_FemaleSmall_BS) {bIgnoreForRelevancyTest=true}
    BlendPose_2:Pose Link Structure=Blendspace Player(X=@n11.SkydiveAimYaw, Y=@n12, BlendSpace=@n10.Default_TurnAdditive_FemaleLarge_BS) {bIgnoreForRelevancyTest=true}
    BlendTime_0=0.1
    BlendTime_1=0.1
    BlendTime_2:float=0.1
    ActiveEnumValue=Get AnimBodyType() -> (AnimBodyType=Small)
   ) -> (Pose) {
    blendType=Linear
    blendPose [{linkId=-1, sourceLinkId=-1}, {linkId=-1, sourceLinkId=-1}, {linkId=-1, sourceLinkId=-1}]
    blendTime [0.1,0.1,0.1]
   }
   n15 = "Blend Poses (EFortCustomGender)"(BlendPose_0=@n13, BlendPose_1=@n14, BlendTime_0=0, BlendTime_1=0, ActiveEnumValue=Get Gender() -> (Gender=Invalid)) {blendType=Linear, blendPose [{linkId=-1, sourceLinkId=-1}, {linkId=-1, sourceLinkId=-1}], blendTime [0.1,0.1]}
   n16 = Apply Additive(
    Base=Apply Additive!(Base=Blendspace Player(X=@n7.SkydiveAimYaw, Y=@n9, BlendSpace=@n8.Into_Base_BS) {bLoop=false}, Additive=@n6, Alpha:float) -> (Pose) {alphaBoolBlend {blendOption=Linear}}
    Additive=@n15
    AlphaCurveName=ParaGlideTurn
   ) {alphaInputType=Curve, alphaBoolBlend {blendOption=Linear}, alphaScaleBiasClamp {bMapRange=true, outRange {max=0, min=1}}}
   Output Animation Pose(Result=@n16)
   n17 = Break Fort Anim Input Player Glider Anim Asset(FortAnimInput_PlayerGliderAnimAsset=Get CurrentGliderAnimSet()) -> (
    RootModPitchMin=~
    RootModPitchMax=~
    RootModYOffsetMin=~
    RootModYOffsetMax=~
    bEnableSpringMods=~
    bAllowPlayerDeployRootMod=~
    bUseSurfStyle=~
    bUseLeanPlayRateAdjust=~
    bUseForwardAs_Y_BS=~
    bUseDeployCurveForIntoPlayRateAdjust=~
    LeanForwardValue=~
    IdleValue=~
    LeanBackValue=~
    IdleInterpToSpeed=~
    LeanForwardInterpToSpeed=~
    LeanBackInterpToSpeed=~
   )
   n18 = Break Fort Anim Input Skydiving(FortAnimInput_Skydiving=Get Skydiving()) -> (
    ParaGilderRootOffset=~
    ParaGliderRootRotationOverride=~
    bUseParaGlideRootModifier=~
    bIsSkydivingFromLaunchPad=~
    bIsSkydivingFromBus=~
    bIsInVortex=~
    bIsDiving=~
    bIsParachuteOpen=~
    bIsSkydiving=~
    bIsParachuteLeaning=~
    bIsSkydiveLeaning=~
    bIsLeaning=~
    bIsSkydiveDiveMode=~
    bPlaySkydiveDrift=~
    bSkydiveDriftDelayActive=~
    bSkydiveDriftAnimAllowed=~
    bIsGliderRight=~
    bIsGliderCenter=~
    bIsGliderLeft=~
    bIsGliderForward=~
    bIsGliderBack=~
    bHasDeployedAndIsStandingUpright=~
    LocalAccelForward=~
    LocalAccelRight=~
    LocalVelocityRight=~
    SkydiveAimPitch=~
    SkydiveAimYaw=~
    DeployChuteAnimRate=~
    SkydiveDriftAnimRate=~
    SkydiveDriftAnimRateCurrent=~
    SkydiveFidgetAnimRate=~
    SkydiveFidgetAnimRateCurrent=~
    SkydiveAdditiveAlpha=~
    SkydiveDriftDelay=~
   )
   n19 = Get AimPitch()
   n20 = "Blend Poses (EFortPlayerAnimBodyType)"!(
    BlendPose_0=Blendspace Player(X=@n18.SkydiveAimYaw, Y=@n19, BlendSpace=@n17.Into_TurnAdditive_MaleMedium_BS) {bIgnoreForRelevancyTest=true}
    BlendPose_1=Blendspace Player(X=@n18.SkydiveAimYaw, Y=@n19, BlendSpace=@n17.Into_TurnAdditive_MaleLarge_BS) {bIgnoreForRelevancyTest=true}
    BlendTime_0=0.1
    BlendTime_1=0.1
    ActiveEnumValue=Get AnimBodyType() -> (AnimBodyType=Small)
   ) -> (Pose) {blendType=Linear, blendPose [{linkId=-1, sourceLinkId=-1}, {linkId=-1, sourceLinkId=-1}], blendTime [0.1,0.1]}
   n21 = "Blend Poses (EFortPlayerAnimBodyType)"!(
    BlendPose_0=Blendspace Player(X=@n18.SkydiveAimYaw, Y=@n19, BlendSpace=@n17.Into_TurnAdditive_FemaleMedium_BS) {bIgnoreForRelevancyTest=true}
    BlendPose_1=Blendspace Player(X=@n18.SkydiveAimYaw, Y=@n19, BlendSpace=@n17.Into_TurnAdditive_FemaleSmall_BS) {bIgnoreForRelevancyTest=true}
    BlendPose_2:Pose Link Structure=Blendspace Player(X=@n18.SkydiveAimYaw, Y=@n19, BlendSpace=@n17.Into_TurnAdditive_FemaleLarge_BS) {bIgnoreForRelevancyTest=true}
    BlendTime_0=0.1
    BlendTime_1=0.1
    BlendTime_2:float=0.1
    ActiveEnumValue=Get AnimBodyType() -> (AnimBodyType=Small)
   ) -> (Pose) {
    blendType=Linear
    blendPose [{linkId=-1, sourceLinkId=-1}, {linkId=-1, sourceLinkId=-1}, {linkId=-1, sourceLinkId=-1}]
    blendTime [0.1,0.1,0.1]
   }
   n22 = "Blend Poses (EFortCustomGender)"(BlendPose_0=@n20, BlendPose_1=@n21, BlendTime_0=0, BlendTime_1=0, ActiveEnumValue=Get Gender() -> (Gender=Invalid)) {blendType=Linear, blendPose [{linkId=-1, sourceLinkId=-1}, {linkId=-1, sourceLinkId=-1}], blendTime [0.1,0.1]}
   Two Way Blend(A=@n22, AlphaCurveName=BlendIn) {blendNode {alphaInputType=Curve, alphaBoolBlend {blendOption=Linear}, alphaScaleBiasClamp {bInterpResult=true}}}
  ]
 }
 ParaGlide {
  path=AnimGraph.AnimGraphNode_StateMachine_0.SkyDiveState.AnimStateNode_1.DiveOrGlide.AnimGraphNode_StateMachine_212.SkydiveOrGlide.AnimStateNode_3.ParaGlide.AnimGraphNode_StateMachine_73.NoGlider.AnimStateNode_3.ParaGlide
  nodes [
   n1 = Break Fort Anim Input Skydiving!(FortAnimInput_Skydiving=Get Skydiving()) -> (bHasDeployedAndIsStandingUpright=~, LocalAccelForward=~, LocalAccelRight=~, SkydiveAimYaw=~, PoseAdditive)
   n2 = Break Fort Anim Input Player Glider Anim Asset(FortAnimInput_PlayerGliderAnimAsset=Get CurrentGliderAnimSet()) -> (
    RootModPitchMin=~
    RootModPitchMax=~
    RootModYOffsetMin=~
    RootModYOffsetMax=~
    bEnableSpringMods=~
    bAllowPlayerDeployRootMod=~
    bUseSurfStyle=~
    bUseLeanPlayRateAdjust=~
    bUseForwardAs_Y_BS=~
    bUseDeployCurveForIntoPlayRateAdjust=~
    LeanForwardValue=~
    IdleValue=~
    LeanBackValue=~
    IdleInterpToSpeed=~
    LeanForwardInterpToSpeed=~
    LeanBackInterpToSpeed=~
   )
   n3 = Get AimPitch()
   n4 = "Blend Poses (EFortPlayerAnimBodyType)"!(
    BlendPose_0=Blendspace Player(X=@n1.SkydiveAimYaw, Y=@n3, BlendSpace=@n2.Default_BodyAdditive_MaleMedium_BS)
    BlendPose_1=Blendspace Player(X=@n1.SkydiveAimYaw, Y=@n3, BlendSpace=@n2.Default_BodyAdditive_MaleLarge_BS)
    BlendTime_0=0.1
    BlendTime_1=0.1
    ActiveEnumValue=Get AnimBodyType() -> (AnimBodyType=Small)
   ) -> (Pose) {blendType=Linear, blendPose [{linkId=-1, sourceLinkId=-1}, {linkId=-1, sourceLinkId=-1}], blendTime [0.1,0.1]}
   n5 = "Blend Poses (EFortPlayerAnimBodyType)"!(
    BlendPose_0=Blendspace Player(X=@n1.SkydiveAimYaw, Y=@n3, BlendSpace=@n2.Default_BodyAdditive_FemaleMedium_BS)
    BlendPose_1=Blendspace Player(X=@n1.SkydiveAimYaw, Y=@n3, BlendSpace=@n2.Default_BodyAdditive_FemaleSmall_BS)
    BlendPose_2:Pose Link Structure=Blendspace Player(X=@n1.SkydiveAimYaw, Y=@n3, BlendSpace=@n2.Default_BodyAdditive_FemaleLarge_BS)
    BlendTime_0=0.1
    BlendTime_1=0.1
    BlendTime_2:float=0.1
    ActiveEnumValue=Get AnimBodyType() -> (AnimBodyType=Small)
   ) -> (Pose) {
    blendType=Linear
    blendPose [{linkId=-1, sourceLinkId=-1}, {linkId=-1, sourceLinkId=-1}, {linkId=-1, sourceLinkId=-1}]
    blendTime [0.1,0.1,0.1]
   }
   n6 = "Blend Poses (EFortCustomGender)"(BlendPose_0=@n4, BlendPose_1=@n5, BlendTime_0=0, BlendTime_1=0, ActiveEnumValue=Get Gender() -> (Gender=Invalid)) {blendType=Linear, blendPose [{linkId=-1, sourceLinkId=-1}, {linkId=-1, sourceLinkId=-1}], blendTime [0.1,0.1]}
   n7 = Break Fort Anim Input Player Glider Anim Asset(FortAnimInput_PlayerGliderAnimAsset=Get CurrentGliderAnimSet()) -> (
    RootModPitchMin=~
    RootModPitchMax=~
    RootModYOffsetMin=~
    RootModYOffsetMax=~
    bEnableSpringMods=~
    bAllowPlayerDeployRootMod=~
    bUseSurfStyle=~
    bUseLeanPlayRateAdjust=~
    bUseForwardAs_Y_BS=~
    bUseDeployCurveForIntoPlayRateAdjust=~
    LeanForwardValue=~
    IdleValue=~
    LeanBackValue=~
    IdleInterpToSpeed=~
    LeanForwardInterpToSpeed=~
    LeanBackInterpToSpeed=~
   )
   n8 = Get Glider_BaseBS_Y_Axis()
   n9 = "Blend Poses (EFortPlayerAnimBodyType)"!(
    BlendPose_0=Blendspace Player(X=@n1.SkydiveAimYaw, Y=@n1.LocalAccelForward, BlendSpace=@n2.Default_TurnAdditive_MaleMedium_BS)
    BlendPose_1=Blendspace Player(X=@n1.SkydiveAimYaw, Y=@n1.LocalAccelForward, BlendSpace=@n2.Default_TurnAdditive_MaleLarge_BS)
    BlendTime_0=0.1
    BlendTime_1=0.1
    ActiveEnumValue=Get AnimBodyType() -> (AnimBodyType=Small)
   ) -> (Pose) {blendType=Linear, blendPose [{linkId=-1, sourceLinkId=-1}, {linkId=-1, sourceLinkId=-1}], blendTime [0.1,0.1]}
   n10 = "Blend Poses (EFortPlayerAnimBodyType)"!(
    BlendPose_0=Blendspace Player(X=@n1.SkydiveAimYaw, Y=@n1.LocalAccelForward, BlendSpace=@n2.Default_TurnAdditive_FemaleMedium_BS)
    BlendPose_1=Blendspace Player(X=@n1.SkydiveAimYaw, Y=@n1.LocalAccelForward, BlendSpace=@n2.Default_TurnAdditive_FemaleSmall_BS)
    BlendPose_2:Pose Link Structure=Blendspace Player(X=@n1.SkydiveAimYaw, Y=@n1.LocalAccelForward, BlendSpace=@n2.Default_TurnAdditive_FemaleLarge_BS)
    BlendTime_0=0.1
    BlendTime_1=0.1
    BlendTime_2:float=0.1
    ActiveEnumValue=Get AnimBodyType() -> (AnimBodyType=Small)
   ) -> (Pose) {
    blendType=Linear
    blendPose [{linkId=-1, sourceLinkId=-1}, {linkId=-1, sourceLinkId=-1}, {linkId=-1, sourceLinkId=-1}]
    blendTime [0.1,0.1,0.1]
   }
   n11 = "Blend Poses (EFortCustomGender)"(BlendPose_0=@n9, BlendPose_1=@n10, BlendTime_0=0, BlendTime_1=0, ActiveEnumValue=Get Gender() -> (Gender=Invalid)) {blendType=Linear, blendPose [{linkId=-1, sourceLinkId=-1}, {linkId=-1, sourceLinkId=-1}], blendTime [0.1,0.1]}
   n12 = Apply Additive!(
    Base=Apply Additive!(Base=Blendspace Player(X=@n1.SkydiveAimYaw, Y=@n8, BlendSpace=@n7), Additive=@n6, Alpha:float) -> (Pose) {alphaBoolBlend {blendOption=Linear}}
    Additive=@n11
    AlphaCurveName:Name=ParaGlideTurn
   ) -> (Pose) {alphaInputType=Curve, alphaBoolBlend {blendOption=Linear}, alphaScaleBiasClamp {bMapRange=true, outRange {max=0, min=1}}}
   Output Animation Pose(Result=@n12)
   Reroute Node(InputPin:float=Get AimPitch()) -> (OutputPin:float)
  ]
 }
 Transition {
  path=AnimGraph.AnimGraphNode_StateMachine_0.SkyDiveState.AnimStateNode_1.DiveOrGlide.AnimGraphNode_StateMachine_212.SkydiveOrGlide.AnimStateNode_3.ParaGlide.AnimGraphNode_StateMachine_73.NoGlider.AnimStateTransitionNode_0.Transition
  nodes [Result(bCanEnterTransition=Get bTimeInParachutingAboveIntroLength())]
 }
 AnimationTransitionGraph_1 {
  path=AnimGraph.AnimGraphNode_StateMachine_0.SkyDiveState.AnimStateNode_1.DiveOrGlide.AnimGraphNode_StateMachine_212.SkydiveOrGlide.AnimStateNode_3.ParaGlide.AnimGraphNode_StateMachine_73.NoGlider.AnimStateTransitionNode_2.AnimationTransitionGraph_1
  nodes [Result()]
 }
 ParaGlide_Out {
  path=AnimGraph.AnimGraphNode_StateMachine_0.SkyDiveState.AnimStateNode_1.DiveOrGlide.AnimGraphNode_StateMachine_212.SkydiveOrGlide.AnimStateNode_4.ParaGlide_Out
  nodes [
   AnimGraphNode_StateMachine_1 = ParaGlide_Out() {is=title AnimGraphNode_StateMachine}
   Output Animation Pose(Result=@AnimGraphNode_StateMachine_1)
  ]
 }
 ParaGlide_Out {
  entry=Passthrough
  path=AnimGraph.AnimGraphNode_StateMachine_0.SkyDiveState.AnimStateNode_1.DiveOrGlide.AnimGraphNode_StateMachine_212.SkydiveOrGlide.AnimStateNode_4.ParaGlide_Out.AnimGraphNode_StateMachine_1.ParaGlide_Out
  ids=id
  nodes [
   AnimStateNode_1 = Dive_Out() {is=title AnimStateNode}
   AnimStateNode_2 = to_Dive()
   AnimStateNode_3 = to_Glide()
   AnimStateNode_4 = Passthrough()
   AnimStateNode_5 = Glide_Out() {is=title AnimStateNode}
   AnimStateTransitionNode_0 = AnimStateTransitionNode() {crossfadeDuration=0, priorityOrder=2}
   AnimStateTransitionNode_1 = AnimStateTransitionNode() {crossfadeDuration=0}
   AnimStateTransitionNode_2 = AnimStateTransitionNode() {crossfadeDuration=0.15}
   AnimStateTransitionNode_3 = AnimStateTransitionNode() {crossfadeDuration=0.15}
   AnimStateTransitionNode_4 = AnimStateTransitionNode() {crossfadeDuration=0.15}
   AnimStateTransitionNode_5 = AnimStateTransitionNode() {crossfadeDuration=0.15}
  ]
  links [
   {
    dst=to_Glide
    rule=AnimGraph.AnimGraphNode_StateMachine_0.SkyDiveState.AnimStateNode_1.DiveOrGlide.AnimGraphNode_StateMachine_212.SkydiveOrGlide.AnimStateNode_4.ParaGlide_Out.AnimGraphNode_StateMachine_1.ParaGlide_Out.AnimStateTransitionNode_3.Transition
    src=Dive_Out
    via=AnimStateTransitionNode_3
   }
   {
    dst=to_Dive
    rule=AnimGraph.AnimGraphNode_StateMachine_0.SkyDiveState.AnimStateNode_1.DiveOrGlide.AnimGraphNode_StateMachine_212.SkydiveOrGlide.AnimStateNode_4.ParaGlide_Out.AnimGraphNode_StateMachine_1.ParaGlide_Out.AnimStateTransitionNode_2.Transition
    src=Glide_Out
    via=AnimStateTransitionNode_2
   }
   {
    dst=Dive_Out
    rule=AnimGraph.AnimGraphNode_StateMachine_0.SkyDiveState.AnimStateNode_1.DiveOrGlide.AnimGraphNode_StateMachine_212.SkydiveOrGlide.AnimStateNode_4.ParaGlide_Out.AnimGraphNode_StateMachine_1.ParaGlide_Out.AnimStateTransitionNode_1.Transition
    src=Passthrough
    via=AnimStateTransitionNode_1
   }
   {
    dst=Glide_Out
    rule=AnimGraph.AnimGraphNode_StateMachine_0.SkyDiveState.AnimStateNode_1.DiveOrGlide.AnimGraphNode_StateMachine_212.SkydiveOrGlide.AnimStateNode_4.ParaGlide_Out.AnimGraphNode_StateMachine_1.ParaGlide_Out.AnimStateTransitionNode_0.Transition
    src=Passthrough
    via=AnimStateTransitionNode_0
   }
   {
    dst=to_Glide
    rule=AnimGraph.AnimGraphNode_StateMachine_0.SkyDiveState.AnimStateNode_1.DiveOrGlide.AnimGraphNode_StateMachine_212.SkydiveOrGlide.AnimStateNode_4.ParaGlide_Out.AnimGraphNode_StateMachine_1.ParaGlide_Out.AnimStateTransitionNode_4.Transition
    src=to_Dive
    via=AnimStateTransitionNode_4
   }
   {
    dst=to_Dive
    rule=AnimGraph.AnimGraphNode_StateMachine_0.SkyDiveState.AnimStateNode_1.DiveOrGlide.AnimGraphNode_StateMachine_212.SkydiveOrGlide.AnimStateNode_4.ParaGlide_Out.AnimGraphNode_StateMachine_1.ParaGlide_Out.AnimStateTransitionNode_5.Transition
    src=to_Glide
    via=AnimStateTransitionNode_5
   }
  ]
 }
 Dive_Out {
  path=AnimGraph.AnimGraphNode_StateMachine_0.SkyDiveState.AnimStateNode_1.DiveOrGlide.AnimGraphNode_StateMachine_212.SkydiveOrGlide.AnimStateNode_4.ParaGlide_Out.AnimGraphNode_StateMachine_1.ParaGlide_Out.AnimStateNode_1.Dive_Out
  nodes [
   n1 = Break Fort Anim Input Player Glider Anim Asset(FortAnimInput_PlayerGliderAnimAsset=Get CurrentGliderAnimSet()) -> (
    RootModPitchMin=~
    RootModPitchMax=~
    RootModYOffsetMin=~
    RootModYOffsetMax=~
    bEnableSpringMods=~
    bAllowPlayerDeployRootMod=~
    bUseSurfStyle=~
    bUseLeanPlayRateAdjust=~
    bUseForwardAs_Y_BS=~
    bUseDeployCurveForIntoPlayRateAdjust=~
    LeanForwardValue=~
    IdleValue=~
    LeanBackValue=~
    IdleInterpToSpeed=~
    LeanForwardInterpToSpeed=~
    LeanBackInterpToSpeed=~
   )
   n2 = Break Fort Anim Input Skydiving!(FortAnimInput_Skydiving=Get Skydiving()) -> (
    ParaGilderRootOffset:Vector=~
    ParaGliderRootRotationOverride:Rotator=~
    bUseParaGlideRootModifier:bool=~
    bIsSkydivingFromLaunchPad:bool=~
    bIsSkydivingFromBus:bool=~
    bIsInVortex:bool=~
    bIsDiving:bool=~
    bIsParachuteOpen:bool=~
    bIsSkydiving:bool=~
    bIsParachuteLeaning:bool=~
    bIsSkydiveLeaning:bool=~
    bIsLeaning:bool=~
    bIsSkydiveDiveMode:bool=~
    bPlaySkydiveDrift:bool=~
    bSkydiveDriftDelayActive:bool=~
    bSkydiveDriftAnimAllowed:bool=~
    bIsGliderRight:bool=~
    bIsGliderCenter:bool=~
    bIsGliderLeft:bool=~
    bIsGliderForward:bool=~
    bIsGliderBack:bool=~
    bHasDeployedAndIsStandingUpright=~
    LocalAccelForward=~
    LocalAccelRight=~
    LocalVelocityRight:float=~
    SkydiveAimPitch:float=~
    SkydiveAimYaw=~
    DeployChuteAnimRate:float=~
    SkydiveDriftAnimRate:float=~
    SkydiveDriftAnimRateCurrent:float=~
    SkydiveFidgetAnimRate:float=~
    SkydiveFidgetAnimRateCurrent:float=~
    SkydiveAdditiveAlpha:float=~
    SkydiveDriftDelay:float=~
    SkydiveDriftAnim:int
    SkydiveDriftAnimMax:int
    LaunchpadAnim:int
    LocalAccelDir:ESkydivingDirection Enum
    DirectionLast:ESkydivingDirection Enum
    PoseAdditive
   )
   n3 = Get AimPitch()
   Output Animation Pose(Result=Apply Additive(
     Base=Blendspace Player(X=@n2.SkydiveAimYaw, Y=@n3, BlendSpace=@n1.ToDive_BS) {bResetPlayTimeWhenBlendSpaceChanges=false, groupName=glider_to, method=SyncGroup}
     Additive=Use cached pose 'FreeFall_TurnAndLean_Dive_Pose'()
    ) {alphaBoolBlend {blendOption=Linear}})
  ]
 }
 to_Dive {
  path=AnimGraph.AnimGraphNode_StateMachine_0.SkyDiveState.AnimStateNode_1.DiveOrGlide.AnimGraphNode_StateMachine_212.SkydiveOrGlide.AnimStateNode_4.ParaGlide_Out.AnimGraphNode_StateMachine_1.ParaGlide_Out.AnimStateNode_2.to_Dive
  nodes [
   n1 = Break Fort Anim Input Player Skydive Anim Asset(FortAnimInput_PlayerSkydiveAnimAsset=Get CurrentSkydiveAnimSet())
   n2 = Break Fort Anim Input Skydiving(FortAnimInput_Skydiving=Get Skydiving()) -> (
    ParaGilderRootOffset=~
    ParaGliderRootRotationOverride=~
    bUseParaGlideRootModifier=~
    bIsSkydivingFromLaunchPad=~
    bIsSkydivingFromBus=~
    bIsInVortex=~
    bIsDiving=~
    bIsParachuteOpen=~
    bIsSkydiving=~
    bIsParachuteLeaning=~
    bIsSkydiveLeaning=~
    bIsLeaning=~
    bIsSkydiveDiveMode=~
    bPlaySkydiveDrift=~
    bSkydiveDriftDelayActive=~
    bSkydiveDriftAnimAllowed=~
    bIsGliderRight=~
    bIsGliderCenter=~
    bIsGliderLeft=~
    bIsGliderForward=~
    bIsGliderBack=~
    bHasDeployedAndIsStandingUpright=~
    LocalAccelForward=~
    LocalAccelRight=~
    LocalVelocityRight=~
    SkydiveAimPitch=~
    SkydiveAimYaw=~
    DeployChuteAnimRate=~
    SkydiveDriftAnimRate=~
    SkydiveDriftAnimRateCurrent=~
    SkydiveFidgetAnimRate=~
    SkydiveFidgetAnimRateCurrent=~
    SkydiveAdditiveAlpha=~
    SkydiveDriftDelay=~
   )
   n3 = Get FreeFall_AimPitch()
   n4 = "Blend Poses (EFortCustomGender)"(
    BlendPose_0=Blendspace Player 'BS_Glide_to_Dive'(X=@n2.SkydiveAimYaw, Y=@n3, BlendSpace=@n1.Glide_ToDive_BS_Male) {bLoop=false, bResetPlayTimeWhenBlendSpaceChanges=false}
    BlendPose_1=Blendspace Player 'BS_Glide_to_Dive'(X=@n2.SkydiveAimYaw, Y=@n3, BlendSpace=@n1.Glide_ToDive_BS_Female) {bLoop=false, bResetPlayTimeWhenBlendSpaceChanges=false, groupName=to_Dive, method=SyncGroup}
    BlendTime_0=0
    BlendTime_1=0
    ActiveEnumValue=Get Gender() -> (Gender=Invalid)
   ) {blendType=Linear, blendPose [{linkId=-1, sourceLinkId=-1}, {linkId=-1, sourceLinkId=-1}], blendTime [0.1,0.1]}
   n5 = Get FreeFall_AimPitch()
   n6 = Two Way Blend(
    A=Blendspace Player 'BS_Glide_2D_Turn_Additive'(X=@n2.SkydiveAimYaw, Y=@n5, BlendSpace=@n1.Glide_Turn_Additive_BS_Female)
    B=Blendspace Player 'BS_Dive_2D_Turn_Additive'(X=@n2.SkydiveAimYaw, Y=@n5, BlendSpace=@n1.Dive_Turn_Additive_BS_Female)
    AlphaCurveName=BlendIn
   ) {blendNode {alphaInputType=Curve, alphaBoolBlend {blendOption=Linear}}}
   n7 = Get FreeFall_AimPitch()
   n8 = "Blend Poses (EFortCustomGender)"(
    BlendPose_0=Blendspace Player 'BS_Dive_Lean_Additive'(X=@n2.LocalVelocityRight, Y=@n7, BlendSpace=@n1.Dive_Lean_Additive_BS_Male)
    BlendPose_1=Blendspace Player 'BS_Dive_Lean_Additive'(X=@n2.LocalVelocityRight, Y=@n7, BlendSpace=@n1.Dive_Lean_Additive_BS_Female)
    BlendTime_0=0
    BlendTime_1=0
    ActiveEnumValue=Get Gender() -> (Gender=Invalid)
   ) {blendType=Linear, blendPose [{linkId=-1, sourceLinkId=-1}, {linkId=-1, sourceLinkId=-1}], blendTime [0.1,0.1]}
   n9 = Blendspace Player 'BS_Glide_Lean_Additive'(X=@n2.LocalVelocityRight, Y=Get FreeFall_AimPitch(), BlendSpace=@n1.Glide_Lean_Additive_BS_Female)
   Output Animation Pose(Result=Apply Additive!(
     Base=Apply Additive!(Base=@n4, Additive=@n6, Alpha:float=0.8) -> (Pose) {alphaBoolBlend {blendOption=Linear}}
     Additive=Two Way Blend(A=@n9, B=@n8, AlphaCurveName=BlendIn) {blendNode {alphaInputType=Curve, alphaBoolBlend {blendOption=Linear}}}
     bAlphaBoolEnabled:bool=@n2.bIsSkydiveLeaning
    ) -> (Pose) {alphaInputType=Bool, alphaBoolBlend {blendInTime=0.5, blendOption=Linear, blendOutTime=0.5}})
  ]
 }
 to_Glide {
  path=AnimGraph.AnimGraphNode_StateMachine_0.SkyDiveState.AnimStateNode_1.DiveOrGlide.AnimGraphNode_StateMachine_212.SkydiveOrGlide.AnimStateNode_4.ParaGlide_Out.AnimGraphNode_StateMachine_1.ParaGlide_Out.AnimStateNode_3.to_Glide
  nodes [
   n1 = Break Fort Anim Input Player Skydive Anim Asset(FortAnimInput_PlayerSkydiveAnimAsset=Get CurrentSkydiveAnimSet())
   n2 = Break Fort Anim Input Skydiving(FortAnimInput_Skydiving=Get Skydiving()) -> (
    ParaGilderRootOffset=~
    ParaGliderRootRotationOverride=~
    bUseParaGlideRootModifier=~
    bIsSkydivingFromLaunchPad=~
    bIsSkydivingFromBus=~
    bIsInVortex=~
    bIsDiving=~
    bIsParachuteOpen=~
    bIsSkydiving=~
    bIsParachuteLeaning=~
    bIsSkydiveLeaning=~
    bIsLeaning=~
    bIsSkydiveDiveMode=~
    bPlaySkydiveDrift=~
    bSkydiveDriftDelayActive=~
    bSkydiveDriftAnimAllowed=~
    bIsGliderRight=~
    bIsGliderCenter=~
    bIsGliderLeft=~
    bIsGliderForward=~
    bIsGliderBack=~
    bHasDeployedAndIsStandingUpright=~
    LocalAccelForward=~
    LocalAccelRight=~
    LocalVelocityRight=~
    SkydiveAimPitch=~
    SkydiveAimYaw=~
    DeployChuteAnimRate=~
    SkydiveDriftAnimRate=~
    SkydiveDriftAnimRateCurrent=~
    SkydiveFidgetAnimRate=~
    SkydiveFidgetAnimRateCurrent=~
    SkydiveAdditiveAlpha=~
    SkydiveDriftDelay=~
   )
   n3 = Blendspace Player 'BS_Dive_to_Glide'(X=@n2.SkydiveAimYaw, Y=Get FreeFall_AimPitch(), BlendSpace=@n1.Dive_ToGlide_BS_Female) {bLoop=false, bResetPlayTimeWhenBlendSpaceChanges=false, groupName=to_Gliding, method=SyncGroup}
   n4 = Get FreeFall_AimPitch()
   n5 = Two Way Blend(
    A=Blendspace Player 'BS_Dive_2D_Turn_Additive'(X=@n2.SkydiveAimYaw, Y=@n4, BlendSpace=@n1.Dive_Turn_Additive_BS_Female)
    B=Blendspace Player 'BS_Glide_2D_Turn_Additive'(X=@n2.SkydiveAimYaw, Y=@n4, BlendSpace=@n1.Glide_Turn_Additive_BS_Female)
    AlphaCurveName=BlendIn
   ) {blendNode {alphaInputType=Curve, alphaBoolBlend {blendOption=Linear}}}
   n6 = Get FreeFall_AimPitch()
   n7 = Blendspace Player 'BS_Glide_Lean_Additive'(X=@n2.LocalVelocityRight, Y=@n6, BlendSpace=@n1.Glide_Lean_Additive_BS_Female)
   n8 = Blendspace Player 'BS_Dive_Lean_Additive'(X=@n2.LocalVelocityRight, Y=@n6, BlendSpace=@n1.Dive_Lean_Additive_BS_Female)
   Output Animation Pose(Result=Apply Additive!(
     Base=Apply Additive!(Base=@n3, Additive=@n5, Alpha:float=0.8) -> (Pose) {alphaBoolBlend {blendOption=Linear}}
     Additive=Two Way Blend(A=@n8, B=@n7, AlphaCurveName=BlendIn) {blendNode {alphaInputType=Curve, alphaBoolBlend {blendOption=Linear}}}
     bAlphaBoolEnabled:bool=@n2.bIsSkydiveLeaning
    ) -> (Pose) {alphaInputType=Bool, alphaBoolBlend {blendInTime=0.5, blendOption=Linear, blendOutTime=0.5}})
  ]
 }
 Passthrough {
  path=AnimGraph.AnimGraphNode_StateMachine_0.SkyDiveState.AnimStateNode_1.DiveOrGlide.AnimGraphNode_StateMachine_212.SkydiveOrGlide.AnimStateNode_4.ParaGlide_Out.AnimGraphNode_StateMachine_1.ParaGlide_Out.AnimStateNode_4.Passthrough
  nodes [Output Animation Pose(Result=Local Space Ref Pose())]
 }
 Glide_Out {
  path=AnimGraph.AnimGraphNode_StateMachine_0.SkyDiveState.AnimStateNode_1.DiveOrGlide.AnimGraphNode_StateMachine_212.SkydiveOrGlide.AnimStateNode_4.ParaGlide_Out.AnimGraphNode_StateMachine_1.ParaGlide_Out.AnimStateNode_5.Glide_Out
  nodes [
   n1 = Break Fort Anim Input Player Glider Anim Asset(FortAnimInput_PlayerGliderAnimAsset=Get CurrentGliderAnimSet()) -> (
    RootModPitchMin=~
    RootModPitchMax=~
    RootModYOffsetMin=~
    RootModYOffsetMax=~
    bEnableSpringMods=~
    bAllowPlayerDeployRootMod=~
    bUseSurfStyle=~
    bUseLeanPlayRateAdjust=~
    bUseForwardAs_Y_BS=~
    bUseDeployCurveForIntoPlayRateAdjust=~
    LeanForwardValue=~
    IdleValue=~
    LeanBackValue=~
    IdleInterpToSpeed=~
    LeanForwardInterpToSpeed=~
    LeanBackInterpToSpeed=~
   )
   n2 = Break Fort Anim Input Skydiving(FortAnimInput_Skydiving=Get Skydiving()) -> (
    ParaGilderRootOffset=~
    ParaGliderRootRotationOverride=~
    bUseParaGlideRootModifier=~
    bIsSkydivingFromLaunchPad=~
    bIsSkydivingFromBus=~
    bIsInVortex=~
    bIsDiving=~
    bIsParachuteOpen=~
    bIsSkydiving=~
    bIsParachuteLeaning=~
    bIsSkydiveLeaning=~
    bIsLeaning=~
    bIsSkydiveDiveMode=~
    bPlaySkydiveDrift=~
    bSkydiveDriftDelayActive=~
    bSkydiveDriftAnimAllowed=~
    bIsGliderRight=~
    bIsGliderCenter=~
    bIsGliderLeft=~
    bIsGliderForward=~
    bIsGliderBack=~
    bHasDeployedAndIsStandingUpright=~
    LocalAccelForward=~
    LocalAccelRight=~
    LocalVelocityRight=~
    SkydiveAimPitch=~
    SkydiveAimYaw=~
    DeployChuteAnimRate=~
    SkydiveDriftAnimRate=~
    SkydiveDriftAnimRateCurrent=~
    SkydiveFidgetAnimRate=~
    SkydiveFidgetAnimRateCurrent=~
    SkydiveAdditiveAlpha=~
    SkydiveDriftDelay=~
   )
   n3 = Get AimPitch()
   Output Animation Pose(Result=Apply Additive(
     Base=Blendspace Player(X=@n2.SkydiveAimYaw, Y=@n3, BlendSpace=@n1.ToGlide_BS) {bResetPlayTimeWhenBlendSpaceChanges=false, groupName=glider_to, method=SyncGroup}
     Additive=Use cached pose 'FreeFall_TurnAndLean_Glide_Pose'()
    ) {alphaBoolBlend {blendOption=Linear}})
  ]
 }
 Transition {
  path=AnimGraph.AnimGraphNode_StateMachine_0.SkyDiveState.AnimStateNode_1.DiveOrGlide.AnimGraphNode_StateMachine_212.SkydiveOrGlide.AnimStateNode_4.ParaGlide_Out.AnimGraphNode_StateMachine_1.ParaGlide_Out.AnimStateTransitionNode_0.Transition
  nodes [
   n1 = Get Skydiving()
   Result(bCanEnterTransition=NOT Boolean(A=Break Fort Anim Input Skydiving!(FortAnimInput_Skydiving=@n1) -> (
      bIsSkydiveDiveMode=~
      bIsGliderRight=~
      bIsGliderCenter=~
      bIsGliderLeft=~
      bIsGliderForward=~
      bIsGliderBack=~
      bHasDeployedAndIsStandingUpright=~
      PoseAdditive
     )))
  ]
 }
 Transition {
  path=AnimGraph.AnimGraphNode_StateMachine_0.SkyDiveState.AnimStateNode_1.DiveOrGlide.AnimGraphNode_StateMachine_212.SkydiveOrGlide.AnimStateNode_4.ParaGlide_Out.AnimGraphNode_StateMachine_1.ParaGlide_Out.AnimStateTransitionNode_1.Transition
  nodes [Result(bCanEnterTransition=Break Fort Anim Input Skydiving(FortAnimInput_Skydiving=Get Skydiving()) -> (
     bIsSkydiveDiveMode=~
     bIsGliderRight=~
     bIsGliderCenter=~
     bIsGliderLeft=~
     bIsGliderForward=~
     bIsGliderBack=~
     bHasDeployedAndIsStandingUpright=~
    ))]
 }
 Transition {
  path=AnimGraph.AnimGraphNode_StateMachine_0.SkyDiveState.AnimStateNode_1.DiveOrGlide.AnimGraphNode_StateMachine_212.SkydiveOrGlide.AnimStateNode_4.ParaGlide_Out.AnimGraphNode_StateMachine_1.ParaGlide_Out.AnimStateTransitionNode_2.Transition
  nodes [Result(bCanEnterTransition=Break Fort Anim Input Skydiving(FortAnimInput_Skydiving=Get Skydiving()) -> (
     bIsSkydiveDiveMode=~
     bIsGliderRight=~
     bIsGliderCenter=~
     bIsGliderLeft=~
     bIsGliderForward=~
     bIsGliderBack=~
     bHasDeployedAndIsStandingUpright=~
    ))]
 }
 Transition {
  path=AnimGraph.AnimGraphNode_StateMachine_0.SkyDiveState.AnimStateNode_1.DiveOrGlide.AnimGraphNode_StateMachine_212.SkydiveOrGlide.AnimStateNode_4.ParaGlide_Out.AnimGraphNode_StateMachine_1.ParaGlide_Out.AnimStateTransitionNode_3.Transition
  nodes [
   n1 = Get Skydiving()
   Result(bCanEnterTransition=NOT Boolean(A=Break Fort Anim Input Skydiving(FortAnimInput_Skydiving=@n1) -> (
      bIsSkydiveDiveMode=~
      bIsGliderRight=~
      bIsGliderCenter=~
      bIsGliderLeft=~
      bIsGliderForward=~
      bIsGliderBack=~
      bHasDeployedAndIsStandingUpright=~
     )))
  ]
 }
 Transition {
  path=AnimGraph.AnimGraphNode_StateMachine_0.SkyDiveState.AnimStateNode_1.DiveOrGlide.AnimGraphNode_StateMachine_212.SkydiveOrGlide.AnimStateNode_4.ParaGlide_Out.AnimGraphNode_StateMachine_1.ParaGlide_Out.AnimStateTransitionNode_4.Transition
  nodes [
   n1 = Get Skydiving()
   Result(bCanEnterTransition=NOT Boolean(A=Break Fort Anim Input Skydiving(FortAnimInput_Skydiving=@n1) -> (
      bIsSkydiveDiveMode=~
      bIsGliderRight=~
      bIsGliderCenter=~
      bIsGliderLeft=~
      bIsGliderForward=~
      bIsGliderBack=~
      bHasDeployedAndIsStandingUpright=~
     )))
  ]
 }
 Transition {
  path=AnimGraph.AnimGraphNode_StateMachine_0.SkyDiveState.AnimStateNode_1.DiveOrGlide.AnimGraphNode_StateMachine_212.SkydiveOrGlide.AnimStateNode_4.ParaGlide_Out.AnimGraphNode_StateMachine_1.ParaGlide_Out.AnimStateTransitionNode_5.Transition
  nodes [Result(bCanEnterTransition=Break Fort Anim Input Skydiving(FortAnimInput_Skydiving=Get Skydiving()) -> (
     bIsSkydiveDiveMode=~
     bIsGliderRight=~
     bIsGliderCenter=~
     bIsGliderLeft=~
     bIsGliderForward=~
     bIsGliderBack=~
     bHasDeployedAndIsStandingUpright=~
    ))]
 }
 AnimationTransitionGraph_1 {
  path=AnimGraph.AnimGraphNode_StateMachine_0.SkyDiveState.AnimStateNode_1.DiveOrGlide.AnimGraphNode_StateMachine_212.SkydiveOrGlide.AnimStateTransitionNode_0.AnimationTransitionGraph_1
  nodes [Result(bCanEnterTransition=Break Fort Anim Input Skydiving!(FortAnimInput_Skydiving=Get Skydiving()) -> (
     ParaGilderRootOffset:Vector=~
     ParaGliderRootRotationOverride:Rotator=~
     bUseParaGlideRootModifier:bool=~
     bIsSkydivingFromLaunchPad:bool=~
     bIsSkydivingFromBus:bool=~
     bIsInVortex:bool=~
     bIsDiving:bool=~
     bIsParachuteOpen:bool=~
     bIsSkydiving:bool=~
     bIsParachuteLeaning:bool=~
     bIsSkydiveLeaning:bool=~
     bIsLeaning:bool=~
     bIsSkydiveDiveMode=~
     bPlaySkydiveDrift:bool=~
     bSkydiveDriftDelayActive:bool=~
     bSkydiveDriftAnimAllowed:bool=~
     bIsGliderRight=~
     bIsGliderCenter=~
     bIsGliderLeft=~
     bIsGliderForward=~
     bIsGliderBack=~
     bHasDeployedAndIsStandingUpright=~
     LocalAccelForward:float=~
     LocalAccelRight:float=~
     LocalVelocityRight:float=~
     SkydiveAimPitch:float=~
     SkydiveAimYaw:float=~
     DeployChuteAnimRate:float=~
     SkydiveDriftAnimRate:float=~
     SkydiveDriftAnimRateCurrent:float=~
     SkydiveFidgetAnimRate:float=~
     SkydiveFidgetAnimRateCurrent:float=~
     SkydiveAdditiveAlpha:float=~
     SkydiveDriftDelay:float=~
     SkydiveDriftAnim:int
     SkydiveDriftAnimMax:int
     LaunchpadAnim:int
     LocalAccelDir:ESkydivingDirection Enum
     DirectionLast:ESkydivingDirection Enum
     PoseAdditive
    ).bIsParachuteOpen)]
 }
 AnimationTransitionGraph_1 {
  path=AnimGraph.AnimGraphNode_StateMachine_0.SkyDiveState.AnimStateNode_1.DiveOrGlide.AnimGraphNode_StateMachine_212.SkydiveOrGlide.AnimStateTransitionNode_1.AnimationTransitionGraph_1
  nodes [Result()]
 }
 AnimationTransitionGraph_1 {
  path=AnimGraph.AnimGraphNode_StateMachine_0.SkyDiveState.AnimStateNode_1.DiveOrGlide.AnimGraphNode_StateMachine_212.SkydiveOrGlide.AnimStateTransitionNode_2.AnimationTransitionGraph_1
  nodes [
   n1 = Get Skydiving()
   Result(bCanEnterTransition=NOT Boolean(A=Break Fort Anim Input Skydiving(FortAnimInput_Skydiving=@n1) -> (
      ParaGilderRootOffset=~
      ParaGliderRootRotationOverride=~
      bUseParaGlideRootModifier=~
      bIsSkydivingFromLaunchPad=~
      bIsSkydivingFromBus=~
      bIsInVortex=~
      bIsDiving=~
      bIsParachuteOpen=~
      bIsSkydiving=~
      bIsParachuteLeaning=~
      bIsSkydiveLeaning=~
      bIsLeaning=~
      bIsSkydiveDiveMode=~
      bPlaySkydiveDrift=~
      bSkydiveDriftDelayActive=~
      bSkydiveDriftAnimAllowed=~
      bIsGliderRight=~
      bIsGliderCenter=~
      bIsGliderLeft=~
      bIsGliderForward=~
      bIsGliderBack=~
      bHasDeployedAndIsStandingUpright=~
      LocalAccelForward=~
      LocalAccelRight=~
      LocalVelocityRight=~
      SkydiveAimPitch=~
      SkydiveAimYaw=~
      DeployChuteAnimRate=~
      SkydiveDriftAnimRate=~
      SkydiveDriftAnimRateCurrent=~
      SkydiveFidgetAnimRate=~
      SkydiveFidgetAnimRateCurrent=~
      SkydiveAdditiveAlpha=~
      SkydiveDriftDelay=~
     ).bIsParachuteOpen))
  ]
 }
 AnimationTransitionGraph_1 {
  path=AnimGraph.AnimGraphNode_StateMachine_0.SkyDiveState.AnimStateNode_1.DiveOrGlide.AnimGraphNode_StateMachine_212.SkydiveOrGlide.AnimStateTransitionNode_3.AnimationTransitionGraph_1
  nodes [Result(bCanEnterTransition=Break Fort Anim Input Skydiving(FortAnimInput_Skydiving=Get Skydiving()) -> (
     ParaGilderRootOffset=~
     ParaGliderRootRotationOverride=~
     bUseParaGlideRootModifier=~
     bIsSkydivingFromLaunchPad=~
     bIsSkydivingFromBus=~
     bIsInVortex=~
     bIsDiving=~
     bIsParachuteOpen=~
     bIsSkydiving=~
     bIsParachuteLeaning=~
     bIsSkydiveLeaning=~
     bIsLeaning=~
     bIsSkydiveDiveMode=~
     bPlaySkydiveDrift=~
     bSkydiveDriftDelayActive=~
     bSkydiveDriftAnimAllowed=~
     bIsGliderRight=~
     bIsGliderCenter=~
     bIsGliderLeft=~
     bIsGliderForward=~
     bIsGliderBack=~
     bHasDeployedAndIsStandingUpright=~
     LocalAccelForward=~
     LocalAccelRight=~
     LocalVelocityRight=~
     SkydiveAimPitch=~
     SkydiveAimYaw=~
     DeployChuteAnimRate=~
     SkydiveDriftAnimRate=~
     SkydiveDriftAnimRateCurrent=~
     SkydiveFidgetAnimRate=~
     SkydiveFidgetAnimRateCurrent=~
     SkydiveAdditiveAlpha=~
     SkydiveDriftDelay=~
    ).bIsParachuteOpen)]
 }
 Launchpad_TakeOff {
  path=AnimGraph.AnimGraphNode_StateMachine_0.SkyDiveState.AnimStateNode_2.Launchpad_TakeOff
  nodes [
   n1 = Blend Poses by int!(
    BlendPose_0:Pose Link Structure=Sequence Player 'LaunchPad_A_Into'() {bLoopAnimation=false, sequence=/Game/Animation/Game/MainPlayer/Skydive/LaunchPad/Medium/Female/LaunchPad_A_Into}
    BlendPose_1:Pose Link Structure=Sequence Player 'LaunchPad_B_Into'() {bLoopAnimation=false, sequence=/Game/Animation/Game/MainPlayer/Skydive/LaunchPad/Medium/Female/LaunchPad_B_Into}
    BlendPose_2:Pose Link Structure=Sequence Player 'LaunchPad_C_Into'() {bLoopAnimation=false, sequence=/Game/Animation/Game/MainPlayer/Skydive/LaunchPad/Medium/Female/LaunchPad_C_Into}
    BlendTime_0:float=0.1
    BlendTime_1:float=0.1
    BlendTime_2:float=0.1
    ActiveChildIndex:int=Break Fort Anim Input Skydiving(FortAnimInput_Skydiving=Get Skydiving()) -> (
     ParaGilderRootOffset=~
     ParaGliderRootRotationOverride=~
     bUseParaGlideRootModifier=~
     bIsSkydivingFromLaunchPad=~
     bIsSkydivingFromBus=~
     bIsInVortex=~
     bIsDiving=~
     bIsParachuteOpen=~
     bIsSkydiving=~
     bIsParachuteLeaning=~
     bIsSkydiveLeaning=~
     bIsLeaning=~
     bIsSkydiveDiveMode=~
     bPlaySkydiveDrift=~
     bSkydiveDriftDelayActive=~
     bSkydiveDriftAnimAllowed=~
     bIsGliderRight=~
     bIsGliderCenter=~
     bIsGliderLeft=~
     bIsGliderForward=~
     bIsGliderBack=~
     bHasDeployedAndIsStandingUpright=~
     LocalAccelForward=~
     LocalAccelRight=~
     LocalVelocityRight=~
     SkydiveAimPitch=~
     SkydiveAimYaw=~
     DeployChuteAnimRate=~
     SkydiveDriftAnimRate=~
     SkydiveDriftAnimRateCurrent=~
     SkydiveFidgetAnimRate=~
     SkydiveFidgetAnimRateCurrent=~
     SkydiveAdditiveAlpha=~
     SkydiveDriftDelay=~
    ).LaunchpadAnim
   ) -> (Pose:Pose Link Structure) {
    is=title AnimGraphNode_BlendListByInt
    blendType=Linear
    transitionType=Inertialization
    blendPose [{linkId=-1, sourceLinkId=-1}, {linkId=-1, sourceLinkId=-1}, {linkId=-1, sourceLinkId=-1}]
    blendTime [0.1,0.1,0.1]
   }
   Output Animation Pose(Result=@n1)
  ]
 }
 LunchPad_Transition {
  path=AnimGraph.AnimGraphNode_StateMachine_0.SkyDiveState.AnimStateNode_3.LunchPad_Transition
  nodes [
   n1 = Blend Poses by int(
    BlendPose_0=Sequence Player 'LaunchPad_A_to_Glide'() {bLoopAnimation=false, sequence=/Game/Animation/Game/MainPlayer/Skydive/LaunchPad/Medium/Female/LaunchPad_A_to_Glide}
    BlendPose_1=Sequence Player 'LaunchPad_B_to_Glide'() {bLoopAnimation=false, sequence=/Game/Animation/Game/MainPlayer/Skydive/LaunchPad/Medium/Female/LaunchPad_B_to_Glide}
    BlendPose_2=Sequence Player 'LaunchPad_C_to_Glide'() {bLoopAnimation=false, sequence=/Game/Animation/Game/MainPlayer/Skydive/LaunchPad/Medium/Female/LaunchPad_C_to_Glide}
    BlendTime_0=0.1
    BlendTime_1=0.1
    BlendTime_2=0.1
    ActiveChildIndex=Break Fort Anim Input Skydiving(FortAnimInput_Skydiving=Get Skydiving()) -> (
     ParaGilderRootOffset=~
     ParaGliderRootRotationOverride=~
     bUseParaGlideRootModifier=~
     bIsSkydivingFromLaunchPad=~
     bIsSkydivingFromBus=~
     bIsInVortex=~
     bIsDiving=~
     bIsParachuteOpen=~
     bIsSkydiving=~
     bIsParachuteLeaning=~
     bIsSkydiveLeaning=~
     bIsLeaning=~
     bIsSkydiveDiveMode=~
     bPlaySkydiveDrift=~
     bSkydiveDriftDelayActive=~
     bSkydiveDriftAnimAllowed=~
     bIsGliderRight=~
     bIsGliderCenter=~
     bIsGliderLeft=~
     bIsGliderForward=~
     bIsGliderBack=~
     bHasDeployedAndIsStandingUpright=~
     LocalAccelForward=~
     LocalAccelRight=~
     LocalVelocityRight=~
     SkydiveAimPitch=~
     SkydiveAimYaw=~
     DeployChuteAnimRate=~
     SkydiveDriftAnimRate=~
     SkydiveDriftAnimRateCurrent=~
     SkydiveFidgetAnimRate=~
     SkydiveFidgetAnimRateCurrent=~
     SkydiveAdditiveAlpha=~
     SkydiveDriftDelay=~
    ).LaunchpadAnim
   ) {
    blendType=Linear
    transitionType=Inertialization
    blendPose [{linkId=-1, sourceLinkId=-1}, {linkId=-1, sourceLinkId=-1}, {linkId=-1, sourceLinkId=-1}]
    blendTime [0.1,0.1,0.1]
   }
   n2 = Break Fort Anim Input Skydiving(FortAnimInput_Skydiving=Get Skydiving()) -> (
    ParaGilderRootOffset=~
    ParaGliderRootRotationOverride=~
    bUseParaGlideRootModifier=~
    bIsSkydivingFromLaunchPad=~
    bIsSkydivingFromBus=~
    bIsInVortex=~
    bIsDiving=~
    bIsParachuteOpen=~
    bIsSkydiving=~
    bIsParachuteLeaning=~
    bIsSkydiveLeaning=~
    bIsLeaning=~
    bIsSkydiveDiveMode=~
    bPlaySkydiveDrift=~
    bSkydiveDriftDelayActive=~
    bSkydiveDriftAnimAllowed=~
    bIsGliderRight=~
    bIsGliderCenter=~
    bIsGliderLeft=~
    bIsGliderForward=~
    bIsGliderBack=~
    bHasDeployedAndIsStandingUpright=~
    LocalAccelForward=~
    LocalAccelRight=~
    LocalVelocityRight=~
    SkydiveAimPitch=~
    SkydiveAimYaw=~
    DeployChuteAnimRate=~
    SkydiveDriftAnimRate=~
    SkydiveDriftAnimRateCurrent=~
    SkydiveFidgetAnimRate=~
    SkydiveFidgetAnimRateCurrent=~
    SkydiveAdditiveAlpha=~
    SkydiveDriftDelay=~
   )
   n3 = Get FreeFall_AimPitch()
   n4 = Two Way Blend(
    A=Blendspace Player 'BS_Glide_2D_Turn_Additive'!(X=@n2.SkydiveAimYaw, Y=@n3) -> (Pose) {blendSpace=/Game/Animation/Game/MainPlayer/Skydive/Freefall/Medium/Female/BS_Glide_2D_Turn_Additive}
    B=Blendspace Player 'BS_Dive_2D_Turn_Additive'!(X=@n2.SkydiveAimYaw, Y=@n3) -> (Pose) {blendSpace=/Game/Animation/Game/MainPlayer/Skydive/Freefall/Medium/Female/BS_Dive_2D_Turn_Additive}
    AlphaCurveName=BlendIn
   ) {blendNode {alphaInputType=Curve, alphaBoolBlend {blendOption=Linear}}}
   n5 = Break Fort Anim Input Skydiving(FortAnimInput_Skydiving=Get Skydiving()) -> (
    ParaGilderRootOffset=~
    ParaGliderRootRotationOverride=~
    bUseParaGlideRootModifier=~
    bIsSkydivingFromLaunchPad=~
    bIsSkydivingFromBus=~
    bIsInVortex=~
    bIsDiving=~
    bIsParachuteOpen=~
    bIsSkydiving=~
    bIsParachuteLeaning=~
    bIsSkydiveLeaning=~
    bIsLeaning=~
    bIsSkydiveDiveMode=~
    bPlaySkydiveDrift=~
    bSkydiveDriftDelayActive=~
    bSkydiveDriftAnimAllowed=~
    bIsGliderRight=~
    bIsGliderCenter=~
    bIsGliderLeft=~
    bIsGliderForward=~
    bIsGliderBack=~
    bHasDeployedAndIsStandingUpright=~
    LocalAccelForward=~
    LocalAccelRight=~
    LocalVelocityRight=~
    SkydiveAimPitch=~
    SkydiveAimYaw=~
    DeployChuteAnimRate=~
    SkydiveDriftAnimRate=~
    SkydiveDriftAnimRateCurrent=~
    SkydiveFidgetAnimRate=~
    SkydiveFidgetAnimRateCurrent=~
    SkydiveAdditiveAlpha=~
    SkydiveDriftDelay=~
   )
   n6 = Get FreeFall_AimPitch()
   n7 = Blend Poses by int(
    BlendPose_0=Blendspace Player 'BS_LaunchPad_A_to_Dive'(X=@n5.SkydiveAimYaw, Y=@n6) {bLoop=false, blendSpace=/Game/Animation/Game/MainPlayer/Skydive/LaunchPad/Medium/Female/BS_LaunchPad_A_to_Dive}
    BlendPose_1=Blendspace Player 'BS_LaunchPad_B_to_Dive'(X=@n5.SkydiveAimYaw, Y=@n6) {bLoop=false, blendSpace=/Game/Animation/Game/MainPlayer/Skydive/LaunchPad/Medium/Female/BS_LaunchPad_B_to_Dive}
    BlendPose_2=Blendspace Player 'BS_LaunchPad_C_to_Dive'(X=@n5.SkydiveAimYaw, Y=@n6) {bLoop=false, blendSpace=/Game/Animation/Game/MainPlayer/Skydive/LaunchPad/Medium/Female/BS_LaunchPad_C_to_Dive}
    BlendTime_0=0.1
    BlendTime_1=0.1
    BlendTime_2=0.1
    ActiveChildIndex=@n5.LaunchpadAnim
   ) {
    blendType=Linear
    transitionType=Inertialization
    blendPose [{linkId=-1, sourceLinkId=-1}, {linkId=-1, sourceLinkId=-1}, {linkId=-1, sourceLinkId=-1}]
    blendTime [0.1,0.1,0.1]
   }
   n8 = Get FreeFall_AimPitch()
   n9 = Two Way Blend(
    A=Blendspace Player 'BS_Dive_2D_Turn_Additive'(X=@n5.SkydiveAimYaw, Y=@n8) {blendSpace=&BS_Dive_2D_Turn_Additive}
    B=Blendspace Player 'BS_Glide_2D_Turn_Additive'(X=@n5.SkydiveAimYaw, Y=@n8) {blendSpace=&BS_Glide_2D_Turn_Additive}
    AlphaCurveName=BlendIn
   ) {blendNode {alphaInputType=Curve, alphaBoolBlend {blendOption=Linear}}}
   n10 = Get Skydiving()
   Output Animation Pose(Result=Blend Poses by bool(
     BlendPose_0=Apply Additive(Base=@n7, Additive=@n9, Alpha=0.8) {alphaBoolBlend {blendOption=Linear}}
     BlendPose_1=Apply Additive(Base=@n1, Additive=@n4, Alpha=0.8) {alphaBoolBlend {blendOption=Linear}}
     bActiveValue=Break Fort Anim Input Skydiving(FortAnimInput_Skydiving=@n10) -> (
      ParaGilderRootOffset=~
      ParaGliderRootRotationOverride=~
      bUseParaGlideRootModifier=~
      bIsSkydivingFromLaunchPad=~
      bIsSkydivingFromBus=~
      bIsInVortex=~
      bIsDiving=~
      bIsParachuteOpen=~
      bIsSkydiving=~
      bIsParachuteLeaning=~
      bIsSkydiveLeaning=~
      bIsLeaning=~
      bIsSkydiveDiveMode=~
      bPlaySkydiveDrift=~
      bSkydiveDriftDelayActive=~
      bSkydiveDriftAnimAllowed=~
      bIsGliderRight=~
      bIsGliderCenter=~
      bIsGliderLeft=~
      bIsGliderForward=~
      bIsGliderBack=~
      bHasDeployedAndIsStandingUpright=~
      LocalAccelForward=~
      LocalAccelRight=~
      LocalVelocityRight=~
      SkydiveAimPitch=~
      SkydiveAimYaw=~
      DeployChuteAnimRate=~
      SkydiveDriftAnimRate=~
      SkydiveDriftAnimRateCurrent=~
      SkydiveFidgetAnimRate=~
      SkydiveFidgetAnimRateCurrent=~
      SkydiveAdditiveAlpha=~
      SkydiveDriftDelay=~
     ).bIsSkydiveDiveMode
    ) {blendType=Linear})
  ]
 }
 Transition {
  path=AnimGraph.AnimGraphNode_StateMachine_0.SkyDiveState.AnimStateTransitionNode_1.Transition
  nodes [Result(bCanEnterTransition=Break Fort Anim Input Skydiving(FortAnimInput_Skydiving=Get Skydiving()) -> (
     ParaGilderRootOffset=~
     ParaGliderRootRotationOverride=~
     bUseParaGlideRootModifier=~
     bIsSkydivingFromLaunchPad=~
     bIsSkydivingFromBus=~
     bIsInVortex=~
     bIsDiving=~
     bIsParachuteOpen=~
     bIsSkydiving=~
     bIsParachuteLeaning=~
     bIsSkydiveLeaning=~
     bIsLeaning=~
     bIsSkydiveDiveMode=~
     bPlaySkydiveDrift=~
     bSkydiveDriftDelayActive=~
     bSkydiveDriftAnimAllowed=~
     bIsGliderRight=~
     bIsGliderCenter=~
     bIsGliderLeft=~
     bIsGliderForward=~
     bIsGliderBack=~
     bHasDeployedAndIsStandingUpright=~
     LocalAccelForward=~
     LocalAccelRight=~
     LocalVelocityRight=~
     SkydiveAimPitch=~
     SkydiveAimYaw=~
     DeployChuteAnimRate=~
     SkydiveDriftAnimRate=~
     SkydiveDriftAnimRateCurrent=~
     SkydiveFidgetAnimRate=~
     SkydiveFidgetAnimRateCurrent=~
     SkydiveAdditiveAlpha=~
     SkydiveDriftDelay=~
    ).bIsSkydivingFromLaunchPad)]
 }
 Transition {
  path=AnimGraph.AnimGraphNode_StateMachine_0.SkyDiveState.AnimStateTransitionNode_14.Transition
  nodes [
   Result(bCanEnterTransition=true)
   NOT Boolean(A=Get bIsFalling())
  ]
 }
 Transition {
  path=AnimGraph.AnimGraphNode_StateMachine_0.SkyDiveState.AnimStateTransitionNode_5.Transition
  nodes [Result(bCanEnterTransition=Break Fort Anim Input Skydiving(FortAnimInput_Skydiving=Get Skydiving()) -> (
     ParaGilderRootOffset=~
     ParaGliderRootRotationOverride=~
     bUseParaGlideRootModifier=~
     bIsSkydivingFromLaunchPad=~
     bIsSkydivingFromBus=~
     bIsInVortex=~
     bIsDiving=~
     bIsParachuteOpen=~
     bIsSkydiving=~
     bIsParachuteLeaning=~
     bIsSkydiveLeaning=~
     bIsLeaning=~
     bIsSkydiveDiveMode=~
     bPlaySkydiveDrift=~
     bSkydiveDriftDelayActive=~
     bSkydiveDriftAnimAllowed=~
     bIsGliderRight=~
     bIsGliderCenter=~
     bIsGliderLeft=~
     bIsGliderForward=~
     bIsGliderBack=~
     bHasDeployedAndIsStandingUpright=~
     LocalAccelForward=~
     LocalAccelRight=~
     LocalVelocityRight=~
     SkydiveAimPitch=~
     SkydiveAimYaw=~
     DeployChuteAnimRate=~
     SkydiveDriftAnimRate=~
     SkydiveDriftAnimRateCurrent=~
     SkydiveFidgetAnimRate=~
     SkydiveFidgetAnimRateCurrent=~
     SkydiveAdditiveAlpha=~
     SkydiveDriftDelay=~
    ).bIsParachuteOpen)]
 }
 Transition {path=AnimGraph.AnimGraphNode_StateMachine_0.SkyDiveState.AnimStateTransitionNode_6.Transition, nodes [Result(bCanEnterTransition=true)]}
 Transition {
  path=AnimGraph.AnimGraphNode_StateMachine_0.SkyDiveState.AnimStateTransitionNode_7.Transition
  nodes [Result(bCanEnterTransition=Break Fort Anim Input Skydiving(FortAnimInput_Skydiving=Get Skydiving()) -> (
     ParaGilderRootOffset=~
     ParaGliderRootRotationOverride=~
     bUseParaGlideRootModifier=~
     bIsSkydivingFromLaunchPad=~
     bIsSkydivingFromBus=~
     bIsInVortex=~
     bIsDiving=~
     bIsParachuteOpen=~
     bIsSkydiving=~
     bIsParachuteLeaning=~
     bIsSkydiveLeaning=~
     bIsLeaning=~
     bIsSkydiveDiveMode=~
     bPlaySkydiveDrift=~
     bSkydiveDriftDelayActive=~
     bSkydiveDriftAnimAllowed=~
     bIsGliderRight=~
     bIsGliderCenter=~
     bIsGliderLeft=~
     bIsGliderForward=~
     bIsGliderBack=~
     bHasDeployedAndIsStandingUpright=~
     LocalAccelForward=~
     LocalAccelRight=~
     LocalVelocityRight=~
     SkydiveAimPitch=~
     SkydiveAimYaw=~
     DeployChuteAnimRate=~
     SkydiveDriftAnimRate=~
     SkydiveDriftAnimRateCurrent=~
     SkydiveFidgetAnimRate=~
     SkydiveFidgetAnimRateCurrent=~
     SkydiveAdditiveAlpha=~
     SkydiveDriftDelay=~
    ).bIsParachuteOpen)]
 }
 Transition {
  path=AnimGraph.AnimGraphNode_StateMachine_0.SkyDiveState.AnimStateTransitionNode_8.Transition
  nodes [Result(bCanEnterTransition=Break Fort Anim Input Skydiving(FortAnimInput_Skydiving=Get Skydiving()) -> (
     ParaGilderRootOffset=~
     ParaGliderRootRotationOverride=~
     bUseParaGlideRootModifier=~
     bIsSkydivingFromLaunchPad=~
     bIsSkydivingFromBus=~
     bIsInVortex=~
     bIsDiving=~
     bIsParachuteOpen=~
     bIsSkydiving=~
     bIsParachuteLeaning=~
     bIsSkydiveLeaning=~
     bIsLeaning=~
     bIsSkydiveDiveMode=~
     bPlaySkydiveDrift=~
     bSkydiveDriftDelayActive=~
     bSkydiveDriftAnimAllowed=~
     bIsGliderRight=~
     bIsGliderCenter=~
     bIsGliderLeft=~
     bIsGliderForward=~
     bIsGliderBack=~
     bHasDeployedAndIsStandingUpright=~
     LocalAccelForward=~
     LocalAccelRight=~
     LocalVelocityRight=~
     SkydiveAimPitch=~
     SkydiveAimYaw=~
     DeployChuteAnimRate=~
     SkydiveDriftAnimRate=~
     SkydiveDriftAnimRateCurrent=~
     SkydiveFidgetAnimRate=~
     SkydiveFidgetAnimRateCurrent=~
     SkydiveAdditiveAlpha=~
     SkydiveDriftDelay=~
    ).bIsParachuteOpen)]
 }
 FreeFall_TurnAndLean_Glide {
  entry=TurnAndLeanAdd_Glide
  path=AnimGraph.AnimGraphNode_StateMachine_1.FreeFall_TurnAndLean_Glide
  ids=id
  nodes [AnimStateNode_1 = TurnAndLeanAdd_Glide() {is=title AnimStateNode}]
 }
 TurnAndLeanAdd_Glide {
  path=AnimGraph.AnimGraphNode_StateMachine_1.FreeFall_TurnAndLean_Glide.AnimStateNode_1.TurnAndLeanAdd_Glide
  nodes [
   n1 = Break Fort Anim Input Player Skydive Anim Asset(FortAnimInput_PlayerSkydiveAnimAsset=Get CurrentSkydiveAnimSet())
   n2 = Break Fort Anim Input Skydiving(FortAnimInput_Skydiving=Get Skydiving()) -> (
    ParaGilderRootOffset=~
    ParaGliderRootRotationOverride=~
    bUseParaGlideRootModifier=~
    bIsSkydivingFromLaunchPad=~
    bIsSkydivingFromBus=~
    bIsInVortex=~
    bIsDiving=~
    bIsParachuteOpen=~
    bIsSkydiving=~
    bIsParachuteLeaning=~
    bIsSkydiveLeaning=~
    bIsLeaning=~
    bIsSkydiveDiveMode=~
    bPlaySkydiveDrift=~
    bSkydiveDriftDelayActive=~
    bSkydiveDriftAnimAllowed=~
    bIsGliderRight=~
    bIsGliderCenter=~
    bIsGliderLeft=~
    bIsGliderForward=~
    bIsGliderBack=~
    bHasDeployedAndIsStandingUpright=~
    LocalAccelForward=~
    LocalAccelRight=~
    LocalVelocityRight=~
    SkydiveAimPitch=~
    SkydiveAimYaw=~
    DeployChuteAnimRate=~
    SkydiveDriftAnimRate=~
    SkydiveDriftAnimRateCurrent=~
    SkydiveFidgetAnimRate=~
    SkydiveFidgetAnimRateCurrent=~
    SkydiveAdditiveAlpha=~
    SkydiveDriftDelay=~
   )
   n3 = Get AimPitch()
   n4 = Apply Additive(
    Base=Apply Additive(
     Base=Local Space Ref Pose()
     Additive=Blendspace Player 'BS_Glide_2D_Turn_Additive'!(X=@n2.SkydiveAimYaw, Y=@n3, BlendSpace:Blend Space Object Reference=@n1.Glide_Turn_Additive_BS_Female) -> (Pose)
    ) {alphaBoolBlend {blendOption=Linear}}
    Additive=Blendspace Player 'BS_Glide_Lean_Additive'(X=@n2.LocalVelocityRight, Y=@n2.SkydiveAimPitch, BlendSpace=@n1.Glide_Lean_Additive_BS_Female)
   ) {alphaBoolBlend {blendInTime=0.5, blendOption=Linear, blendOutTime=0.5}}
   n5 = "Sequence Player (additive) 'Glide_Additive_Drift'"!(Sequence:Anim Sequence Base Object Reference=@n1.Glide_DriftA_Additive) -> (Pose)
   n6 = Break Fort Anim Input Skydiving(FortAnimInput_Skydiving=Get Skydiving()) -> (
    ParaGilderRootOffset=~
    ParaGliderRootRotationOverride=~
    bUseParaGlideRootModifier=~
    bIsSkydivingFromLaunchPad=~
    bIsSkydivingFromBus=~
    bIsInVortex=~
    bIsDiving=~
    bIsParachuteOpen=~
    bIsSkydiving=~
    bIsParachuteLeaning=~
    bIsSkydiveLeaning=~
    bIsLeaning=~
    bIsSkydiveDiveMode=~
    bPlaySkydiveDrift=~
    bSkydiveDriftDelayActive=~
    bSkydiveDriftAnimAllowed=~
    bIsGliderRight=~
    bIsGliderCenter=~
    bIsGliderLeft=~
    bIsGliderForward=~
    bIsGliderBack=~
    bHasDeployedAndIsStandingUpright=~
    LocalAccelForward=~
    LocalAccelRight=~
    LocalVelocityRight=~
    SkydiveAimPitch=~
    SkydiveAimYaw=~
    DeployChuteAnimRate=~
    SkydiveDriftAnimRate=~
    SkydiveDriftAnimRateCurrent=~
    SkydiveFidgetAnimRate=~
    SkydiveFidgetAnimRateCurrent=~
    SkydiveAdditiveAlpha=~
    SkydiveDriftDelay=~
   )
   n7 = Break Fort Anim Input Player Glider Anim Asset(FortAnimInput_PlayerGliderAnimAsset=Get CurrentGliderAnimSet()) -> (
    RootModPitchMin=~
    RootModPitchMax=~
    RootModYOffsetMin=~
    RootModYOffsetMax=~
    bEnableSpringMods=~
    bAllowPlayerDeployRootMod=~
    bUseSurfStyle=~
    bUseLeanPlayRateAdjust=~
    bUseForwardAs_Y_BS=~
    bUseDeployCurveForIntoPlayRateAdjust=~
    LeanForwardValue=~
    IdleValue=~
    LeanBackValue=~
    IdleInterpToSpeed=~
    LeanForwardInterpToSpeed=~
    LeanBackInterpToSpeed=~
   )
   Output Animation Pose(Result=Apply Additive!(
     Base=Apply Additive!(Base=@n4, Additive=@n5, Alpha:float=@n6.SkydiveAdditiveAlpha) -> (Pose) {alphaBoolBlend {blendOption=Linear}}
     Additive=Sequence Player!(Sequence=@n7.Glide_WeaponR_Additive_Anim) -> (Pose)
     bAlphaBoolEnabled:bool=True
    ) -> (Pose) {alphaInputType=Bool, alphaBoolBlend {blendOption=Linear}})
  ]
 }
 FreeFall_TurnAndLean_Dive {
  entry=TurnAndLeanAdd_Dive
  path=AnimGraph.AnimGraphNode_StateMachine_8.FreeFall_TurnAndLean_Dive
  ids=id
  nodes [AnimStateNode_1 = TurnAndLeanAdd_Dive() {is=title AnimStateNode}]
 }
 TurnAndLeanAdd_Dive {
  path=AnimGraph.AnimGraphNode_StateMachine_8.FreeFall_TurnAndLean_Dive.AnimStateNode_1.TurnAndLeanAdd_Dive
  nodes [
   n1 = Break Fort Anim Input Player Skydive Anim Asset(FortAnimInput_PlayerSkydiveAnimAsset=Get CurrentSkydiveAnimSet())
   n2 = Break Fort Anim Input Skydiving(FortAnimInput_Skydiving=Get Skydiving()) -> (
    ParaGilderRootOffset=~
    ParaGliderRootRotationOverride=~
    bUseParaGlideRootModifier=~
    bIsSkydivingFromLaunchPad=~
    bIsSkydivingFromBus=~
    bIsInVortex=~
    bIsDiving=~
    bIsParachuteOpen=~
    bIsSkydiving=~
    bIsParachuteLeaning=~
    bIsSkydiveLeaning=~
    bIsLeaning=~
    bIsSkydiveDiveMode=~
    bPlaySkydiveDrift=~
    bSkydiveDriftDelayActive=~
    bSkydiveDriftAnimAllowed=~
    bIsGliderRight=~
    bIsGliderCenter=~
    bIsGliderLeft=~
    bIsGliderForward=~
    bIsGliderBack=~
    bHasDeployedAndIsStandingUpright=~
    LocalAccelForward=~
    LocalAccelRight=~
    LocalVelocityRight=~
    SkydiveAimPitch=~
    SkydiveAimYaw=~
    DeployChuteAnimRate=~
    SkydiveDriftAnimRate=~
    SkydiveDriftAnimRateCurrent=~
    SkydiveFidgetAnimRate=~
    SkydiveFidgetAnimRateCurrent=~
    SkydiveAdditiveAlpha=~
    SkydiveDriftDelay=~
   )
   n3 = Get FreeFall_AimPitch()
   n4 = "Blend Poses (EFortCustomGender)"(
    BlendPose_0=Blendspace Player 'BS_Dive_2D_Turn_Additive'!(X=@n2.SkydiveAimYaw, Y=@n3, BlendSpace:Blend Space Object Reference=@n1.Dive_Turn_Additive_BS_Male) -> (Pose)
    BlendPose_1=Blendspace Player 'BS_Dive_2D_Turn_Additive'(X=@n2.SkydiveAimYaw, Y=@n3, BlendSpace=@n1.Dive_Turn_Additive_BS_Female)
    BlendTime_0=0
    BlendTime_1=0
    ActiveEnumValue=Get Gender() -> (Gender=Invalid)
   ) {blendType=Linear, blendPose [{linkId=-1, sourceLinkId=-1}, {linkId=-1, sourceLinkId=-1}], blendTime [0.1,0.1]}
   n5 = Get FreeFall_AimPitch()
   n6 = "Blend Poses (EFortCustomGender)"(
    BlendPose_0=Blendspace Player 'BS_Dive_Lean_Additive'(X=@n2.LocalVelocityRight, Y=@n5, BlendSpace=@n1.Dive_Lean_Additive_BS_Male) {bLoop=false, bResetPlayTimeWhenBlendSpaceChanges=false}
    BlendPose_1=Blendspace Player 'BS_Dive_Lean_Additive'(X=@n2.LocalVelocityRight, Y=@n5, BlendSpace=@n1.Dive_Lean_Additive_BS_Female)
    BlendTime_0=0
    BlendTime_1=0
    ActiveEnumValue=Get Gender() -> (Gender=Invalid)
   ) {blendType=Linear, blendPose [{linkId=-1, sourceLinkId=-1}, {linkId=-1, sourceLinkId=-1}], blendTime [0.1,0.1]}
   n7 = Apply Additive!(
    Base=Apply Additive!(Base=Local Space Ref Pose(), Additive=@n4, Alpha:float) -> (Pose) {alphaBoolBlend {blendOption=Linear}}
    Additive=@n6
    bAlphaBoolEnabled:bool=@n2.bIsSkydiveLeaning
   ) -> (Pose) {alphaInputType=Bool, alphaBoolBlend {blendInTime=0.5, blendOption=Linear, blendOutTime=0.5}}
   n8 = Break Fort Anim Input Skydiving(FortAnimInput_Skydiving=Get Skydiving()) -> (
    ParaGilderRootOffset=~
    ParaGliderRootRotationOverride=~
    bUseParaGlideRootModifier=~
    bIsSkydivingFromLaunchPad=~
    bIsSkydivingFromBus=~
    bIsInVortex=~
    bIsDiving=~
    bIsParachuteOpen=~
    bIsSkydiving=~
    bIsParachuteLeaning=~
    bIsSkydiveLeaning=~
    bIsLeaning=~
    bIsSkydiveDiveMode=~
    bPlaySkydiveDrift=~
    bSkydiveDriftDelayActive=~
    bSkydiveDriftAnimAllowed=~
    bIsGliderRight=~
    bIsGliderCenter=~
    bIsGliderLeft=~
    bIsGliderForward=~
    bIsGliderBack=~
    bHasDeployedAndIsStandingUpright=~
    LocalAccelForward=~
    LocalAccelRight=~
    LocalVelocityRight=~
    SkydiveAimPitch=~
    SkydiveAimYaw=~
    DeployChuteAnimRate=~
    SkydiveDriftAnimRate=~
    SkydiveDriftAnimRateCurrent=~
    SkydiveFidgetAnimRate=~
    SkydiveFidgetAnimRateCurrent=~
    SkydiveAdditiveAlpha=~
    SkydiveDriftDelay=~
   )
   n9 = Blend Poses by int!(
    BlendPose_0=Blendspace Player 'BS_Dive_Additive_Drift'!(
     X=@n8.SkydiveAimYaw
     Y=@n8.SkydiveAimPitch
     PlayRate:float=@n8.SkydiveDriftAnimRateCurrent
     BlendSpace:Blend Space Object Reference=@n1.Dive_DriftA_Additive_BS
    ) -> (Pose) {playRate=2}
    BlendPose_1=Blendspace Player 'BS_Dive_Additive_Drift_B'!(
     X=@n8.SkydiveAimYaw
     Y=@n8.SkydiveAimPitch
     PlayRate:float=@n8.SkydiveDriftAnimRateCurrent
     BlendSpace:Blend Space Object Reference=@n1.Dive_DriftB_Additive_BS
    ) -> (Pose) {playRate=2}
    BlendPose_2=Blendspace Player 'BS_Dive_Additive_Drift_C'!(
     X=@n8.SkydiveAimYaw
     Y=@n8.SkydiveAimPitch
     PlayRate:float=@n8.SkydiveDriftAnimRateCurrent
     BlendSpace:Blend Space Object Reference=@n1.Dive_DriftC_Additive_BS
    ) -> (Pose)
    BlendPose_3:Pose Link Structure=Blendspace Player 'BS_Dive_Additive_Fidget_A'!(
     X=@n8.SkydiveAimYaw
     Y=@n8.SkydiveAimPitch
     PlayRate:float=@n8.SkydiveFidgetAnimRateCurrent
     BlendSpace:Blend Space Object Reference=@n1.Dive_FidgetA_Additive_BS
    ) -> (Pose)
    BlendPose_4:Pose Link Structure=Blendspace Player 'BS_Dive_Additive_Fidget_B'!(
     X=@n8.SkydiveAimYaw
     Y=@n8.SkydiveAimPitch
     PlayRate:float=@n8.SkydiveFidgetAnimRateCurrent
     BlendSpace:Blend Space Object Reference=@n1.Dive_FidgetB_Additive_BS
    ) -> (Pose)
    BlendPose_5:Pose Link Structure=Blendspace Player 'BS_Dive_Additive_Fidget_C'!(
     X=@n8.SkydiveAimYaw
     Y=@n8.SkydiveAimPitch
     PlayRate:float=@n8.SkydiveFidgetAnimRateCurrent
     BlendSpace:Blend Space Object Reference=@n1.Dive_FidgetC_Additive_BS
    ) -> (Pose)
    BlendTime_0=.15
    BlendTime_1=.15
    BlendTime_2=.15
    BlendTime_3:float=.15
    BlendTime_4:float=.15
    BlendTime_5:float=.15
    ActiveChildIndex=@n8.SkydiveDriftAnim
   ) -> (Pose) {
    blendType=Linear
    childUpateMode=ResetChildOnActivate
    blendPose [
     {linkId=-1, sourceLinkId=-1}
     {linkId=-1, sourceLinkId=-1}
     {linkId=-1, sourceLinkId=-1}
     {linkId=-1, sourceLinkId=-1}
     {linkId=-1, sourceLinkId=-1}
     {linkId=-1, sourceLinkId=-1}
    ]
    blendTime [0.1,0.1,0.1,0.1,0.1,0.1]
   }
   n10 = Break Fort Anim Input Skydiving(FortAnimInput_Skydiving=Get Skydiving()) -> (
    ParaGilderRootOffset=~
    ParaGliderRootRotationOverride=~
    bUseParaGlideRootModifier=~
    bIsSkydivingFromLaunchPad=~
    bIsSkydivingFromBus=~
    bIsInVortex=~
    bIsDiving=~
    bIsParachuteOpen=~
    bIsSkydiving=~
    bIsParachuteLeaning=~
    bIsSkydiveLeaning=~
    bIsLeaning=~
    bIsSkydiveDiveMode=~
    bPlaySkydiveDrift=~
    bSkydiveDriftDelayActive=~
    bSkydiveDriftAnimAllowed=~
    bIsGliderRight=~
    bIsGliderCenter=~
    bIsGliderLeft=~
    bIsGliderForward=~
    bIsGliderBack=~
    bHasDeployedAndIsStandingUpright=~
    LocalAccelForward=~
    LocalAccelRight=~
    LocalVelocityRight=~
    SkydiveAimPitch=~
    SkydiveAimYaw=~
    DeployChuteAnimRate=~
    SkydiveDriftAnimRate=~
    SkydiveDriftAnimRateCurrent=~
    SkydiveFidgetAnimRate=~
    SkydiveFidgetAnimRateCurrent=~
    SkydiveAdditiveAlpha=~
    SkydiveDriftDelay=~
   )
   n11 = Break Fort Anim Input Player Glider Anim Asset(FortAnimInput_PlayerGliderAnimAsset=Get CurrentGliderAnimSet()) -> (
    RootModPitchMin=~
    RootModPitchMax=~
    RootModYOffsetMin=~
    RootModYOffsetMax=~
    bEnableSpringMods=~
    bAllowPlayerDeployRootMod=~
    bUseSurfStyle=~
    bUseLeanPlayRateAdjust=~
    bUseForwardAs_Y_BS=~
    bUseDeployCurveForIntoPlayRateAdjust=~
    LeanForwardValue=~
    IdleValue=~
    LeanBackValue=~
    IdleInterpToSpeed=~
    LeanForwardInterpToSpeed=~
    LeanBackInterpToSpeed=~
   )
   n12 = Get AimPitch()
   Output Animation Pose(Result=Apply Additive!(
     Base=Apply Additive!(Base=@n7, Additive=@n9, Alpha:float=@n10.SkydiveAdditiveAlpha) -> (Pose) {alphaBoolBlend {blendOption=Linear}}
     Additive=Blendspace Player(X=@n10.SkydiveAimYaw, Y=@n12, BlendSpace=@n11.Dive_WeaponR_Additive_BS)
     bAlphaBoolEnabled:bool=True
    ) -> (Pose) {alphaInputType=Bool, alphaBoolBlend {blendOption=Linear}})
  ]
 }
 EventGraph {links [], nodes []}
]
variables [
 BlockAnimRule {type=boolean}
 ParaGlide_TriggerAnim {type=boolean}
]
