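# Linked player Anim Blueprint that animates tactical sprint and tactical jump.
# AnimGraph switches between a jump state machine with additive leans and TacticalSprintLayer, then applies ConsumablesLayer upper-body slots and a weapon-bone fix.
# Sprint state machines use anim-set sequences for starts and loops, plus hurdle and mantle landings chosen by height, mirroring and gender.
# TacticalSprintLayer passes the input pose through when motion-matched sprint is enabled, smoothed by Dead Blending.
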AnimBlueprint /Game/Characters/Player/AnimationBlueprints/LinkedAnimGraphs/Player_TacticalState_AnimBP
target_skeleton=&Fortnite_M_Avg_Player_Skeleton
graphs [
 AnimGraph {nodes [
   n1 = Get YawDeltaSmoothed() -> (YawDeltaSmoothed:float)
   n2 = Get PawnSpeed2D()
   AnimGraphNode_StateMachine_1 = TacticalJump() {is=title AnimGraphNode_StateMachine, bAllowConduitEntryStates=true, maxTransitionsPerFrame=0}
   n3 = Blend Poses by bool(
    BlendPose_0=Apply Additive(
     Base=@AnimGraphNode_StateMachine_1
     Additive=Blendspace Player 'TacSprint_Jump_Leans_BS'(X=@n1, Y=@n2) {blendSpace=/Game/Animation/Game/MainPlayerFN/Poses/TacSprint/TacSprint_Jump_Leans_BS}
    ) {alphaBoolBlend {blendInTime=0.1, blendOutTime=0.1}}
    BlendPose_1=TacticalSprintLayer!(InPoseTacticalSprint:Pose Link Structure=Use cached pose 'InputPose'()) -> (Pose:Pose Link Structure) {is=title AnimGraphNode_LinkedAnimLayer, layer=TacticalSprintLayer}
    bActiveValue=Get bUseTacticalJump() -> (bUseTacticalJump:bool)
   ) {blendType=Linear}
   Output Pose(Result=Layered blend per bone(
     BasePose=ConsumablesLayer!(InPose:Pose Link Structure=@n3) -> (Pose:Pose Link Structure) {is=title AnimGraphNode_LinkedAnimLayer, layer=ConsumablesLayer}
     BlendPoses_0=Cached Sequence Evaluator!(Sequence:Anim Sequence Base Object Reference) -> (Pose:Pose Link Structure) {is=title FortAnimGraphNode_CachedSequenceEvaluator, cacheName=SourcePose}
     BlendWeights_0=Get WeaponBoneFixAlpha() -> (WeaponBoneFixAlpha:float)
    ) {layerSetup [{branchFilters [{boneName=weapon_l}, {boneName=weapon_r}]}]}) {layerGroup=None}
   Save cached pose 'InputPose'(Pose=Input Pose() {name=InPoseTacticalState}) {cacheName=InputPose}
  ]}
 TacticalJump {
  entry=Conduit
  path=AnimGraph.AnimGraphNode_StateMachine_1.TacticalJump
  ids=id
  nodes [
   AnimStateConduitNode_0 = Conduit()
   AnimStateNode_0 = TacticalJumpStart() {is=title AnimStateNode}
   AnimStateNode_1 = TacticalJumpFall() {is=title AnimStateNode}
   AnimStateNode_3 = TacticalDrop() {is=title AnimStateNode}
   AnimStateTransitionNode_0 = AnimStateTransitionNode() {crossfadeDuration=0.1, logicType=TLT_Inertialization, priorityOrder=2}
   AnimStateTransitionNode_1 = AnimStateTransitionNode() {bAutomaticRuleBasedOnSequencePlayerInState=true, crossfadeDuration=0.7}
   AnimStateTransitionNode_2 = AnimStateTransitionNode() {priorityOrder=100}
   AnimStateTransitionNode_3 = AnimStateTransitionNode() {crossfadeDuration=1}
   AnimStateTransitionNode_4 = AnimStateTransitionNode() {crossfadeDuration=0.4}
   AnimStateTransitionNode_5 = AnimStateTransitionNode() {crossfadeDuration=0.1, logicType=TLT_Inertialization}
  ]
  links [
   {
    dst=TacticalDrop
    rule=AnimGraph.AnimGraphNode_StateMachine_1.TacticalJump.AnimStateTransitionNode_2.Transition
    src=Conduit
    via=AnimStateTransitionNode_2
   }
   {
    dst=TacticalJumpFall
    rule=AnimGraph.AnimGraphNode_StateMachine_1.TacticalJump.AnimStateTransitionNode_3.Transition
    src=Conduit
    via=AnimStateTransitionNode_3
   }
   {
    dst=TacticalJumpStart
    rule=AnimGraph.AnimGraphNode_StateMachine_1.TacticalJump.AnimStateTransitionNode_0.Transition
    src=Conduit
    via=AnimStateTransitionNode_0
   }
   {
    dst=TacticalJumpFall
    rule=AnimGraph.AnimGraphNode_StateMachine_1.TacticalJump.AnimStateTransitionNode_4.Transition
    src=TacticalDrop
    via=AnimStateTransitionNode_4
   }
   {
    dst=TacticalJumpStart
    rule=AnimGraph.AnimGraphNode_StateMachine_1.TacticalJump.AnimStateTransitionNode_5.Transition
    src=TacticalJumpFall
    via=AnimStateTransitionNode_5
   }
   {
    dst=TacticalJumpFall
    rule=AnimGraph.AnimGraphNode_StateMachine_1.TacticalJump.AnimStateTransitionNode_1.Transition
    src=TacticalJumpStart
    via=AnimStateTransitionNode_1
   }
  ]
 }
 Conduit {path=AnimGraph.AnimGraphNode_StateMachine_1.TacticalJump.AnimStateConduitNode_0.Conduit, nodes [Result(bCanEnterTransition=true)]}
 TacticalJumpStart {
  path=AnimGraph.AnimGraphNode_StateMachine_1.TacticalJump.AnimStateNode_0.TacticalJumpStart
  nodes [
   n1 = Get CurrentTacticalStateAnimSet() -> (CurrentTacticalStateAnimSet:Fort Tactical State Anim Set Structure)
   Output Animation Pose(Result=Sequence Player!(Sequence:Anim Sequence Base Object Reference=Break Fort Tactical State Anim Set!(FortTacticalStateAnimSet:"Fort Tactical State Anim Set Structure (by ref)"=@n1) -> (
      SprintStart:Anim Sequence Base Object Reference
      SprintLoop:Anim Sequence Base Object Reference
      FallToSprint:Anim Sequence Base Object Reference
      DoorBashToSprint:Anim Sequence Base Object Reference
      JumpStart:Anim Sequence Base Object Reference
      FallLoop:Anim Sequence Base Object Reference
     ) {is=title K2Node_BreakStruct}.JumpStart) -> (Pose) {bLoopAnimation=false})
  ]
 }
 TacticalJumpFall {
  path=AnimGraph.AnimGraphNode_StateMachine_1.TacticalJump.AnimStateNode_1.TacticalJumpFall
  nodes [
   n1 = Get CurrentTacticalStateAnimSet()
   Output Animation Pose(Result=Sequence Player(Sequence=Break Fort Tactical State Anim Set(FortTacticalStateAnimSet=@n1).FallLoop))
  ]
 }
 TacticalDrop {
  path=AnimGraph.AnimGraphNode_StateMachine_1.TacticalJump.AnimStateNode_3.TacticalDrop
  nodes [
   n1 = Sequence Player(Sequence=Break Fort Tactical State Anim Set(FortTacticalStateAnimSet=Get CurrentTacticalStateAnimSet()).SprintLoop) {groupName=Locomotion, method=SyncGroup}
   n2 = Use cached pose 'InputPose'()
   n3 = Get bEnableMotionMatchedTacticalSprint() -> (bEnableMotionMatchedTacticalSprint:bool)
   Output Animation Pose(Result=Modify Curve(
     SourcePose=Blend Poses by bool(BlendPose_0=@n2, BlendPose_1=@n1, BlendTime_0=0, BlendTime_1=0, bActiveValue=@n3) {blendType=Linear}
     CurveValues_0=1
     Alpha=1
    ) {curveNames [Hurdle_Fall], curveValues [0]})
  ]
 }
 Transition {
  path=AnimGraph.AnimGraphNode_StateMachine_1.TacticalJump.AnimStateTransitionNode_0.Transition
  nodes [Result(bCanEnterTransition=Get bIsJumping())]
 }
 Transition {path=AnimGraph.AnimGraphNode_StateMachine_1.TacticalJump.AnimStateTransitionNode_1.Transition, nodes [Result()]}
 Transition {path=AnimGraph.AnimGraphNode_StateMachine_1.TacticalJump.AnimStateTransitionNode_2.Transition, nodes [Result(bCanEnterTransition=true)]}
 Transition {
  path=AnimGraph.AnimGraphNode_StateMachine_1.TacticalJump.AnimStateTransitionNode_3.Transition
  nodes [Result(bCanEnterTransition=Get bIsHurdleFall() -> (bIsHurdleFall:bool))]
 }
 Transition {path=AnimGraph.AnimGraphNode_StateMachine_1.TacticalJump.AnimStateTransitionNode_4.Transition, nodes [Result(bCanEnterTransition=true)]}
 Transition {
  path=AnimGraph.AnimGraphNode_StateMachine_1.TacticalJump.AnimStateTransitionNode_5.Transition
  nodes [Result(bCanEnterTransition=Get bIsJumping())]
 }
 ConsumablesLayer {nodes [
   n1 = Layered blend per bone(BasePose=Use cached pose 'PreTacticalSprintUpperBodySlot'(), BlendPoses_0=Use cached pose 'TacticalSprintUpperBodyPose'()) {layerSetup [{branchFilters [
       {blendDepth=7, boneName=Spine_01}
       {blendDepth=-1, boneName=clavicle_l}
       {boneName=clavicle_r}
       {boneName=ik_hand_root}
       {blendDepth=-1, boneName=ik_hand_l}
      ]}]}
   n2 = Layered blend per bone(BasePose=Use cached pose 'PreTacticalSprintUpperBodySlot'(), BlendPoses_0=Use cached pose 'TacticalSprintUpperBodyPose'()) {
    bMeshSpaceRotationBlend=true
    layerSetup [{branchFilters [{blendDepth=6, boneName=spine_01}, {boneName=clavicle_l}, {boneName=clavicle_r}, {boneName=ik_hand_root}]}]
   }
   n3 = Layered blend per bone(BasePose=Use cached pose 'PreTacticalSprintUpperBodySlot'(), BlendPoses_0=Use cached pose 'TacticalSprintUpperBodyPose'()) {layerSetup [{branchFilters [{blendDepth=6, boneName=spine_01}, {boneName=clavicle_l}, {boneName=clavicle_r}, {boneName=ik_hand_root}]}]}
   Output Pose(Result=Two Way Blend(A=Two Way Blend(A=@n3, B=@n2, Alpha=0) {blendNode {alphaBoolBlend {blendOption=Linear}}}, B=@n1, Alpha=0) {blendNode {alphaBoolBlend {blendOption=Linear}}}) {layerGroup=None}
   Save cached pose 'PreTacticalSprintUpperBodySlot'(Pose=Input Pose()) {cacheName=PreTacticalSprintUpperBodySlot}
   n4 = "Blend Poses (EFortCustomGender)"!(
    BlendPose_0:Pose Link Structure=Use cached pose 'PostTacticalSprintUpperBody'()
    BlendPose_1:Pose Link Structure=Slot 'AdditiveCorrectiveUpperBody'(Source=Use cached pose 'PostTacticalSprintUpperBody'()) {slotName=AdditiveCorrectiveUpperBody}
    BlendTime_0:float=0.1
    BlendTime_1:float=0.1
    ActiveEnumValue:Byte=Get Gender() -> (Gender:EFortCustomGender Enum=Invalid)
   ) -> (Pose:Pose Link Structure) {
    is=title AnimGraphNode_BlendListByEnum
    blendType=Linear
    blendPose [{linkId=-1, sourceLinkId=-1}, {linkId=-1, sourceLinkId=-1}]
    blendTime [0.1,0.1]
   }
   Save cached pose 'TacticalSprintUpperBodyPose'(Pose=@n4) {cacheName=TacticalSprintUpperBodyPose}
   Save cached pose 'PostTacticalSprintUpperBody'(Pose=Slot 'TacticalSprintUpperBody'(Source=Use cached pose 'PreTacticalSprintUpperBodySlot'()) {slotName=TacticalSprintUpperBody}) {cacheName=PostTacticalSprintUpperBody}
  ]}
 EventGraph {links [], nodes []}
 TacticalSprintLayer {nodes [
   AnimGraphNode_StateMachine_16 = SprintingPoseAdditives() {is=title AnimGraphNode_StateMachine}
   n1 = Use cached pose 'InputPoseTacticalSprint'()
   n2 = Get bEnableMotionMatchedTacticalSprint()
   Output Pose(Result=Dead Blending!(Source:Pose Link Structure=Blend Poses by bool(BlendPose_0=@n1, BlendPose_1=@AnimGraphNode_StateMachine_16, bActiveValue=@n2) {blendType=Linear}) -> (Pose:Pose Link Structure) {is=title AnimGraphNode_DeadBlending, defaultBlendMode=Linear}) {layerGroup=None}
   n3 = Get SprintYawAngle() -> (SprintYawAngle:float)
   AnimGraphNode_StateMachine_0 = TacticalSprint() {is=title AnimGraphNode_StateMachine, bAllowConduitEntryStates=true}
   n4 = Get SprintYawAngle()
   Save cached pose 'TacticalSprintPose'(Pose=Apply Additive(
     Base=Rotate Root Bone!(BasePose:Pose Link Structure=@AnimGraphNode_StateMachine_0, Yaw:float=@n4) -> (Pose:Pose Link Structure) {is=title AnimGraphNode_RotateRootBone}
     Additive=Blendspace Player 'Sprint_LookAt_BS_1D'!(X=@n3) -> (Pose) {blendSpace=/Game/Animation/Game/MainPlayer/Locomotion/Medium/Male/Sprint/Sprint_LookAt_BS_1D}
     Alpha=Get LookAtBlendSpaceAlpha() -> (LookAtBlendSpaceAlpha:float)
    ) {lODThreshold=1, alphaBoolBlend {blendOption=Linear}}) {cacheName=TacticalSprintPose}
   Save cached pose 'InputPoseTacticalSprint'(Pose=Input Pose() {name=InPoseTacticalSprint}) {cacheName=InputPoseTacticalSprint}
  ]}
 TacticalSprint {
  entry=Conduit
  path=TacticalSprintLayer.AnimGraphNode_StateMachine_0.TacticalSprint
  ids=id
  nodes [
   AnimStateConduitNode_0 = Conduit()
   AnimStateNode_0 = TacticalSprintStart() {is=title AnimStateNode}
   AnimStateNode_1 = TacticalSprintLoop() {is=title AnimStateNode}
   AnimStateNode_2 = DoorBash_To_TacticalSprint() {is=title AnimStateNode}
   AnimStateNode_3 = Hurdle_To_TacticalSprint() {is=title AnimStateNode}
   AnimStateNode_4 = Hurdle_Mantle_To_TacticalSprint() {is=title AnimStateNode}
   AnimStateNode_5 = Fall_To_TacticalSprint() {is=title AnimStateNode}
   AnimStateTransitionNode_0 = AnimStateTransitionNode() {bAutomaticRuleBasedOnSequencePlayerInState=true, crossfadeDuration=0.25, logicType=TLT_Inertialization}
   AnimStateTransitionNode_1 = AnimStateTransitionNode() {crossfadeDuration=0.1, logicType=TLT_Inertialization, priorityOrder=100}
   AnimStateTransitionNode_10 = AnimStateTransitionNode() {logicType=TLT_Inertialization, priorityOrder=2}
   AnimStateTransitionNode_2 = AnimStateTransitionNode() {crossfadeDuration=0.1, logicType=TLT_Inertialization}
   AnimStateTransitionNode_3 = AnimStateTransitionNode() {bAutomaticRuleBasedOnSequencePlayerInState=true, crossfadeDuration=0.3}
   AnimStateTransitionNode_4 = AnimStateTransitionNode() {crossfadeDuration=0.4, logicType=TLT_Inertialization, priorityOrder=3}
   AnimStateTransitionNode_5 = AnimStateTransitionNode() {bAutomaticRuleBasedOnSequencePlayerInState=true}
   AnimStateTransitionNode_6 = AnimStateTransitionNode() {crossfadeDuration=0.4, logicType=TLT_Inertialization, priorityOrder=4}
   AnimStateTransitionNode_7 = AnimStateTransitionNode() {bAutomaticRuleBasedOnSequencePlayerInState=true}
   AnimStateTransitionNode_9 = AnimStateTransitionNode() {bAutomaticRuleBasedOnSequencePlayerInState=true, logicType=TLT_Inertialization}
  ]
  links [
   {
    dst=DoorBash_To_TacticalSprint
    rule=TacticalSprintLayer.AnimGraphNode_StateMachine_0.TacticalSprint.AnimStateTransitionNode_2.AnimationTransitionGraph_0
    src=Conduit
    via=AnimStateTransitionNode_2
   }
   {
    dst=Fall_To_TacticalSprint
    rule=TacticalSprintLayer.AnimGraphNode_StateMachine_0.TacticalSprint.AnimStateTransitionNode_10.Transition
    src=Conduit
    via=AnimStateTransitionNode_10
   }
   {
    dst=Hurdle_Mantle_To_TacticalSprint
    rule=TacticalSprintLayer.AnimGraphNode_StateMachine_0.TacticalSprint.AnimStateTransitionNode_6.Transition
    src=Conduit
    via=AnimStateTransitionNode_6
   }
   {
    dst=Hurdle_To_TacticalSprint
    rule=TacticalSprintLayer.AnimGraphNode_StateMachine_0.TacticalSprint.AnimStateTransitionNode_4.Transition
    src=Conduit
    via=AnimStateTransitionNode_4
   }
   {
    dst=TacticalSprintStart
    rule=TacticalSprintLayer.AnimGraphNode_StateMachine_0.TacticalSprint.AnimStateTransitionNode_1.AnimationTransitionGraph_0
    src=Conduit
    via=AnimStateTransitionNode_1
   }
   {
    dst=TacticalSprintLoop
    rule=TacticalSprintLayer.AnimGraphNode_StateMachine_0.TacticalSprint.AnimStateTransitionNode_3.AnimationTransitionGraph_0
    src=DoorBash_To_TacticalSprint
    via=AnimStateTransitionNode_3
   }
   {
    dst=TacticalSprintLoop
    rule=TacticalSprintLayer.AnimGraphNode_StateMachine_0.TacticalSprint.AnimStateTransitionNode_9.Transition
    src=Fall_To_TacticalSprint
    via=AnimStateTransitionNode_9
   }
   {
    dst=TacticalSprintLoop
    rule=TacticalSprintLayer.AnimGraphNode_StateMachine_0.TacticalSprint.AnimStateTransitionNode_7.Transition
    src=Hurdle_Mantle_To_TacticalSprint
    via=AnimStateTransitionNode_7
   }
   {
    dst=TacticalSprintLoop
    rule=TacticalSprintLayer.AnimGraphNode_StateMachine_0.TacticalSprint.AnimStateTransitionNode_5.Transition
    src=Hurdle_To_TacticalSprint
    via=AnimStateTransitionNode_5
   }
   {
    dst=TacticalSprintLoop
    rule=TacticalSprintLayer.AnimGraphNode_StateMachine_0.TacticalSprint.AnimStateTransitionNode_0.AnimationTransitionGraph_0
    src=TacticalSprintStart
    via=AnimStateTransitionNode_0
   }
  ]
 }
 Conduit {
  path=TacticalSprintLayer.AnimGraphNode_StateMachine_0.TacticalSprint.AnimStateConduitNode_0.Conduit
  nodes [Result(bCanEnterTransition=true)]
 }
 TacticalSprintStart {
  path=TacticalSprintLayer.AnimGraphNode_StateMachine_0.TacticalSprint.AnimStateNode_0.TacticalSprintStart
  nodes [
   n1 = Get CurrentTacticalStateAnimSet()
   Output Animation Pose(Result=Sequence Player(Sequence=Break Fort Tactical State Anim Set(FortTacticalStateAnimSet=@n1)) {bLoopAnimation=false, groupName=Locomotion, method=SyncGroup})
  ]
 }
 TacticalSprintLoop {
  path=TacticalSprintLayer.AnimGraphNode_StateMachine_0.TacticalSprint.AnimStateNode_1.TacticalSprintLoop
  nodes [
   n1 = Get CurrentTacticalStateAnimSet()
   Output Animation Pose(Result=Sequence Player(Sequence=Break Fort Tactical State Anim Set(FortTacticalStateAnimSet=@n1).SprintLoop) {groupName=Locomotion, method=SyncGroup})
  ]
 }
 DoorBash_To_TacticalSprint {
  path=TacticalSprintLayer.AnimGraphNode_StateMachine_0.TacticalSprint.AnimStateNode_2.DoorBash_To_TacticalSprint
  nodes [
   n1 = Get CurrentTacticalStateAnimSet()
   Output Animation Pose(Result=Sequence Player(Sequence=Break Fort Tactical State Anim Set(FortTacticalStateAnimSet=@n1).DoorBashToSprint) {bLoopAnimation=false, groupName=Locomotion, method=SyncGroup})
  ]
 }
 Hurdle_To_TacticalSprint {
  path=TacticalSprintLayer.AnimGraphNode_StateMachine_0.TacticalSprint.AnimStateNode_3.Hurdle_To_TacticalSprint
  nodes [
   AnimGraphNode_StateMachine_0 = Hurdle_Land() {is=title AnimGraphNode_StateMachine, bAllowConduitEntryStates=true}
   Output Animation Pose(Result=@AnimGraphNode_StateMachine_0)
  ]
 }
 Hurdle_Land {
  entry=Conduit
  path=TacticalSprintLayer.AnimGraphNode_StateMachine_0.TacticalSprint.AnimStateNode_3.Hurdle_To_TacticalSprint.AnimGraphNode_StateMachine_0.Hurdle_Land
  ids=id
  nodes [
   AnimStateConduitNode_0 = Conduit()
   AnimStateNode_0 = 200up_250fwd_Land() {is=title AnimStateNode}
   AnimStateNode_1 = 100up_250fwd_Land() {is=title AnimStateNode}
   AnimStateNode_2 = 100up_100fwd_Land() {is=title AnimStateNode}
   AnimStateNode_3 = 200up_100fwd_Land() {is=title AnimStateNode}
   AnimStateNode_4 = 100up_100fwd_Land_Mirrored() {is=title AnimStateNode}
   AnimStateNode_5 = 100up_250fwd_Land_Mirrored() {is=title AnimStateNode}
   AnimStateNode_6 = 50up_100fwd_Land() {is=title AnimStateNode}
   AnimStateNode_7 = 50up_250fwd_Land() {is=title AnimStateNode}
   AnimStateTransitionNode_0 = AnimStateTransitionNode() {crossfadeDuration=0.35, logicType=TLT_Inertialization, priorityOrder=4}
   AnimStateTransitionNode_1 = AnimStateTransitionNode() {crossfadeDuration=0.35, logicType=TLT_Inertialization, priorityOrder=6}
   AnimStateTransitionNode_2 = AnimStateTransitionNode() {crossfadeDuration=0.35, logicType=TLT_Inertialization, priorityOrder=7}
   AnimStateTransitionNode_3 = AnimStateTransitionNode() {crossfadeDuration=0.35, logicType=TLT_Inertialization, priorityOrder=3}
   AnimStateTransitionNode_4 = AnimStateTransitionNode() {crossfadeDuration=0.35, logicType=TLT_Inertialization, priorityOrder=5}
   AnimStateTransitionNode_5 = AnimStateTransitionNode() {crossfadeDuration=0.35, logicType=TLT_Inertialization, priorityOrder=100}
   AnimStateTransitionNode_6 = AnimStateTransitionNode() {crossfadeDuration=0.3, logicType=TLT_Inertialization}
   AnimStateTransitionNode_7 = AnimStateTransitionNode() {crossfadeDuration=0.3, logicType=TLT_Inertialization, priorityOrder=2}
  ]
  links [
   {
    dst=100up_100fwd_Land
    rule=TacticalSprintLayer.AnimGraphNode_StateMachine_0.TacticalSprint.AnimStateNode_3.Hurdle_To_TacticalSprint.AnimGraphNode_StateMachine_0.Hurdle_Land.AnimStateTransitionNode_3.Transition
    src=Conduit
    via=AnimStateTransitionNode_3
   }
   {
    dst=100up_100fwd_Land_Mirrored
    rule=TacticalSprintLayer.AnimGraphNode_StateMachine_0.TacticalSprint.AnimStateNode_3.Hurdle_To_TacticalSprint.AnimGraphNode_StateMachine_0.Hurdle_Land.AnimStateTransitionNode_0.Transition
    src=Conduit
    via=AnimStateTransitionNode_0
   }
   {
    dst=100up_250fwd_Land
    rule=TacticalSprintLayer.AnimGraphNode_StateMachine_0.TacticalSprint.AnimStateNode_3.Hurdle_To_TacticalSprint.AnimGraphNode_StateMachine_0.Hurdle_Land.AnimStateTransitionNode_4.Transition
    src=Conduit
    via=AnimStateTransitionNode_4
   }
   {
    dst=100up_250fwd_Land_Mirrored
    rule=TacticalSprintLayer.AnimGraphNode_StateMachine_0.TacticalSprint.AnimStateNode_3.Hurdle_To_TacticalSprint.AnimGraphNode_StateMachine_0.Hurdle_Land.AnimStateTransitionNode_1.Transition
    src=Conduit
    via=AnimStateTransitionNode_1
   }
   {
    dst=200up_100fwd_Land
    rule=TacticalSprintLayer.AnimGraphNode_StateMachine_0.TacticalSprint.AnimStateNode_3.Hurdle_To_TacticalSprint.AnimGraphNode_StateMachine_0.Hurdle_Land.AnimStateTransitionNode_2.Transition
    src=Conduit
    via=AnimStateTransitionNode_2
   }
   {
    dst=200up_250fwd_Land
    rule=TacticalSprintLayer.AnimGraphNode_StateMachine_0.TacticalSprint.AnimStateNode_3.Hurdle_To_TacticalSprint.AnimGraphNode_StateMachine_0.Hurdle_Land.AnimStateTransitionNode_5.Transition
    src=Conduit
    via=AnimStateTransitionNode_5
   }
   {
    dst=50up_100fwd_Land
    rule=TacticalSprintLayer.AnimGraphNode_StateMachine_0.TacticalSprint.AnimStateNode_3.Hurdle_To_TacticalSprint.AnimGraphNode_StateMachine_0.Hurdle_Land.AnimStateTransitionNode_6.Transition
    src=Conduit
    via=AnimStateTransitionNode_6
   }
   {
    dst=50up_250fwd_Land
    rule=TacticalSprintLayer.AnimGraphNode_StateMachine_0.TacticalSprint.AnimStateNode_3.Hurdle_To_TacticalSprint.AnimGraphNode_StateMachine_0.Hurdle_Land.AnimStateTransitionNode_7.Transition
    src=Conduit
    via=AnimStateTransitionNode_7
   }
  ]
 }
 Conduit {
  path=TacticalSprintLayer.AnimGraphNode_StateMachine_0.TacticalSprint.AnimStateNode_3.Hurdle_To_TacticalSprint.AnimGraphNode_StateMachine_0.Hurdle_Land.AnimStateConduitNode_0.Conduit
  nodes [Result(bCanEnterTransition=true)]
 }
 200up_250fwd_Land {
  path=TacticalSprintLayer.AnimGraphNode_StateMachine_0.TacticalSprint.AnimStateNode_3.Hurdle_To_TacticalSprint.AnimGraphNode_StateMachine_0.Hurdle_Land.AnimStateNode_0.200up_250fwd_Land
  nodes [
   n1 = Sequence Player 'Hurdle_LandToSprint_250Fwd_200Up_CMF'() {
    bLoopAnimation=false
    groupName=Locomotion
    method=SyncGroup
    sequence=/Game/Animation/Game/MainPlayer/Locomotion/Medium/Male/TacticalSprint/Hurdle_LandToSprint_250Fwd_200Up_CMF
   }
   n2 = Sequence Player 'Hurdle_LandToSprint_250Fwd_200Up'() {
    bLoopAnimation=false
    groupName=Locomotion
    method=SyncGroup
    sequence=/Game/Animation/Game/MainPlayer/Locomotion/Medium/Male/TacticalSprint/Hurdle_LandToSprint_250Fwd_200Up
   }
   Output Animation Pose(Result=Blend Poses by bool(BlendPose_0=@n1, BlendPose_1=@n2, bActiveValue=Get bIsFemale() -> (bIsFemale:bool)) {blendType=Linear})
  ]
 }
 100up_250fwd_Land {
  path=TacticalSprintLayer.AnimGraphNode_StateMachine_0.TacticalSprint.AnimStateNode_3.Hurdle_To_TacticalSprint.AnimGraphNode_StateMachine_0.Hurdle_Land.AnimStateNode_1.100up_250fwd_Land
  nodes [
   n1 = Sequence Player 'Hurdle_LandToSprint_250Fwd_100Up_CMF'() {
    bLoopAnimation=false
    groupName=Locomotion
    method=SyncGroup
    sequence=/Game/Animation/Game/MainPlayer/Locomotion/Medium/Male/TacticalSprint/Hurdle_LandToSprint_250Fwd_100Up_CMF
   }
   n2 = Sequence Player 'Hurdle_LandToSprint_250Fwd_100Up'() {
    bLoopAnimation=false
    groupName=Locomotion
    method=SyncGroup
    sequence=/Game/Animation/Game/MainPlayer/Locomotion/Medium/Male/TacticalSprint/Hurdle_LandToSprint_250Fwd_100Up
   }
   Output Animation Pose(Result=Blend Poses by bool(BlendPose_0=@n1, BlendPose_1=@n2, bActiveValue=Get bIsFemale()) {blendType=Linear})
  ]
 }
 100up_100fwd_Land {
  path=TacticalSprintLayer.AnimGraphNode_StateMachine_0.TacticalSprint.AnimStateNode_3.Hurdle_To_TacticalSprint.AnimGraphNode_StateMachine_0.Hurdle_Land.AnimStateNode_2.100up_100fwd_Land
  nodes [
   n1 = Sequence Player 'Hurdle_LandToSprint_100Fwd_100Up_CMF'() {
    bLoopAnimation=false
    groupName=Locomotion
    method=SyncGroup
    sequence=/Game/Animation/Game/MainPlayer/Locomotion/Medium/Male/TacticalSprint/Hurdle_LandToSprint_100Fwd_100Up_CMF
   }
   n2 = Sequence Player 'Hurdle_LandToSprint_100Fwd_100Up'() {
    bLoopAnimation=false
    groupName=Locomotion
    method=SyncGroup
    sequence=/Game/Animation/Game/MainPlayer/Locomotion/Medium/Male/TacticalSprint/Hurdle_LandToSprint_100Fwd_100Up
   }
   Output Animation Pose(Result=Blend Poses by bool(BlendPose_0=@n1, BlendPose_1=@n2, bActiveValue=Get bIsFemale()) {blendType=Linear})
  ]
 }
 200up_100fwd_Land {
  path=TacticalSprintLayer.AnimGraphNode_StateMachine_0.TacticalSprint.AnimStateNode_3.Hurdle_To_TacticalSprint.AnimGraphNode_StateMachine_0.Hurdle_Land.AnimStateNode_3.200up_100fwd_Land
  nodes [
   n1 = Sequence Player 'Hurdle_LandToSprint_100Fwd_200Up_CMF'() {
    bLoopAnimation=false
    groupName=Locomotion
    method=SyncGroup
    sequence=/Game/Animation/Game/MainPlayer/Locomotion/Medium/Male/TacticalSprint/Hurdle_LandToSprint_100Fwd_200Up_CMF
   }
   n2 = Sequence Player 'Hurdle_LandToSprint_100Fwd_200Up'() {
    bLoopAnimation=false
    groupName=Locomotion
    method=SyncGroup
    sequence=/Game/Animation/Game/MainPlayer/Locomotion/Medium/Male/TacticalSprint/Hurdle_LandToSprint_100Fwd_200Up
   }
   Output Animation Pose(Result=Blend Poses by bool(BlendPose_0=@n1, BlendPose_1=@n2, bActiveValue=Get bIsFemale()) {blendType=Linear})
  ]
 }
 100up_100fwd_Land_Mirrored {
  path=TacticalSprintLayer.AnimGraphNode_StateMachine_0.TacticalSprint.AnimStateNode_3.Hurdle_To_TacticalSprint.AnimGraphNode_StateMachine_0.Hurdle_Land.AnimStateNode_4.100up_100fwd_Land_Mirrored
  nodes [
   n1 = Sequence Player 'Hurdle_LandToSprint_100Fwd_100Up_Mirrored'() {
    bLoopAnimation=false
    groupName=Locomotion
    method=SyncGroup
    sequence=/Game/Animation/Game/MainPlayer/Locomotion/Medium/Male/TacticalSprint/Hurdle_LandToSprint_100Fwd_100Up_Mirrored
   }
   n2 = Sequence Player 'Hurdle_LandToSprint_100Fwd_100Up_CMF_Mirrored'() {
    bLoopAnimation=false
    groupName=Locomotion
    method=SyncGroup
    sequence=/Game/Animation/Game/MainPlayer/Locomotion/Medium/Male/TacticalSprint/Hurdle_LandToSprint_100Fwd_100Up_CMF_Mirrored
   }
   Output Animation Pose(Result=Blend Poses by bool(BlendPose_0=@n2, BlendPose_1=@n1, bActiveValue=Get bIsFemale()) {blendType=Linear})
  ]
 }
 100up_250fwd_Land_Mirrored {
  path=TacticalSprintLayer.AnimGraphNode_StateMachine_0.TacticalSprint.AnimStateNode_3.Hurdle_To_TacticalSprint.AnimGraphNode_StateMachine_0.Hurdle_Land.AnimStateNode_5.100up_250fwd_Land_Mirrored
  nodes [
   n1 = Sequence Player 'Hurdle_LandToSprint_250Fwd_100Up_CMF_Mirrored'() {
    bLoopAnimation=false
    groupName=Locomotion
    method=SyncGroup
    sequence=/Game/Animation/Game/MainPlayer/Locomotion/Medium/Male/TacticalSprint/Hurdle_LandToSprint_250Fwd_100Up_CMF_Mirrored
   }
   n2 = Sequence Player 'Hurdle_LandToSprint_250Fwd_100Up_Mirrored'() {
    bLoopAnimation=false
    groupName=Locomotion
    method=SyncGroup
    sequence=/Game/Animation/Game/MainPlayer/Locomotion/Medium/Male/TacticalSprint/Hurdle_LandToSprint_250Fwd_100Up_Mirrored
   }
   Output Animation Pose(Result=Blend Poses by bool(BlendPose_0=@n1, BlendPose_1=@n2, bActiveValue=Get bIsFemale()) {blendType=Linear})
  ]
 }
 50up_100fwd_Land {
  path=TacticalSprintLayer.AnimGraphNode_StateMachine_0.TacticalSprint.AnimStateNode_3.Hurdle_To_TacticalSprint.AnimGraphNode_StateMachine_0.Hurdle_Land.AnimStateNode_6.50up_100fwd_Land
  nodes [
   n1 = Sequence Player 'Hurdle_LandToSprint_100Fwd_50Up'() {
    bLoopAnimation=false
    groupName=Locomotion
    method=SyncGroup
    sequence=/Game/Animation/Game/MainPlayer/Locomotion/Medium/Male/TacticalSprint/Hurdle_LandToSprint_100Fwd_50Up
   }
   n2 = Sequence Player 'Hurdle_LandToSprint_100Fwd_50Up_CMF'() {
    bLoopAnimation=false
    groupName=Locomotion
    method=SyncGroup
    sequence=/Game/Animation/Game/MainPlayer/Locomotion/Medium/Male/TacticalSprint/Hurdle_LandToSprint_100Fwd_50Up_CMF
   }
   Output Animation Pose(Result=Blend Poses by bool(BlendPose_0=@n2, BlendPose_1=@n1, bActiveValue=Get bIsFemale()) {blendType=Linear})
  ]
 }
 50up_250fwd_Land {
  path=TacticalSprintLayer.AnimGraphNode_StateMachine_0.TacticalSprint.AnimStateNode_3.Hurdle_To_TacticalSprint.AnimGraphNode_StateMachine_0.Hurdle_Land.AnimStateNode_7.50up_250fwd_Land
  nodes [
   n1 = Sequence Player 'Hurdle_LandToSprint_250Fwd_50Up_CMF'() {
    bLoopAnimation=false
    groupName=Locomotion
    method=SyncGroup
    sequence=/Game/Animation/Game/MainPlayer/Locomotion/Medium/Male/TacticalSprint/Hurdle_LandToSprint_250Fwd_50Up_CMF
   }
   n2 = Sequence Player 'Hurdle_LandToSprint_250Fwd_50Up'() {
    bLoopAnimation=false
    groupName=Locomotion
    method=SyncGroup
    sequence=/Game/Animation/Game/MainPlayer/Locomotion/Medium/Male/TacticalSprint/Hurdle_LandToSprint_250Fwd_50Up
   }
   Output Animation Pose(Result=Blend Poses by bool(BlendPose_0=@n1, BlendPose_1=@n2, bActiveValue=Get bIsFemale()) {blendType=Linear})
  ]
 }
 Transition {
  path=TacticalSprintLayer.AnimGraphNode_StateMachine_0.TacticalSprint.AnimStateNode_3.Hurdle_To_TacticalSprint.AnimGraphNode_StateMachine_0.Hurdle_Land.AnimStateTransitionNode_0.Transition
  nodes [Result(bCanEnterTransition=Get bTransitionHurdle100up_100fwd_Land_Mirrored() -> (bTransitionHurdle100up_100fwd_Land_Mirrored:bool))]
 }
 Transition {
  path=TacticalSprintLayer.AnimGraphNode_StateMachine_0.TacticalSprint.AnimStateNode_3.Hurdle_To_TacticalSprint.AnimGraphNode_StateMachine_0.Hurdle_Land.AnimStateTransitionNode_1.Transition
  nodes [Result(bCanEnterTransition=Get bTransitionHurdle100up_250fwd_Land_Mirrored() -> (bTransitionHurdle100up_250fwd_Land_Mirrored:bool))]
 }
 Transition {
  path=TacticalSprintLayer.AnimGraphNode_StateMachine_0.TacticalSprint.AnimStateNode_3.Hurdle_To_TacticalSprint.AnimGraphNode_StateMachine_0.Hurdle_Land.AnimStateTransitionNode_2.Transition
  nodes [Result(bCanEnterTransition=Get bTransitionHurdle200up_100fwd_Land() -> (bTransitionHurdle200up_100fwd_Land:bool))]
 }
 Transition {
  path=TacticalSprintLayer.AnimGraphNode_StateMachine_0.TacticalSprint.AnimStateNode_3.Hurdle_To_TacticalSprint.AnimGraphNode_StateMachine_0.Hurdle_Land.AnimStateTransitionNode_3.Transition
  nodes [Result(bCanEnterTransition=Get bTransitionHurdle100up_100fwd_Land() -> (bTransitionHurdle100up_100fwd_Land:bool))]
 }
 Transition {
  path=TacticalSprintLayer.AnimGraphNode_StateMachine_0.TacticalSprint.AnimStateNode_3.Hurdle_To_TacticalSprint.AnimGraphNode_StateMachine_0.Hurdle_Land.AnimStateTransitionNode_4.Transition
  nodes [Result(bCanEnterTransition=Get bTransitionHurdle100up_250fwd_Land() -> (bTransitionHurdle100up_250fwd_Land:bool))]
 }
 Transition {
  path=TacticalSprintLayer.AnimGraphNode_StateMachine_0.TacticalSprint.AnimStateNode_3.Hurdle_To_TacticalSprint.AnimGraphNode_StateMachine_0.Hurdle_Land.AnimStateTransitionNode_5.Transition
  nodes [Result(bCanEnterTransition=true)]
 }
 Transition {
  path=TacticalSprintLayer.AnimGraphNode_StateMachine_0.TacticalSprint.AnimStateNode_3.Hurdle_To_TacticalSprint.AnimGraphNode_StateMachine_0.Hurdle_Land.AnimStateTransitionNode_6.Transition
  nodes [Result(bCanEnterTransition=Get bTransitionHurdle50up_100fwd_Land() -> (bTransitionHurdle50up_100fwd_Land:bool))]
 }
 Transition {
  path=TacticalSprintLayer.AnimGraphNode_StateMachine_0.TacticalSprint.AnimStateNode_3.Hurdle_To_TacticalSprint.AnimGraphNode_StateMachine_0.Hurdle_Land.AnimStateTransitionNode_7.Transition
  nodes [Result(bCanEnterTransition=Get bTransitionHurdle50up_250fwd_Land() -> (bTransitionHurdle50up_250fwd_Land:bool))]
 }
 Hurdle_Mantle_To_TacticalSprint {
  path=TacticalSprintLayer.AnimGraphNode_StateMachine_0.TacticalSprint.AnimStateNode_4.Hurdle_Mantle_To_TacticalSprint
  nodes [
   AnimGraphNode_StateMachine_0 = Hurdle_Mantle_Land() {is=title AnimGraphNode_StateMachine, bAllowConduitEntryStates=true}
   Output Animation Pose(Result=@AnimGraphNode_StateMachine_0)
  ]
 }
 Hurdle_Mantle_Land {
  entry=Conduit
  path=TacticalSprintLayer.AnimGraphNode_StateMachine_0.TacticalSprint.AnimStateNode_4.Hurdle_Mantle_To_TacticalSprint.AnimGraphNode_StateMachine_0.Hurdle_Mantle_Land
  ids=id
  nodes [
   AnimStateConduitNode_0 = Conduit()
   AnimStateNode_0 = Hurdle_Mantle_100up() {is=title AnimStateNode}
   AnimStateNode_1 = Mantle_Hurdle_200up() {is=title AnimStateNode}
   AnimStateNode_2 = Hurdle_Mantle_50up() {is=title AnimStateNode}
   AnimStateNode_3 = Hurdle_Mantle_100up_Mirrored() {is=title AnimStateNode}
   AnimStateNode_4 = Mantle_Hurdle_200up_Mirrored() {is=title AnimStateNode}
   AnimStateTransitionNode_0 = AnimStateTransitionNode() {crossfadeDuration=0.3, logicType=TLT_Inertialization}
   AnimStateTransitionNode_2 = AnimStateTransitionNode() {crossfadeDuration=0.3, logicType=TLT_Inertialization, priorityOrder=2}
   AnimStateTransitionNode_3 = AnimStateTransitionNode() {crossfadeDuration=0.3, logicType=TLT_Inertialization, priorityOrder=4}
   AnimStateTransitionNode_4 = AnimStateTransitionNode() {crossfadeDuration=0.3, logicType=TLT_Inertialization, priorityOrder=3}
   AnimStateTransitionNode_5 = AnimStateTransitionNode() {crossfadeDuration=0.3, logicType=TLT_Inertialization, priorityOrder=100}
  ]
  links [
   {
    dst=Hurdle_Mantle_100up
    rule=TacticalSprintLayer.AnimGraphNode_StateMachine_0.TacticalSprint.AnimStateNode_4.Hurdle_Mantle_To_TacticalSprint.AnimGraphNode_StateMachine_0.Hurdle_Mantle_Land.AnimStateTransitionNode_2.Transition
    src=Conduit
    via=AnimStateTransitionNode_2
   }
   {
    dst=Hurdle_Mantle_100up_Mirrored
    rule=TacticalSprintLayer.AnimGraphNode_StateMachine_0.TacticalSprint.AnimStateNode_4.Hurdle_Mantle_To_TacticalSprint.AnimGraphNode_StateMachine_0.Hurdle_Mantle_Land.AnimStateTransitionNode_4.Transition
    src=Conduit
    via=AnimStateTransitionNode_4
   }
   {
    dst=Hurdle_Mantle_50up
    rule=TacticalSprintLayer.AnimGraphNode_StateMachine_0.TacticalSprint.AnimStateNode_4.Hurdle_Mantle_To_TacticalSprint.AnimGraphNode_StateMachine_0.Hurdle_Mantle_Land.AnimStateTransitionNode_0.Transition
    src=Conduit
    via=AnimStateTransitionNode_0
   }
   {
    dst=Mantle_Hurdle_200up
    rule=TacticalSprintLayer.AnimGraphNode_StateMachine_0.TacticalSprint.AnimStateNode_4.Hurdle_Mantle_To_TacticalSprint.AnimGraphNode_StateMachine_0.Hurdle_Mantle_Land.AnimStateTransitionNode_3.Transition
    src=Conduit
    via=AnimStateTransitionNode_3
   }
   {
    dst=Mantle_Hurdle_200up_Mirrored
    rule=TacticalSprintLayer.AnimGraphNode_StateMachine_0.TacticalSprint.AnimStateNode_4.Hurdle_Mantle_To_TacticalSprint.AnimGraphNode_StateMachine_0.Hurdle_Mantle_Land.AnimStateTransitionNode_5.Transition
    src=Conduit
    via=AnimStateTransitionNode_5
   }
  ]
 }
 Conduit {
  path=TacticalSprintLayer.AnimGraphNode_StateMachine_0.TacticalSprint.AnimStateNode_4.Hurdle_Mantle_To_TacticalSprint.AnimGraphNode_StateMachine_0.Hurdle_Mantle_Land.AnimStateConduitNode_0.Conduit
  nodes [Result(bCanEnterTransition=true)]
 }
 Hurdle_Mantle_100up {
  path=TacticalSprintLayer.AnimGraphNode_StateMachine_0.TacticalSprint.AnimStateNode_4.Hurdle_Mantle_To_TacticalSprint.AnimGraphNode_StateMachine_0.Hurdle_Mantle_Land.AnimStateNode_0.Hurdle_Mantle_100up
  nodes [
   n1 = Sequence Player 'Hurdle_Mantle_LandToSprint_100Up'() {
    groupName=Locomotion
    method=SyncGroup
    sequence=/Game/Animation/Game/MainPlayer/Locomotion/Medium/Male/TacticalSprint/Hurdle_Mantle_LandToSprint_100Up
   }
   n2 = Sequence Player 'Hurdle_Mantle_LandToSprint_100Up_CMF'() {
    bLoopAnimation=false
    groupName=Locomotion
    method=SyncGroup
    sequence=/Game/Animation/Game/MainPlayer/Locomotion/Medium/Male/TacticalSprint/Hurdle_Mantle_LandToSprint_100Up_CMF
   }
   Output Animation Pose(Result=Blend Poses by bool(BlendPose_0=@n2, BlendPose_1=@n1, bActiveValue=Get bIsFemale()) {blendType=Linear})
  ]
 }
 Mantle_Hurdle_200up {
  path=TacticalSprintLayer.AnimGraphNode_StateMachine_0.TacticalSprint.AnimStateNode_4.Hurdle_Mantle_To_TacticalSprint.AnimGraphNode_StateMachine_0.Hurdle_Mantle_Land.AnimStateNode_1.Mantle_Hurdle_200up
  nodes [
   n1 = Sequence Player 'Hurdle_Mantle_LandToSprint_200Up'() {
    groupName=Locomotion
    method=SyncGroup
    sequence=/Game/Animation/Game/MainPlayer/Locomotion/Medium/Male/TacticalSprint/Hurdle_Mantle_LandToSprint_200Up
   }
   n2 = Sequence Player 'Hurdle_Mantle_LandToSprint_200Up_CMF'() {
    bLoopAnimation=false
    groupName=Locomotion
    method=SyncGroup
    sequence=/Game/Animation/Game/MainPlayer/Locomotion/Medium/Male/TacticalSprint/Hurdle_Mantle_LandToSprint_200Up_CMF
   }
   Output Animation Pose(Result=Blend Poses by bool(BlendPose_0=@n2, BlendPose_1=@n1, bActiveValue=Get bIsFemale()) {blendType=Linear})
  ]
 }
 Hurdle_Mantle_50up {
  path=TacticalSprintLayer.AnimGraphNode_StateMachine_0.TacticalSprint.AnimStateNode_4.Hurdle_Mantle_To_TacticalSprint.AnimGraphNode_StateMachine_0.Hurdle_Mantle_Land.AnimStateNode_2.Hurdle_Mantle_50up
  nodes [
   n1 = Sequence Player 'Hurdle_Mantle_LandToSprint_50Up'() {
    bLoopAnimation=false
    groupName=Locomotion
    method=SyncGroup
    sequence=/Game/Animation/Game/MainPlayer/Locomotion/Medium/Male/TacticalSprint/Hurdle_Mantle_LandToSprint_50Up
   }
   n2 = Sequence Player 'Hurdle_Mantle_LandToSprint_50Up_CMF'() {
    bLoopAnimation=false
    groupName=Locomotion
    method=SyncGroup
    sequence=/Game/Animation/Game/MainPlayer/Locomotion/Medium/Male/TacticalSprint/Hurdle_Mantle_LandToSprint_50Up_CMF
   }
   Output Animation Pose(Result=Blend Poses by bool(BlendPose_0=@n2, BlendPose_1=@n1, bActiveValue=Get bIsFemale()) {blendType=Linear})
  ]
 }
 Hurdle_Mantle_100up_Mirrored {
  path=TacticalSprintLayer.AnimGraphNode_StateMachine_0.TacticalSprint.AnimStateNode_4.Hurdle_Mantle_To_TacticalSprint.AnimGraphNode_StateMachine_0.Hurdle_Mantle_Land.AnimStateNode_3.Hurdle_Mantle_100up_Mirrored
  nodes [
   n1 = Sequence Player 'Hurdle_Mantle_LandToSprint_100Up_Mirrored'() {
    bLoopAnimation=false
    groupName=Locomotion
    method=SyncGroup
    sequence=/Game/Animation/Game/MainPlayer/Locomotion/Medium/Male/TacticalSprint/Hurdle_Mantle_LandToSprint_100Up_Mirrored
   }
   n2 = Sequence Player 'Hurdle_Mantle_LandToSprint_100Up_CMF_Mirrored'() {
    bLoopAnimation=false
    groupName=Locomotion
    method=SyncGroup
    sequence=/Game/Animation/Game/MainPlayer/Locomotion/Medium/Male/TacticalSprint/Hurdle_Mantle_LandToSprint_100Up_CMF_Mirrored
   }
   Output Animation Pose(Result=Blend Poses by bool(BlendPose_0=@n2, BlendPose_1=@n1, bActiveValue=Get bIsFemale()) {blendType=Linear})
  ]
 }
 Mantle_Hurdle_200up_Mirrored {
  path=TacticalSprintLayer.AnimGraphNode_StateMachine_0.TacticalSprint.AnimStateNode_4.Hurdle_Mantle_To_TacticalSprint.AnimGraphNode_StateMachine_0.Hurdle_Mantle_Land.AnimStateNode_4.Mantle_Hurdle_200up_Mirrored
  nodes [
   n1 = Sequence Player 'Hurdle_Mantle_LandToSprint_200Up_CMF_Mirrored'() {
    bLoopAnimation=false
    groupName=Locomotion
    method=SyncGroup
    sequence=/Game/Animation/Game/MainPlayer/Locomotion/Medium/Male/TacticalSprint/Hurdle_Mantle_LandToSprint_200Up_CMF_Mirrored
   }
   n2 = Sequence Player 'Hurdle_Mantle_LandToSprint_200Up_Mirrored'() {
    bLoopAnimation=false
    groupName=Locomotion
    method=SyncGroup
    sequence=/Game/Animation/Game/MainPlayer/Locomotion/Medium/Male/TacticalSprint/Hurdle_Mantle_LandToSprint_200Up_Mirrored
   }
   Output Animation Pose(Result=Blend Poses by bool(BlendPose_0=@n1, BlendPose_1=@n2, bActiveValue=Get bIsFemale()) {blendType=Linear})
  ]
 }
 Transition {
  path=TacticalSprintLayer.AnimGraphNode_StateMachine_0.TacticalSprint.AnimStateNode_4.Hurdle_Mantle_To_TacticalSprint.AnimGraphNode_StateMachine_0.Hurdle_Mantle_Land.AnimStateTransitionNode_0.Transition
  nodes [Result(bCanEnterTransition=Get bTransitionHurdleMantle50up() -> (bTransitionHurdleMantle50up:bool))]
 }
 Transition {
  path=TacticalSprintLayer.AnimGraphNode_StateMachine_0.TacticalSprint.AnimStateNode_4.Hurdle_Mantle_To_TacticalSprint.AnimGraphNode_StateMachine_0.Hurdle_Mantle_Land.AnimStateTransitionNode_2.Transition
  nodes [Result(bCanEnterTransition=Get bTransitionHurdleMantle100up() -> (bTransitionHurdleMantle100up:bool))]
 }
 Transition {
  path=TacticalSprintLayer.AnimGraphNode_StateMachine_0.TacticalSprint.AnimStateNode_4.Hurdle_Mantle_To_TacticalSprint.AnimGraphNode_StateMachine_0.Hurdle_Mantle_Land.AnimStateTransitionNode_3.Transition
  nodes [Result(bCanEnterTransition=Get bTransitionHurdleMantle200up() -> (bTransitionHurdleMantle200up:bool))]
 }
 Transition {
  path=TacticalSprintLayer.AnimGraphNode_StateMachine_0.TacticalSprint.AnimStateNode_4.Hurdle_Mantle_To_TacticalSprint.AnimGraphNode_StateMachine_0.Hurdle_Mantle_Land.AnimStateTransitionNode_4.Transition
  nodes [Result(bCanEnterTransition=Get bTransitionHurdleMantle100upMirrored() -> (bTransitionHurdleMantle100upMirrored:bool))]
 }
 Transition {
  path=TacticalSprintLayer.AnimGraphNode_StateMachine_0.TacticalSprint.AnimStateNode_4.Hurdle_Mantle_To_TacticalSprint.AnimGraphNode_StateMachine_0.Hurdle_Mantle_Land.AnimStateTransitionNode_5.Transition
  nodes [Result(bCanEnterTransition=true)]
 }
 Fall_To_TacticalSprint {
  path=TacticalSprintLayer.AnimGraphNode_StateMachine_0.TacticalSprint.AnimStateNode_5.Fall_To_TacticalSprint
  nodes [
   n1 = Get CurrentTacticalStateAnimSet()
   Output Animation Pose(Result=Sequence Player(Sequence=Break Fort Tactical State Anim Set(FortTacticalStateAnimSet=@n1).FallToSprint) {bLoopAnimation=false, groupName=Locomotion, groupRole=TransitionFollower, method=SyncGroup})
  ]
 }
 AnimationTransitionGraph_0 {
  path=TacticalSprintLayer.AnimGraphNode_StateMachine_0.TacticalSprint.AnimStateTransitionNode_0.AnimationTransitionGraph_0
  nodes [Result(bCanEnterTransition=true)]
 }
 AnimationTransitionGraph_0 {
  path=TacticalSprintLayer.AnimGraphNode_StateMachine_0.TacticalSprint.AnimStateTransitionNode_1.AnimationTransitionGraph_0
  nodes [Result(bCanEnterTransition=true)]
 }
 Transition {
  path=TacticalSprintLayer.AnimGraphNode_StateMachine_0.TacticalSprint.AnimStateTransitionNode_10.Transition
  nodes [Result(bCanEnterTransition=Get bIsHardLanding() -> (bIsHardLanding:bool))]
 }
 AnimationTransitionGraph_0 {
  path=TacticalSprintLayer.AnimGraphNode_StateMachine_0.TacticalSprint.AnimStateTransitionNode_2.AnimationTransitionGraph_0
  nodes [Result(bCanEnterTransition=Get bIsDoorBashExit() -> (bIsDoorBashExit:bool))]
 }
 AnimationTransitionGraph_0 {path=TacticalSprintLayer.AnimGraphNode_StateMachine_0.TacticalSprint.AnimStateTransitionNode_3.AnimationTransitionGraph_0, nodes [Result()]}
 Transition {
  path=TacticalSprintLayer.AnimGraphNode_StateMachine_0.TacticalSprint.AnimStateTransitionNode_4.Transition
  nodes [Result(bCanEnterTransition=Get bIsHurdleLanding() -> (bIsHurdleLanding:bool))]
 }
 Transition {path=TacticalSprintLayer.AnimGraphNode_StateMachine_0.TacticalSprint.AnimStateTransitionNode_5.Transition, nodes [Result()]}
 Transition {
  path=TacticalSprintLayer.AnimGraphNode_StateMachine_0.TacticalSprint.AnimStateTransitionNode_6.Transition
  nodes [Result(bCanEnterTransition=Get bIsHurdleMantleLanding() -> (bIsHurdleMantleLanding:bool))]
 }
 Transition {path=TacticalSprintLayer.AnimGraphNode_StateMachine_0.TacticalSprint.AnimStateTransitionNode_7.Transition, nodes [Result()]}
 Transition {path=TacticalSprintLayer.AnimGraphNode_StateMachine_0.TacticalSprint.AnimStateTransitionNode_9.Transition, nodes [Result()]}
 SprintingPoseAdditives {
  entry=SprintingPoseAdditivesState
  path=TacticalSprintLayer.AnimGraphNode_StateMachine_16.SprintingPoseAdditives
  ids=id
  nodes [AnimStateNode_0 = SprintingPoseAdditivesState() {is=title AnimStateNode}]
 }
 SprintingPoseAdditivesState {
  path=TacticalSprintLayer.AnimGraphNode_StateMachine_16.SprintingPoseAdditives.AnimStateNode_0.SprintingPoseAdditivesState
  nodes [
   n1 = Get PawnSpeed2D()
   n2 = Get SprintLeanAngle() -> (SprintLeanAngle:float)
   Output Animation Pose(Result=Apply Additive(
     Base=Use cached pose 'TacticalSprintPose'()
     Additive=Blendspace Player 'TacSprint_Leans_BS'(X=@n2, Y=@n1) {blendSpace=/Game/Animation/Game/MainPlayerFN/Poses/TacSprint/TacSprint_Leans_BS}
     Alpha=Get LeanAdditiveAlpha() -> (LeanAdditiveAlpha:float)
    ) {lODThreshold=1, alphaBoolBlend {blendOption=Linear}})
  ]
 }
]
variables []
