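# Animation layer interface for player facial animation on the Male_Avg_Base character, declaring the FacialLayerInterface layer group.
# It defines two layer functions with no logic of their own.
# FacialLayer_FaceCurves takes two input poses, one from before and one from after the facial curves, and returns a pose.
# FacialLayer_FullBody takes a single input pose and returns a pose.
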AnimBlueprint /Game/Characters/Player/Male/Male_Avg_Base/AnimLayerInterface/FacialCurvesLayerInterface
target_skeleton=null
graphs [
 FacialLayer_FaceCurves {nodes [
   Input Pose() {name=InPosePreFacialCurves}
   Input Pose() {name=InPosePostFacialCurves}
   Output Pose() {layerGroup=FacialLayerInterface}
  ]}
 FacialLayer_FullBody {nodes [
   Input Pose()
   Output Pose() {layerGroup=FacialLayerInterface}
  ]}
]
variables []
