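# Animation Layer Interface for procedural pose layering on the average-build male player character, with no target skeleton.
# It declares two ungrouped layers and contains no blend logic or variables of its own.
# ProceduralAdditivePose takes source locomotion, motion matching and source poses and returns one pose.
# ProceduralLayeredPose takes a single locomotion pose and returns one pose.
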AnimBlueprint /Game/Characters/Player/Male/Male_Avg_Base/AnimLayerInterface/ProceduralLayeringAnimInterface
target_skeleton=null
graphs [
 ProceduralAdditivePose {nodes [
   Input Pose() {name=InSourceLocomotionPose}
   Input Pose() {name=InMotionMatchingPose}
   Input Pose() {name=InSourcePose}
   Output Pose() {layerGroup=None}
  ]}
 ProceduralLayeredPose {nodes [
   Input Pose() {name=InLocomotionPose}
   Output Pose() {layerGroup=None}
  ]}
]
variables []
