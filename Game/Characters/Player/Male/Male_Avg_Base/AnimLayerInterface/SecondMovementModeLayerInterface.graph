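# SecondMovementModeLayerInterface is an animation layer interface for the average male player character's animation blueprints, tied to no target skeleton.
# It declares one layer, SecondMovementMode_FinalPoseOverride, in the SecondMovementMode layer group.
# The layer takes the InFinalPose input pose and outputs a pose, so implementing blueprints can override the final pose.
# The interface holds no variables or logic of its own.
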AnimBlueprint /Game/Characters/Player/Male/Male_Avg_Base/AnimLayerInterface/SecondMovementModeLayerInterface
target_skeleton=null
graphs [SecondMovementMode_FinalPoseOverride {nodes [
   Input Pose() {name=InFinalPose}
   Output Pose() {layerGroup=SecondMovementMode}
  ]}]
variables []
