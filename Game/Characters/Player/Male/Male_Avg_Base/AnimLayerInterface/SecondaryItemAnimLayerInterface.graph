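# Animation layer interface that declares the SecondaryItemAimOffsets layer for aim offsets on the player's secondary item.
# The layer takes two input poses, InPoseAimOffsets and InPosePreAimOffsets, with AimOffsetAlpha, Yaw and Pitch inputs on the first.
# Its output pose belongs to the SecondaryItemLayer group, and the graph holds only this signature, with no blending logic or variables.
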AnimBlueprint /Game/Characters/Player/Male/Male_Avg_Base/AnimLayerInterface/SecondaryItemAnimLayerInterface
target_skeleton=null
graphs [SecondaryItemAimOffsets {nodes [
   Input Pose!() -> (Pose, AimOffsetAlpha:double=~, Yaw:double=~, Pitch:double=~) {
    name=InPoseAimOffsets
    inputs [
     AimOffsetAlpha {type {
       bIsConst=false
       bIsReference=false
       bIsUObjectWrapper=false
       bIsWeakPointer=false
       bSerializeAsSinglePrecisionFloat=false
       containerType=None
       pinCategory=real
       pinSubCategory=double
       pinSubCategoryObject=null
       pinSubCategoryMemberReference {memberGuid=#1, memberName=None, memberParent=null}
       pinValueType {
        bTerminalIsConst=false
        bTerminalIsUObjectWrapper=false
        bTerminalIsWeakPointer=false
        terminalCategory=None
        terminalSubCategory=None
        terminalSubCategoryObject=null
       }
      }}
     Yaw {type {
       bIsConst=false
       bIsReference=false
       bIsUObjectWrapper=false
       bIsWeakPointer=false
       bSerializeAsSinglePrecisionFloat=false
       containerType=None
       pinCategory=real
       pinSubCategory=double
       pinSubCategoryObject=null
       pinSubCategoryMemberReference {memberGuid=#1, memberName=None, memberParent=null}
       pinValueType {
        bTerminalIsConst=false
        bTerminalIsUObjectWrapper=false
        bTerminalIsWeakPointer=false
        terminalCategory=None
        terminalSubCategory=None
        terminalSubCategoryObject=null
       }
      }}
     Pitch {type {
       bIsConst=false
       bIsReference=false
       bIsUObjectWrapper=false
       bIsWeakPointer=false
       bSerializeAsSinglePrecisionFloat=false
       containerType=None
       pinCategory=real
       pinSubCategory=double
       pinSubCategoryObject=null
       pinSubCategoryMemberReference {memberGuid=#1, memberName=None, memberParent=null}
       pinValueType {
        bTerminalIsConst=false
        bTerminalIsUObjectWrapper=false
        bTerminalIsWeakPointer=false
        terminalCategory=None
        terminalSubCategory=None
        terminalSubCategoryObject=null
       }
      }}
    ]
   }
   Input Pose!() -> (Pose) {name=InPosePreAimOffsets}
   Output Pose() {layerGroup=SecondaryItemLayer}
  ]}]
variables []
