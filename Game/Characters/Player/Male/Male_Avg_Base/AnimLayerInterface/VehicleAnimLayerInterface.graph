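# Animation Layer Interface that declares pose override layers for player character vehicle animation, with no skeleton, variables or logic of its own.
# FinalPose and FullBody layers each take one input pose, LowerBody takes upper-body and default lower-body poses.
# SplitBody takes split-body, upper-and-lower-body and root-removed upper-and-lower-body poses.
# SplitBody's group is "Vehicle Override" with a space, separating it from the other three in "VehicleOverride".
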AnimBlueprint /Game/Characters/Player/Male/Male_Avg_Base/AnimLayerInterface/VehicleAnimLayerInterface
target_skeleton=null
graphs [
 VehicleFinalPoseOverride {nodes [
   Input Pose() {name=InFinalPose}
   Output Pose() {layerGroup=VehicleOverride}
  ]}
 VehicleFullBodyOverride {nodes [
   Output Pose() {layerGroup=VehicleOverride}
   AnimGraphNode_SubInput_0 = Input Pose() {name=InPoseFullBody}
  ]}
 VehicleLowerBodyOverride {nodes [
   Input Pose() {name=InPoseUpperBody}
   Output Pose() {layerGroup=VehicleOverride}
   AnimGraphNode_SubInput_0 = Input Pose() {name=InPoseLowerBodyDefault}
  ]}
 VehicleSplitBodyOverride {nodes [
   Input Pose() {name=InPoseSplitBody}
   Input Pose() {name=InPoseUpperAndLowerBody}
   Input Pose() {name=InPoseUpperAndLowerBodyRemoveRoot}
   Output Pose() {layerGroup=Vehicle Override}
  ]}
]
variables []
