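# Main Fortnite player character AnimBlueprint, assembling the full body pose from locomotion, state machines and swappable linked layers.
# Motion matching locomotion feeds procedural layering and a cached SourcePose, used by nested state machines for in-air, tactical sprint, skydive, cannon and stun.
# Layer interfaces let locomotion, movement mode, vehicle and item graphs override stages before procedural adjustments, rigid body and facial layers.
# An offscreen LOD flag switches between this pipeline and LowLODLayer.
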AnimBlueprint /Game/Characters/Player/Male/Male_Avg_Base/Fortnite_M_Avg_Player_AnimBlueprint
target_skeleton=&Fortnite_M_Avg_Player_Skeleton
graphs [
 ADSReload {nodes [
   n1 = ADS Reload!(
    ComponentPose:Component Space Pose Link Structure=Local To Component(LocalPose=Use cached pose 'InputPose'())
    Alpha:float=1
    CurrentReloadMontage:Anim Montage Object Reference
    ReloadMontageSlotIndex:int=0
    WeaponAttachBoneName:Name=None
    AttachmentRotationCorrection:Rotator="0,0,0"
    TargetingWeight:float=0
   ) -> (Pose:Component Space Pose Link Structure) {
    is=title FortAnimGraphNode_AdsReload
    tag=ADSReload
    node {
     alphaBoolBlend {blendOption=Linear}
     leftArm {iKBone {bone {boneName=hand_l}}}
     rightArm {iKBone {bone {boneName=hand_r}}}
     updateFunction {functionName=OnAdsReloadNodeUpdate}
    }
    updateFunction {bSelfContext=true, memberName=OnAdsReloadNodeUpdate}
   }
   n2 = Get ADSReloadInput() -> (ADSReloadInput:Fort Anim Input ADSReload Structure)
   Output Pose(Result=Blend Poses by bool(
     BlendPose_0=Component To Local(ComponentPose=@n1)
     BlendPose_1=Use cached pose 'InputPose'()
     BlendTime_0=0
     BlendTime_1=0
     bActiveValue=Break Fort Anim Input ADSReload!(FortAnimInput_ADSReload:"Fort Anim Input ADSReload Structure (by ref)"=@n2) -> (
      WeaponAttachBoneName:Name
      AttachmentRotationCorrection:Rotator=~
      AdsReloadLeftHandIKAlpha:float=~
      AdsReloadRightHandIKAlpha:float=~
      TargetingWeight:float=~
      bActive:bool=~
      CurrentReloadMontage:Anim Montage Weak object
      ReloadMontageSlotIndex:int
     ) {is=title K2Node_BreakStruct}.bActive
    ))
   Save cached pose 'InputPose'(Pose=Input Pose() {name=InPose_ADSReload}) {cacheName=InputPose}
  ]}
 AnimGraph {nodes [
   n1 = LowLODLayer!() -> (Pose:Pose Link Structure) {is=title AnimGraphNode_LinkedAnimLayer, layer=LowLODLayer}
   n2 = Player_ProceduralAdjustments_AnimBP!(
    InPosePostTurnInPlace:Pose Link Structure=Use cached pose 'FullBody'()
    IsHidingInProp:bool=Get bIsHidingInProp() -> (bIsHidingInProp:bool)
    RightHandOverride:Vector=Get RightHandOverride() -> (RightHandOverride:Vector)
    LeftHandOverride:Vector=Get LeftHandOverride() -> (LeftHandOverride:Vector)
    EmoteAnimOffsetCurveName:Name=Get EmoteAnimOffsetCurveName() -> (EmoteAnimOffsetCurveName:Name=None)
   ) -> (Pose:Pose Link Structure) {
    is=title AnimGraphNode_LinkedAnimGraph
    tag=PlayerProceduralAdjustments
    instanceClass=/Game/Characters/Player/AnimationBlueprints/LinkedAnimGraphs/Player_ProceduralAdjustments_AnimBP.Player_ProceduralAdjustments_AnimBP_C
   }
   n3 = MovementModeLayerInterface - MovementMode_FinalPoseOverride!(InFinalPose:Pose Link Structure=Two Way Blend(
     A=Use cached pose 'FullBody'()
     B=Linked Graph Gate!(Source:Pose Link Structure=@n2) -> (Pose:Pose Link Structure) {is=title FortAnimGraphNode_LinkedGraphGate, linkedGraphTag=PlayerProceduralAdjustments}
     Alpha=Get ProceduralAdjustmentsAlpha() -> (ProceduralAdjustmentsAlpha:float)
    )) -> (Pose:Pose Link Structure) {
    is=title AnimGraphNode_LinkedAnimLayer
    tag=MovementMode_FinalPoseOverride
    interface=/Game/Characters/Player/Male/Male_Avg_Base/AnimLayerInterface/MovementModeLayerInterface.MovementModeLayerInterface_C
    layer=MovementMode_FinalPoseOverride
   }
   n4 = SecondMovementModeLayerInterface - SecondMovementMode_FinalPoseOverride!(InFinalPose:Pose Link Structure=@n3) -> (Pose:Pose Link Structure) {
    is=title AnimGraphNode_LinkedAnimLayer
    tag=SecondMovementMode_FinalPoseOverride
    interface=/Game/Characters/Player/Male/Male_Avg_Base/AnimLayerInterface/SecondMovementModeLayerInterface.SecondMovementModeLayerInterface_C
    layer=SecondMovementMode_FinalPoseOverride
   }
   n5 = VehicleAnimLayerInterface - VehicleFinalPoseOverride!(InFinalPose:Pose Link Structure=@n4) -> (Pose:Pose Link Structure) {
    is=title AnimGraphNode_LinkedAnimLayer
    tag=VehicleFinalPoseOverride
    interface=/Game/Characters/Player/Male/Male_Avg_Base/AnimLayerInterface/VehicleAnimLayerInterface.VehicleAnimLayerInterface_C
    layer=VehicleFinalPoseOverride
   }
   n6 = ItemAnimLayerInterface - ItemFinalPoseOverride!(InFinalPose:Pose Link Structure=@n5) -> (Pose:Pose Link Structure) {
    is=title AnimGraphNode_LinkedAnimLayer
    tag=ItemFinalPoseOverride
    interface=/Game/Characters/Player/Male/Male_Avg_Base/AnimLayerInterface/ItemAnimLayerInterface.ItemAnimLayerInterface_C
    layer=ItemFinalPoseOverride
   }
   n7 = PlayerLocomotionLayerInterface - LocomotionLayer_FinalPose!(InPoseFinalPose:Pose Link Structure=@n6) -> (Pose:Pose Link Structure) {
    is=title AnimGraphNode_LinkedAnimLayer
    interface=/Game/Characters/Player/Male/Male_Avg_Base/AnimLayerInterface/PlayerLocomotionLayerInterface.PlayerLocomotionLayerInterface_C
    layer=LocomotionLayer_FinalPose
   }
   n8 = Local To Component(LocalPose=RigidBodyPBAProto!(InPoseRigidBodyPBA:Pose Link Structure=@n7) -> (Pose:Pose Link Structure) {is=title AnimGraphNode_LinkedAnimLayer, layer=RigidBodyPBAProto})
   n9 = Slot 'SequencerDefault'(Source=Component To Local(ComponentPose=Scale Character Bones!(ComponentPose:Component Space Pose Link Structure=@n8, Alpha:float=1) -> (Pose:Component Space Pose Link Structure) {is=title FortAnimGraphNode_ScaleCharacterBones, alphaBoolBlend {blendOption=Linear}})) {slotName=SequencerDefault}
   n10 = HideWeapon!(InPoseHideWeapon:Pose Link Structure=Dead Blending!(Source:Pose Link Structure=FacialCurvesLayer!(InPoseFacialCurves:Pose Link Structure=@n9) -> (Pose:Pose Link Structure) {is=title AnimGraphNode_LinkedAnimLayer, layer=FacialCurvesLayer}) -> (Pose:Pose Link Structure) {is=title AnimGraphNode_DeadBlending, bForwardRequestsThroughSkippedCachedPoseNodes=true, defaultBlendMode=Linear}) -> (Pose:Pose Link Structure) {is=title AnimGraphNode_LinkedAnimLayer, layer=HideWeapon}
   Output Pose(Result=Blend Poses by bool(
     BlendPose_0=Profiler Stat Event!(Source:Pose Link Structure=@n1) -> (Pose:Pose Link Structure) {is=title FortAnimGraphNode_ProfilerStatEvent, profilerEventName=MainPlayerAnimBP_LowLOD}
     BlendPose_1=Profiler Stat Event(Source=@n10) {profilerEventName=MainPlayerAnimBP_All}
     BlendTime_0=0
     BlendTime_1=0
     bActiveValue=Get bEnableOffscreenLOD() -> (bEnableOffscreenLOD:bool)
    )) {layerGroup=None}
   n11 = Layered blend per bone(
    BasePose=Slot 'FullBodyInAir'(Source=Use cached pose 'FullBodyPostAdditives'()) {slotName=FullBodyInAir}
    BlendPoses_0=Use cached pose 'FullBodyPostAdditives'()
    BlendWeights_0=Get FullBodyInAirLowerBodyWeight() -> (FullBodyInAirLowerBodyWeight:float)
   ) {
    bMeshSpaceRotationBlend=true
    layerSetup [{branchFilters [{boneName=Root}, {blendDepth=-1, boneName=Ik_hand_root}, {blendDepth=-4, boneName=Spine_01}]}]
   }
   Save cached pose 'FullBody'(Pose=@n11) {cacheName=FullBody}
   AnimGraphNode_StateMachine_6 = FullBody() {is=title AnimGraphNode_StateMachine, bAllowConduitEntryStates=true}
   n12 = PlayerLocomotionLayerInterface - LocomotionLayer_FullBody!(InPose_FullBody:Pose Link Structure=Blend Poses by bool(
     BlendPose_0=Linked Anim Graph!() -> (Pose:Pose Link Structure) {is=title AnimGraphNode_LinkedAnimGraph, tag=FullBodyState}
     BlendPose_1=@AnimGraphNode_StateMachine_6
     BlendTime_0=0.25
     BlendTime_1=0.25
     bActiveValue=Get bUseFullBodyStateLinkedGraph() -> (bUseFullBodyStateLinkedGraph:bool)
    ) {blendType=Linear}) -> (Pose:Pose Link Structure) {
    is=title AnimGraphNode_LinkedAnimLayer
    interface=&PlayerLocomotionLayerInterface_C
    layer=LocomotionLayer_FullBody
   }
   AnimGraphNode_Layer_1 = VehicleAnimLayerInterface - VehicleFullBodyOverride!(InPoseFullBody:Pose Link Structure=@n12) -> (Pose:Pose Link Structure) {
    is=title AnimGraphNode_LinkedAnimLayer
    tag=VehicleFullBodyOverride
    interface=&VehicleAnimLayerInterface_C
    layer=VehicleFullBodyOverride
   }
   n13 = MovementModeLayerInterface - MovementMode_FullBodyOverride!(InputPoseFullBody:Pose Link Structure=@AnimGraphNode_Layer_1) -> (Pose:Pose Link Structure) {
    is=title AnimGraphNode_LinkedAnimLayer
    tag=MovementMode_FullBodyOverride
    interface=&MovementModeLayerInterface_C
    layer=MovementMode_FullBodyOverride
   }
   n14 = ItemAnimLayerInterface - ItemFullBodyOverride!(InPoseFullBody:Pose Link Structure=@n13) -> (Pose:Pose Link Structure) {
    is=title AnimGraphNode_LinkedAnimLayer
    tag=ItemFullBodyOverride
    interface=&ItemAnimLayerInterface_C
    layer=ItemFullBodyOverride
   }
   n15 = PlayerAnimOverrideLayerInterface - FullBodyOverride!(PassThroughFullBodyPose:Pose Link Structure=@n14) -> (Pose:Pose Link Structure) {
    is=title AnimGraphNode_LinkedAnimLayer
    tag=FullBodyOverride
    interface=/Game/Characters/Player/Male/Male_Avg_Base/AnimLayerInterface/PlayerAnimOverrideLayerInterface.PlayerAnimOverrideLayerInterface_C
    layer=FullBodyOverride
   }
   Save cached pose 'FullBodyBase'(Pose=Profiler Stat Event(Source=FullBodySlotLayer!(InFullBodyPose:Pose Link Structure=@n15) -> (Pose:Pose Link Structure) {is=title AnimGraphNode_LinkedAnimLayer, layer=FullBodySlotLayer}) {profilerEventName=MainPlayerAnimBP_FullBody}) {cacheName=FullBodyBase, cachePoseName=FullBodyBase}
   n16 = PlayerLocomotionLayerInterface - LocomotionLayer_BaseLocomotion!(InPose_BaseLocomotion:Pose Link Structure=BaseLocomotionLayer!() -> (Pose:Pose Link Structure) {is=title AnimGraphNode_LinkedAnimLayer, layer=BaseLocomotionLayer}) -> (Pose:Pose Link Structure) {
    is=title AnimGraphNode_LinkedAnimLayer
    interface=&PlayerLocomotionLayerInterface_C
    layer=LocomotionLayer_BaseLocomotion
   }
   n17 = MovementModeLayerInterface - MovementMode_BaseLocomotion!(InPoseBaseLocomotion:Pose Link Structure=@n16) -> (Pose:Pose Link Structure) {
    is=title AnimGraphNode_LinkedAnimLayer
    interface=&MovementModeLayerInterface_C
    layer=MovementMode_BaseLocomotion
   }
   Save cached pose 'MotionMatchingLocoPose'(Pose=Profiler Stat Event(Source=@n17) {profilerEventName=MainPlayerAnimBP_BaseLocomotion}) {cacheName=MotionMatchingLocoPose, cachePoseName=MotionMatchingLocoPose}
   n18 = Use cached pose 'FullBodyBase'()
   Save cached pose 'FullBodyPostAdditives'(Pose=FirstPersonAdditives!(InPose_FullBody:Pose Link Structure=FullBodyAdditivesLayer!(InPoseFullBodyAdditives:Pose Link Structure=@n18) -> (Pose:Pose Link Structure) {is=title AnimGraphNode_LinkedAnimLayer, layer=FullBodyAdditivesLayer}) -> (Pose:Pose Link Structure) {is=title AnimGraphNode_LinkedAnimLayer, layer=FirstPersonAdditives}) {cacheName=FullBodyPostAdditives, cachePoseName=FullBodyPostAdditives}
   n19 = Use cached pose 'MotionMatchingLocoPose'()
   n20 = MovementModeLayerInterface - MovementMode_SourcePose!(InSourcePoseBase:Pose Link Structure=Profiler Stat Event(Source=ProceduralLayeringStep1!(InPoseProceduralLayering:Pose Link Structure=@n19) -> (Pose:Pose Link Structure) {is=title AnimGraphNode_LinkedAnimLayer, layer=ProceduralLayeringStep1}) {profilerEventName=MainPlayerAnimBP_ProceduralLayeirngStep1}) -> (Pose:Pose Link Structure) {
    is=title AnimGraphNode_LinkedAnimLayer
    interface=&MovementModeLayerInterface_C
    layer=MovementMode_SourcePose
   }
   n21 = Get MotionMatchingWithoutProceduralLayeringBlendTime() -> (MotionMatchingWithoutProceduralLayeringBlendTime:float)
   n22 = Blend Poses by bool(
    BlendPose_0=Use cached pose 'MotionMatchingLocoPose'()
    BlendPose_1=@n20
    BlendTime_0=@n21
    BlendTime_1=@n21
    bActiveValue=Get bUseMotionMatchingWithoutProceduralLayering() -> (bUseMotionMatchingWithoutProceduralLayering:bool)
   ) {
    becomeRelevantFunction {bSelfContext=true, memberName=ResetBlendListBase}
    node {blendType=Linear, becomeRelevantFunction {functionName=ResetBlendListBase}}
   }
   Save cached pose 'PostProceduralLayering'(Pose=@n22) {cacheName=PostProceduralLayering}
   n23 = SourcePoseLayer!(InSourcePoseInput:Pose Link Structure=Use cached pose 'PostProceduralLayering'()) -> (Pose:Pose Link Structure) {is=title AnimGraphNode_LinkedAnimLayer, layer=SourcePoseLayer}
   n24 = Use cached pose 'MotionMatchingLocoPose'()
   n25 = Get bUseNoWeaponPoseDebug() -> (bUseNoWeaponPoseDebug:bool)
   Save cached pose 'SourcePose'(Pose=Profiler Stat Event(Source=Blend Poses by bool(BlendPose_0=@n24, BlendPose_1=@n23, bActiveValue=@n25)) {profilerEventName=MainPlayerAnimBP_SourcePose}) {
    cacheName=SourcePose
    node {cachePoseName=SourcePose, updateFunction {functionName=OnUpdateSourcePose}}
    updateFunction {bSelfContext=true, memberName=OnUpdateSourcePose}
   }
  ]}
 FullBody {
  entry=Conduit
  path=AnimGraph.AnimGraphNode_StateMachine_6.FullBody
  ids=id
  nodes [
   AnimStateConduitNode_0 = Conduit()
   AnimStateNode_1 = PushCannonFired() {is=title AnimStateNode}
   AnimStateNode_11 = SkyDive() {is=title AnimStateNode}
   AnimStateNode_2 = PushCannon() {is=title AnimStateNode}
   AnimStateNode_4 = Default() {is=title AnimStateNode}
   AnimStateNode_8 = Stunned() {is=title AnimStateNode}
   AnimStateTransitionNode_0 = AnimStateTransitionNode() {crossfadeDuration=0.4, priorityOrder=2}
   AnimStateTransitionNode_10 = AnimStateTransitionNode() {logicType=TLT_Inertialization, priorityOrder=8}
   AnimStateTransitionNode_11 = AnimStateTransitionNode()
   AnimStateTransitionNode_12 = AnimStateTransitionNode()
   AnimStateTransitionNode_21 = AnimStateTransitionNode() {logicType=TLT_Inertialization, priorityOrder=5}
   AnimStateTransitionNode_25 = AnimStateTransitionNode()
   AnimStateTransitionNode_27 = AnimStateTransitionNode()
   AnimStateTransitionNode_28 = AnimStateTransitionNode() {logicType=TLT_Inertialization, priorityOrder=6}
   AnimStateTransitionNode_5 = AnimStateTransitionNode() {logicType=TLT_Inertialization, priorityOrder=100}
   AnimStateTransitionNode_6 = AnimStateTransitionNode() {logicType=TLT_Inertialization, priorityOrder=7}
   AnimStateTransitionNode_7 = AnimStateTransitionNode()
   AnimStateTransitionNode_8 = AnimStateTransitionNode() {crossfadeDuration=0, logicType=TLT_Inertialization}
  ]
  links [
   {
    dst=Default
    rule=AnimGraph.AnimGraphNode_StateMachine_6.FullBody.AnimStateTransitionNode_0.Transition
    src=Conduit
    via=AnimStateTransitionNode_0
   }
   {
    dst=Default
    rule=AnimGraph.AnimGraphNode_StateMachine_6.FullBody.AnimStateTransitionNode_5.Transition
    src=Conduit
    via=AnimStateTransitionNode_5
   }
   {
    dst=Default
    rule=AnimGraph.AnimGraphNode_StateMachine_6.FullBody.AnimStateTransitionNode_8.Transition
    src=Conduit
    via=AnimStateTransitionNode_8
   }
   {
    dst=PushCannon
    rule=AnimGraph.AnimGraphNode_StateMachine_6.FullBody.AnimStateTransitionNode_6.Transition
    src=Conduit
    via=AnimStateTransitionNode_6
   }
   {
    dst=PushCannonFired
    rule=AnimGraph.AnimGraphNode_StateMachine_6.FullBody.AnimStateTransitionNode_10.Transition
    src=Conduit
    via=AnimStateTransitionNode_10
   }
   {
    dst=SkyDive
    rule=AnimGraph.AnimGraphNode_StateMachine_6.FullBody.AnimStateTransitionNode_28.Transition
    src=Conduit
    via=AnimStateTransitionNode_28
   }
   {
    dst=Stunned
    rule=AnimGraph.AnimGraphNode_StateMachine_6.FullBody.AnimStateTransitionNode_21.Transition
    src=Conduit
    via=AnimStateTransitionNode_21
   }
   {
    dst=Conduit
    rule=AnimGraph.AnimGraphNode_StateMachine_6.FullBody.AnimStateTransitionNode_7.Transition
    src=Default
    via=AnimStateTransitionNode_7
   }
   {
    dst=Conduit
    rule=AnimGraph.AnimGraphNode_StateMachine_6.FullBody.AnimStateTransitionNode_12.Transition
    src=PushCannon
    via=AnimStateTransitionNode_12
   }
   {
    dst=Conduit
    rule=AnimGraph.AnimGraphNode_StateMachine_6.FullBody.AnimStateTransitionNode_11.Transition
    src=PushCannonFired
    via=AnimStateTransitionNode_11
   }
   {
    dst=Conduit
    rule=AnimGraph.AnimGraphNode_StateMachine_6.FullBody.AnimStateTransitionNode_27.Transition
    src=SkyDive
    via=AnimStateTransitionNode_27
   }
   {
    dst=Conduit
    rule=AnimGraph.AnimGraphNode_StateMachine_6.FullBody.AnimStateTransitionNode_25.Transition
    src=Stunned
    via=AnimStateTransitionNode_25
   }
  ]
 }
 Conduit {path=AnimGraph.AnimGraphNode_StateMachine_6.FullBody.AnimStateConduitNode_0.Conduit, nodes [Result(bCanEnterTransition=true)]}
 PushCannonFired {
  path=AnimGraph.AnimGraphNode_StateMachine_6.FullBody.AnimStateNode_1.PushCannonFired
  nodes [
   n1 = "Blend Poses (EFortCustomGender)"!(
    BlendPose_0:Pose Link Structure=Sequence Player 'PushCannon_Passenger_Idle_Fired_CMM'() {sequence=/Game/Animation/Game/MainPlayer/Vehicles/PushCannon/PushCannon_Passenger_Idle_Fired_CMM}
    BlendPose_1:Pose Link Structure=Sequence Player 'PushCannon_Passenger_Idle_Fired_CMF'() {sequence=/Game/Animation/Game/MainPlayer/Vehicles/PushCannon/PushCannon_Passenger_Idle_Fired_CMF}
    BlendTime_0:float=0.1
    BlendTime_1:float=0.1
    ActiveEnumValue:Byte=Get Gender() -> (Gender:EFortCustomGender Enum=Invalid)
   ) -> (Pose:Pose Link Structure) {
    is=title AnimGraphNode_BlendListByEnum
    blendType=Linear
    blendPose [{linkId=-1, sourceLinkId=-1}, {linkId=-1, sourceLinkId=-1}]
    blendTime [0.1,0.1]
   }
   Output Animation Pose(Result=@n1)
  ]
 }
 SkyDive {
  path=AnimGraph.AnimGraphNode_StateMachine_6.FullBody.AnimStateNode_11.SkyDive
  nodes [
   n1 = Player_SkyDiveAndGlide_AnimBP!(ParaGlide_TriggerAnim:bool=false) -> (Pose:Pose Link Structure) {
    is=title AnimGraphNode_LinkedAnimGraph
    tag=SkydiveAndGlide
    instanceClass=/Game/Characters/Player/AnimationBlueprints/LinkedAnimGraphs/Player_SkyDiveAndGlide_AnimBP.Player_SkyDiveAndGlide_AnimBP_C
   }
   Output Animation Pose(Result=Linked Graph Gate(Source=@n1) {linkedGraphTag=SkydiveAndGlide})
  ]
 }
 PushCannon {
  path=AnimGraph.AnimGraphNode_StateMachine_6.FullBody.AnimStateNode_2.PushCannon
  nodes [Output Animation Pose(Result=Sequence Player 'PushCannon_Passenger_Idle_Pose'() {sequence=/Game/Animation/Game/MainPlayer/Vehicles/PushCannon/PushCannon_Passenger_Idle_Pose})]
 }
 Default {
  path=AnimGraph.AnimGraphNode_StateMachine_6.FullBody.AnimStateNode_4.Default
  nodes [
   n1 = PlayerLocomotionLayerInterface - LocomotionLayer_Additives!() -> (Pose:Pose Link Structure) {
    is=title AnimGraphNode_LinkedAnimLayer
    interface=&PlayerLocomotionLayerInterface_C
    layer=LocomotionLayer_Additives
   }
   n2 = MovementModeLayerInterface - MovementMode_LocomotionAdditives!(InLocomotionAdditivesPose:Pose Link Structure=@n1) -> (Pose:Pose Link Structure) {
    is=title AnimGraphNode_LinkedAnimLayer
    interface=&MovementModeLayerInterface_C
    layer=MovementMode_LocomotionAdditives
   }
   AnimGraphNode_StateMachine_1 = FullBodyWithProceduralLayering() {is=title AnimGraphNode_StateMachine, bAllowConduitEntryStates=true}
   n3 = PlayerLocomotionLayerInterface - LocomotionLayer_FullLocomotion!(InPose_FullLocomotion:Pose Link Structure=ProceduralLayeringStep2!(
     InSourceLocomotionPose:Pose Link Structure=Profiler Stat Event(Source=@AnimGraphNode_StateMachine_1) {profilerEventName=MainPlayerAnimBP_FullBodyWithProcLayering}
     In_MotionMatchingPose:Pose Link Structure=Use cached pose 'MotionMatchingLocoPose'()
     In_SourcePose:Pose Link Structure=Use cached pose 'SourcePose'()
    ) -> (Pose:Pose Link Structure) {is=title AnimGraphNode_LinkedAnimLayer, layer=ProceduralLayeringStep2}) -> (Pose:Pose Link Structure) {
    is=title AnimGraphNode_LinkedAnimLayer
    interface=&PlayerLocomotionLayerInterface_C
    layer=LocomotionLayer_FullLocomotion
   }
   Output Animation Pose(Result=Apply Additive(Base=@n3, Additive=@n2, Alpha=Get LocomotionAdditiveAlpha() -> (LocomotionAdditiveAlpha:float)) {alphaBoolBlend {blendOption=Linear}})
  ]
 }
 FullBodyWithProceduralLayering {
  entry=Conduit
  path=AnimGraph.AnimGraphNode_StateMachine_6.FullBody.AnimStateNode_4.Default.AnimGraphNode_StateMachine_1.FullBodyWithProceduralLayering
  ids=id
  nodes [
   AnimStateConduitNode_0 = Conduit()
   AnimStateNode_0 = Default()
   AnimStateNode_3 = TacticalState() {is=title AnimStateNode, bAlwaysResetOnEntry=true}
   AnimStateTransitionNode_0 = AnimStateTransitionNode() {logicType=TLT_Inertialization, priorityOrder=100}
   AnimStateTransitionNode_1 = AnimStateTransitionNode() {crossfadeDuration=0.4, priorityOrder=3}
   AnimStateTransitionNode_2 = AnimStateTransitionNode() {crossfadeDuration=0.4, logicType=TLT_Inertialization, priorityOrder=2}
   AnimStateTransitionNode_3 = AnimStateTransitionNode() {crossfadeDuration=0, logicType=TLT_Inertialization}
   AnimStateTransitionNode_4 = AnimStateTransitionNode()
   AnimStateTransitionNode_5 = AnimStateTransitionNode()
  ]
  links [
   {
    dst=Default
    rule=AnimGraph.AnimGraphNode_StateMachine_6.FullBody.AnimStateNode_4.Default.AnimGraphNode_StateMachine_1.FullBodyWithProceduralLayering.AnimStateTransitionNode_0.Transition
    src=Conduit
    via=AnimStateTransitionNode_0
   }
   {
    dst=Default
    rule=AnimGraph.AnimGraphNode_StateMachine_6.FullBody.AnimStateNode_4.Default.AnimGraphNode_StateMachine_1.FullBodyWithProceduralLayering.AnimStateTransitionNode_3.Transition
    src=Conduit
    via=AnimStateTransitionNode_3
   }
   {
    dst=TacticalState
    rule=AnimGraph.AnimGraphNode_StateMachine_6.FullBody.AnimStateNode_4.Default.AnimGraphNode_StateMachine_1.FullBodyWithProceduralLayering.AnimStateTransitionNode_1.Transition
    src=Conduit
    via=AnimStateTransitionNode_1
   }
   {
    dst=TacticalState
    rule=AnimGraph.AnimGraphNode_StateMachine_6.FullBody.AnimStateNode_4.Default.AnimGraphNode_StateMachine_1.FullBodyWithProceduralLayering.AnimStateTransitionNode_2.Transition
    src=Conduit
    via=AnimStateTransitionNode_2
   }
   {
    dst=Conduit
    rule=AnimGraph.AnimGraphNode_StateMachine_6.FullBody.AnimStateNode_4.Default.AnimGraphNode_StateMachine_1.FullBodyWithProceduralLayering.AnimStateTransitionNode_4.Transition
    src=Default
    via=AnimStateTransitionNode_4
   }
   {
    dst=Conduit
    rule=AnimGraph.AnimGraphNode_StateMachine_6.FullBody.AnimStateNode_4.Default.AnimGraphNode_StateMachine_1.FullBodyWithProceduralLayering.AnimStateTransitionNode_5.Transition
    src=TacticalState
    via=AnimStateTransitionNode_5
   }
  ]
 }
 Conduit {
  path=AnimGraph.AnimGraphNode_StateMachine_6.FullBody.AnimStateNode_4.Default.AnimGraphNode_StateMachine_1.FullBodyWithProceduralLayering.AnimStateConduitNode_0.Conduit
  nodes [Result(bCanEnterTransition=true)]
 }
 Default {
  path=AnimGraph.AnimGraphNode_StateMachine_6.FullBody.AnimStateNode_4.Default.AnimGraphNode_StateMachine_1.FullBodyWithProceduralLayering.AnimStateNode_0.Default
  nodes [
   AnimGraphNode_StateMachine_0 = BaseLocomotion() {is=title AnimGraphNode_StateMachine, bAllowConduitEntryStates=true}
   Output Animation Pose(Result=FullBodyAimOffsets!(InPose_FullBodyAimOffsets:Pose Link Structure=@AnimGraphNode_StateMachine_0) -> (Pose:Pose Link Structure) {is=title AnimGraphNode_LinkedAnimLayer, layer=FullBodyAimOffsets})
  ]
 }
 BaseLocomotion {
  entry=Conduit
  path=AnimGraph.AnimGraphNode_StateMachine_6.FullBody.AnimStateNode_4.Default.AnimGraphNode_StateMachine_1.FullBodyWithProceduralLayering.AnimStateNode_0.Default.AnimGraphNode_StateMachine_0.BaseLocomotion
  ids=id
  nodes [
   AnimStateConduitNode_0 = Conduit()
   AnimStateNode_0 = InAir() {is=title AnimStateNode}
   AnimStateNode_1 = Default()
   AnimStateTransitionNode_0 = AnimStateTransitionNode()
   AnimStateTransitionNode_1 = AnimStateTransitionNode() {logicType=TLT_Inertialization, priorityOrder=100}
   AnimStateTransitionNode_2 = AnimStateTransitionNode()
   AnimStateTransitionNode_3 = AnimStateTransitionNode()
  ]
  links [
   {
    dst=Default
    rule=AnimGraph.AnimGraphNode_StateMachine_6.FullBody.AnimStateNode_4.Default.AnimGraphNode_StateMachine_1.FullBodyWithProceduralLayering.AnimStateNode_0.Default.AnimGraphNode_StateMachine_0.BaseLocomotion.AnimStateTransitionNode_1.AnimationTransitionGraph_0
    src=Conduit
    via=AnimStateTransitionNode_1
   }
   {
    dst=InAir
    rule=AnimGraph.AnimGraphNode_StateMachine_6.FullBody.AnimStateNode_4.Default.AnimGraphNode_StateMachine_1.FullBodyWithProceduralLayering.AnimStateNode_0.Default.AnimGraphNode_StateMachine_0.BaseLocomotion.AnimStateTransitionNode_0.AnimationTransitionGraph_0
    src=Conduit
    via=AnimStateTransitionNode_0
   }
   {
    dst=Conduit
    rule=AnimGraph.AnimGraphNode_StateMachine_6.FullBody.AnimStateNode_4.Default.AnimGraphNode_StateMachine_1.FullBodyWithProceduralLayering.AnimStateNode_0.Default.AnimGraphNode_StateMachine_0.BaseLocomotion.AnimStateTransitionNode_3.AnimationTransitionGraph_0
    src=Default
    via=AnimStateTransitionNode_3
   }
   {
    dst=Conduit
    rule=AnimGraph.AnimGraphNode_StateMachine_6.FullBody.AnimStateNode_4.Default.AnimGraphNode_StateMachine_1.FullBodyWithProceduralLayering.AnimStateNode_0.Default.AnimGraphNode_StateMachine_0.BaseLocomotion.AnimStateTransitionNode_2.AnimationTransitionGraph_1
    src=InAir
    via=AnimStateTransitionNode_2
   }
  ]
 }
 Conduit {
  path=AnimGraph.AnimGraphNode_StateMachine_6.FullBody.AnimStateNode_4.Default.AnimGraphNode_StateMachine_1.FullBodyWithProceduralLayering.AnimStateNode_0.Default.AnimGraphNode_StateMachine_0.BaseLocomotion.AnimStateConduitNode_0.Conduit
  nodes [Result(bCanEnterTransition=true)]
 }
 InAir {
  path=AnimGraph.AnimGraphNode_StateMachine_6.FullBody.AnimStateNode_4.Default.AnimGraphNode_StateMachine_1.FullBodyWithProceduralLayering.AnimStateNode_0.Default.AnimGraphNode_StateMachine_0.BaseLocomotion.AnimStateNode_0.InAir
  nodes [
   n1 = Layered blend per bone(BasePose=Use cached pose 'SourcePose'(), BlendPoses_0=Use cached pose 'PostProceduralLayering'()) {
    bMeshSpaceRotationBlend=true
    bMeshSpaceScaleBlend=true
    curveBlendOption=UseMaxValue
    layerSetup [{branchFilters [
       {boneName=Root}
       {blendDepth=-1, boneName=Ik_hand_root}
       {blendDepth=-4, boneName=spine_01}
       {blendDepth=-1, boneName=VB pelvis_ik_hand_l}
       {blendDepth=-1, boneName=root_fp}
      ]}]
   }
   n2 = Player_OldJump_AnimBP_Template!(InPose:Pose Link Structure=Use cached pose 'SourcePose'()) -> (Pose:Pose Link Structure) {
    is=title AnimGraphNode_LinkedAnimGraph
    tag=JumpNoProcedural
    instanceClass=/Game/Characters/Player/AnimationBlueprints/LinkedAnimGraphTemplates/Player_OldJump_AnimBP_Template.Player_OldJump_AnimBP_Template_C
   }
   n3 = PlayerLocomotionLayerInterface - LocomotionLayer_InAir!(
    InPose_InAir:Pose Link Structure=Blend Poses by bool(BlendPose_0=@n2, BlendPose_1=@n1, bActiveValue=Get bUseOldJump() -> (bUseOldJump:bool)) {blendType=Linear}
    InPose_InAirSourcePose:Pose Link Structure=Use cached pose 'SourcePose'()
   ) -> (Pose:Pose Link Structure) {
    is=title AnimGraphNode_LinkedAnimLayer
    interface=&PlayerLocomotionLayerInterface_C
    layer=LocomotionLayer_InAir
   }
   n4 = MovementModeLayerInterface - MovementMode_InAirOverride!(InSource_Pose:Pose Link Structure=Use cached pose 'SourcePose'(), InAirPose:Pose Link Structure=@n3) -> (Pose:Pose Link Structure) {
    is=title AnimGraphNode_LinkedAnimLayer
    interface=&MovementModeLayerInterface_C
    layer=MovementMode_InAirOverride
   }
   Output Animation Pose(Result=@n4)
  ]
 }
 Default {
  path=AnimGraph.AnimGraphNode_StateMachine_6.FullBody.AnimStateNode_4.Default.AnimGraphNode_StateMachine_1.FullBodyWithProceduralLayering.AnimStateNode_0.Default.AnimGraphNode_StateMachine_0.BaseLocomotion.AnimStateNode_1.Default
  nodes [
   n1 = Use cached pose 'PostProceduralLayering'()
   n2 = Use cached pose 'SourcePose'()
   Output Animation Pose(Result=Blend Poses by bool(
     BlendPose_0=Use cached pose 'PostProceduralLayering'()
     BlendPose_1=LocomotionLayer!(InSourcePoseLoco:Pose Link Structure=@n2, MotionMatchingLocoPose:Pose Link Structure=@n1) -> (Pose:Pose Link Structure) {is=title AnimGraphNode_LinkedAnimLayer, layer=LocomotionLayer}
     bActiveValue=Get bEnableGraphLOD() -> (bEnableGraphLOD:bool)
    ))
  ]
 }
 AnimationTransitionGraph_0 {
  path=AnimGraph.AnimGraphNode_StateMachine_6.FullBody.AnimStateNode_4.Default.AnimGraphNode_StateMachine_1.FullBodyWithProceduralLayering.AnimStateNode_0.Default.AnimGraphNode_StateMachine_0.BaseLocomotion.AnimStateTransitionNode_0.AnimationTransitionGraph_0
  nodes [Result(bCanEnterTransition=Get bIsInAir() -> (bIsInAir:bool))]
 }
 AnimationTransitionGraph_0 {
  path=AnimGraph.AnimGraphNode_StateMachine_6.FullBody.AnimStateNode_4.Default.AnimGraphNode_StateMachine_1.FullBodyWithProceduralLayering.AnimStateNode_0.Default.AnimGraphNode_StateMachine_0.BaseLocomotion.AnimStateTransitionNode_1.AnimationTransitionGraph_0
  nodes [Result(bCanEnterTransition=true)]
 }
 AnimationTransitionGraph_1 {
  path=AnimGraph.AnimGraphNode_StateMachine_6.FullBody.AnimStateNode_4.Default.AnimGraphNode_StateMachine_1.FullBodyWithProceduralLayering.AnimStateNode_0.Default.AnimGraphNode_StateMachine_0.BaseLocomotion.AnimStateTransitionNode_2.AnimationTransitionGraph_1
  nodes [Result(bCanEnterTransition=NOT Boolean(A=Get bIsInAir()))]
 }
 AnimationTransitionGraph_0 {
  path=AnimGraph.AnimGraphNode_StateMachine_6.FullBody.AnimStateNode_4.Default.AnimGraphNode_StateMachine_1.FullBodyWithProceduralLayering.AnimStateNode_0.Default.AnimGraphNode_StateMachine_0.BaseLocomotion.AnimStateTransitionNode_3.AnimationTransitionGraph_0
  nodes [Result(bCanEnterTransition=true)]
 }
 TacticalState {
  path=AnimGraph.AnimGraphNode_StateMachine_6.FullBody.AnimStateNode_4.Default.AnimGraphNode_StateMachine_1.FullBodyWithProceduralLayering.AnimStateNode_3.TacticalState
  nodes [
   n1 = Player_TacticalState_AnimBP!(InPoseTacticalState:Pose Link Structure=Blend Poses by bool(
     BlendPose_0=Use cached pose 'PostProceduralLayering'()
     BlendPose_1=Use cached pose 'MotionMatchingLocoPose'()
     bActiveValue=Get bEnableProceduralLayeredTacticalSprint() -> (bEnableProceduralLayeredTacticalSprint:bool)
    ) {blendType=Linear}) -> (Pose:Pose Link Structure) {
    is=title AnimGraphNode_LinkedAnimGraph
    tag=TacticalState
    instanceClass=/Game/Characters/Player/AnimationBlueprints/LinkedAnimGraphs/Player_TacticalState_AnimBP.Player_TacticalState_AnimBP_C
   }
   Output Animation Pose(Result=Linked Graph Gate(Source=@n1) {linkedGraphTag=TacticalState})
  ]
 }
 Transition {
  path=AnimGraph.AnimGraphNode_StateMachine_6.FullBody.AnimStateNode_4.Default.AnimGraphNode_StateMachine_1.FullBodyWithProceduralLayering.AnimStateTransitionNode_0.Transition
  nodes [Result(bCanEnterTransition=true)]
 }
 Transition {
  path=AnimGraph.AnimGraphNode_StateMachine_6.FullBody.AnimStateNode_4.Default.AnimGraphNode_StateMachine_1.FullBodyWithProceduralLayering.AnimStateTransitionNode_1.Transition
  nodes [Result(bCanEnterTransition=Get bTransition_FullBody_TacticalSprint() -> (bTransition_FullBody_TacticalSprint:bool))]
 }
 Transition {
  path=AnimGraph.AnimGraphNode_StateMachine_6.FullBody.AnimStateNode_4.Default.AnimGraphNode_StateMachine_1.FullBodyWithProceduralLayering.AnimStateTransitionNode_2.Transition
  nodes [Result(bCanEnterTransition=Get bTransition_FullBody_TacticalSprint_Inertialized() -> (bTransition_FullBody_TacticalSprint_Inertialized:bool))]
 }
 Transition {
  path=AnimGraph.AnimGraphNode_StateMachine_6.FullBody.AnimStateNode_4.Default.AnimGraphNode_StateMachine_1.FullBodyWithProceduralLayering.AnimStateTransitionNode_3.Transition
  nodes [Result(bCanEnterTransition=Get bRecentlyFired() -> (bRecentlyFired:bool))]
 }
 Transition {
  path=AnimGraph.AnimGraphNode_StateMachine_6.FullBody.AnimStateNode_4.Default.AnimGraphNode_StateMachine_1.FullBodyWithProceduralLayering.AnimStateTransitionNode_4.Transition
  nodes [Result(bCanEnterTransition=true)]
 }
 Transition {
  path=AnimGraph.AnimGraphNode_StateMachine_6.FullBody.AnimStateNode_4.Default.AnimGraphNode_StateMachine_1.FullBodyWithProceduralLayering.AnimStateTransitionNode_5.Transition
  nodes [Result(bCanEnterTransition=true)]
 }
 Stunned {
  path=AnimGraph.AnimGraphNode_StateMachine_6.FullBody.AnimStateNode_8.Stunned
  nodes [Output Animation Pose(Result=Sequence Player 'Standing_Stun'() {sequence=/Game/Animation/Game/MainPlayer/Locomotion/Small/idle/Standing_Stun})]
 }
 Transition {
  path=AnimGraph.AnimGraphNode_StateMachine_6.FullBody.AnimStateTransitionNode_0.Transition
  nodes [Result(bCanEnterTransition=Get bTransition_FullBody_TacticalSprint())]
 }
 Transition {
  path=AnimGraph.AnimGraphNode_StateMachine_6.FullBody.AnimStateTransitionNode_10.Transition
  nodes [Result(bCanEnterTransition=Get bIsFiredFromCannon() -> (bIsFiredFromCannon:bool))]
 }
 Transition {path=AnimGraph.AnimGraphNode_StateMachine_6.FullBody.AnimStateTransitionNode_11.Transition, nodes [Result(bCanEnterTransition=true)]}
 Transition {path=AnimGraph.AnimGraphNode_StateMachine_6.FullBody.AnimStateTransitionNode_12.Transition, nodes [Result(bCanEnterTransition=true)]}
 Transition {
  path=AnimGraph.AnimGraphNode_StateMachine_6.FullBody.AnimStateTransitionNode_21.Transition
  nodes [Result(bCanEnterTransition=Get bTransition_FullBody_Stunned() -> (bTransition_FullBody_Stunned:bool))]
 }
 Transition {path=AnimGraph.AnimGraphNode_StateMachine_6.FullBody.AnimStateTransitionNode_25.Transition, nodes [Result(bCanEnterTransition=true)]}
 Transition {path=AnimGraph.AnimGraphNode_StateMachine_6.FullBody.AnimStateTransitionNode_27.Transition, nodes [Result(bCanEnterTransition=true)]}
 Transition {
  path=AnimGraph.AnimGraphNode_StateMachine_6.FullBody.AnimStateTransitionNode_28.Transition
  nodes [Result(bCanEnterTransition=Break Fort Anim Input Skydiving!(FortAnimInput_Skydiving:"Fort Anim Input Skydiving Structure (by ref)"=Get Skydiving() -> (Skydiving:Fort Anim Input Skydiving Structure)) -> (
     ParaGilderRootOffset:Vector=~
     ParaGliderRootRotationOverride:Rotator=~
     bUseParaGlideRootModifier:bool=~
     bIsSkydivingFromLaunchPad:bool=~
     bIsSkydivingFromBus:bool=~
     bIsInVortex:bool=~
     bIsDiving:bool=~
     bIsParachuteOpen:bool=~
     bIsSkydiving:bool=~
     bIsParachuteLeaning:bool=~
     bIsSkydiveLeaning:bool=~
     bIsLeaning:bool=~
     bIsSkydiveDiveMode:bool=~
     bPlaySkydiveDrift:bool=~
     bSkydiveDriftDelayActive:bool=~
     bSkydiveDriftAnimAllowed:bool=~
     bIsGliderRight:bool=~
     bIsGliderCenter:bool=~
     bIsGliderLeft:bool=~
     bIsGliderForward:bool=~
     bIsGliderBack:bool=~
     bHasDeployedAndIsStandingUpright:bool=~
     LocalAccelForward:float=~
     LocalAccelRight:float=~
     LocalVelocityRight:float=~
     SkydiveAimPitch:float=~
     SkydiveAimYaw:float=~
     DeployChuteAnimRate:float=~
     SkydiveDriftAnimRate:float=~
     SkydiveDriftAnimRateCurrent:float=~
     SkydiveFidgetAnimRate:float=~
     SkydiveFidgetAnimRateCurrent:float=~
     SkydiveAdditiveAlpha:float=~
     SkydiveDriftDelay:float=~
     SkydiveDriftAnim:int
     SkydiveDriftAnimMax:int
     LaunchpadAnim:int
     LocalAccelDir:ESkydivingDirection Enum
     DirectionLast:ESkydivingDirection Enum
     PoseAdditive:Anim Sequence Object Reference
    ) {is=title K2Node_BreakStruct}.bIsSkydiving)]
 }
 Transition {path=AnimGraph.AnimGraphNode_StateMachine_6.FullBody.AnimStateTransitionNode_5.Transition, nodes [Result(bCanEnterTransition=true)]}
 Transition {
  path=AnimGraph.AnimGraphNode_StateMachine_6.FullBody.AnimStateTransitionNode_6.Transition
  nodes [Result(bCanEnterTransition=Get bIsInCannon() -> (bIsInCannon:bool))]
 }
 Transition {
  path=AnimGraph.AnimGraphNode_StateMachine_6.FullBody.AnimStateTransitionNode_7.Transition
  nodes [Result(bCanEnterTransition=NOT Boolean(A=Get bIsInDefaultFullBodyState() -> (bIsInDefaultFullBodyState:bool)))]
 }
 Transition {
  path=AnimGraph.AnimGraphNode_StateMachine_6.FullBody.AnimStateTransitionNode_8.Transition
  nodes [Result(bCanEnterTransition=Get bRecentlyFired())]
 }
 BaseLocomotionLayer {nodes [
   AnimGraphNode_StateMachine_1 = MotionMatchingLocomotion() {is=title AnimGraphNode_StateMachine, bAllowConduitEntryStates=true}
   n1 = Layered blend per bone(
    BasePose=Profiler Stat Event(Source=@AnimGraphNode_StateMachine_1) {profilerEventName=MainPlayerAnimBP_MotionMatching}
    BlendPoses_0=Cached Sequence Evaluator!(Sequence:Anim Sequence Base Object Reference) -> (Pose:Pose Link Structure) {is=title FortAnimGraphNode_CachedSequenceEvaluator, cacheName=SourcePose}
    BlendWeights_0=Get ApplySourcePoseLeftHandMaskAlpha() -> (ApplySourcePoseLeftHandMaskAlpha:float)
   ) {blendMode=BlendMask, blendMasks [&Fortnite_M_Avg_Player_Skeleton:LeftHandChildren], layerSetup []}
   n2 = Layered blend per bone(
    BasePose=@n1
    BlendPoses_0=Cached Sequence Evaluator() {cacheName=SourcePose}
    BlendWeights_0=Get ApplySourcePoseRightHandMaskAlpha() -> (ApplySourcePoseRightHandMaskAlpha:float)
   ) {blendMode=BlendMask, blendMasks [&Fortnite_M_Avg_Player_Skeleton:RightHandChildren], layerSetup []}
   Output Pose(Result=@n2)
   Save cached pose 'MotionMatchingMainPose'(Pose=Linked Graph Gate(Source=Linked Anim Graph() {tag=MotionMatching}) {linkedGraphTag=SourcePoseDefault}) {cacheName=MotionMatchingMainPose}
  ]}
 MotionMatchingLocomotion {
  entry=Conduit
  path=BaseLocomotionLayer.AnimGraphNode_StateMachine_1.MotionMatchingLocomotion
  ids=id
  nodes [
   AnimStateConduitNode_0 = Conduit()
   AnimStateNode_0 = InAir()
   AnimStateNode_1 = Default()
   AnimStateTransitionNode_0 = AnimStateTransitionNode()
   AnimStateTransitionNode_1 = AnimStateTransitionNode() {priorityOrder=100}
   AnimStateTransitionNode_2 = AnimStateTransitionNode()
   AnimStateTransitionNode_3 = AnimStateTransitionNode()
  ]
  links [
   {
    dst=Default
    rule=BaseLocomotionLayer.AnimGraphNode_StateMachine_1.MotionMatchingLocomotion.AnimStateTransitionNode_1.AnimationTransitionGraph_0
    src=Conduit
    via=AnimStateTransitionNode_1
   }
   {
    dst=InAir
    rule=BaseLocomotionLayer.AnimGraphNode_StateMachine_1.MotionMatchingLocomotion.AnimStateTransitionNode_0.AnimationTransitionGraph_0
    src=Conduit
    via=AnimStateTransitionNode_0
   }
   {
    dst=Conduit
    rule=BaseLocomotionLayer.AnimGraphNode_StateMachine_1.MotionMatchingLocomotion.AnimStateTransitionNode_3.AnimationTransitionGraph_0
    src=Default
    via=AnimStateTransitionNode_3
   }
   {
    dst=Conduit
    rule=BaseLocomotionLayer.AnimGraphNode_StateMachine_1.MotionMatchingLocomotion.AnimStateTransitionNode_2.AnimationTransitionGraph_1
    src=InAir
    via=AnimStateTransitionNode_2
   }
  ]
 }
 Conduit {
  path=BaseLocomotionLayer.AnimGraphNode_StateMachine_1.MotionMatchingLocomotion.AnimStateConduitNode_0.Conduit
  nodes [Result(bCanEnterTransition=true)]
 }
 InAir {
  path=BaseLocomotionLayer.AnimGraphNode_StateMachine_1.MotionMatchingLocomotion.AnimStateNode_0.InAir
  nodes [
   n1 = Player_InAir_AnimBP!(SourcePose:Pose Link Structure=Use cached pose 'MotionMatchingMainPose'()) -> (Pose:Pose Link Structure) {
    is=title AnimGraphNode_LinkedAnimGraph
    tag=Jump
    instanceClass=/Game/Characters/Player/AnimationBlueprints/LinkedAnimGraphs/Player_InAir_AnimBP.Player_InAir_AnimBP_C
   }
   Output Animation Pose(Result=Blend Poses by bool(
     BlendPose_0=Use cached pose 'MotionMatchingMainPose'()
     BlendPose_1=Linked Graph Gate(Source=@n1) {linkedGraphTag=Jump}
     bActiveValue=Get bUseOldJump()
    ) {blendType=Linear})
  ]
 }
 Default {
  path=BaseLocomotionLayer.AnimGraphNode_StateMachine_1.MotionMatchingLocomotion.AnimStateNode_1.Default
  nodes [Output Animation Pose(Result=Use cached pose 'MotionMatchingMainPose'())]
 }
 AnimationTransitionGraph_0 {
  path=BaseLocomotionLayer.AnimGraphNode_StateMachine_1.MotionMatchingLocomotion.AnimStateTransitionNode_0.AnimationTransitionGraph_0
  nodes [Result(bCanEnterTransition=Get bIsInAir())]
 }
 AnimationTransitionGraph_0 {
  path=BaseLocomotionLayer.AnimGraphNode_StateMachine_1.MotionMatchingLocomotion.AnimStateTransitionNode_1.AnimationTransitionGraph_0
  nodes [Result(bCanEnterTransition=true)]
 }
 AnimationTransitionGraph_1 {
  path=BaseLocomotionLayer.AnimGraphNode_StateMachine_1.MotionMatchingLocomotion.AnimStateTransitionNode_2.AnimationTransitionGraph_1
  nodes [Result(bCanEnterTransition=NOT Boolean(A=Get bIsInAir()))]
 }
 AnimationTransitionGraph_0 {
  path=BaseLocomotionLayer.AnimGraphNode_StateMachine_1.MotionMatchingLocomotion.AnimStateTransitionNode_3.AnimationTransitionGraph_0
  nodes [Result(bCanEnterTransition=true)]
 }
 CharacterSkeletalControlPostLegIK {nodes [Output Pose(Result=Input Pose() {name=InPosePostLegIK}) {layerGroup=PlayerOverride}]}
 EventGraph {nodes [
   AnimNotify_Melee_Swing_L_End() {is=title K2Node_Event}
   > n1 = Cast To FortPawn!(execute, Object=TryGetPawnOwner()) -> (then, CastFailed, AsFort Pawn:Fort Pawn Object Reference) {is=title K2Node_DynamicCast}
   > MeleeSwingLeft_End!(execute:Exec, self:Object Reference=@n1) -> (then:Exec) {is=title K2Node_Message}
   AnimNotify_IndigoLaunch() {is=title K2Node_Event}
   > Call Indigo Boost!(execute:Exec, self:Fortnite M Avg Player Anim Blueprint Object Reference) -> (then:Exec) {is=title K2Node_CallDelegate}
   AnimNotify_Melee_Swing_L() {is=title K2Node_Event}
   > n2 = Cast To FortPawn(Object=TryGetPawnOwner())
   > MeleeSwingLeft!(execute:Exec, self:Object Reference=@n2, First Left:bool=true) -> (then:Exec) {is=title K2Node_Message}
   AnimNotify_Melee_Swing_R() {is=title K2Node_Event}
   > n3 = Cast To FortPawn(Object=TryGetPawnOwner())
   > MeleeSwingRight!(execute:Exec, self:Object Reference=@n3, First Right:bool=true) -> (then:Exec) {is=title K2Node_Message}
   n4 = AnimNotify_CreateToolFlyModeLoop_Entered() {is=title K2Node_Event}
   Set bIsWaterJump(execute=[@n4.then, @n6.then], bIsWaterJump:bool=false) -> (Output_Get:bool)
   AnimNotify_Melee_Swing_R_End() {is=title K2Node_Event}
   > n5 = Cast To FortPawn(Object=TryGetPawnOwner())
   > MeleeSwingRight_End!(execute:Exec, self:Object Reference=@n5) -> (then:Exec) {is=title K2Node_Message}
   n6 = AnimNotify_CreateToolFlyModeStart_Entered() {is=title K2Node_Event}
   AnimNotify_playFacialAnim() {is=title K2Node_Event}
   > n7 = Branch(Condition=Cast To FortGameStateZone!(Object=GetGameState!() -> (ReturnValue:Game State Base Object Reference)) -> (AsFort Game State Zone:Fort Game State Zone Object Reference, bSuccess:bool=~) {is=title K2Node_DynamicCast}.bSuccess)
   Call Menu Screen Dispatcher!(execute:Exec=@n7.else, self:Fortnite M Avg Player Anim Blueprint Object Reference, CharPart:Fort Player Anim Instance Object Reference) -> (then:Exec) {is=title K2Node_CallDelegate}
   AnimNotify_Melee_Swing_L_2() {is=title K2Node_Event}
   > n8 = Cast To FortPawn(Object=TryGetPawnOwner())
   > MeleeSwingLeft(self=@n8, First Left=false)
   AnimNotify_Melee_Swing_R_2() {is=title K2Node_Event}
   > n9 = Cast To FortPawn(Object=TryGetPawnOwner())
   > MeleeSwingRight(self=@n9, First Right=false)
   AnimNotify_OpenGlider() {is=title K2Node_Event}
   > OnAnimNotifyOpenGlider(self:Fort Player Animation Interface Object Reference=GetAnimationInterface!(self:Fort Player Pawn Object Reference=Get FortPlayerPawn() -> (FortPlayerPawn:Fort Player Pawn Object Reference)) -> (ReturnValue:Fort Player Animation Interface Object Reference))
   > Set bParaGlide_OpenGlider(bParaGlide_OpenGlider:bool=true) -> (Output_Get:bool)
   > Delay(Duration=0.2)
   > Set bParaGlide_OpenGlider(bParaGlide_OpenGlider=false)
  ]}
 FacialCurvesLayer {nodes [
   n1 = Gameplay_FaceSystem_StateMachine_AnimBP!() -> (Pose:Pose Link Structure) {
    is=title AnimGraphNode_LinkedAnimGraph
    tag=FaceSystemStateMachine
    instanceClass=/Game/Animation/Game/MainPlayer/Facial/FaceGameplay/Gameplay_FaceSystem_StateMachine_AnimBP.Gameplay_FaceSystem_StateMachine_AnimBP_C
   }
   n2 = FacialCurvesLayerInterface - FacialLayer_FaceCurves!(
    InPosePreFacialCurves:Pose Link Structure=Use cached pose 'InPoseFacialCurves'()
    InPosePostFacialCurves:Pose Link Structure=Slot 'Default'(Source=Linked Graph Gate(Source=@n1) {linkedGraphTag=FaceSystemStateMachine}) {slotName=Default}
   ) -> (Pose:Pose Link Structure) {
    is=title AnimGraphNode_LinkedAnimLayer
    interface=/Game/Characters/Player/Male/Male_Avg_Base/AnimLayerInterface/FacialCurvesLayerInterface.FacialCurvesLayerInterface_C
    layer=FacialLayer_FaceCurves
   }
   n3 = Blend Facial Curves!(FirstPose:Pose Link Structure=Use cached pose 'InPoseFacialCurves'(), SecondPose:Pose Link Structure=@n2, bIs3LRig:bool=False) -> (Pose:Pose Link Structure) {
    is=title FortAnimGraphNode_BlendFacialCurves
    alphaCurveName=FaceStateMachineIdle
    bInvertAlphaCurveValue=true
    blendPreset=AlphaCurveBlend
    facialCurveDefinitions=/Game/Characters/Player/Common/FacialCurvesDefinitions
    secondPoseCurvePassthrough [baseheaddefaultfaceidle]
   }
   Output Pose(Result=Blend Poses by bool(BlendPose_0=@n3, BlendPose_1=Use cached pose 'InPose'(), bActiveValue=Get bEnableFacialCurves() -> (bEnableFacialCurves:bool))) {layerGroup=None}
   n4 = FacialCurvesLayerInterface - FacialLayer_FullBody!(InPose:Pose Link Structure=Use cached pose 'InPose'()) -> (Pose:Pose Link Structure) {
    is=title AnimGraphNode_LinkedAnimLayer
    interface=&FacialCurvesLayerInterface_C
    layer=FacialLayer_FullBody
   }
   Save cached pose 'InPoseFacialCurves'(Pose=@n4) {cacheName=InPoseFacialCurves}
   Save cached pose 'InPose'(Pose=Input Pose() {name=InPoseFacialCurves}) {cacheName=InPose}
  ]}
 FacialLayer_FaceCurves {nodes [
   Input Pose() {name=InPosePreFacialCurves}
   Output Pose(Result=Input Pose() {name=InPosePostFacialCurves}) {layerGroup=FacialLayerInterface}
  ]}
 FacialLayer_FullBody {nodes [Output Pose(Result=Input Pose()) {layerGroup=FacialLayerInterface}]}
 FirstPersonAdditives {nodes [
   n1 = Player_FirstPersonAdditives_AnimBP_Template!(InPose_FullBody:Pose Link Structure=Use cached pose 'InPoseFullBody'()) -> (Pose:Pose Link Structure) {
    is=title AnimGraphNode_LinkedAnimGraph
    tag=FirstPersonAdditives
    instanceClass=/Game/Characters/Player/AnimationBlueprints/LinkedAnimGraphTemplates/Player_FirstPersonAdditives_AnimBP_Template.Player_FirstPersonAdditives_AnimBP_Template_C
   }
   Output Pose(Result=Two Way Blend(
     A=Use cached pose 'InPoseFullBody'()
     B=Linked Graph Gate(Source=@n1) {bFallbackToAdditiveIdentity=true, linkedGraphTag=FirstPersonAdditives}
     Alpha=Get FirstPersonAdditiveAlpha() -> (FirstPersonAdditiveAlpha:float)
    ) {blendNode {alphaBoolBlend {blendOption=Linear}}}) {layerGroup=None}
   Save cached pose 'InPoseFullBody'(Pose=Input Pose() {name=InPose_FullBody}) {cacheName=InPoseFullBody}
  ]}
 FullBodyAdditivesLayer {nodes [
   n1 = Slot 'LargeBodyAdditiveCorrective'(Source=Use cached pose 'PostAdditiveFullBodySlot'()) {slotName=LargeBodyAdditiveCorrective}
   n2 = Use cached pose 'PostAdditiveFullBodySlot'()
   n3 = Get LargeBodyAdditiveSlotAlpha() -> (LargeBodyAdditiveSlotAlpha:float)
   AnimGraphNode_StateMachine_1 = FullBodyAdditives() {is=title AnimGraphNode_StateMachine, bAllowConduitEntryStates=true, bSkipFirstUpdateTransition=false, maxTransitionsPerFrame=0}
   n4 = Layered blend per bone(
    BasePose=Blend Poses by bool(
     BlendPose_0=Linked Anim Graph() {tag=FullBodyAdditive}
     BlendPose_1=@AnimGraphNode_StateMachine_1
     BlendTime_0=0.25
     BlendTime_1=0.5
     bActiveValue=Get bUseFullBodyJumpAdditivesLinkedGraph() -> (bUseFullBodyJumpAdditivesLinkedGraph:bool)
    ) {blendType=Linear}
    BlendPoses_0=Additive Identity Pose()
    BlendWeights_0=Get JumpUpperBodyBlendWeight() -> (JumpUpperBodyBlendWeight:float)
   ) {lODThreshold=2, layerSetup [{branchFilters [{blendDepth=2, boneName=Spine_01}, {boneName=Ik_hand_root}]}]}
   n5 = Layered blend per bone(
    BasePose=@n4
    BlendPoses_0=Additive Identity Pose()
    BlendWeights_0=Get DisableUpperBodyAdditiveMeshSpaceCurveAlpha() -> (DisableUpperBodyAdditiveMeshSpaceCurveAlpha:float)
   ) {bMeshSpaceRotationBlend=true, lODThreshold=2, layerSetup [{branchFilters [{blendDepth=4, boneName=spine_01}, {boneName=ik_hand_root}]}]}
   Output Pose(Result=Apply Additive(
     Base=Two Way Blend(A=@n2, B=@n1, Alpha=@n3) {blendNode {alphaBoolBlend {blendOption=Linear}}}
     Additive=@n5
     Alpha=Get JumpAdditiveLayerAlpha() -> (JumpAdditiveLayerAlpha:float)
    ) {alphaBoolBlend {blendOption=Linear}}) {layerGroup=None}
   n6 = Input Pose() {name=InPoseFullBodyAdditives}
   n7 = Hand IK Retargeting!(
    ComponentPose:Component Space Pose Link Structure=Local To Component(LocalPose=Slot 'AdditiveCorrective'(Source=@n6) {slotName=AdditiveCorrective})
    AlphaCurveName:Name=DisableHandIKRetargeting
    PerAxisAlpha:Vector="1,1,1"
    HandFKWeight:float=Get HandIKRetargetingWeight() -> (HandIKRetargetingWeight:float)
   ) -> (Pose:Component Space Pose Link Structure) {
    is=title AnimGraphNode_HandIKRetargeting
    alphaInputType=Curve
    alphaBoolBlend {blendOption=Linear}
    alphaScaleBiasClamp {bias=1, scale=-1}
    iKBonesToMove [{boneName=ik_hand_gun}]
    leftHandFK {boneName=hand_l}
    leftHandIK {boneName=ik_hand_l}
    rightHandFK {boneName=hand_r}
    rightHandIK {boneName=ik_hand_r}
   }
   Save cached pose 'PostHandIKRetarget'(Pose=Component To Local(ComponentPose=@n7)) {cacheName=PostHandIKRetarget}
   n8 = Use cached pose 'PostHandIKRetarget'()
   Save cached pose 'PostAdditiveFullBodySlot'(Pose=Two Way Blend(
     A=Use cached pose 'PostHandIKRetarget'()
     B=Slot 'AdditiveFullBody'(Source=@n8) {slotName=AdditiveFullBody}
     Alpha=Get AdditiveFullBodySlotAlpha() -> (AdditiveFullBodySlotAlpha:float)
    ) {blendNode {alphaBoolBlend {blendOption=Linear}}}) {cacheName=PostAdditiveFullBodySlot}
  ]}
 FullBodyAdditives {
  entry=IdleAdditive
  path=FullBodyAdditivesLayer.AnimGraphNode_StateMachine_1.FullBodyAdditives
  ids=id
  nodes [
   AnimStateNode_0 = Landing() {
    is=title AnimStateNode
    stateEntered {guid=#1}
    stateFullyBlended {guid=#2}
    stateLeft {guid=#3}
   }
   AnimStateNode_2 = SlopeSlidingState() {
    is=title AnimStateNode
    stateEntered {guid=#4}
    stateFullyBlended {guid=#5}
    stateLeft {guid=#6}
   }
   AnimStateNode_8 = IdleAdditive() {
    is=title AnimStateNode
    stateEntered {guid=#7}
    stateFullyBlended {guid=#8}
    stateLeft {guid=#9}
   }
   AnimStateTransitionNode_0 = AnimStateTransitionNode()
   AnimStateTransitionNode_1 = AnimStateTransitionNode()
   AnimStateTransitionNode_10 = AnimStateTransitionNode() {crossfadeDuration=0.1, logicType=TLT_Inertialization}
   AnimStateTransitionNode_11 = AnimStateTransitionNode()
   AnimStateTransitionNode_12 = AnimStateTransitionNode() {crossfadeDuration=0.3, logicType=TLT_Inertialization, priorityOrder=2}
   AnimStateTransitionNode_13 = AnimStateTransitionNode() {automaticRuleTriggerTime=0, crossfadeDuration=0.25, logicType=TLT_Inertialization}
   AnimStateTransitionNode_14 = AnimStateTransitionNode() {logicType=TLT_Inertialization}
   AnimStateTransitionNode_16 = AnimStateTransitionNode() {crossfadeDuration=0.3, logicType=TLT_Inertialization}
   AnimStateTransitionNode_2 = AnimStateTransitionNode() {priorityOrder=100}
   AnimStateTransitionNode_3 = AnimStateTransitionNode() {crossfadeDuration=0.1, logicType=TLT_Inertialization}
  ]
  links [
   {
    dst=Landing
    rule=FullBodyAdditivesLayer.AnimGraphNode_StateMachine_1.FullBodyAdditives.AnimStateTransitionNode_10.Transition
    src=IdleAdditive
    via=AnimStateTransitionNode_10
   }
   {
    dst=Landing
    rule=FullBodyAdditivesLayer.AnimGraphNode_StateMachine_1.FullBodyAdditives.AnimStateTransitionNode_12.Transition
    src=IdleAdditive
    via=AnimStateTransitionNode_12
   }
   {
    dst=SlopeSlidingState
    rule=FullBodyAdditivesLayer.AnimGraphNode_StateMachine_1.FullBodyAdditives.AnimStateTransitionNode_16.AnimationTransitionGraph_0
    src=IdleAdditive
    via=AnimStateTransitionNode_16
   }
   {
    dst=IdleAdditive
    rule=FullBodyAdditivesLayer.AnimGraphNode_StateMachine_1.FullBodyAdditives.AnimStateTransitionNode_0.Transition
    src=Landing
    via=AnimStateTransitionNode_0
   }
   {
    dst=IdleAdditive
    rule=FullBodyAdditivesLayer.AnimGraphNode_StateMachine_1.FullBodyAdditives.AnimStateTransitionNode_1.Transition
    src=Landing
    via=AnimStateTransitionNode_1
   }
   {
    dst=IdleAdditive
    rule=FullBodyAdditivesLayer.AnimGraphNode_StateMachine_1.FullBodyAdditives.AnimStateTransitionNode_11.Transition
    src=Landing
    via=AnimStateTransitionNode_11
   }
   {
    dst=IdleAdditive
    rule=FullBodyAdditivesLayer.AnimGraphNode_StateMachine_1.FullBodyAdditives.AnimStateTransitionNode_13.Transition
    src=Landing
    via=AnimStateTransitionNode_13
   }
   {
    dst=IdleAdditive
    rule=FullBodyAdditivesLayer.AnimGraphNode_StateMachine_1.FullBodyAdditives.AnimStateTransitionNode_14.Transition
    src=Landing
    via=AnimStateTransitionNode_14
   }
   {
    dst=IdleAdditive
    rule=FullBodyAdditivesLayer.AnimGraphNode_StateMachine_1.FullBodyAdditives.AnimStateTransitionNode_2.Transition
    src=SlopeSlidingState
    via=AnimStateTransitionNode_2
   }
   {
    dst=Landing
    rule=FullBodyAdditivesLayer.AnimGraphNode_StateMachine_1.FullBodyAdditives.AnimStateTransitionNode_3.Transition
    src=SlopeSlidingState
    via=AnimStateTransitionNode_3
   }
  ]
 }
 Landing {
  path=FullBodyAdditivesLayer.AnimGraphNode_StateMachine_1.FullBodyAdditives.AnimStateNode_0.Landing
  nodes [Output Animation Pose(Result=Linked Graph Gate(Source=Player_Land_AnimBP!() -> (Pose:Pose Link Structure) {
      is=title AnimGraphNode_LinkedAnimGraph
      tag=Land
      instanceClass=/Game/Characters/Player/AnimationBlueprints/LinkedAnimGraphs/Player_Land_AnimBP.Player_Land_AnimBP_C
     }) {bFallbackToAdditiveIdentity=true, linkedGraphTag=Land})]
 }
 SlopeSlidingState {
  path=FullBodyAdditivesLayer.AnimGraphNode_StateMachine_1.FullBodyAdditives.AnimStateNode_2.SlopeSlidingState
  nodes [
   n1 = Get SlopeSlidingPitch() -> (SlopeSlidingPitch:float)
   n2 = Get SlopeSlidingRoll() -> (SlopeSlidingRoll:float)
   n3 = Blendspace Player 'SlopeSliding_NoWep_BS'(X=@n2, Y=@n1) {
    blendSpace=/Game/Animation/Game/MainPlayer/Locomotion/Medium/Male/Sliding/SlopeSliding_NoWep_BS
    groupName=SlopeSliding
    groupRole=AlwaysFollower
    method=SyncGroup
   }
   n4 = Blendspace Player 'SlopeSliding_AR_BS'(X=@n2, Y=@n1) {blendSpace=/Game/Animation/Game/MainPlayer/Locomotion/Medium/Male/Sliding/SlopeSliding_AR_BS, groupName=SlopeSliding, method=SyncGroup}
   n5 = Blendspace Player 'SlopeSliding_NoWep_BS'(X=@n2, Y=@n1) {
    blendSpace=&SlopeSliding_NoWep_BS
    groupName=SlopeSliding
    groupRole=AlwaysFollower
    method=SyncGroup
   }
   n6 = Blendspace Player 'SlopeSliding_Pistol_BS'(X=@n2, Y=@n1) {
    blendSpace=/Game/Animation/Game/MainPlayer/Locomotion/Medium/Male/Sliding/SlopeSliding_Pistol_BS
    groupName=SlopeSliding
    groupRole=AlwaysFollower
    method=SyncGroup
   }
   n7 = "Blend Poses (EFortWeaponCoreAnimation)"!(
    BlendPose_0:Pose Link Structure=@n4
    BlendPose_1:Pose Link Structure=@n6
    BlendPose_2:Pose Link Structure=@n3
    BlendPose_3:Pose Link Structure=@n5
    BlendTime_0:float=.2
    BlendTime_1:float=.2
    BlendTime_2:float=0.1
    BlendTime_3:float=0.1
    ActiveEnumValue:Byte=Get WeaponCoreAnim!() -> (WeaponCoreAnim)
   ) -> (Pose:Pose Link Structure) {
    is=title AnimGraphNode_BlendListByEnum
    blendPose [{linkId=-1, sourceLinkId=-1}, {linkId=-1, sourceLinkId=-1}, {linkId=-1, sourceLinkId=-1}, {linkId=-1, sourceLinkId=-1}]
    blendTime [0.1,0.1,0.1,0.1]
   }
   n8 = "Sequence Player (additive) 'Slide_Center_CMM_Additive'"() {sequence=/Game/Animation/Game/MainPlayer/Locomotion/Medium/Male/Sliding/Poses/Slide_Center_CMM_Additive}
   Output Animation Pose(Result=Two Way Blend(
     A=Additive Identity Pose()
     B=Apply Additive(Base=@n7, Additive=@n8) {lODThreshold=1, alphaBoolBlend {blendOption=Linear}}
     Alpha=Get SlopeSlidingAlpha() -> (SlopeSlidingAlpha:float)
    ) {blendNode {alphaBoolBlend {blendOption=Linear}}})
  ]
 }
 IdleAdditive {
  path=FullBodyAdditivesLayer.AnimGraphNode_StateMachine_1.FullBodyAdditives.AnimStateNode_8.IdleAdditive
  nodes [Output Animation Pose(Result=Additive Identity Pose())]
 }
 Transition {
  path=FullBodyAdditivesLayer.AnimGraphNode_StateMachine_1.FullBodyAdditives.AnimStateTransitionNode_0.Transition
  nodes [Result(bCanEnterTransition=Get bTransition_SwimJumpSurfaceLoop_to_IdleAdditive() -> (bTransition_SwimJumpSurfaceLoop_to_IdleAdditive:bool))]
 }
 Transition {
  path=FullBodyAdditivesLayer.AnimGraphNode_StateMachine_1.FullBodyAdditives.AnimStateTransitionNode_1.Transition
  nodes [Result(bCanEnterTransition=Get bIsRiding() -> (bIsRiding:bool))]
 }
 Transition {
  path=FullBodyAdditivesLayer.AnimGraphNode_StateMachine_1.FullBodyAdditives.AnimStateTransitionNode_10.Transition
  nodes [Result(bCanEnterTransition=Get bEnterLanding() -> (bEnterLanding:bool))]
 }
 Transition {
  path=FullBodyAdditivesLayer.AnimGraphNode_StateMachine_1.FullBodyAdditives.AnimStateTransitionNode_11.Transition
  nodes [Result(bCanEnterTransition=Get bIsWaterJump() -> (bIsWaterJump:bool))]
 }
 Transition {
  path=FullBodyAdditivesLayer.AnimGraphNode_StateMachine_1.FullBodyAdditives.AnimStateTransitionNode_12.Transition
  nodes [Result(bCanEnterTransition=Get bSwimmingIsWaterLanding() -> (bSwimmingIsWaterLanding:bool))]
 }
 Transition {
  path=FullBodyAdditivesLayer.AnimGraphNode_StateMachine_1.FullBodyAdditives.AnimStateTransitionNode_13.Transition
  nodes [Result(bCanEnterTransition=Get bExitLanding() -> (bExitLanding:bool))]
 }
 Transition {
  path=FullBodyAdditivesLayer.AnimGraphNode_StateMachine_1.FullBodyAdditives.AnimStateTransitionNode_14.Transition
  nodes [Result(bCanEnterTransition=Get bIsJumping())]
 }
 AnimationTransitionGraph_0 {
  path=FullBodyAdditivesLayer.AnimGraphNode_StateMachine_1.FullBodyAdditives.AnimStateTransitionNode_16.AnimationTransitionGraph_0
  nodes [Result(bCanEnterTransition=Get bIsSlopeSliding() -> (bIsSlopeSliding:bool))]
 }
 Transition {
  path=FullBodyAdditivesLayer.AnimGraphNode_StateMachine_1.FullBodyAdditives.AnimStateTransitionNode_2.Transition
  nodes [Result(bCanEnterTransition=NOT Boolean(A=Get bIsSlopeSliding()))]
 }
 Transition {
  path=FullBodyAdditivesLayer.AnimGraphNode_StateMachine_1.FullBodyAdditives.AnimStateTransitionNode_3.Transition
  nodes [Result(bCanEnterTransition=Get bEnterLanding())]
 }
 FullBodyAimOffsets {nodes [
   n1 = VehicleAnimLayerInterface - VehicleSplitBodyOverride!(
    InPoseSplitBody:Pose Link Structure=Use cached pose 'PostFullBodyAimOffsetsPose'()
    InPoseUpperAndLowerBody:Pose Link Structure=Use cached pose 'PostFullBodyAimOffsetsPose'()
    InPoseUpperAndLowerBodyRemoveRoot:Pose Link Structure=Use cached pose 'PostFullBodyAimOffsetsNoRoot'()
   ) -> (Pose:Pose Link Structure) {
    is=title AnimGraphNode_LinkedAnimLayer
    interface=&VehicleAnimLayerInterface_C
    layer=VehicleSplitBodyOverride
   }
   Output Pose(Result=@n1)
   Save cached pose 'InputPose2'(Pose=Input Pose() {name=InPose_FullBodyAimOffsets}) {cacheName=InputPose2}
   n2 = Break Fort Anim Input Adjusted Aim!(FortAnimInput_AdjustedAim:"Fort Anim Input Adjusted Aim Structure (by ref)"=Get AdjustedAim() -> (AdjustedAim:Fort Anim Input Adjusted Aim Structure)) -> (ResultingYaw:float=~, ResultingPitch:float=~, ZeroOutPitchWeightCurveName:Name) {is=title K2Node_BreakStruct}
   n3 = Get FullBodyAimOffsetAlpha() -> (FullBodyAimOffsetAlpha:float)
   n4 = Get bEnableAimOffsetsRootYawCompensation() -> (bEnableAimOffsetsRootYawCompensation:bool)
   n5 = Compensated Aim Offset 'M_Melee_TwoHanded_AO'!(
    BasePose:Pose Link Structure=Use cached pose 'InputPose2'()
    X:float=@n2
    Y:float=@n2.ResultingPitch
    bApplyYawCompensation:bool=@n4
    bApplyAdditiveInRootSpace:bool=False
    Alpha:float=@n3
   ) -> (Pose:Pose Link Structure) {
    is=title FortAnimGraphNode_AimOffsets
    blendSpace=/Game/Animation/Game/MainPlayer/Poses/Medium/Male/AO/M_Melee_TwoHanded_AO
    yawCompensationScale=-1
    compensationBone {boneName=Root}
   }
   n6 = Compensated Aim Offset 'M_NoWep_Relaxed_AO'!(
    BasePose:Pose Link Structure=Use cached pose 'InputPose2'()
    X:float=@n2
    Y:float=@n2.ResultingPitch
    bApplyYawCompensation:bool=@n4
    bApplyAdditiveInRootSpace:bool=False
    Alpha:float=@n3
   ) -> (Pose:Pose Link Structure) {
    is=title FortAnimGraphNode_AimOffsets
    blendSpace=/Game/Animation/Game/MainPlayer/Poses/Medium/Male/AO/M_NoWep_Relaxed_AO
    yawCompensationScale=-1
    compensationBone {boneName=Root}
   }
   n7 = ItemAnimLayerInterface - ItemFullBodyAimOffsets!(
    InPoseFullBodyAimOffsets:Pose Link Structure=Two Way Blend(A=@n6, B=@n5, Alpha=Get PlayMeleeAttackAO() -> (PlayMeleeAttackAO:float)) {blendNode {alphaBoolBlend {blendOption=Linear}}}
    InPosePreFullBodyAimOffsets:Pose Link Structure=Use cached pose 'InputPose2'()
    FullBody_AimOffsetAlpha:double=0
    Aim_Yaw:double=0
    Aim_Pitch:double=0
   ) -> (Pose:Pose Link Structure) {
    is=title AnimGraphNode_LinkedAnimLayer
    interface=&ItemAnimLayerInterface_C
    layer=ItemFullBodyAimOffsets
   }
   Save cached pose 'PostFullBodyAimOffsetsPose'(Pose=Blend Poses by bool(
     BlendPose_0=@n7
     BlendPose_1=Use cached pose 'InputPose2'()
     BlendTime_0=0.3
     BlendTime_1=0.3
     bActiveValue=Get bEnableAimOffsets() -> (bEnableAimOffsets:bool)
    ) {blendType=QuadraticInOut}) {cacheName=PostFullBodyAimOffsetsPose}
   n8 = Local To Component(LocalPose=Use cached pose 'PostFullBodyAimOffsetsPose'())
   Save cached pose 'PostFullBodyAimOffsetsNoRoot'(Pose=Component To Local(ComponentPose=Reset Root Transform!(ComponentPose:Component Space Pose Link Structure=@n8, Alpha:float=1) -> (Pose:Component Space Pose Link Structure) {is=title AnimGraphNode_ResetRoot, alphaBoolBlend {blendOption=Linear}})) {cacheName=PostFullBodyAimOffsetsNoRoot}
  ]}
 FullBodyOverride {nodes [Output Pose(Result=Input Pose() {name=PassThroughFullBodyPose}) {layerGroup=PlayerOverride}]}
 FullBodySlotLayer {nodes [
   n1 = Layered blend per bone(BasePose=Use cached pose 'PostFullBodySlots'(), BlendPoses_0=Use cached pose 'InFullBodyPose'()) {
    bMeshSpaceRotationBlend=true
    bMeshSpaceScaleBlend=true
    curveBlendOption=UseMaxValue
    layerSetup [{branchFilters [
       {boneName=Root}
       {blendDepth=-1, boneName=Ik_hand_root}
       {blendDepth=-4, boneName=spine_01}
       {blendDepth=-1, boneName=attach_fp}
       {blendDepth=-1, boneName=camera_root}
      ]}]
   }
   n2 = Layered blend per bone(
    BasePose=Slot 'UpperBodyInMotion'(Source=Use cached pose 'InFullBodyPose'()) {slotName=UpperBodyInMotion}
    BlendPoses_0=Use cached pose 'InFullBodyPose'()
   ) {
    bMeshSpaceRotationBlend=true
    bMeshSpaceScaleBlend=true
    curveBlendOption=UseMaxValue
    layerSetup [{branchFilters [
       {boneName=Root}
       {blendDepth=-1, boneName=Ik_hand_root}
       {blendDepth=-4, boneName=spine_01}
       {blendDepth=-1, boneName=attach_fp}
       {blendDepth=-1, boneName=camera_root}
      ]}]
   }
   n3 = Get FullBodyToUpperBodyInMotionBlendTime() -> (FullBodyToUpperBodyInMotionBlendTime:float)
   n4 = Blend Poses by bool(
    BlendPose_0=@n2
    BlendPose_1=Blend Poses by bool(
     BlendPose_0=@n1
     BlendPose_1=Use cached pose 'PostFullBodySlots'()
     BlendTime_0=0.3
     BlendTime_1=0.4
     bActiveValue=Get bPlayLowerBodyLocomotionOverFullBodySlot() -> (bPlayLowerBodyLocomotionOverFullBodySlot:bool)
    )
    BlendTime_0=@n3
    BlendTime_1=@n3
    bActiveValue=Get bUseUpperBodyInMotionSlot() -> (bUseUpperBodyInMotionSlot:bool)
   ) {blendType=Linear}
   n5 = Slot 'FullBodyInMotion'(Source=Use cached pose 'InFullBodyPose'()) {slotName=FullBodyInMotion}
   n6 = Get StaticToTraversalBlendTime() -> (StaticToTraversalBlendTime:float)
   n7 = Get TraversalToStaticBlendTime() -> (TraversalToStaticBlendTime:float)
   n8 = Get bIsMovingAndInMotionEmote() -> (bIsMovingAndInMotionEmote:bool)
   n9 = Get MissingCosmeticLowerBodyOverride() -> (MissingCosmeticLowerBodyOverride:Anim Sequence Object Reference)
   n10 = Get MissingCosmeticLowerBodyInMotionOverride() -> (MissingCosmeticLowerBodyInMotionOverride:Anim Sequence Object Reference)
   n11 = Layered blend per bone(
    BasePose=Blend Poses by bool(
     BlendPose_0=Sequence Player!(Sequence:Anim Sequence Base Object Reference=@n10) -> (Pose)
     BlendPose_1=Sequence Player(Sequence=@n9)
     bActiveValue=Get bIsMoving2D()
    ) {blendType=Linear}
    BlendPoses_0=Sequence Player(Sequence=Get MissingCosmeticUpperBodyOverride() -> (MissingCosmeticUpperBodyOverride:Anim Sequence Object Reference))
   ) {bMeshSpaceRotationBlend=true, layerSetup [{branchFilters [{blendDepth=1, boneName=spine_05}, {blendDepth=-1, boneName=neck_01}]}]}
   Output Pose(Result=Blend Poses by bool(
     BlendPose_0=@n11
     BlendPose_1=Blend Poses by bool(BlendPose_0=@n5, BlendPose_1=@n4, BlendTime_0=@n6, BlendTime_1=@n7, bActiveValue=@n8)
     BlendTime_1=0.3
     bActiveValue=Get bEnableMissingCosmeticOverride() -> (bEnableMissingCosmeticOverride:bool)
    ) {blendType=Linear}) {layerGroup=None}
   Save cached pose 'InFullBodyPose'(Pose=Input Pose() {name=InFullBodyPose}) {cacheName=InFullBodyPose}
   n12 = Slot 'FullBody'(Source=Use cached pose 'InFullBodyPose'()) {slotName=FullBody}
   n13 = Slot 'SwimFullBody'(Source=Use cached pose 'InFullBodyPose'()) {slotName=SwimFullBody}
   n14 = Get bUseSwimFullBodySlot() -> (bUseSwimFullBodySlot:bool)
   Save cached pose 'PostFullBodySlots'(Pose=Slot 'OverrideFullBody'(Source=Blend Poses by bool(BlendPose_0=@n13, BlendPose_1=@n12, bActiveValue=@n14)) {slotName=OverrideFullBody}) {cacheName=PostFullBodySlots}
  ]}
 GetSequencerAnimSlotNames {nodes [
   GetSequencerAnimSlotNames() {is=title K2Node_FunctionEntry}
   > Return Node!(execute, ReturnValue:Array of Names=Make Array("[0]":Name=SequencerDefault) -> (Array:Array of Names))
  ]}
 HideWeapon {nodes [
   n1 = Get HideLeftWeaponAlpha() -> (HideLeftWeaponAlpha:float)
   n2 = "Transform (Modify) Bone - Bone: weapon_l"!(ComponentPose=Local To Component(LocalPose=Use cached pose 'InPoseHideWeapon'()), Alpha=@n1, Scale="0.01,0.01,0.01") -> (Pose) {scaleMode=BMM_Replace, alphaBoolBlend {blendOption=Linear}, boneToModify {boneName=weapon_l}, scale=<0.01 0.01 0.01>}
   n3 = Get HideRightWeaponAlpha() -> (HideRightWeaponAlpha:float)
   n4 = "Transform (Modify) Bone - Bone: weapon_r"!(ComponentPose=@n2, Alpha=@n3, Scale="0.01,0.01,0.01") -> (Pose) {scaleMode=BMM_Replace, alphaBoolBlend {blendOption=Linear}, boneToModify {boneName=weapon_r}, scale=<0.01 0.01 0.01>}
   n5 = "Transform (Modify) Bone - Bone: hand_attach_l"!(ComponentPose=@n4, Alpha=@n1, Scale="0.01,0.01,0.01") -> (Pose) {scaleMode=BMM_Replace, alphaBoolBlend {blendOption=Linear}, boneToModify {boneName=hand_attach_l}, scale=<0.01 0.01 0.01>}
   n6 = "Transform (Modify) Bone - Bone: hand_attach_r"!(ComponentPose=@n5, Alpha=@n3, Scale="0.01,0.01,0.01") -> (Pose) {scaleMode=BMM_Replace, alphaBoolBlend {blendOption=Linear}, boneToModify {boneName=hand_attach_r}, scale=<0.01 0.01 0.01>}
   Output Pose(Result=Blend Poses by bool(
     BlendPose_0=Component To Local(ComponentPose=@n6)
     BlendPose_1=Use cached pose 'InPoseHideWeapon'()
     BlendTime_0=0
     BlendTime_1=0
     bActiveValue=Get bHideWeapon() -> (bHideWeapon:bool)
    ) {blendType=Linear})
   Save cached pose 'InPoseHideWeapon'(Pose=Input Pose() {name=InPoseHideWeapon}) {cacheName=InPoseHideWeapon}
  ]}
 Indigo_Boost {nodes [Indigo_Boost() {is=title K2Node_FunctionEntry}]}
 ItemAimOffset {nodes [
   Input Pose() {name=InPose_PreAimOffsets}
   Output Pose(Result=Input Pose!() -> (Pose, AimOffsetAlpha:double, Yaw:double, Pitch:double) {name=InPose_AimOffset}) {layerGroup=Item Override}
  ]}
 ItemFallAdditive {nodes [
   Input Pose!() -> (Pose) {name=InPoseItemFallAdditive}
   Output Pose() {layerGroup=Item Override}
  ]}
 ItemFallLandAdditive {nodes [
   Input Pose() {name=InPoseItemFallLandAdditive}
   Output Pose() {layerGroup=Item Override}
  ]}
 ItemFallLevel2Override {nodes [
   Input Pose() {name=InPose_FallLevel2}
   Output Pose() {layerGroup=Item Override}
  ]}
 ItemFinalPoseOverride {nodes [Output Pose(Result=Input Pose() {name=InFinalPose}) {layerGroup=Item Override}]}
 ItemFullBodyAimOffsets {nodes [
   Input Pose() {name=InPosePreFullBodyAimOffsets}
   Output Pose(Result=Input Pose!() -> (Pose, FullBody_AimOffsetAlpha:double, Aim_Yaw:double, Aim_Pitch:double) {name=InPoseFullBodyAimOffsets}) {layerGroup=Item Override}
  ]}
 ItemFullBodyOverride {nodes [Output Pose(Result=Input Pose!() -> (Pose) {name=InPoseFullBody}) {layerGroup=Item Override}]}
 ItemIdleOverride {nodes [
   Input Pose() {name=InPose_Idle}
   Output Pose() {layerGroup=Item Override}
  ]}
 ItemJetPackJumpAdditive {nodes [Output Pose() {layerGroup=Item Override}]}
 ItemJetPackStartAdditive {nodes [Output Pose() {layerGroup=Item Override}]}
 ItemJumpLoopAdditive {nodes [
   Input Pose() {name=InPoseItemJumpLoopAdditive}
   Output Pose() {layerGroup=Item Override}
  ]}
 ItemJumpOverride {nodes [
   Input Pose() {name=InPoseJump}
   Output Pose() {layerGroup=Item Override}
  ]}
 ItemJumpUpAdditive {nodes [
   Input Pose() {name=InPoseItemJumpUpAdditive}
   Output Pose() {layerGroup=Item Override}
  ]}
 ItemPreAimOffsets {nodes [Output Pose(Result=Input Pose() {name=InPosePreAimOffsets}) {layerGroup=Item Override}]}
 ItemPreIK {nodes [
   Input Pose() {name=InPose_PreIK}
   Output Pose() {layerGroup=Item Override}
  ]}
 ItemRightHandIK {nodes [
   Input Pose() {name=InPosePreRightHandIK}
   Input Pose() {name=InPosePostRightHandIK}
   Output Pose() {layerGroup=Item Override}
  ]}
 ItemSkeletalControl {nodes [
   Input Pose()
   Output Pose() {layerGroup=Item Override}
  ]}
 ItemSwimJumpFallLoop {nodes [Output Pose() {layerGroup=Item Override}]}
 ItemSwimJumpSurfaceEnd {nodes [Output Pose() {layerGroup=Item Override}]}
 ItemSwimJumpSurfaceLoop {nodes [Output Pose() {layerGroup=Item Override}]}
 ItemUpperBody {nodes [
   AnimGraphNode_SubInput_0 = Input Pose!() -> (Pose, InputParam:Fort Anim Input Adjusted Aim Structure) {name=InPoseUpperBody}
   Output Pose(Result=@AnimGraphNode_SubInput_0) {layerGroup=Item Override}
  ]}
 ItemZipLineStartAdditive {nodes [Output Pose() {layerGroup=Item Override}]}
 LocomotionLayer {nodes [
   n1 = Use cached pose 'LocomotionPoseBase'()
   Output Pose(Result=Blend Poses by bool(
     BlendPose_0=Slot 'CrouchInPlaceAdditive'(Source=@n1) {slotName=CrouchInPlaceAdditive}
     BlendPose_1=Use cached pose 'LocomotionPoseBase'()
     bActiveValue=Get bShouldUseCrouchInPlaceAdditiveSlot() -> (bShouldUseCrouchInPlaceAdditiveSlot:bool)
    ) {blendType=Linear}) {layerGroup=None}
   Save cached pose 'LocomotionBasePosePreAdditives'(Pose=LowerBodyBlendLayer!(BasePose:Pose Link Structure=Use cached pose 'InSourcePoseLoco'(), LowerBody:Pose Link Structure=Use cached pose 'LowerBody'()) -> (Pose:Pose Link Structure) {is=title AnimGraphNode_LinkedAnimLayer, layer=LowerBodyBlendLayer}) {cacheName=LocomotionBasePosePreAdditives}
   n2 = "Sequence Player (additive) 'NoHeadOrNeckType_Crouch_AdditiveCorrection'"() {bLoopAnimation=false, sequence=/Game/Animation/Game/MainPlayerFN/Poses/AdditiveCorrection/NoHeadOrNeckType_Crouch_AdditiveCorrection}
   n3 = Layered blend per bone(
    BasePose=Use cached pose 'LocomotionBasePosePreAdditives'()
    BlendPoses_0=Apply Additive(Base=Use cached pose 'LocomotionBasePosePreAdditives'(), Additive=@n2) {alphaBoolBlend {blendOption=Linear}}
   ) {
    bMeshSpaceRotationBlend=true
    curveBlendOption=UseMaxValue
    layerSetup [{branchFilters [{boneName=spine_01}, {blendDepth=-1, boneName=clavicle_l}, {blendDepth=-1, boneName=clavicle_r}]}]
   }
   Save cached pose 'LocomotionPoseBase'(Pose=Blend Poses by bool(
     BlendPose_0=@n3
     BlendPose_1=Use cached pose 'LocomotionBasePosePreAdditives'()
     bActiveValue=Get bApplyCrouchMovingAdditiveSkinFix() -> (bApplyCrouchMovingAdditiveSkinFix:bool)
    ) {blendType=Linear}) {cacheName=LocomotionPoseBase, cachePoseName=LocomotionPoseBase}
   n4 = Input Pose!() -> (Pose) {name=MotionMatchingLocoPose}
   AnimGraphNode_StateMachine_0 = MotionMatchingIdle() {is=title AnimGraphNode_StateMachine}
   n5 = Get bTransitionMotionMatchingIdle() -> (bTransitionMotionMatchingIdle:bool)
   n6 = Get IdleToMovementLowerBodyBlendTime() -> (IdleToMovementLowerBodyBlendTime:float)
   n7 = Local To Component(LocalPose=Modify Curve!(
     SourcePose=Blend Poses by bool(BlendPose_0=@AnimGraphNode_StateMachine_0, BlendPose_1=@n4, BlendTime_0=0.3, BlendTime_1=@n6, bActiveValue=@n5) {blendType=QuadraticInOut}
     CurveValues_0=0
     CurveValues_1:float=0
     Alpha=1
    ) -> (Pose) {curveNames [BothHandFK,LeftHandFK], curveValues [0,0]})
   n8 = Get IdlePelvisOffset() -> (IdlePelvisOffset:Vector)
   n9 = Get IdlePelvisOffsetAlpha() -> (IdlePelvisOffsetAlpha:float)
   n10 = PlayerLocomotionLayerInterface - LocomotionLayer_LowerBody!(InPose_LowerBody:Pose Link Structure=Component To Local(ComponentPose="Transform (Modify) Bone - Bone: pelvis"(ComponentPose=@n7, Alpha=@n9, Translation=@n8, Rotation="0,0,0", Scale="1,1,1") {lODThreshold=2, translationMode=BMM_Additive, alphaBoolBlend {blendOption=Linear}, boneToModify {boneName=pelvis}})) -> (Pose:Pose Link Structure) {
    is=title AnimGraphNode_LinkedAnimLayer
    interface=&PlayerLocomotionLayerInterface_C
    layer=LocomotionLayer_LowerBody
   }
   n11 = MovementModeLayerInterface - MovementMode_LowerBodyOverride!(InLowerBodyPose:Pose Link Structure=@n10, InSourcePose:Pose Link Structure=Use cached pose 'InSourcePoseLoco'()) -> (Pose:Pose Link Structure) {
    is=title AnimGraphNode_LinkedAnimLayer
    interface=&MovementModeLayerInterface_C
    layer=MovementMode_LowerBodyOverride
   }
   AnimGraphNode_Layer_0 = VehicleAnimLayerInterface - VehicleLowerBodyOverride!(InPoseLowerBodyDefault:Pose Link Structure=@n11, InPoseUpperBody:Pose Link Structure=Use cached pose 'InSourcePoseLoco'()) -> (Pose:Pose Link Structure) {
    is=title AnimGraphNode_LinkedAnimLayer
    interface=&VehicleAnimLayerInterface_C
    layer=VehicleLowerBodyOverride
   }
   Save cached pose 'LowerBody'(Pose=@AnimGraphNode_Layer_0) {cacheName=LowerBody, cachePoseName=LowerBody}
   Save cached pose 'InSourcePoseLoco'(Pose=Input Pose() {name=InSourcePoseLoco}) {cacheName=InSourcePoseLoco}
  ]}
 MotionMatchingIdle {
  entry=Idle
  path=LocomotionLayer.AnimGraphNode_StateMachine_0.MotionMatchingIdle
  ids=id
  nodes [
   AnimStateNode_1 = Turns() {
    is=title AnimStateNode
    stateEntered {guid=#10}
    stateFullyBlended {guid=#11}
    stateLeft {guid=#12}
   }
   AnimStateNode_2 = Idle()
   AnimStateTransitionNode_0 = AnimStateTransitionNode() {crossfadeDuration=0}
   AnimStateTransitionNode_1 = AnimStateTransitionNode() {bAllowInertializationForSelfTransitions=true, crossfadeDuration=0}
   AnimStateTransitionNode_2 = AnimStateTransitionNode() {bAutomaticRuleBasedOnSequencePlayerInState=true, crossfadeDuration=0}
  ]
  links [
   {
    dst=Turns
    rule=LocomotionLayer.AnimGraphNode_StateMachine_0.MotionMatchingIdle.AnimStateTransitionNode_0.AnimationTransitionGraph_0
    src=Idle
    via=AnimStateTransitionNode_0
   }
   {
    dst=Idle
    rule=LocomotionLayer.AnimGraphNode_StateMachine_0.MotionMatchingIdle.AnimStateTransitionNode_1.Transition
    src=Turns
    via=AnimStateTransitionNode_1
   }
   {
    dst=Idle
    rule=LocomotionLayer.AnimGraphNode_StateMachine_0.MotionMatchingIdle.AnimStateTransitionNode_2.AnimationTransitionGraph_0
    src=Turns
    via=AnimStateTransitionNode_2
   }
  ]
 }
 Turns {
  path=LocomotionLayer.AnimGraphNode_StateMachine_0.MotionMatchingIdle.AnimStateNode_1.Turns
  nodes [
   n1 = Break Fort Anim Input Player Anim Asset!(FortAnimInput_PlayerAnimAsset:"Fort Anim Input Player Anim Asset Structure (by ref)"=Get CurrentAnimSet() -> (CurrentAnimSet:Fort Anim Input Player Anim Asset Structure)) -> (
    bPlayUpperBodySlotOnFullBodyInAir:bool=~
    bDisableFullBodyAimOffsetDuringMelee:bool=~
    bShouldApplyAimOffsetFullBody:bool=~
    FullBodyAimOffsetAlpha:float=~
    UpperBodyAimOffsetAlpha:float=~
    bOverrideDisableArmsHeadAdditive:bool=~
    DisableArmsHeadAdditiveOverride:float=~
    SprintPlayrateClampRange:Fort Anim Playrate Range Structure
    SprintTargetingPlayrateClampRange:Fort Anim Playrate Range Structure
    CrouchSprintPlayrateClampRange:Fort Anim Playrate Range Structure
    CrouchSprintTargetingPlayrateClampRange:Fort Anim Playrate Range Structure
    TargetingAimOffset:Aim Offset Blend Space Object Reference
    NonTargetingAimOffset:Aim Offset Blend Space Object Reference
    RelaxedAimOffset:Aim Offset Blend Space Object Reference
    SwimTargetingPose:Anim Sequence Object Reference
    SwimNonTargetingPose:Anim Sequence Object Reference
    SwimJogAdditiveBlendSpace:Blend Space Object Reference
    SwimTargetingAimOffset:Aim Offset Blend Space Object Reference
    SwimIdleNoise:Anim Sequence Object Reference
    SwimJumpSurfaceEndAdditve:Anim Sequence Object Reference
    SwimJumpSurfaceLoopAdditve:Anim Sequence Object Reference
    SwimJumpFallLoopAdditve:Anim Sequence Object Reference
    FlyModeLoopAdditive:Anim Sequence Object Reference
    FlyModeStartAdditive:Anim Sequence Object Reference
    ZipLineStartAdditve:Anim Sequence Object Reference
    JetPackJumpAdditve:Anim Sequence Object Reference
    JetPackStartAdditve:Anim Sequence Object Reference
    FallAdditive:Anim Sequence Object Reference
    JumpLoopAdditive:Anim Sequence Object Reference
    JumpUpAdditive:Anim Sequence Object Reference
    FullBodyJumpAdditive:Anim Sequence Object Reference
    WeaponChargeLoop:Anim Sequence Object Reference
    LandingAdditiveOverride:Anim Sequence Object Reference
    TurnInPlaceLeftAdditiveOverride:Anim Sequence Object Reference
    TurnInPlaceRightAdditiveOverride:Anim Sequence Object Reference
    bOverridePitchAndYawOffsets:bool=~
    AnimSetOffsets:Fort Anim Input Adjusted Aim Offset Structure
    BonesToHide:Array of Names
    bOverrideHandIKRetargetingWeight:bool=~
    HandIKRetargetingWeight:float=~
    bForceKneeSliding:bool=~
    TargetingPose:Anim Sequence Object Reference
    NonTargetingPose:Anim Sequence Object Reference
    IdleNoise:Anim Sequence Object Reference
   ) {is=title K2Node_BreakStruct}
   n2 = Sequence Player(Sequence=@n1.TurnInPlaceRightAdditiveOverride) {bLoopAnimation=false}
   n3 = Sequence Player(Sequence=@n1.TurnInPlaceLeftAdditiveOverride) {bLoopAnimation=false}
   n4 = Break Fort Anim Input Turn in Place!(FortAnimInput_TurnInPlace:"Fort Anim Input Turn In Place Structure (by ref)"=Get TurnInPlaceInput() -> (TurnInPlaceInput:Fort Anim Input Turn In Place Structure)) -> (
    RootYawOffset:float=~
    TurnAnimPlayRate:float=~
    bWantsToTurn:bool=~
    bExitTurn:bool=~
    bUseTurnA:bool=~
    TurnACurveAlpha:float=~
    TurnBCurveAlpha:float=~
    bWantsToTurnInVehicle:bool=~
    bWantsToTurnAgain:bool=~
    bTurningLeft:bool=~
    ShufflePlayRate:float=~
    CrouchShufflePlayRate:float=~
    CrouchShuffleStartPosition:float=~
    CrouchShuffleSpeedScaling:float=~
    bTransition_LowerBody_Shuffle:bool=~
    bTransition_LowerBody_ShuffleADS:bool=~
    bPlayOutro:bool=~
   ) {is=title K2Node_BreakStruct}
   n5 = Modify Curve(
    SourcePose=Blend Poses by bool(
     BlendPose_0=Blend Poses by bool(BlendPose_0=@n3, BlendPose_1=@n2, bActiveValue=@n4.bTurningLeft) {blendType=Linear}
     BlendPose_1=Additive Identity Pose()
     BlendTime_0=0
     BlendTime_1=0.3
     bActiveValue=@n4.bUseTurnA
    ) {blendType=Linear, childUpateMode=ResetChildOnActivate}
    CurveValues_0=0
    CurveValues_1=0
    Alpha=@n4.TurnACurveAlpha
   ) {curveNames [TurnRotationAmount, TurnInPlaceAnim], curveValues [0,0]}
   n6 = Sequence Player(Sequence=@n1.TurnInPlaceLeftAdditiveOverride) {bLoopAnimation=false}
   n7 = Sequence Player(Sequence=@n1.TurnInPlaceRightAdditiveOverride) {bLoopAnimation=false}
   n8 = Modify Curve(
    SourcePose=Blend Poses by bool(
     BlendPose_0=Additive Identity Pose()
     BlendPose_1=Blend Poses by bool(BlendPose_0=@n6, BlendPose_1=@n7, bActiveValue=@n4.bTurningLeft) {blendType=Linear}
     BlendTime_0=0.3
     BlendTime_1=0
     bActiveValue=@n4.bUseTurnA
    ) {blendType=Linear, childUpateMode=ResetChildOnActivate}
    CurveValues_0=0
    CurveValues_1=0
    Alpha=@n4.TurnBCurveAlpha
   ) {curveNames [TurnRotationAmount, TurnInPlaceAnim], curveValues [0,0]}
   Output Animation Pose(Result=Apply Additive(Base=Use cached pose 'InSourcePoseLoco'(), Additive=Apply Additive(Base=@n5, Additive=@n8) {alphaBoolBlend {blendOption=Linear}}) {alphaBoolBlend {blendOption=Linear}})
  ]
 }
 Idle {
  path=LocomotionLayer.AnimGraphNode_StateMachine_0.MotionMatchingIdle.AnimStateNode_2.Idle
  nodes [Output Animation Pose(Result=Use cached pose 'InSourcePoseLoco'())]
 }
 AnimationTransitionGraph_0 {
  path=LocomotionLayer.AnimGraphNode_StateMachine_0.MotionMatchingIdle.AnimStateTransitionNode_0.AnimationTransitionGraph_0
  nodes [Result(bCanEnterTransition=Get bLocomotion_Idles_to_Turns() -> (bLocomotion_Idles_to_Turns:bool))]
 }
 Transition {
  path=LocomotionLayer.AnimGraphNode_StateMachine_0.MotionMatchingIdle.AnimStateTransitionNode_1.Transition
  nodes [Result(bCanEnterTransition=Break Fort Anim Input Turn in Place(FortAnimInput_TurnInPlace=Get TurnInPlaceInput()) -> (
     RootYawOffset=~
     TurnAnimPlayRate=~
     bWantsToTurn=~
     bExitTurn=~
     bUseTurnA=~
     TurnACurveAlpha=~
     TurnBCurveAlpha=~
     bWantsToTurnInVehicle=~
     bWantsToTurnAgain=~
     bTurningLeft=~
     ShufflePlayRate=~
     CrouchShufflePlayRate=~
     CrouchShuffleStartPosition=~
     CrouchShuffleSpeedScaling=~
     bTransition_LowerBody_Shuffle=~
     bTransition_LowerBody_ShuffleADS=~
     bPlayOutro=~
    ).bExitTurn)]
 }
 AnimationTransitionGraph_0 {path=LocomotionLayer.AnimGraphNode_StateMachine_0.MotionMatchingIdle.AnimStateTransitionNode_2.AnimationTransitionGraph_0, nodes [Result()]}
 LocomotionLayer_Additives {nodes [Output Pose() {layerGroup=LocomotionOverride}]}
 LocomotionLayer_BaseLocomotion {nodes [Output Pose(Result=Input Pose() {name=InPose_BaseLocomotion}) {layerGroup=LocomotionOverride}]}
 LocomotionLayer_FinalPose {nodes [Output Pose(Result=Input Pose() {name=InPoseFinalPose}) {layerGroup=LocomotionOverride}]}
 LocomotionLayer_FullBody {nodes [Output Pose(Result=Input Pose() {name=InPose_FullBody}) {layerGroup=LocomotionOverride}]}
 LocomotionLayer_FullLocomotion {nodes [Output Pose(Result=Input Pose() {name=InPose_FullLocomotion}) {layerGroup=LocomotionOverride}]}
 LocomotionLayer_GroundLandAdditive {nodes [
   Input Pose()
   Output Pose() {layerGroup=LocomotionOverride}
  ]}
 LocomotionLayer_InAir {nodes [
   Input Pose() {name=InPose_InAirSourcePose}
   Output Pose(Result=Input Pose() {name=InPose_InAir}) {layerGroup=LocomotionOverride}
  ]}
 LocomotionLayer_LowerBody {nodes [Output Pose(Result=Input Pose() {name=InPose_LowerBody}) {layerGroup=LocomotionOverride}]}
 LocomotionLayer_UpperBody {nodes [Output Pose(Result=Input Pose() {name=InPose_UpperBody}) {layerGroup=LocomotionOverride}]}
 LowLODLayer {nodes [
   n1 = Use cached pose 'LocoPoseLowLOD'()
   n2 = Cached Sequence Evaluator!(Sequence, ExplicitTime:float=0) -> (Pose) {bShouldLoop=false, cacheName=SourcePose}
   n3 = MovementModeLayerInterface - MovementMode_OffScreen!(InPoseOffscreen:Pose Link Structure=Blend Poses by bool(
     BlendPose_0=Profile Blend!(BasePose:Pose Link Structure=@n2, BlendPose:Pose Link Structure=@n1, BlendWeight:float=1) -> (Pose:Pose Link Structure) {
      is=title AnimGraphNode_BlendProfileLayeredBlend
      blendProfileAsset=/Game/Characters/Player/BlendMasks/LowerBody_BlendMask
      blendPose {linkId=16842751}
     }
     BlendPose_1=Use cached pose 'LocoPoseLowLOD'()
     bActiveValue=Get bApplyOffscreenSourcePose() -> (bApplyOffscreenSourcePose:bool)
    ) {transitionType=Inertialization}) -> (Pose:Pose Link Structure) {
    is=title AnimGraphNode_LinkedAnimLayer
    tag=MovementMode_Offscreen
    interface=&MovementModeLayerInterface_C
    layer=MovementMode_OffScreen
   }
   Output Pose(Result=Dead Blending(Source=@n3))
   n4 = Get OffscreenSequence() -> (OffscreenSequence:Anim Sequence Object Reference)
   n5 = Get Skydiving()
   Save cached pose 'LocoPoseLowLOD'(Pose=Blend Poses by bool(
     BlendPose_0=Sequence Player 'HangGlider_Glide'() {sequence=/Game/Animation/Game/MainPlayer/Skydive/ParaGlide/HangGlider/Medium/Female/HangGlider_Glide}
     BlendPose_1=Sequence Player!(Sequence=@n4, PlayRate:float=1) -> (Pose)
     bActiveValue=Break Fort Anim Input Skydiving(FortAnimInput_Skydiving=@n5) -> (
      ParaGilderRootOffset=~
      ParaGliderRootRotationOverride=~
      bUseParaGlideRootModifier=~
      bIsSkydivingFromLaunchPad=~
      bIsSkydivingFromBus=~
      bIsInVortex=~
      bIsDiving=~
      bIsParachuteOpen=~
      bIsSkydiving=~
      bIsParachuteLeaning=~
      bIsSkydiveLeaning=~
      bIsLeaning=~
      bIsSkydiveDiveMode=~
      bPlaySkydiveDrift=~
      bSkydiveDriftDelayActive=~
      bSkydiveDriftAnimAllowed=~
      bIsGliderRight=~
      bIsGliderCenter=~
      bIsGliderLeft=~
      bIsGliderForward=~
      bIsGliderBack=~
      bHasDeployedAndIsStandingUpright=~
      LocalAccelForward=~
      LocalAccelRight=~
      LocalVelocityRight=~
      SkydiveAimPitch=~
      SkydiveAimYaw=~
      DeployChuteAnimRate=~
      SkydiveDriftAnimRate=~
      SkydiveDriftAnimRateCurrent=~
      SkydiveFidgetAnimRate=~
      SkydiveFidgetAnimRateCurrent=~
      SkydiveAdditiveAlpha=~
      SkydiveDriftDelay=~
     ).bIsParachuteOpen
    )) {cacheName=LocoPoseLowLOD}
  ]}
 LowerBodyBlendLayer {nodes [
   AnimGraphNode_BlendStack_0 = Blend Stack!(
    AnimationAsset:Animation Asset Object Reference=/Game/Animation/Game/MainPlayer/Poses/Medium/Male/AssaultRifle/AR_NonTargeted_Pose_CMM
    BlendTime:float=0.2
   ) -> (Pose:Pose Link Structure) {
    is=title AnimGraphNode_BlendStack
    node {bUseInertialBlend=true, blendOption=Linear, updateFunction {functionName=OnLowerBodyBlendStackUpdate}}
    updateFunction {bSelfContext=true, memberName=OnLowerBodyBlendStackUpdate}
   }
   n1 = Layered blend per bone(BasePose=@AnimGraphNode_BlendStack_0, BlendPoses_0=Use cached pose 'UpperBodyBlendPose'()) {
    blendMode=BlendMask
    curveBlendOption=UseMaxValue
    blendMasks [&Fortnite_M_Avg_Player_Skeleton:FirstPersonAdditives]
    layerSetup []
   }
   Output Pose(Result=@n1) {layerGroup=None}
   Save cached pose 'UpperBodyInputPose'(Pose=Input Pose() {name=BasePose}) {cacheName=UpperBodyInputPose, cachePoseName=UpperBodyInputPose}
   Save cached pose 'LowerBodyBlendInputPose'(Pose=Input Pose() {name=LowerBody}) {cacheName=LowerBodyBlendInputPose}
   n2 = Use cached pose 'UpperBodyInputPose'()
   n3 = Get AimYaw() -> (AimYaw:float)
   Save cached pose 'UpperBodyBlendPose'(Pose=Blend Poses by bool(
     BlendPose_0=Rotate Root Bone!(BasePose:Pose Link Structure=@n2, Pitch:float=0, Yaw:float=@n3, MeshToComponent:Rotator="0,0,0") -> (Pose:Pose Link Structure) {is=title AnimGraphNode_RotateRootBone}
     BlendPose_1=Use cached pose 'UpperBodyInputPose'()
     bActiveValue=Get bApplyAimYawToRoot() -> (bApplyAimYawToRoot:bool)
    ) {blendType=Linear}) {cacheName=UpperBodyBlendPose}
  ]}
 AnimationBlendStackGraph_0 {
  path=LowerBodyBlendLayer.AnimGraphNode_BlendStack_0.AnimationBlendStackGraph_0
  nodes [
   Blend Stack Input!() -> (Pose:Pose Link Structure) {is=title AnimGraphNode_BlendStackInput}
   Output Pose!(Result:Pose Link Structure=Profile Blend!(
     BasePose=Use cached pose 'UpperBodyBlendPose'()
     BlendPose=Use cached pose 'LowerBodyBlendInputPose'()
     BlendProfileAsset:Blend Profile Standalone Object Reference
     RotationBlendSpace:EBlendProfileRotationBlendSpace Enum=MeshSpace
     BlendWeight=1
    ) -> (Pose) {bCustomCurveBlending=true, curveBlendingOption=UseMaxValue, rotationBlendSpace=MeshSpace, blendPose {linkId=16842751}}) {is=title AnimGraphNode_BlendStackResult}
  ]
 }
 MenuScreenDispatcher {nodes [MenuScreenDispatcher!() -> (then, CharPart:Fort Player Anim Instance Object Reference) {is=title K2Node_FunctionEntry}]}
 MovementMode_AimOffset {nodes [
   Input Pose() {name=InSourcePosePreAimOffsets}
   Output Pose(Result=Input Pose() {name=InAimOffsetsPose}) {is=title AnimGraphNode_Root, layerGroup=MovementMode}
  ]}
 MovementMode_BaseLocomotion {nodes [Output Pose(Result=Input Pose() {name=InPoseBaseLocomotion}) {layerGroup=MovementMode}]}
 MovementMode_FinalPoseOverride {nodes [Output Pose(Result=Input Pose() {name=InFinalPose}) {layerGroup=MovementMode}]}
 MovementMode_FullBodyOverride {nodes [Output Pose(Result=Input Pose() {name=InputPoseFullBody}) {layerGroup=MovementMode}]}
 MovementMode_InAirOverride {nodes [
   Input Pose() {name=InSource_Pose}
   Output Pose(Result=Input Pose() {name=InAirPose}) {layerGroup=MovementMode}
  ]}
 MovementMode_LocomotionAdditives {nodes [Output Pose(Result=Input Pose() {name=InLocomotionAdditivesPose}) {layerGroup=MovementMode}]}
 MovementMode_LowerBodyOverride {nodes [
   Input Pose() {name=InSourcePose}
   Output Pose(Result=Input Pose() {name=InLowerBodyPose}) {layerGroup=MovementMode}
  ]}
 MovementMode_OffScreen {nodes [Output Pose(Result=Input Pose() {name=InPoseOffscreen}) {layerGroup=MovementMode}]}
 MovementMode_SourcePose {nodes [Output Pose(Result=Input Pose() {name=InSourcePoseBase}) {layerGroup=MovementMode}]}
 PreLegIK {nodes [
   Input Pose()
   Output Pose() {layerGroup=PlayerOverride}
  ]}
 ProceduralAdditivePose {nodes [
   Input Pose() {name=InSourceLocomotionPose}
   Input Pose() {name=InMotionMatchingPose}
   Input Pose() {name=InSourcePose}
   Output Pose() {layerGroup=None}
  ]}
 ProceduralLayeredPose {nodes [
   Input Pose() {name=InLocomotionPose}
   Output Pose() {layerGroup=None}
  ]}
 ProceduralLayeringStep1 {nodes [
   n1 = ProceduralLayeringAnimInterface - ProceduralLayeredPose!(InLocomotionPose:Pose Link Structure=Input Pose() {name=InPoseProceduralLayering}) -> (Pose:Pose Link Structure) {
    is=title AnimGraphNode_LinkedAnimLayer
    tag=ProceduralLayering
    instanceClass=/Game/Characters/Player/AnimationBlueprints/LinkedAnimGraphs/Player_ProceduralLayering_AnimBP.Player_ProceduralLayering_AnimBP_C
    interface=/Game/Characters/Player/Male/Male_Avg_Base/AnimLayerInterface/ProceduralLayeringAnimInterface.ProceduralLayeringAnimInterface_C
    layer=ProceduralLayeredPose
   }
   Output Pose(Result=Linked Graph Gate(Source=@n1)) {layerGroup=None}
  ]}
 ProceduralLayeringStep2 {nodes [
   n1 = Use cached pose 'InSourceLocomotionPose'()
   n2 = ProceduralLayeringAnimInterface - ProceduralAdditivePose!(
    InSourceLocomotionPose:Pose Link Structure=Use cached pose 'InSourceLocomotionPose'()
    InMotionMatchingPose:Pose Link Structure=Input Pose() {name=In_MotionMatchingPose}
    InSourcePose:Pose Link Structure=Input Pose() {name=In_SourcePose}
   ) -> (Pose:Pose Link Structure) {
    is=title AnimGraphNode_LinkedAnimLayer
    tag=ProceduralLayeringStep2
    instanceClass=&Player_ProceduralLayering_AnimBP_C
    interface=&ProceduralLayeringAnimInterface_C
    layer=ProceduralAdditivePose
   }
   n3 = Linked Graph Gate(Source=@n2)
   n4 = Get bEnableProceduralLayeringStep2() -> (bEnableProceduralLayeringStep2:bool)
   n5 = Stride Warping!(
    ComponentPose:Component Space Pose Link Structure=Local To Component(LocalPose=Blend Poses by bool(BlendPose_0=@n3, BlendPose_1=@n1, bActiveValue=@n4))
    Alpha:float=Get StrideWarpingAlpha() -> (StrideWarpingAlpha:float)
    LocomotionSpeed:float=0
    StrideScaleModifier:Input Clamp Constants Structure=(bClampResult=True,bInterpResult=True,ClampMin=0.8,ClampMax=1.2,InterpSpeedIncreasing=10,InterpSpeedDecreasing=10)
   ) -> (Pose:Component Space Pose Link Structure) {
    is=title AnimGraphNode_StrideWarping
    bClampIKUsingFKLimits=false
    bDisableIfMissingRootMotion=false
    mode=Graph
    alphaBoolBlend {blendOption=Linear}
    footDefinitions [
     {fKFootBone {boneName=foot_l}, iKFootBone {boneName=ik_foot_l}, thighBone {boneName=thigh_l}}
     {fKFootBone {boneName=foot_r}, iKFootBone {boneName=ik_foot_r}, thighBone {boneName=thigh_r}}
    ]
    iKFootRootBone {boneName=ik_foot_root}
    pelvisBone {boneName=pelvis}
    strideScaleModifier {bClampResult=true, bInterpResult=true, clampMax=1.2, clampMin=0.8}
   }
   n6 = Component To Local(ComponentPose=@n5)
   n7 = Use cached pose 'InSourceLocomotionPose'()
   n8 = Get bUseProceduralLayering() -> (bUseProceduralLayering:bool)
   Output Pose(Result=Profiler Stat Event(Source=Blend Poses by bool(BlendPose_0=@n6, BlendPose_1=@n7, bActiveValue=@n8) {blendType=Linear}) {profilerEventName=MainPlayerAnimBP_ProceduralLayeringStep2}) {layerGroup=None}
   Save cached pose 'InSourceLocomotionPose'(Pose=Input Pose() {name=InSourceLocomotionPose}) {cacheName=InSourceLocomotionPose}
  ]}
 RigidBodyPBAProto {nodes [
   n1 = Player_RigidBodyPBA_AnimBP_Template!(InPose:Pose Link Structure=Use cached pose 'InputPoseRigidBodyPBA'()) -> (Pose:Pose Link Structure) {
    is=title AnimGraphNode_LinkedAnimGraph
    tag=RigidBodyPBA
    instanceClass=/Game/Characters/Player/AnimationBlueprints/LinkedAnimGraphTemplates/Player_RigidBodyPBA_AnimBP_Template.Player_RigidBodyPBA_AnimBP_Template_C
   }
   Output Pose(Result=Blend Poses by bool(BlendPose_0=@n1, BlendPose_1=Use cached pose 'InputPoseRigidBodyPBA'(), bActiveValue=Get bEnableRigidBodyPBA() -> (bEnableRigidBodyPBA:bool)) {blendType=Linear}) {layerGroup=None}
   Save cached pose 'InputPoseRigidBodyPBA'(Pose=Input Pose() {name=InPoseRigidBodyPBA}) {cacheName=InputPoseRigidBodyPBA}
  ]}
 SecondMovementMode_FinalPoseOverride {nodes [Output Pose(Result=Input Pose() {name=InFinalPose}) {layerGroup=SecondMovementMode}]}
 SecondaryItemAimOffsets {nodes [
   Input Pose() {name=InPosePreAimOffsets}
   Output Pose(Result=Input Pose!() -> (Pose, AimOffsetAlpha:double, Yaw:double, Pitch:double) {name=InPoseAimOffsets}) {layerGroup=SecondaryItemLayer}
  ]}
 SourcePoseAdjustmentsLayer {nodes [
   Output Pose(Result=UpperBodyAdjustmentsLayer!(
     InPoseUpperBodyAdjustments:Pose Link Structure=Use cached pose 'InSourcePose'()
     InPosePostUpperBodySlot:Pose Link Structure=Use cached pose 'PostUpperBodySlot'()
    ) -> (Pose:Pose Link Structure) {is=title AnimGraphNode_LinkedAnimLayer, layer=UpperBodyAdjustmentsLayer}) {layerGroup=None}
   Save cached pose 'InSourcePose'(Pose=Input Pose!() -> (Pose) {name=InSourcePose}) {cacheName=InSourcePose}
   n1 = Use cached pose 'PreCrouchUpperBodySlot'()
   Save cached pose 'PostUpperBodySlot'(Pose=Blend Poses by bool(
     BlendPose_0=Slot 'CrouchUpperBody'(Source=@n1) {slotName=CrouchUpperBody}
     BlendPose_1=Use cached pose 'PreCrouchUpperBodySlot'()
     bActiveValue=Get bShouldUseCrouchUpperBodySlot() -> (bShouldUseCrouchUpperBodySlot:bool)
    ) {blendType=Linear}) {cacheName=PostUpperBodySlot}
   n2 = Use cached pose 'InSourcePose'()
   n3 = Use cached pose 'InSourcePose'()
   Save cached pose 'PreSwimUpperBodySlot'(Pose=Blend Poses by bool(
     BlendPose_0=Slot 'FP_UpperBody'(Source=@n3) {slotName=FP_UpperBody}
     BlendPose_1=Slot 'UpperBody'(Source=@n2) {slotName=UpperBody}
     BlendTime_0=0
     BlendTime_1=0
     bActiveValue=Get bUseFirstPersonUpperBodySlot() -> (bUseFirstPersonUpperBodySlot:bool)
    ) {blendType=Linear}) {cacheName=PreSwimUpperBodySlot}
   n4 = Slot 'UpperBodySwim'(Source=Use cached pose 'PreSwimUpperBodySlot'()) {slotName=UpperBodySwim}
   n5 = Use cached pose 'PreSwimUpperBodySlot'()
   n6 = Get bIsSurfaceSwimming()
   Save cached pose 'PreCrouchUpperBodySlot'(Pose=Slot 'OverrideUpperBody'(Source=Blend Poses by bool(BlendPose_0=@n4, BlendPose_1=@n5, bActiveValue=@n6) {blendType=Linear}) {slotName=OverrideUpperBody}) {cacheName=PreCrouchUpperBodySlot}
  ]}
 SourcePoseLayer {nodes [
   n1 = Layered blend per bone(
    BasePose=Use cached pose 'PostSourcePoseAdjLayerPose'()
    BlendPoses_0=Use cached pose 'SourcePosePreUpperBodySlot'()
    BlendWeights_0=Get UpperBodySlotLowerBodyMaskAlpha() -> (UpperBodySlotLowerBodyMaskAlpha:float)
   ) {
    bMeshSpaceRotationBlend=true
    bMeshSpaceScaleBlend=true
    blendMode=BlendMask
    curveBlendOption=UseBasePose
    blendMasks [&Fortnite_M_Avg_Player_Skeleton:LowerBodyNoFeathering]
    layerSetup []
   }
   n2 = Layered blend per bone(
    BasePose=Use cached pose 'PostSourcePoseAdjLayerPose'()
    BlendPoses_0=Use cached pose 'SourcePosePreUpperBodySlot'()
    BlendWeights_0=Get UpperBodySlotLowerBodyMaskAlpha()
   ) {
    bMeshSpaceRotationBlend=true
    bMeshSpaceScaleBlend=true
    blendMode=BlendMask
    curveBlendOption=UseBasePose
    blendMasks [&Fortnite_M_Avg_Player_Skeleton:LowerBody]
    layerSetup []
   }
   Output Pose(Result=Blend Poses by bool(BlendPose_0=@n1, BlendPose_1=@n2, bActiveValue=Get bUseNoFeatheringLowerBodyBlendMask() -> (bUseNoFeatheringLowerBodyBlendMask:bool)) {blendType=Linear}) {layerGroup=None}
   n3 = ItemAnimLayerInterface - ItemAimOffset!(
    InPose_AimOffset:Pose Link Structure=UpperBodyAimingOffsetsLayer!(InPose:Pose Link Structure=Use cached pose 'SourcePosePreAimOffsets'()) -> (Pose:Pose Link Structure) {is=title AnimGraphNode_LinkedAnimLayer, layer=UpperBodyAimingOffsetsLayer}
    InPose_PreAimOffsets:Pose Link Structure=Use cached pose 'SourcePosePreAimOffsets'()
    AimOffsetAlpha:double=0
    Yaw:double=0
    Pitch:double=0
   ) -> (Pose:Pose Link Structure) {
    is=title AnimGraphNode_LinkedAnimLayer
    interface=&ItemAnimLayerInterface_C
    layer=ItemAimOffset
   }
   n4 = SecondaryItemAnimLayerInterface - SecondaryItemAimOffsets!(
    InPoseAimOffsets:Pose Link Structure=@n3
    InPosePreAimOffsets:Pose Link Structure=Use cached pose 'SourcePosePreAimOffsets'()
    AimOffsetAlpha:double=0
    Yaw:double=0
    Pitch:double=0
   ) -> (Pose:Pose Link Structure) {
    is=title AnimGraphNode_LinkedAnimLayer
    interface=/Game/Characters/Player/Male/Male_Avg_Base/AnimLayerInterface/SecondaryItemAnimLayerInterface.SecondaryItemAnimLayerInterface_C
    layer=SecondaryItemAimOffsets
   }
   n5 = MovementModeLayerInterface - MovementMode_AimOffset!(InAimOffsetsPose:Pose Link Structure=@n4, InSourcePosePreAimOffsets:Pose Link Structure=Use cached pose 'SourcePosePreAimOffsets'()) -> (Pose:Pose Link Structure) {
    is=title AnimGraphNode_LinkedAnimLayer
    interface=&MovementModeLayerInterface_C
    layer=MovementMode_AimOffset
   }
   Save cached pose 'SourcePosePreUpperBodySlot'(Pose=Blend Poses by bool(
     BlendPose_0=@n5
     BlendPose_1=Use cached pose 'SourcePosePreAimOffsets'()
     BlendTime_0=0.3
     BlendTime_1=0.3
     bActiveValue=Get bEnableAimOffsets()
    ) {blendType=QuadraticInOut}) {cacheName=SourcePosePreUpperBodySlot}
   n6 = PlayerLocomotionLayerInterface - LocomotionLayer_UpperBody!(InPose_UpperBody:Pose Link Structure=Use cached pose 'SourcePosePreUpperBodySlot'()) -> (Pose:Pose Link Structure) {
    is=title AnimGraphNode_LinkedAnimLayer
    interface=&PlayerLocomotionLayerInterface_C
    layer=LocomotionLayer_UpperBody
   }
   AnimGraphNode_Layer_4 = ItemAnimLayerInterface - ItemUpperBody!(InPoseUpperBody:Pose Link Structure=@n6, InputParam:Fort Anim Input Adjusted Aim Structure=Get AdjustedAim()) -> (Pose:Pose Link Structure) {
    is=title AnimGraphNode_LinkedAnimLayer
    interface=&ItemAnimLayerInterface_C
    layer=ItemUpperBody
   }
   n7 = PlayerAnimOverrideLayerInterface - UpperBody!(InPoseUpperBody:Pose Link Structure=@AnimGraphNode_Layer_4) -> (Pose:Pose Link Structure) {
    is=title AnimGraphNode_LinkedAnimLayer
    interface=&PlayerAnimOverrideLayerInterface_C
    layer=UpperBody
   }
   Save cached pose 'PostSourcePoseAdjLayerPose'(Pose=SourcePoseAdjustmentsLayer!(InSourcePose:Pose Link Structure=@n7) -> (Pose:Pose Link Structure) {is=title AnimGraphNode_LinkedAnimLayer, layer=SourcePoseAdjustmentsLayer}) {cacheName=PostSourcePoseAdjLayerPose}
   Save cached pose 'SourcePoseBase'(Pose=Input Pose() {name=InSourcePoseInput}) {cacheName=SourcePoseBase, cachePoseName=SourcePoseBase}
   n8 = "Blend Poses (EPreAimOffsetSlot)"!(
    BlendPose_0:Pose Link Structure=Slot 'PreAimOffsets'(Source=Use cached pose 'SourcePoseBase'()) {slotName=PreAimOffsets}
    BlendPose_1:Pose Link Structure=Slot 'PreAimOffsets_ADS'(Source=Use cached pose 'SourcePoseBase'()) {slotName=PreAimOffsets_ADS}
    BlendPose_2:Pose Link Structure=Slot 'FP_PreAimOffsets'(Source=Use cached pose 'SourcePoseBase'()) {slotName=FP_PreAimOffsets}
    BlendPose_3:Pose Link Structure=Slot 'FP_PreAimOffsets_ADS'(Source=Use cached pose 'SourcePoseBase'()) {slotName=FP_PreAimOffsets_ADS}
    BlendTime_0:float=0.1
    BlendTime_1:float=0.1
    BlendTime_2:float=0.1
    BlendTime_3:float=0.1
    ActiveEnumValue:Byte=Get PreAimOffsetSlot() -> (PreAimOffsetSlot:EPreAimOffsetSlot Enum=PreAimOffsets)
   ) -> (Pose:Pose Link Structure) {
    is=title AnimGraphNode_BlendListByEnum
    blendPose [{linkId=-1, sourceLinkId=-1}, {linkId=-1, sourceLinkId=-1}, {linkId=-1, sourceLinkId=-1}, {linkId=-1, sourceLinkId=-1}]
    blendTime [0.1,0.1,0.1,0.1]
   }
   n9 = ItemAnimLayerInterface - ItemPreAimOffsets!(InPosePreAimOffsets:Pose Link Structure=Slot 'PreAimOffsetsItemLayerOffHand'(Source=ADSReload!(InPose_ADSReload:Pose Link Structure=@n8) -> (Pose:Pose Link Structure) {is=title AnimGraphNode_LinkedAnimLayer, layer=ADSReload}) {slotName=PreAimOffsetsItemLayerOffHand}) -> (Pose:Pose Link Structure) {
    is=title AnimGraphNode_LinkedAnimLayer
    tag=ItemPreAimOffsets
    interface=&ItemAnimLayerInterface_C
    layer=ItemPreAimOffsets
   }
   n10 = Layered blend per bone(BasePose=@n9, BlendPoses_0=Use cached pose 'SourcePoseBase'(), BlendWeights_0=Get UpperBodySlotLowerBodyMaskAlpha()) {
    bMeshSpaceRotationBlend=true
    bMeshSpaceScaleBlend=true
    blendMode=BlendMask
    curveBlendOption=UseMaxValue
    blendMasks [&Fortnite_M_Avg_Player_Skeleton:LowerBody]
    layerSetup []
   }
   Save cached pose 'PreMeshRotationUpperBodySlot'(Pose=@n10) {cacheName=PreMeshRotationUpperBodySlot}
   n11 = Blend Poses by bool(
    BlendPose_0=Slot 'FP_MeshRotationUpperBody'(Source=Use cached pose 'PreMeshRotationUpperBodySlot'()) {slotName=FP_MeshRotationUpperBody}
    BlendPose_1=Slot 'MeshRotationUpperBody'(Source=Use cached pose 'PreMeshRotationUpperBodySlot'()) {slotName=MeshRotationUpperBody}
    BlendTime_0=0
    BlendTime_1=0
    bActiveValue=Get bUseFirstPersonMeshRotationUpperBodySlot() -> (bUseFirstPersonMeshRotationUpperBodySlot:bool)
   ) {blendType=Linear}
   Save cached pose 'SourcePosePreAimOffsets'(Pose=UpperBodyAdditivesLayer!(InPoseSourcePosePreAimOffsets:Pose Link Structure=Slot 'MeshRotationUpperBodyOffhand'(Source=@n11) {slotName=MeshRotationUpperBodyOffhand}) -> (Pose:Pose Link Structure) {is=title AnimGraphNode_LinkedAnimLayer, layer=UpperBodyAdditivesLayer}) {cacheName=SourcePosePreAimOffsets}
  ]}
 UpperBody {nodes [Output Pose(Result=Input Pose() {name=InPoseUpperBody}) {layerGroup=PlayerOverride}]}
 UpperBodyAdditivesLayer {nodes [
   n1 = Blendspace Evaluator!(
    X:float=Get ADSLeanWeight() -> (ADSLeanWeight:float)
    Y:float=0
    BlendSpace:Blend Space Object Reference=Get ADSLeanBlendSpace() -> (ADSLeanBlendSpace:Blend Space Object Reference)
    NormalizedTime:float=0
   ) -> (Pose:Pose Link Structure) {is=title AnimGraphNode_BlendSpaceEvaluator}
   n2 = Input Pose() {name=InPoseSourcePosePreAimOffsets}
   n3 = Get ADSLeanAdditiveAlpha() -> (ADSLeanAdditiveAlpha:float)
   n4 = Break Fort Anim Input Weapon Stance Transition!(FortAnimInput_WeaponStanceTransition:"Fort Anim Input Weapon Stance Transition Structure (by ref)"=Get WeaponStanceTransition() -> (WeaponStanceTransition:Fort Anim Input Weapon Stance Transition Structure)) -> (
    Additive:Anim Sequence Object Reference
    Alpha:float=~
    CurrentStance:EFortWeaponStance Enum
    PreviousStance:EFortWeaponStance Enum
    BlendTime:float=~
    BlendCurve:EAlphaBlendOption Enum
   ) {is=title K2Node_BreakStruct}
   AnimGraphNode_BlendStack_0 = Blend Stack!(AnimationAsset=@n4, BlendTime=@n4.BlendTime, BlendOption:EAlphaBlendOption Enum=@n4.BlendCurve) -> (Pose) {tag=WeaponStanceTransitionBlendStack, bLoop=false, bUseInertialBlend=true, inertialBlendNodeTag=WeaponStanceDeadBlend}
   n5 = Remove Curve!(SourcePose:Pose Link Structure=@AnimGraphNode_BlendStack_0, bEnabled:bool=True) -> (Pose:Pose Link Structure) {is=title AnimGraphNode_RemoveCurve, removeMode=KeepOnlySpecifiedCurves}
   Output Pose(Result=Apply Mesh Space Additive!(
     Base:Pose Link Structure=Apply Additive(Base=@n2, Additive=@n1, Alpha=@n3) {alphaBoolBlend {blendOption=Linear}}
     Additive:Pose Link Structure=Dead Blending(Source=@n5) {tag=WeaponStanceDeadBlend}
     Alpha:float=@n4.Alpha
    ) -> (Pose:Pose Link Structure) {is=title AnimGraphNode_ApplyMeshSpaceAdditive}) {layerGroup=None}
  ]}
 AnimationBlendStackGraph_0 {
  path=UpperBodyAdditivesLayer.AnimGraphNode_BlendStack_0.AnimationBlendStackGraph_0
  nodes [Output Pose(Result=Blend Stack Input()) {is=title AnimGraphNode_BlendStackResult}]
 }
 UpperBodyAdjustmentsLayer {nodes [
   n1 = Layered blend per bone(
    BasePose=Use cached pose 'PostAdditiveSlot'()
    BlendPoses_0=Use cached pose 'InUpperBodyAdjustments'()
    BlendWeights_0=Get LeftArmMaskWeight() -> (LeftArmMaskWeight:float)
   ) {
    curveBlendOption=UseBasePose
    layerSetup [{branchFilters [{blendDepth=3, boneName=clavicle_l}, {boneName=ik_hand_root}, {boneName=weapon_l}]}]
   }
   n2 = Layered blend per bone(BasePose=@n1, BlendPoses_0=Use cached pose 'InUpperBodyAdjustments'(), BlendWeights_0=Get RightArmMaskWeight() -> (RightArmMaskWeight:float)) {
    curveBlendOption=UseBasePose
    layerSetup [{branchFilters [{blendDepth=3, boneName=clavicle_r}, {boneName=ik_hand_root}, {boneName=weapon_r}]}]
   }
   n3 = Sequence Player!(Sequence=Get ConsumableOffsetPose() -> (ConsumableOffsetPose:Anim Sequence Object Reference)) -> (Pose)
   n4 = Break Fort Anim Input Weapon Additive Anim Asset!(FortAnimInput_WeaponAdditiveAnimAsset:"Fort Anim Input Weapon Additive Anim Asset Structure (by ref)"=Get CurrentWeaponAdditiveAnimSet() -> (CurrentWeaponAdditiveAnimSet:Fort Anim Input Weapon Additive Anim Asset Structure)) -> (
    AdditivePoseOffset:Anim Sequence Object Reference
    TargetingAdditivePoseOffset:Anim Sequence Object Reference
    NonTargetedAdditivePoseOffset:Anim Sequence Object Reference
    RelaxedAdditivePoseOffset:Anim Sequence Object Reference
    RelaxedWhileSprintingAdditiveAnim:Anim Sequence Object Reference
    RelaxedWhileCrouchSprintingAdditiveAnim:Anim Sequence Object Reference
   ) {is=title K2Node_BreakStruct}
   n5 = Sequence Player(Sequence=@n4.RelaxedAdditivePoseOffset)
   n6 = Sequence Player(Sequence=@n4.NonTargetedAdditivePoseOffset)
   n7 = Get bIsRelaxedLevel1() -> (bIsRelaxedLevel1:bool)
   n8 = Apply Additive!(
    Base=Apply Additive(
     Base=Apply Additive(Base=@n2, Additive=@n3, AlphaCurveName=DisableConsumableOffsetPose) {alphaInputType=Curve, alphaScaleBiasClamp {bClampResult=true, bias=1, scale=-1}}
     Additive=Sequence Player(Sequence=@n4)
     AlphaCurveName=DisableWeaponAdditivePose
    ) {alphaInputType=Curve, alphaScaleBiasClamp {bClampResult=true, bias=1, scale=-1}}
    Additive=Blend Poses by bool(
     BlendPose_0=Sequence Player(Sequence=@n4.TargetingAdditivePoseOffset)
     BlendPose_1=Blend Poses by bool(BlendPose_0=@n5, BlendPose_1=@n6, BlendTime_0=.4, BlendTime_1=0, bActiveValue=@n7) {blendType=Linear}
     BlendTime_0=Get NonADSToADSBlendTime() -> (NonADSToADSBlendTime:float)
     BlendTime_1=Get ADSToNonADSBlendTime() -> (ADSToNonADSBlendTime:float)
     bActiveValue=Get bPlayUpperBodyTargeting() -> (bPlayUpperBodyTargeting:bool)
    ) {blendType=Linear}
    AlphaCurveName:Name=DisableWeaponAdditivePose
   ) -> (Pose) {alphaInputType=Curve, alphaScaleBiasClamp {bClampResult=true, bias=1, scale=-1}}
   n9 = Sequence Player(Sequence=Get WeaponModLeftHandPoseOffset() -> (WeaponModLeftHandPoseOffset:Anim Sequence Object Reference))
   Output Pose(Result=Blend Poses by bool(
     BlendPose_0=Use cached pose 'PostAdditiveSlot'()
     BlendPose_1=Apply Additive(Base=@n8, Additive=@n9, AlphaCurveName=DisableWeaponModLeftHandOffset) {alphaInputType=Curve, alphaScaleBiasClamp {bClampResult=true, bias=1, scale=-1}}
     BlendTime_0=0
     BlendTime_1=0
     bActiveValue=Get bCanDisableUpperBodyAdjustments() -> (bCanDisableUpperBodyAdjustments:bool)
    ) {blendType=Linear}) {is=title AnimGraphNode_Root, layerGroup=None}
   Save cached pose 'InUpperBodyAdjustments'(Pose=Input Pose() {name=InPoseUpperBodyAdjustments}) {cacheName=InUpperBodyAdjustments}
   Save cached pose 'InPostUpperBodySlot'(Pose=Input Pose() {name=InPosePostUpperBodySlot}) {cacheName=InPostUpperBodySlot}
   n10 = "Blend Poses (EFortCustomGender)"(
    BlendPose_0=Use cached pose 'InPostUpperBodySlot'()
    BlendPose_1=Curve Filter Slot 'AdditiveCorrectiveUpperBody'!(Source:Pose Link Structure=Use cached pose 'InPostUpperBodySlot'()) -> (Pose:Pose Link Structure) {is=title FortAnimGraphNode_CurveFilterSlot, filterMode=Closed, montageSlotName=AdditiveCorrectiveUpperBody}
    BlendTime_0=0.1
    BlendTime_1=0.1
    ActiveEnumValue=Get Gender() -> (Gender=Invalid)
   ) {blendType=Linear, blendPose [{linkId=-1, sourceLinkId=-1}, {linkId=-1, sourceLinkId=-1}], blendTime [0.1,0.1]}
   Save cached pose 'PostAdditiveSlot'(Pose=@n10) {cacheName=PostAdditiveSlot}
  ]}
 UpperBodyAimingOffsetsLayer {nodes [
   n1 = Get UpperBodyAimOffsetAlpha() -> (UpperBodyAimOffsetAlpha:float)
   n2 = Get bEnableAimOffsetsRootYawCompensation()
   n3 = Break Fort Anim Input Player Anim Asset(FortAnimInput_PlayerAnimAsset=Get CurrentAnimSet()) -> (
    bPlayUpperBodySlotOnFullBodyInAir=~
    bDisableFullBodyAimOffsetDuringMelee=~
    bShouldApplyAimOffsetFullBody=~
    FullBodyAimOffsetAlpha=~
    UpperBodyAimOffsetAlpha=~
    bOverrideDisableArmsHeadAdditive=~
    DisableArmsHeadAdditiveOverride=~
    bOverridePitchAndYawOffsets=~
    bOverrideHandIKRetargetingWeight=~
    HandIKRetargetingWeight=~
    bForceKneeSliding=~
   )
   n4 = Break Fort Anim Input Adjusted Aim(FortAnimInput_AdjustedAim=Get AdjustedAim()) -> (ResultingYaw=~, ResultingPitch=~)
   n5 = Blend Poses by bool(
    BlendPose_0="Compensated Aim Offset (None)"!(
     BasePose:Pose Link Structure=Use cached pose 'AimingOffsetBasePose'()
     BlendSpace:Blend Space Object Reference=@n3.RelaxedAimOffset
     X:float=@n4
     Y:float=@n4.ResultingPitch
     bApplyYawCompensation:bool=@n2
     bApplyAdditiveInRootSpace:bool=False
     Alpha:float=@n1
    ) -> (Pose:Pose Link Structure) {is=title FortAnimGraphNode_AimOffsets, yawCompensationScale=-1, compensationBone {boneName=Root}}
    BlendPose_1="Compensated Aim Offset (None)"(
     BasePose=Use cached pose 'AimingOffsetBasePose'()
     BlendSpace=@n3.NonTargetingAimOffset
     X=@n4
     Y=@n4.ResultingPitch
     bApplyYawCompensation=@n2
     bApplyAdditiveInRootSpace=False
     Alpha=@n1
    ) {yawCompensationScale=-1, compensationBone {boneName=Root}}
    BlendTime_0=.4
    BlendTime_1=Get WeaponFireBlendTime() -> (WeaponFireBlendTime:float)
    bActiveValue=Get bIsRelaxedLevel1()
   ) {blendType=Linear}
   n6 = "Compensated Aim Offset (None)"(
    BasePose=Use cached pose 'AimingOffsetBasePose'()
    BlendSpace=@n3.TargetingAimOffset
    X=@n4
    Y=@n4.ResultingPitch
    bApplyYawCompensation=@n2
    bApplyAdditiveInRootSpace=False
    Alpha=@n1
   ) {yawCompensationScale=-1, compensationBone {boneName=Root}}
   n7 = Get NonADSToADSBlendTime()
   n8 = Get bPlayUpperBodyTargeting()
   n9 = Get ADSToNonADSBlendTime()
   Output Pose(Result=AimOffset Player 'M_NoWep_Relaxed_AO_120'!(
     X:float=Get HeadAimYaw() -> (HeadAimYaw:float)
     Y:float=Get HeadAimPitch() -> (HeadAimPitch:float)
     BasePose:Pose Link Structure=Blend Poses by bool(BlendPose_0=@n6, BlendPose_1=@n5, BlendTime_0=@n7, BlendTime_1=@n9, bActiveValue=@n8) {blendType=Linear}
     Alpha:float=Get HeadAimAlpha() -> (HeadAimAlpha:float)
    ) -> (Pose:Pose Link Structure) {blendSpace=/Game/Animation/Game/MainPlayer/Poses/Medium/Male/AO/M_NoWep_Relaxed_AO_120}) {layerGroup=None}
   Save cached pose 'AimingOffsetBasePose'(Pose=Input Pose()) {cacheName=AimingOffsetBasePose}
  ]}
 VehicleFinalPoseOverride {nodes [Output Pose(Result=Input Pose() {name=InFinalPose}) {layerGroup=VehicleOverride}]}
 VehicleFullBodyOverride {nodes [
   AnimGraphNode_SubInput_0 = Input Pose() {name=InPoseFullBody}
   Output Pose(Result=@AnimGraphNode_SubInput_0) {layerGroup=VehicleOverride}
  ]}
 VehicleLowerBodyOverride {nodes [
   Input Pose() {name=InPoseUpperBody}
   AnimGraphNode_SubInput_0 = Input Pose() {name=InPoseLowerBodyDefault}
   Output Pose(Result=@AnimGraphNode_SubInput_0) {layerGroup=VehicleOverride}
  ]}
 VehicleSplitBodyOverride {nodes [
   Input Pose() {name=InPoseUpperAndLowerBody}
   Input Pose() {name=InPoseUpperAndLowerBodyRemoveRoot}
   Output Pose(Result=Input Pose() {name=InPoseSplitBody}) {layerGroup=Vehicle Override}
  ]}
]
variables [
 ArmLeftLocomotionWeight {type=number}
 ArmRightLocomotionWeight {type=number}
 CrouchPlayRate {type=number, default=1}
 EmoteAnimOffsetCurveName {type=string, default=UseEmoteAnimOffset}
 Indigo_Boost {default=()}
 IsSnowman {type=boolean}
 LeftHandOverride {type=Vector}
 MenuScreenDispatcher {default=()}
 ParaGlide_TriggerAnim {type=boolean}
 ParaGlide_UseAltGliderCam {type=boolean}
 PivotAnimBlendTime {type=number, default=0.1}
 RightHandOverride {type=Vector}
 StartAnimBlendTime {type=number, default=0.33}
 TorsoLocomotionWeight {type=number}
]
