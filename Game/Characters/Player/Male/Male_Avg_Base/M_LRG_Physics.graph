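# Physics asset for the M_LRG male player character, defining simplified body collision and joints for skeletal physics such as ragdoll.
# Boxes cover pelvis, lower spine, hands and feet, capsules cover upper spine and limbs, and a sphere covers the head.
# Constraints chain bodies out from the pelvis, skipping Spine_04, neck and clavicles, with uniform angle limit values.
# Elbows and knees lock twist and one swing axis to act as hinges.
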PhysicsAsset /Game/Characters/Player/Male/Male_Avg_Base/M_LRG_Physics
bodies {
 Spine_01 [{shapeType=Box, extentX=33.5381, extentY=45.7704, extentZ=9.8556, center=<7.7667 0.8417 0>, rotation=<rot 0 78.75 -90>}]
 Spine_02 [{shapeType=Box, extentX=10.8, extentY=35.8, extentZ=45.8, center=<8.2859 0.8602 0>, rotation=<rot 0 -11.25 0>}]
 Spine_03 [{shapeType=Box, extentX=11.8122, extentY=37.0763, extentZ=52.6254, center=<7.87 -0.7891 -1.4711>}]
 calf_l [{shapeType=Capsule, length=30.5, radius=8.25, center=<-14.5627 0 0>, rotation=<rot -90 19.1416 -19.1416>}]
 calf_r [{shapeType=Capsule, length=30.5, radius=8.25, center=<16.3534 0.0001 0>, rotation=<rot -90 19.1416 -19.1416>}]
 foot_l [{shapeType=Box, extentX=12.1564, extentY=19.3073, extentZ=10.9962, center=<-3.6652 9.4852 -0.096>}]
 foot_r [{shapeType=Box, extentX=12.1564, extentY=19.3073, extentZ=10.9963, center=<3.6657 -9.6953 0.1443>}]
 hand_l [{shapeType=Box, extentX=18.0239, extentY=9.5969, extentZ=18.199, center=<10.6022 -1.0679 0.8028>}]
 hand_r [{shapeType=Box, extentX=18.0239, extentY=9.5969, extentZ=18.199, center=<-10.1406 2.2756 0.8028>}]
 head [{shapeType=Sphere, radius=15.15, center=<5.5971 0 0>}]
 lowerarm_l [{shapeType=Capsule, length=16.5, radius=9.25, center=<12.8813 0 0>, rotation=<rot 90 20.9808 20.9809>}]
 lowerarm_r [{shapeType=Capsule, length=16.5, radius=9.25, center=<-12.9584 0.0001 0>, rotation=<rot 90 20.9808 20.9809>}]
 pelvis [{shapeType=Box, extentX=20.1733, extentY=32.674, extentZ=39.7114, center=<3.6787 4.1914 -0.3355>}]
 spine_05 [{shapeType=Capsule, length=20.5, radius=23.25, center=<2.1737 -5.4066 0>}]
 thigh_l [{shapeType=Capsule, length=21.5, radius=10.5, center=<-15.701 1.873 0.0989>, rotation=<rot -90 175.7815 -175.7816>}]
 thigh_r [{shapeType=Capsule, length=21.5, radius=10.5, center=<16.8996 -2.8189 0.0989>, rotation=<rot -90 175.7815 -175.7816>}]
 upperarm_l [{shapeType=Capsule, length=25.5, radius=12.25, center=<13.7621 0 0.0001>, rotation=<rot -89.8935 -5.6819 5.682>}]
 upperarm_r [{shapeType=Capsule, length=25.5, radius=12.25, center=<-9.0654 0 0>, rotation=<rot -89.8935 -5.6819 5.682>}]
}
constraints [
 {bone1Name=Spine_01, bone2Name=pelvis, swing1LimitDegrees=45, swing2LimitDegrees=45, twistLimitDegrees=45}
 {bone1Name=Spine_02, bone2Name=Spine_01, swing1LimitDegrees=45, swing2LimitDegrees=45, twistLimitDegrees=45}
 {bone1Name=Spine_03, bone2Name=Spine_02, swing1LimitDegrees=45, swing2LimitDegrees=45, twistLimitDegrees=45}
 {bone1Name=spine_05, bone2Name=Spine_03, swing1LimitDegrees=45, swing2LimitDegrees=45, twistLimitDegrees=45}
 {bone1Name=upperarm_l, bone2Name=spine_05, swing1LimitDegrees=45, swing2LimitDegrees=45, twistLimitDegrees=45}
 {
  bone1Name=lowerarm_l
  bone2Name=upperarm_l
  swing1LimitDegrees=45
  swing2LimitDegrees=45
  swing2Motion=Locked
  twistLimitDegrees=45
  twistMotion=Locked
 }
 {bone1Name=hand_l, bone2Name=lowerarm_l, swing1LimitDegrees=45, swing2LimitDegrees=45, twistLimitDegrees=45}
 {bone1Name=upperarm_r, bone2Name=spine_05, swing1LimitDegrees=45, swing2LimitDegrees=45, twistLimitDegrees=45}
 {
  bone1Name=lowerarm_r
  bone2Name=upperarm_r
  swing1LimitDegrees=45
  swing2LimitDegrees=45
  swing2Motion=Locked
  twistLimitDegrees=45
  twistMotion=Locked
 }
 {bone1Name=hand_r, bone2Name=lowerarm_r, swing1LimitDegrees=45, swing2LimitDegrees=45, twistLimitDegrees=45}
 {bone1Name=thigh_l, bone2Name=pelvis, swing1LimitDegrees=45, swing2LimitDegrees=45, twistLimitDegrees=45}
 {bone1Name=thigh_r, bone2Name=pelvis, swing1LimitDegrees=45, swing2LimitDegrees=45, twistLimitDegrees=45}
 {
  bone1Name=calf_l
  bone2Name=thigh_l
  swing1LimitDegrees=45
  swing2LimitDegrees=45
  swing2Motion=Locked
  twistLimitDegrees=45
  twistMotion=Locked
 }
 {
  bone1Name=calf_r
  bone2Name=thigh_r
  swing1LimitDegrees=45
  swing2LimitDegrees=45
  swing2Motion=Locked
  twistLimitDegrees=45
  twistMotion=Locked
 }
 {bone1Name=foot_l, bone2Name=calf_l, swing1LimitDegrees=45, swing2LimitDegrees=45, twistLimitDegrees=45}
 {bone1Name=foot_r, bone2Name=calf_r, swing1LimitDegrees=45, swing2LimitDegrees=45, twistLimitDegrees=45}
 {bone1Name=head, bone2Name=spine_05, swing1LimitDegrees=45, swing2LimitDegrees=45, twistLimitDegrees=45}
]
